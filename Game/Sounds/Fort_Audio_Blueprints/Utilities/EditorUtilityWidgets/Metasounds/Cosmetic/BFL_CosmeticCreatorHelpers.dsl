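; Editor helper function library for the MetaSound cosmetic preset creator widget, covering sound library creation and cosmetic option lookup.
; CreateSoundLibrary makes and casts a SoundLibrary asset through AssetTools, logging the result, and AddMetasoundToLibrary maps sounds into it by tag.
; GetCosmeticOptions reads data table rows, while GetCosmeticCreator finds the owning widget through a child's parent and outers.
; AddNotifyState loops over montages but shows no action.
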(fn AddMetasoundToLibrary
  (params (Sound :type Sound%20Base%20Soft%20Object%20Reference%20%28by%20ref%29) (SoundLibraryTag :type Sound%20Library%20Tag%20Structure) (SoundLibrary :type Sound%20Library%20Object%20Reference))
  (bind _returnvalue (Utilities|IsValid SoundLibrary :!class /Script/Engine.KismetSystemLibrary))
  (bind _sounds (Variables|/Script/SoundLibrary.SoundLibrary|GetSounds SoundLibrary))
  (if _returnvalue
    (Utilities|Map|Add _sounds SoundLibraryTag Sound)
    (else
      _))
  (return _sounds _returnvalue))

(fn AddNotifyState
  (params (Montages :type Array%20of%20Anim%20Sequence%20Base%20Object%20References) (NotifyType :type Anim%20Notify%20State%20Object%20Reference))
  (for _ Montages
    _))

(fn CreateSoundLibrary
  (params (PackagingPath :type String) (AssetsName :type String))
  (bind _returnvalue (EditorScripting|AssetTools|GetAssetTools))
  (bind _returnvalue_2 (EditorScripting|AssetTools|CreateAsset _returnvalue AssetsName PackagingPath "/Script/SoundLibrary.SoundLibrary" 0 "None" false))
  (bind _assound_library (Utilities|Casting|CastToSoundLibrary _returnvalue_2)
    (:then
      (LocalVariables|SetCastPass true))
    (:CastFailed
      (LocalVariables|SetCastPass false)))
  (Utilities|String|LogString (Utilities|String|Append AssetsName (Utilities|String|SelectString " Saved!" " Failed to save!" (LocalVariables|GetCastPass))))
  (return _assound_library (LocalVariables|GetCastPass)))

(fn GetCosmeticCreator
  (params (Target :type Widget%20Object%20Reference))
  (bind _aseuw_meta_sound_cosmetic_preset_creator (Utilities|Casting|CastToEUW_MetaSound_Cosmetic_PresetCreator (Utilities|GetOuterObject (Utilities|GetOuterObject (Widget|GetParent Target)))))
  (return _aseuw_meta_sound_cosmetic_preset_creator))

(fn GetCosmeticOptions
  (params (DataTable :type Data%20Table%20Object%20Reference) (RowName :type Name))
  (bind _returnvalue (Utilities|GetDataTableRow DataTable RowName)
    (:then
      (LocalVariables|SetRowValid true))
    (:RowNotFound
      (LocalVariables|SetRowValid false)))
  (return (LocalVariables|GetRowValid) (Utilities|Struct|BreakStructMetaSoundCosmeticOptions _returnvalue) (.StatManagerTag_7_9A8637544D86A6059632C9B67C1351DE (Utilities|Struct|BreakStructMetaSoundCosmeticOptions _returnvalue)) (.SoundLibraryTag_8_A25CE634434A2AE4075FAFA3645A03FD (Utilities|Struct|BreakStructMetaSoundCosmeticOptions _returnvalue))))
