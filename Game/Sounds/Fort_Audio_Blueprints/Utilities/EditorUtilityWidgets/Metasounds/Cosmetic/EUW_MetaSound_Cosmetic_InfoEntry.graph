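# Editor utility widget that forms an info entry for cosmetic MetaSound tooling, with fields for a MetaSound name and its wave names.
# A canvas panel holds two variable text blocks, InfoMetaSoundName in green above InfoWavesNames, both with placeholder text and drop shadows.
# The event graph is empty and the widget declares no variables of its own, so it contains layout only.
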EditorUtilityWidgetBlueprint /Game/Sounds/Fort_Audio_Blueprints/Utilities/EditorUtilityWidgets/Metasounds/Cosmetic/EUW_MetaSound_Cosmetic_InfoEntry
graphs [EventGraph {links [], nodes []}]
variables []
widgets {
 info {rootWidgetClass=&CanvasPanel, widgetCount=3}
 widgets [
  CanvasPanel_0 {widgetClassPath=&CanvasPanel, widget=CanvasPanel_0}
  InfoMetaSoundName {
   widgetClassPath=&TextBlock
   bIsVariable=true
   parent=CanvasPanel_0
   widget=InfoMetaSoundName
   properties {
    text=EXAMPLE
    colorAndOpacity {colorUseRule=UseColor_Specified, specifiedColor=<rgba 0.2745 1 0.2314 1>}
    font {
     bForceMonospaced=false
     bMaterialIsStencil=false
     fontMaterial=null
     fontObject=/LauncherSocial/BrutalType
     letterSpacing=0
     monospacedWidth=1
     size=15
     skewAmount=0
     typefaceFontName=Regular
     outlineSettings {
      bApplyOutlineToDropShadows=false
      bMiteredCorners=false
      bSeparateFillAlpha=false
      outlineMaterial=null
      outlineSize=0
      outlineColor=<rgba 0 0 0 1>
     }
    }
    shadowColorAndOpacity=<rgba 0 0 0 1>
    shadowOffset=<0.5 0.5>
   }
   slot {class=CanvasPanelSlot, properties {layoutData {alignment=<0 0>, anchors {maximum=<0 0>, minimum=<0 0>}, offsets=<ltrb 10 10 100 30>}}}
  }
  InfoWavesNames {
   widgetClassPath=&TextBlock
   bIsVariable=true
   parent=CanvasPanel_0
   widget=InfoWavesNames
   properties {
    text=Example
    font {
     bForceMonospaced=false
     bMaterialIsStencil=false
     fontMaterial=null
     fontObject=&BrutalType
     letterSpacing=0
     monospacedWidth=1
     size=15
     skewAmount=0
     typefaceFontName=Regular
     outlineSettings {
      bApplyOutlineToDropShadows=false
      bMiteredCorners=false
      bSeparateFillAlpha=false
      outlineMaterial=null
      outlineSize=0
      outlineColor=<rgba 0 0 0 1>
     }
    }
    shadowColorAndOpacity=<rgba 0.2042 0.54 0.8333 1>
    shadowOffset=<0.5 0.5>
   }
   slot {class=CanvasPanelSlot, properties {layoutData {alignment=<0 0>, anchors {maximum=<0 0>, minimum=<0 0>}, offsets=<ltrb 10 40 100 30>}}}
  }
 ]
}
