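# Editor utility widget that serves as a static info panel for a cosmetic MetaSound creation tool.
# A grid of color-coded text blocks explains the Selected Sounds List, save Path and Asset Name fields.
# Text below shows the glider naming convention, mapping looping wave names to MSS_-prefixed MetaSound names, with a concrete example and a template.
# It has no graph logic or variables, so it only displays text.
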EditorUtilityWidgetBlueprint /Game/Sounds/Fort_Audio_Blueprints/Utilities/EditorUtilityWidgets/Metasounds/Cosmetic/EUW_MetaSound_Cosmetic_InfoListView
graphs [EventGraph {links [], nodes []}]
variables []
widgets {
 info {rootWidgetClass=&CanvasPanel, widgetCount=8}
 widgets [
  CanvasPanel_0 {widgetClassPath=&CanvasPanel, widget=CanvasPanel_0}
  ExampleText {
   widgetClassPath=&TextBlock
   bIsVariable=true
   parent=CanvasPanel_0
   widget=ExampleText
   properties {
    text=Example:
    font {
     bForceMonospaced=false
     bMaterialIsStencil=false
     fontMaterial=null
     fontObject=/LauncherSocial/BrutalType
     letterSpacing=0
     monospacedWidth=1
     size=15
     skewAmount=0
     typefaceFontName=Regular
     outlineSettings {
      bApplyOutlineToDropShadows=false
      bMiteredCorners=false
      bSeparateFillAlpha=false
      outlineMaterial=null
      outlineSize=0
      outlineColor=<rgba 0 0 0 1>
     }
    }
    shadowColorAndOpacity=<rgba 0 0 0 1>
   }
   slot {
    class=CanvasPanelSlot
    properties {layoutData {alignment=<0 0>, anchors {maximum=<0 0>, minimum=<0 0>}, offsets=<ltrb 12 500 733.2046 24.852>}}
   }
  }
  ExampleText_1 {
   widgetClassPath=&TextBlock
   parent=CanvasPanel_0
   widget=ExampleText_1
   properties {
    text="Glider_PlotTwist_Thrust_Loop   ->   MSS_Glider_PlotTwist_Thrust"
    colorAndOpacity {colorUseRule=UseColor_Specified, specifiedColor=<rgba 0.0899 0.9203 1 1>}
    font {
     bForceMonospaced=false
     bMaterialIsStencil=false
     fontMaterial=null
     fontObject=&BrutalType
     letterSpacing=0
     monospacedWidth=1
     size=15
     skewAmount=0
     typefaceFontName=Regular
     outlineSettings {
      bApplyOutlineToDropShadows=false
      bMiteredCorners=false
      bSeparateFillAlpha=false
      outlineMaterial=null
      outlineSize=0
      outlineColor=<rgba 0 0 0 1>
     }
    }
    shadowColorAndOpacity=<rgba 0 0 0 1>
   }
   slot {
    class=CanvasPanelSlot
    properties {layoutData {alignment=<0 0>, anchors {maximum=<0 0>, minimum=<0 0>}, offsets=<ltrb 12 530 733.2046 25.71>}}
   }
  }
  ExampleText_2 {
   widgetClassPath=&TextBlock
   parent=CanvasPanel_0
   widget=ExampleText_2
   properties {
    text="Glider_<UniqueName>_<Action>_Loop   ->   MSS_Glider_<UniqueName>_<Action>"
    colorAndOpacity {colorUseRule=UseColor_Specified, specifiedColor=<rgba 0.4291 0.2617 1 1>}
    font {
     bForceMonospaced=false
     bMaterialIsStencil=false
     fontMaterial=null
     fontObject=&BrutalType
     letterSpacing=0
     monospacedWidth=1
     size=15
     skewAmount=0
     typefaceFontName=Regular
     outlineSettings {
      bApplyOutlineToDropShadows=false
      bMiteredCorners=false
      bSeparateFillAlpha=false
      outlineMaterial=null
      outlineSize=0
      outlineColor=<rgba 0 0 0 1>
     }
    }
    shadowColorAndOpacity=<rgba 0 0 0 1>
   }
   slot {
    class=CanvasPanelSlot
    properties {layoutData {alignment=<0 0>, anchors {maximum=<0 0>, minimum=<0 0>}, offsets=<ltrb 12 560 733.2046 29.142>}}
   }
  }
  GridPanel_0 {
   widgetClassPath=&GridPanel
   parent=CanvasPanel_0
   widget=GridPanel_0
   slot {
    class=CanvasPanelSlot
    properties {bAutoSize=true, layoutData {alignment=<0 0>, anchors {maximum=<0 0>, minimum=<0 0>}, offsets=<ltrb 20 20 100 51.45>}}
   }
  }
  ExampleText_3 {
   widgetClassPath=&TextBlock
   parent=GridPanel_0
   widget=ExampleText_3
   properties {
    text="Selected Sounds List: \r\n Collection of identifiers to create metassounds.\r\n Selected waves are added to this list."
    colorAndOpacity {colorUseRule=UseColor_Specified, specifiedColor=<rgba 0.0899 0.9203 1 1>}
    font {
     bForceMonospaced=false
     bMaterialIsStencil=false
     fontMaterial=null
     fontObject=&BrutalType
     letterSpacing=0
     monospacedWidth=1
     size=15
     skewAmount=0
     typefaceFontName=Regular
     outlineSettings {
      bApplyOutlineToDropShadows=false
      bMiteredCorners=false
      bSeparateFillAlpha=false
      outlineMaterial=null
      outlineSize=0
      outlineColor=<rgba 0 0 0 1>
     }
    }
    shadowColorAndOpacity=<rgba 0 0 0 1>
   }
   slot {class=GridSlot, properties {padding=<ltrb 4 4 4 4>}}
  }
  SoundListText_1 {
   widgetClassPath=&TextBlock
   parent=GridPanel_0
   widget=SoundListText_1
   properties {
    autoWrapText=true
    text="Path:\r\n Content browser location that metasounds will be saved."
    colorAndOpacity {colorUseRule=UseColor_Specified, specifiedColor=<rgba 0.2754 1 0.2314 1>}
    font {
     bForceMonospaced=false
     bMaterialIsStencil=false
     fontMaterial=null
     fontObject=&BrutalType
     letterSpacing=0
     monospacedWidth=1
     size=15
     skewAmount=0
     typefaceFontName=Regular
     outlineSettings {
      bApplyOutlineToDropShadows=false
      bMiteredCorners=false
      bSeparateFillAlpha=false
      outlineMaterial=null
      outlineSize=0
      outlineColor=<rgba 0 0 0 1>
     }
    }
    shadowColorAndOpacity=<rgba 0 0 0 1>
   }
   slot {class=GridSlot, properties {row=1}}
  }
  SoundListText_2 {
   widgetClassPath=&TextBlock
   parent=GridPanel_0
   widget=SoundListText_2
   properties {
    autoWrapText=true
    text="Asset Name:\r\n Unique name that will be added between the pre-fix and post-fix in the MetaSound name when created."
    colorAndOpacity {colorUseRule=UseColor_Specified, specifiedColor=<rgba 0.9748 1 0.0165 1>}
    font {
     bForceMonospaced=false
     bMaterialIsStencil=false
     fontMaterial=null
     fontObject=&BrutalType
     letterSpacing=0
     monospacedWidth=1
     size=15
     skewAmount=0
     typefaceFontName=Regular
     outlineSettings {
      bApplyOutlineToDropShadows=false
      bMiteredCorners=false
      bSeparateFillAlpha=false
      outlineMaterial=null
      outlineSize=0
      outlineColor=<rgba 0 0 0 1>
     }
    }
    shadowColorAndOpacity=<rgba 0 0 0 1>
   }
   slot {class=GridSlot, properties {row=3, padding=<ltrb 4 4 4 4>}}
  }
 ]
}
