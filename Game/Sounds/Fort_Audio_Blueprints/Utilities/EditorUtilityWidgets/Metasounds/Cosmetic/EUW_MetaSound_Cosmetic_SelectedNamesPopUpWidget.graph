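# Editor Utility Widget that serves as one selectable sound-name entry, apparently for a popup in the cosmetic MetaSound audio tool.
# Pressing its full-size text button toggles the Selected flag, using a Select node as a NOT, and tints the background blue when selected or white when deselected.
# Both paths then broadcast OnButtonPressed with the widget itself.
# Variables hold the sound name, object, type text and owning widgets, but the graph never uses them.
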EditorUtilityWidgetBlueprint /Game/Sounds/Fort_Audio_Blueprints/Utilities/EditorUtilityWidgets/Metasounds/Cosmetic/EUW_MetaSound_Cosmetic_SelectedNamesPopUpWidget
graphs [
 EventGraph {nodes [
   "On Pressed (TextButton)"() {is=title K2Node_ComponentBoundEvent}
   > n1 = Branch(Condition=Select(Option 0:bool=true, Option 1:bool=false, Index=Get Selected() -> (Selected:bool)) -> (ReturnValue:bool))
   n2 = Get TextButton() -> (TextButton:Editor Utility Button Object Reference)
   > SetBackgroundColor!(execute, self:Button Object Reference=@n2, InBackgroundColor:Linear Color Structure=(R=0,G=0.4667,B=1,A=1)) -> (then)
   > n3 = Set Selected(Selected:bool=true) -> (Output_Get:bool)
   Call On Button Pressed!(
    execute:Exec=[@n4.then, @n3.then]
    self:EUW Meta Sound Cosmetic Selected Names Pop Up Widget Object Reference
    SoundNameWidget:EUW Meta Sound Cosmetic Selected Names Pop Up Widget Object Reference=Self-Reference()
   ) -> (then:Exec) {is=title K2Node_CallDelegate}
   SetBackgroundColor(execute=@n1.else, self=@n2, InBackgroundColor=(R=1,G=1,B=1,A=1))
   > n4 = Set Selected(Selected=false)
   Event Construct()
  ]}
 OnButtonPressed {nodes [OnButtonPressed!() -> (then, SoundNameWidget:EUW Meta Sound Cosmetic Selected Names Pop Up Widget Object Reference) {is=title K2Node_FunctionEntry}]}
]
variables [
 CurrentActionWidget {type=EUW_MetaSound_Cosmetic_ActionNameWidget_C, default=null}
 OnButtonPressed {default=()}
 ParentWidget {type=EUW_MetaSound_Cosmetic_ActionSoundsList_C, default=null}
 Selected {type=boolean}
 SoundName {type=string, default=""}
 SoundObject {type=Object, default=null}
 TypeText {type=string, default=""}
]
widgets {
 info {rootWidgetClass=&Overlay, widgetCount=3}
 widgets [
  Overlay_141 {widgetClassPath=&Overlay, widget=Overlay_141}
  TextButton {
   widgetClassPath=&EditorUtilityButton
   parent=Overlay_141
   widget=TextButton
   properties {widgetStyle {
     clickedSlateSound {resourceObject=null}
     disabled {
      bIsDynamicallyLoaded=false
      drawAs=RoundedBox
      imageType=NoImage
      mirroring=NoMirror
      resourceName=None
      resourceObject=null
      tiling=NoTile
      imageSize=<0 0>
      margin=<ltrb 0 0 0 0>
      outlineSettings {
       bUseBrushTransparency=false
       roundingType=FixedRadius
       width=1
       color {colorUseRule=UseColor_ColorTable, specifiedColor=<rgba 0 0 0 0>}
       cornerRadii=<4 4 4 4>
      }
      tintColor {colorUseRule=UseColor_Specified, specifiedColor=<rgba 0.0395 0.0395 0.0395 1>}
      uVRegion {bIsValid=false, max=<0 0>, min=<0 0>}
     }
     disabledForeground {colorUseRule=UseColor_ColorTable, specifiedColor=<rgba 0 0 0 0>}
     hovered {
      bIsDynamicallyLoaded=false
      drawAs=RoundedBox
      imageType=NoImage
      mirroring=NoMirror
      resourceName=None
      resourceObject=null
      tiling=NoTile
      imageSize=<32 32>
      margin=<ltrb 0 0 0 0>
      outlineSettings {
       bUseBrushTransparency=false
       roundingType=FixedRadius
       width=1
       color {colorUseRule=UseColor_ColorTable, specifiedColor=<rgba 0 0 0 0>}
       cornerRadii=<4 4 4 4>
      }
      tintColor {colorUseRule=UseColor_Specified, specifiedColor=<rgba 0.0953 0.0953 0.0953 1>}
      uVRegion {bIsValid=false, max=<0 0>, min=<0 0>}
     }
     hoveredForeground {colorUseRule=UseColor_ColorTable, specifiedColor=<rgba 0 0 0 0>}
     hoveredSlateSound {resourceObject=null}
     normal {
      bIsDynamicallyLoaded=false
      drawAs=RoundedBox
      imageType=NoImage
      mirroring=NoMirror
      resourceName=None
      resourceObject=null
      tiling=NoTile
      imageSize=<32 32>
      margin=<ltrb 0 0 0 0>
      outlineSettings {
       bUseBrushTransparency=false
       roundingType=FixedRadius
       width=1
       color {colorUseRule=UseColor_ColorTable, specifiedColor=<rgba 0 0 0 0>}
       cornerRadii=<4 4 4 4>
      }
      tintColor {colorUseRule=UseColor_Specified, specifiedColor=<rgba 0.0395 0.0395 0.0395 1>}
      uVRegion {bIsValid=false, max=<0 0>, min=<0 0>}
     }
     normalForeground {colorUseRule=UseColor_ColorTable, specifiedColor=<rgba 0 0 0 0>}
     normalPadding=<ltrb 12 1.5 12 1.5>
     pressed {
      bIsDynamicallyLoaded=false
      drawAs=RoundedBox
      imageType=NoImage
      mirroring=NoMirror
      resourceName=None
      resourceObject=null
      tiling=NoTile
      imageSize=<32 32>
      margin=<ltrb 0 0 0 0>
      outlineSettings {
       bUseBrushTransparency=false
       roundingType=FixedRadius
       width=1
       color {colorUseRule=UseColor_ColorTable, specifiedColor=<rgba 0 0 0 0>}
       cornerRadii=<4 4 4 4>
      }
      tintColor {colorUseRule=UseColor_Specified, specifiedColor=<rgba 0.0284 0.0284 0.0284 1>}
      uVRegion {bIsValid=false, max=<0 0>, min=<0 0>}
     }
     pressedForeground {colorUseRule=UseColor_ColorTable, specifiedColor=<rgba 0 0 0 0>}
     pressedPadding=<ltrb 12 2.5 12 0.5>
     pressedSlateSound {resourceObject=null}
    }}
   slot {class=OverlaySlot, properties {horizontalAlignment=HAlign_Fill, verticalAlignment=VAlign_Fill}}
  }
  TextEntry {
   widgetClassPath=&TextBlock
   bIsVariable=true
   parent=TextButton
   widget=TextEntry
   properties {
    text=Example
    visibility=HitTestInvisible
    colorAndOpacity {colorUseRule=UseColor_Specified, specifiedColor=<rgba 0.2039 0.5412 0.8314 1>}
    font {
     bForceMonospaced=false
     bMaterialIsStencil=false
     fontMaterial=null
     fontObject=/LauncherSocial/BrutalType
     letterSpacing=0
     monospacedWidth=1
     size=15
     skewAmount=0
     typefaceFontName=Regular
     outlineSettings {
      bApplyOutlineToDropShadows=false
      bMiteredCorners=false
      bSeparateFillAlpha=false
      outlineMaterial=null
      outlineSize=0
      outlineColor=<rgba 0 0 0 1>
     }
    }
    shadowColorAndOpacity=<rgba 0 0 0 1>
    shadowOffset=<0.5 0.5>
   }
   slot {class=ButtonSlot, properties {horizontalAlignment=HAlign_Fill, verticalAlignment=VAlign_Fill}}
  }
 ]
}
