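# Editor utility widget for one draggable sound-name row in a MetaSound cosmetic audio tool.
# A type button toggles IsLooping and shows a red (L) or green (S), and an X button broadcasts OnRemovedWidget.
# Left-dragging creates a drag-drop operation carrying the sound name, object, owning action widget and itself.
# Drag start, cancel and drops onto the row are forwarded through event dispatchers, with cancel also printing a debug string.
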EditorUtilityWidgetBlueprint /Game/Sounds/Fort_Audio_Blueprints/Utilities/EditorUtilityWidgets/Metasounds/Cosmetic/EUW_MetaSound_Cosmetic_SoundNameWidget
graphs [
 EventGraph {nodes [
   "On Pressed (TextButton)"() {is=title K2Node_ComponentBoundEvent}
   > Update Type Text!(
    execute
    self
    ShouldLoop:bool=Select(Option 0:bool=true, Option 1:bool=false, Index=Get IsLooping() -> (IsLooping:bool)) -> (ReturnValue:bool)
   ) -> (then)
   "On Pressed (RemoveButton)"() {is=title K2Node_ComponentBoundEvent}
   > Call On Removed Widget!(
    execute:Exec
    self:EUW Meta Sound Cosmetic Sound Name Widget Object Reference
    SoundNameWidget:EUW Meta Sound Cosmetic Sound Name Widget Object Reference=Self-Reference()
   ) -> (then:Exec) {is=title K2Node_CallDelegate}
   Event OnDragCancelled!() -> (OutputDelegate, then, PointerEvent:"Pointer Event Structure (by ref)", Operation:Drag Drop Operation Object Reference) {is=title K2Node_Event}
   > Call On Drag Cancel!(
    execute:Exec
    self:EUW Meta Sound Cosmetic Sound Name Widget Object Reference
    CurrentActionWidget:EUW Meta Sound Cosmetic Action Name Widget Object Reference=Get CurrentActionWidget() -> (CurrentActionWidget:EUW Meta Sound Cosmetic Action Name Widget Object Reference)
    SoundNameWidget:EUW Meta Sound Cosmetic Sound Name Widget Object Reference=Self-Reference()
   ) -> (then:Exec) {is=title K2Node_CallDelegate}
   n1 = "GetText (Text)"!(self:Text Object Reference=Get PresetText!(self:EUW Meta Sound Cosmetic Action Name Widget Object Reference=Get CurrentActionWidget()) -> (PresetText:Text Object Reference)) -> (ReturnValue:Text)
   > PrintString(InString=Append(A="Drag Cancled: ", B="To String (Text)"(InText=@n1)))
   Event Construct()
   > Update Type Text(ShouldLoop=Get IsLooping())
  ]}
 OnDragCancel {nodes [OnDragCancel!() -> (
    then
    CurrentActionWidget:EUW Meta Sound Cosmetic Action Name Widget Object Reference
    SoundNameWidget:EUW Meta Sound Cosmetic Sound Name Widget Object Reference
   ) {is=title K2Node_FunctionEntry}]}
 OnDragDetected {nodes [
   OnDragDetected!() -> (then, MyGeometry:Geometry Structure, PointerEvent:"Pointer Event Structure (by ref)") {is=title K2Node_FunctionEntry}
   > n1 = "Create Drag && Drop Operation"!(
    execute:Exec
    WorldContextObject:Object Reference
    Class:Drag Drop Operation Class Reference=&/B_MetaSound_Cosmeitc_SoundNameDragDrop.B_MetaSound_Cosmeitc_SoundNameDragDrop_C
    Tag:String
    Payload:Object Reference
    DefaultDragVisual:Widget Object Reference=Self-Reference()
    Pivot:EDragPivot Enum=MouseDown
    Offset:Vector 2D Structure=(X=0,Y=0)
    SoundName:Text=Get SoundName() -> (SoundName:Text)
    CurrentActionWidget:EUW Meta Sound Cosmetic Action Name Widget Object Reference=Get CurrentActionWidget()
    SoundObjectRef:Object Reference=Get SoundObject() -> (SoundObject:Object Reference)
    SoundNameWidget:EUW Meta Sound Cosmetic Sound Name Widget Object Reference=Self-Reference()
   ) -> (then:Exec, ReturnValue:B Meta Sound Cosmeitc Sound Name Drag Drop Object Reference) {is=title K2Node_CreateDragDropOperation}
   > Call On Drag Started!(
    execute:Exec
    self:EUW Meta Sound Cosmetic Sound Name Widget Object Reference
    CurrentNameWidget:EUW Meta Sound Cosmetic Action Name Widget Object Reference=Get CurrentActionWidget()
    SoundNameWidget:EUW Meta Sound Cosmetic Sound Name Widget Object Reference=Self-Reference()
   ) -> (then:Exec) {is=title K2Node_CallDelegate}
   > Return Node!(execute, Operation:Drag Drop Operation Object Reference=@n1)
  ]}
 OnDragStarted {nodes [OnDragStarted!() -> (
    then
    CurrentNameWidget:EUW Meta Sound Cosmetic Action Name Widget Object Reference
    SoundNameWidget:EUW Meta Sound Cosmetic Sound Name Widget Object Reference
   ) {is=title K2Node_FunctionEntry}]}
 OnDraggedWidgetDropped {nodes [OnDraggedWidgetDropped!() -> (
    then
    DragDropRef:B Meta Sound Cosmeitc Sound Name Drag Drop Object Reference
    SoundNameWidget:EUW Meta Sound Cosmetic Sound Name Widget Object Reference
   ) {is=title K2Node_FunctionEntry}]}
 OnDrop {nodes [
   n1 = Cast To B_MetaSound_Cosmeitc_SoundNameDragDrop!(execute, Object:Wildcard) -> (then, CastFailed, AsB Meta Sound Cosmeitc Sound Name Drag Drop:B Meta Sound Cosmeitc Sound Name Drag Drop Object Reference) {is=title K2Node_DynamicCast}
   > Call On Dragged Widget Dropped!(
    execute:Exec
    self:EUW Meta Sound Cosmetic Sound Name Widget Object Reference
    DragDropRef:B Meta Sound Cosmeitc Sound Name Drag Drop Object Reference=@n1
    SoundNameWidget:EUW Meta Sound Cosmetic Sound Name Widget Object Reference=Self-Reference()
   ) -> (then:Exec) {is=title K2Node_CallDelegate}
   OnDrop!() -> (then, MyGeometry:Geometry Structure, PointerEvent:Pointer Event Structure, Operation:Drag Drop Operation Object Reference) {is=title K2Node_FunctionEntry}
   Return Node!(execute, ReturnValue:bool=false)
  ]}
 OnMouseButtonDown {nodes [
   n1 = OnMouseButtonDown!() -> (then, MyGeometry:Geometry Structure, MouseEvent:"Pointer Event Structure (by ref)") {is=title K2Node_FunctionEntry}
   > n2 = DetectDragIfPressed!(
    execute
    PointerEvent:"Pointer Event Structure (by ref)"=@n1.MouseEvent
    WidgetDetectingDrag:Widget Object Reference
    DragKey:Key Structure=LeftMouseButton
   ) -> (then, ReturnValue:Event Reply Structure)
   > Return Node(ReturnValue:Event Reply Structure=@n2)
  ]}
 OnRemovedWidget {nodes [OnRemovedWidget!() -> (then, SoundNameWidget:EUW Meta Sound Cosmetic Sound Name Widget Object Reference) {is=title K2Node_FunctionEntry}]}
 Update Type Text {nodes [
   n1 = Update Type Text!() -> (then, ShouldLoop:bool=~) {is=title K2Node_FunctionEntry}
   > n2 = Branch(Condition=@n1)
   > "SetText (Text)"(self=Get TypeTextBlock() -> (TypeTextBlock:Text Object Reference), InText="LOCTEXT(\"(L)\")")
   > SetColorAndOpacity(self=Get TypeTextBlock(), InColorAndOpacity=Make Slate Color(SpecifiedColor=(R=1,G=0.0941,B=0,A=1)))
   > Set IsLooping(IsLooping:bool=true) -> (Output_Get:bool)
   "SetText (Text)"(execute=@n2.else, self=Get TypeTextBlock(), InText="LOCTEXT(\"(S)\")")
   > SetColorAndOpacity(self=Get TypeTextBlock(), InColorAndOpacity=Make Slate Color(SpecifiedColor=(R=0.0541,G=1,B=0.1565,A=1)))
   > Set IsLooping(IsLooping=false)
  ]}
]
variables [
 CurrentActionWidget {type=EUW_MetaSound_Cosmetic_ActionNameWidget_C, default=null}
 IsLooping {type=boolean}
 OnDragCancel {default=()}
 OnDraggedWidgetDropped {default=()}
 OnDragStarted {default=()}
 OnRemovedWidget {default=()}
 ParentWidget {type=EUW_MetaSound_Cosmetic_ActionSoundsList_C, default=null}
 SoundName {type=string, default=""}
 SoundObject {type=Object, default=null}
 TypeText {type=string, default=""}
]
widgets {
 info {rootWidgetClass=&Overlay, widgetCount=11}
 widgets [
  Overlay_141 {widgetClassPath=&Overlay, widget=Overlay_141}
  Border_1 {
   widgetClassPath=&Border
   parent=Overlay_141
   widget=Border_1
   properties {
    verticalAlignment=VAlign_Center
    background {
     bIsDynamicallyLoaded=false
     drawAs=Image
     imageType=NoImage
     mirroring=NoMirror
     resourceName=None
     resourceObject=null
     tiling=NoTile
     imageSize=<32 32>
     margin=<ltrb 0 0 0 0>
     outlineSettings {
      bUseBrushTransparency=false
      roundingType=HalfHeightRadius
      width=0
      color {colorUseRule=UseColor_Specified, specifiedColor=<rgba 0 0 0 0>}
      cornerRadii=<0 0 0 0>
     }
     tintColor {colorUseRule=UseColor_Specified, specifiedColor=<rgba 0.0208 0.0208 0.0208 1>}
     uVRegion {bIsValid=false, max=<0 0>, min=<0 0>}
    }
    padding=<ltrb 0 0 0 0>
   }
   slot {class=OverlaySlot, properties {horizontalAlignment=HAlign_Fill, verticalAlignment=VAlign_Center}}
  }
  Border_0 {
   widgetClassPath=&Border
   parent=Border_1
   widget=Border_0
   properties {background {
     bIsDynamicallyLoaded=false
     drawAs=Border
     imageType=NoImage
     mirroring=NoMirror
     resourceName=None
     resourceObject=null
     tiling=NoTile
     imageSize=<32 32>
     margin=<ltrb 1 1 1 1>
     outlineSettings {
      bUseBrushTransparency=false
      roundingType=HalfHeightRadius
      width=0
      color {colorUseRule=UseColor_Specified, specifiedColor=<rgba 0 0 0 0>}
      cornerRadii=<0 0 0 0>
     }
     tintColor {colorUseRule=UseColor_Specified, specifiedColor=<rgba 0 0 0 1>}
     uVRegion {bIsValid=false, max=<0 0>, min=<0 0>}
    }}
   slot {class=BorderSlot, properties {verticalAlignment=VAlign_Center, padding=<ltrb 0 0 0 0>}}
  }
  HorizontalBox_128 {widgetClassPath=&HorizontalBox, parent=Border_0, widget=HorizontalBox_128, slot {class=BorderSlot}}
  TextButton {
   widgetClassPath=&EditorUtilityButton
   parent=HorizontalBox_128
   widget=TextButton
   properties {widgetStyle {
     clickedSlateSound {resourceObject=null}
     disabled {
      bIsDynamicallyLoaded=false
      drawAs=RoundedBox
      imageType=NoImage
      mirroring=NoMirror
      resourceName=None
      resourceObject=null
      tiling=NoTile
      imageSize=<0 0>
      margin=<ltrb 0 0 0 0>
      outlineSettings {
       bUseBrushTransparency=false
       roundingType=FixedRadius
       width=1
       color {colorUseRule=UseColor_ColorTable, specifiedColor=<rgba 0 0 0 0>}
       cornerRadii=<4 4 4 4>
      }
      tintColor {colorUseRule=UseColor_Specified, specifiedColor=<rgba 0.0395 0.0395 0.0395 1>}
      uVRegion {bIsValid=false, max=<0 0>, min=<0 0>}
     }
     disabledForeground {colorUseRule=UseColor_ColorTable, specifiedColor=<rgba 0 0 0 0>}
     hovered {
      bIsDynamicallyLoaded=false
      drawAs=RoundedBox
      imageType=NoImage
      mirroring=NoMirror
      resourceName=None
      resourceObject=null
      tiling=NoTile
      imageSize=<32 32>
      margin=<ltrb 0 0 0 0>
      outlineSettings {
       bUseBrushTransparency=false
       roundingType=FixedRadius
       width=1
       color {colorUseRule=UseColor_ColorTable, specifiedColor=<rgba 0 0 0 0>}
       cornerRadii=<4 4 4 4>
      }
      tintColor {colorUseRule=UseColor_Specified, specifiedColor=<rgba 0.0953 0.0953 0.0953 1>}
      uVRegion {bIsValid=false, max=<0 0>, min=<0 0>}
     }
     hoveredForeground {colorUseRule=UseColor_ColorTable, specifiedColor=<rgba 0 0 0 0>}
     hoveredSlateSound {resourceObject=null}
     normal {
      bIsDynamicallyLoaded=false
      drawAs=RoundedBox
      imageType=NoImage
      mirroring=NoMirror
      resourceName=None
      resourceObject=null
      tiling=NoTile
      imageSize=<32 32>
      margin=<ltrb 0 0 0 0>
      outlineSettings {
       bUseBrushTransparency=false
       roundingType=FixedRadius
       width=1
       color {colorUseRule=UseColor_ColorTable, specifiedColor=<rgba 0 0 0 0>}
       cornerRadii=<4 4 4 4>
      }
      tintColor {colorUseRule=UseColor_Specified, specifiedColor=<rgba 0.0395 0.0395 0.0395 1>}
      uVRegion {bIsValid=false, max=<0 0>, min=<0 0>}
     }
     normalForeground {colorUseRule=UseColor_ColorTable, specifiedColor=<rgba 0 0 0 0>}
     normalPadding=<ltrb 12 1.5 12 1.5>
     pressed {
      bIsDynamicallyLoaded=false
      drawAs=RoundedBox
      imageType=NoImage
      mirroring=NoMirror
      resourceName=None
      resourceObject=null
      tiling=NoTile
      imageSize=<32 32>
      margin=<ltrb 0 0 0 0>
      outlineSettings {
       bUseBrushTransparency=false
       roundingType=FixedRadius
       width=1
       color {colorUseRule=UseColor_ColorTable, specifiedColor=<rgba 0 0 0 0>}
       cornerRadii=<4 4 4 4>
      }
      tintColor {colorUseRule=UseColor_Specified, specifiedColor=<rgba 0.0284 0.0284 0.0284 1>}
      uVRegion {bIsValid=false, max=<0 0>, min=<0 0>}
     }
     pressedForeground {colorUseRule=UseColor_ColorTable, specifiedColor=<rgba 0 0 0 0>}
     pressedPadding=<ltrb 12 2.5 12 0.5>
     pressedSlateSound {resourceObject=null}
    }}
   slot {class=HorizontalBoxSlot}
  }
  SizeBox_0 {
   widgetClassPath=&SizeBox
   parent=TextButton
   widget=SizeBox_0
   properties {bOverride_WidthOverride=true, widthOverride=25}
   slot {class=ButtonSlot, properties {padding=<ltrb 0 0 0 0>}}
  }
  TypeTextBlock {
   widgetClassPath=&TextBlock
   bIsVariable=true
   parent=SizeBox_0
   widget=TypeTextBlock
   properties {
    text=(L)
    colorAndOpacity {colorUseRule=UseColor_Specified, specifiedColor=<rgba 1 0.0944 0 1>}
    font {
     bForceMonospaced=false
     bMaterialIsStencil=false
     fontMaterial=null
     fontObject=/LauncherSocial/BrutalType
     letterSpacing=0
     monospacedWidth=1
     size=15
     skewAmount=0
     typefaceFontName=Regular
     outlineSettings {
      bApplyOutlineToDropShadows=false
      bMiteredCorners=false
      bSeparateFillAlpha=false
      outlineMaterial=null
      outlineSize=0
      outlineColor=<rgba 0 0 0 1>
     }
    }
    shadowColorAndOpacity=<rgba 0.2042 0.54 0.8333 1>
    shadowOffset=<0.5 0.5>
   }
   slot {class=SizeBoxSlot, properties {verticalAlignment=VAlign_Center}}
  }
  RemoveButton {
   widgetClassPath=&EditorUtilityButton
   parent=HorizontalBox_128
   widget=RemoveButton
   properties {widgetStyle {
     clickedSlateSound {resourceObject=null}
     disabled {
      bIsDynamicallyLoaded=false
      drawAs=RoundedBox
      imageType=NoImage
      mirroring=NoMirror
      resourceName=None
      resourceObject=null
      tiling=NoTile
      imageSize=<0 0>
      margin=<ltrb 0 0 0 0>
      outlineSettings {
       bUseBrushTransparency=false
       roundingType=FixedRadius
       width=1
       color {colorUseRule=UseColor_ColorTable, specifiedColor=<rgba 0 0 0 0>}
       cornerRadii=<4 4 4 4>
      }
      tintColor {colorUseRule=UseColor_Specified, specifiedColor=<rgba 0.0395 0.0395 0.0395 1>}
      uVRegion {bIsValid=false, max=<0 0>, min=<0 0>}
     }
     disabledForeground {colorUseRule=UseColor_ColorTable, specifiedColor=<rgba 0 0 0 0>}
     hovered {
      bIsDynamicallyLoaded=false
      drawAs=RoundedBox
      imageType=NoImage
      mirroring=NoMirror
      resourceName=None
      resourceObject=null
      tiling=NoTile
      imageSize=<32 32>
      margin=<ltrb 0 0 0 0>
      outlineSettings {
       bUseBrushTransparency=false
       roundingType=FixedRadius
       width=1
       color {colorUseRule=UseColor_ColorTable, specifiedColor=<rgba 0 0 0 0>}
       cornerRadii=<4 4 4 4>
      }
      tintColor {colorUseRule=UseColor_Specified, specifiedColor=<rgba 0.0953 0.0953 0.0953 1>}
      uVRegion {bIsValid=false, max=<0 0>, min=<0 0>}
     }
     hoveredForeground {colorUseRule=UseColor_ColorTable, specifiedColor=<rgba 0 0 0 0>}
     hoveredSlateSound {resourceObject=null}
     normal {
      bIsDynamicallyLoaded=false
      drawAs=RoundedBox
      imageType=NoImage
      mirroring=NoMirror
      resourceName=None
      resourceObject=null
      tiling=NoTile
      imageSize=<32 32>
      margin=<ltrb 0 0 0 0>
      outlineSettings {
       bUseBrushTransparency=false
       roundingType=FixedRadius
       width=1
       color {colorUseRule=UseColor_ColorTable, specifiedColor=<rgba 0 0 0 0>}
       cornerRadii=<4 4 4 4>
      }
      tintColor {colorUseRule=UseColor_Specified, specifiedColor=<rgba 0.0395 0.0395 0.0395 1>}
      uVRegion {bIsValid=false, max=<0 0>, min=<0 0>}
     }
     normalForeground {colorUseRule=UseColor_ColorTable, specifiedColor=<rgba 0 0 0 0>}
     normalPadding=<ltrb 12 1.5 12 1.5>
     pressed {
      bIsDynamicallyLoaded=false
      drawAs=RoundedBox
      imageType=NoImage
      mirroring=NoMirror
      resourceName=None
      resourceObject=null
      tiling=NoTile
      imageSize=<32 32>
      margin=<ltrb 0 0 0 0>
      outlineSettings {
       bUseBrushTransparency=false
       roundingType=FixedRadius
       width=1
       color {colorUseRule=UseColor_ColorTable, specifiedColor=<rgba 0 0 0 0>}
       cornerRadii=<4 4 4 4>
      }
      tintColor {colorUseRule=UseColor_Specified, specifiedColor=<rgba 0.0284 0.0284 0.0284 1>}
      uVRegion {bIsValid=false, max=<0 0>, min=<0 0>}
     }
     pressedForeground {colorUseRule=UseColor_ColorTable, specifiedColor=<rgba 0 0 0 0>}
     pressedPadding=<ltrb 12 2.5 12 0.5>
     pressedSlateSound {resourceObject=null}
    }}
   slot {class=HorizontalBoxSlot, properties {horizontalAlignment=HAlign_Left, verticalAlignment=VAlign_Bottom}}
  }
  SizeBox {
   widgetClassPath=&SizeBox
   parent=RemoveButton
   widget=SizeBox
   properties {bOverride_WidthOverride=true, widthOverride=25}
   slot {class=ButtonSlot, properties {padding=<ltrb 0 0 0 0>}}
  }
  TypeTextBlock_1 {
   widgetClassPath=&TextBlock
   bIsVariable=true
   parent=SizeBox
   widget=TypeTextBlock_1
   properties {
    text=X
    colorAndOpacity {colorUseRule=UseColor_Specified, specifiedColor=<rgba 1 0 0.1541 1>}
    font {
     bForceMonospaced=false
     bMaterialIsStencil=false
     fontMaterial=null
     fontObject=&BrutalType
     letterSpacing=0
     monospacedWidth=1
     size=15
     skewAmount=0
     typefaceFontName=Regular
     outlineSettings {
      bApplyOutlineToDropShadows=false
      bMiteredCorners=false
      bSeparateFillAlpha=false
      outlineMaterial=null
      outlineSize=0
      outlineColor=<rgba 0 0 0 1>
     }
    }
    shadowColorAndOpacity=<rgba 0.2042 0.54 0.8333 1>
    shadowOffset=<0.5 0.5>
   }
   slot {class=SizeBoxSlot, properties {horizontalAlignment=HAlign_Center, verticalAlignment=VAlign_Center}}
  }
  TextEntry {
   widgetClassPath=&TextBlock
   bIsVariable=true
   parent=HorizontalBox_128
   widget=TextEntry
   properties {
    text=Example
    colorAndOpacity {colorUseRule=UseColor_Specified, specifiedColor=<rgba 1 1 1 0.007>}
    font {
     bForceMonospaced=false
     bMaterialIsStencil=false
     fontMaterial=null
     fontObject=&BrutalType
     letterSpacing=0
     monospacedWidth=1
     size=15
     skewAmount=0
     typefaceFontName=Regular
     outlineSettings {
      bApplyOutlineToDropShadows=false
      bMiteredCorners=false
      bSeparateFillAlpha=false
      outlineMaterial=null
      outlineSize=0
      outlineColor=<rgba 0 0 0 1>
     }
    }
    shadowColorAndOpacity=<rgba 0.2042 0.54 0.8333 1>
    shadowOffset=<0.5 0.5>
   }
   slot {class=HorizontalBoxSlot, properties {padding=<ltrb 4 2 4 2>}}
  }
 ]
}
