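# Editor utility widget that lists cosmetic MetaSound action rows and the sound assets assigned to each, for audio tooling.
# GetSoundNameReferences builds the rows from reference structs, and UpdateSoundLists feeds in non-MetaSoundSource objects, auto-assigned by matching names against each row's identifiers.
# Sound widgets move between rows by drag and drop, with duplicate checks and loop-option enforcement.
# Each row's add button spawns a Selected Sounds tab for picking additional sounds.
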EditorUtilityWidgetBlueprint /Game/Sounds/Fort_Audio_Blueprints/Utilities/EditorUtilityWidgets/Metasounds/Cosmetic/EUW_MetaSound_Cosmetic_ActionSoundsList
graphs [
 AddSoundNameToActionWdiget {nodes [
   n1 = AddSoundNameToActionWdiget!() -> (then, ActionWidget:EUW Meta Sound Cosmetic Action Name Widget Object Reference, SoundObject:Object Reference) {is=title K2Node_FunctionEntry}
   > n2 = Check if Action Has Sound!(Check:Exec, ActionWidget:EUW Meta Sound Cosmetic Action Name Widget Object Reference=@n1, SoundObject:Object Reference=@n1.SoundObject) -> (Exists:Exec, DoesNotExist:Exec) {is=title K2Node_MacroInstance}
   n3 = Get ActionWidget() -> (ActionWidget:EUW Meta Sound Cosmetic Action Name Widget Object Reference)
   n4 = Construct Object from Class!(
    execute=@n2.DoesNotExist
    Class=&/EUW_MetaSound_Cosmetic_SoundNameWidget.EUW_MetaSound_Cosmetic_SoundNameWidget_C
    self=@n3
    SoundName:Text="To Text (Object)"!(InObj:Object Reference=Get SoundObject() -> (SoundObject:Object Reference)) -> (ReturnValue:Text)
    CurrentActionWidget:EUW Meta Sound Cosmetic Action Name Widget Object Reference=@n3
    SoundObject:Object Reference=Get SoundObject()
    ParentWidget:EUW Meta Sound Cosmetic Action Sounds List Object Reference=Self-Reference()
   ) -> (then, ReturnValue:EUW Meta Sound Cosmetic Sound Name Widget Object Reference)
   > Bind Event to On Drag Started(self:EUW Meta Sound Cosmetic Sound Name Widget Object Reference=@n4, Delegate=Create Event()) {is=title K2Node_AddDelegate}
   > Bind Event to On Drag Cancel(self:EUW Meta Sound Cosmetic Sound Name Widget Object Reference=@n4, Delegate=Create Event()) {is=title K2Node_AddDelegate}
   > Add Unique(
    TargetArray:Array of EUW Meta Sound Cosmetic Sound Name Widget Object References=Get SoundNameWidgets() -> (SoundNameWidgets:Array of EUW Meta Sound Cosmetic Sound Name Widget Object References)
    NewItem:"EUW Meta Sound Cosmetic Sound Name Widget Object Reference (by ref)"=@n4
   )
   > Add Child Widget!(
    execute
    self
    ActionNameWidget:EUW Meta Sound Cosmetic Action Name Widget Object Reference=Get ActionWidget()
    SoundNameWidget:EUW Meta Sound Cosmetic Sound Name Widget Object Reference=@n4
   ) -> (then)
   > Bind Event to On Removed Widget(self:EUW Meta Sound Cosmetic Sound Name Widget Object Reference=@n4, Delegate=Create Event()) {is=title K2Node_AddDelegate}
  ]}
 AttemptToAutomateObjectWidgetToActionWidgets {nodes [
   AttemptToAutomateObjectWidgetToActionWidgets!() -> (then, SoundNameWidget:EUW Meta Sound Cosmetic Sound Name Widget Object Reference) {is=title K2Node_FunctionEntry}
   > n1 = For Each Loop(Array:Array of EUW Meta Sound Cosmetic Action Name Widget Object References=Get ActionNameWidgets() -> (ActionNameWidgets:Array of EUW Meta Sound Cosmetic Action Name Widget Object References)) -> (Array Element:EUW Meta Sound Cosmetic Action Name Widget Object Reference)
   > n2 = For Each Loop(Array:Array of Strings=Get SameNameIdentifiers!(self:EUW Meta Sound Cosmetic Action Name Widget Object Reference=@n1) -> (SameNameIdentifiers:Array of Strings)) -> (Array Element:String)
   n3 = "To Text (Object)"(InObj=Get SoundObject!(self:EUW Meta Sound Cosmetic Sound Name Widget Object Reference=Get SoundNameWidget() -> (SoundNameWidget:EUW Meta Sound Cosmetic Sound Name Widget Object Reference)) -> (SoundObject))
   > Branch(Condition=IsProperString!(Substring:String=@n2, SearchIn:String="To String (Text)"(InText=@n3)) -> (ReturnValue:bool))
   n4 = Get SoundNameWidget()
   > MoveChildWidget!(
    execute
    self
    PreviousActionWidget:EUW Meta Sound Cosmetic Action Name Widget Object Reference=Get CurrentActionWidget!(self:EUW Meta Sound Cosmetic Sound Name Widget Object Reference=@n4) -> (CurrentActionWidget:EUW Meta Sound Cosmetic Action Name Widget Object Reference)
    NewActionWidget:EUW Meta Sound Cosmetic Action Name Widget Object Reference=@n1
    SoundNameWidget:EUW Meta Sound Cosmetic Sound Name Widget Object Reference=@n4
   ) -> (then)
   > Set CurrentActionWidget!(
    execute
    CurrentActionWidget:EUW Meta Sound Cosmetic Action Name Widget Object Reference=@n1
    self:EUW Meta Sound Cosmetic Sound Name Widget Object Reference=Get SoundNameWidget()
   ) -> (then, Output_Get:EUW Meta Sound Cosmetic Action Name Widget Object Reference)
  ]}
 CheckIfActionHasSound {nodes [
   n1 = Inputs!() -> (Check:Exec, ActionWidget:EUW Meta Sound Cosmetic Action Name Widget Object Reference, SoundObject:Object Reference)
   > Set SoundExist(SoundExist:bool=false) -> (Output_Get:bool)
   > n2 = For Each Loop(Array:Array of Widget Object References=GetAllChildren(self=Get ScrollBox!(self:EUW Meta Sound Cosmetic Action Name Widget Object Reference=@n1) -> (ScrollBox:Editor Utility Scroll Box Object Reference))) -> (Array Element:Widget Object Reference)
   n3 = Branch(execute=@n2.Completed, Condition=Get SoundExist() -> (SoundExist:bool))
   > Outputs!(Exists:Exec, DoesNotExist:Exec=@n3.else)
   n4 = Cast To EUW_MetaSound_Cosmetic_SoundNameWidget!(execute=@n2.LoopBody, Object=@n2) -> (then, CastFailed, AsEUW Meta Sound Cosmetic Sound Name Widget:EUW Meta Sound Cosmetic Sound Name Widget Object Reference) {is=title K2Node_DynamicCast}
   > Branch(Condition="Equal Exactly (Text)"!(
     A:"Text (by ref)"=Get SoundName!(self:EUW Meta Sound Cosmetic Sound Name Widget Object Reference=@n4) -> (SoundName:Text)
     B:"Text (by ref)"="To Text (Object)"(InObj=@n1.SoundObject)
    ) -> (ReturnValue:bool))
   > Set SoundExist(SoundExist=true)
  ]}
 EventGraph {nodes [
   ConfirmAddedSoundsButtonPressed() {is=title K2Node_CustomEvent}
   > n1 = For Each Loop(Array:Array of EUW Meta Sound Cosmetic Selected Names Pop Up Widget Object References=Get ListEntries!(self:EUW Meta Sound Cosmetic Selected Sounds Widget Object Reference=Get PopUpSoundsList() -> (PopUpSoundsList:EUW Meta Sound Cosmetic Selected Sounds Widget Object Reference)) -> (ListEntries:Array of EUW Meta Sound Cosmetic Selected Names Pop Up Widget Object References)) -> (Array Element:EUW Meta Sound Cosmetic Selected Names Pop Up Widget Object Reference)
   n2 = SpawnAndRegisterTabAndGetID!(
    execute=@n4.then
    self:Editor Utility Subsystem Object Reference=Get EditorUtilitySubsystem()
    InBlueprint:Editor Utility Widget Blueprint Object Reference=&/EUW_MetaSound_Cosmetic_SelectedSoundsWidget
   ) -> (then, NewTabID:Name=None, ReturnValue:Editor Utility Widget Object Reference)
   UnregisterTabByID!(execute=@n1.Completed, self:Editor Utility Subsystem Object Reference=Get EditorUtilitySubsystem(), TabID:Name=@n2) -> (then, ReturnValue:bool)
   n3 = Branch(
    execute=@n1.LoopBody
    Condition=Get Selected!(self:EUW Meta Sound Cosmetic Selected Names Pop Up Widget Object Reference=@n1) -> (Selected:bool)
   )
   n4 = AddSoundButton!() -> (OutputDelegate, then, ActionWidget:EUW Meta Sound Cosmetic Action Name Widget Object Reference) {is=title K2Node_CustomEvent}
   AddSoundNameToActionWdiget!(
    execute=@n3.then
    self
    ActionWidget:EUW Meta Sound Cosmetic Action Name Widget Object Reference=@n4.ActionWidget
    SoundObject:Object Reference=Get SoundObject(self:EUW Meta Sound Cosmetic Selected Names Pop Up Widget Object Reference=@n1)
   ) -> (then) {is=title K2Node_CallFunction}
   n5 = OnDragCancel!() -> (
    OutputDelegate
    then
    CurrentActionWidget:EUW Meta Sound Cosmetic Action Name Widget Object Reference
    SoundNameWidget:EUW Meta Sound Cosmetic Sound Name Widget Object Reference
   ) {is=title K2Node_CustomEvent}
   > AddChild(self=Get ScrollBox(self=@n5.CurrentActionWidget), Content=@n5.SoundNameWidget)
   > Set CurrentActionWidget(CurrentActionWidget=@n5.CurrentActionWidget, self=@n5.SoundNameWidget)
   n6 = "GetText (Text)"!(self:Text Object Reference=Get PresetText!(self:EUW Meta Sound Cosmetic Action Name Widget Object Reference=@n5.CurrentActionWidget) -> (PresetText:Text Object Reference)) -> (ReturnValue:Text)
   > PrintString(InString=Append(A="Dropped Complete: ", B="To String (Text)"(InText=@n6)))
   n7 = OnDragDropComplete!() -> (
    OutputDelegate
    then
    DragDropRef:B Meta Sound Cosmeitc Sound Name Drag Drop Object Reference
    ActionNameWdiget:EUW Meta Sound Cosmetic Action Name Widget Object Reference
   ) {is=title K2Node_CustomEvent}
   > Add Child Widget(
    ActionNameWidget=@n7.ActionNameWdiget
    SoundNameWidget=Get SoundNameWidget!(self:B Meta Sound Cosmeitc Sound Name Drag Drop Object Reference=@n7.DragDropRef) -> (SoundNameWidget)
   )
   n8 = "GetText (Text)"(self=Get PresetText(self=@n7.ActionNameWdiget))
   > PrintString(InString=Append(A="Dropped Complete: ", B="To String (Text)"(InText=@n8)))
   n9 = Add Child Widget!() -> (
    OutputDelegate
    then
    ActionNameWidget:EUW Meta Sound Cosmetic Action Name Widget Object Reference
    SoundNameWidget:EUW Meta Sound Cosmetic Sound Name Widget Object Reference
   ) {is=title K2Node_CustomEvent}
   > n10 = Check if Action Has Sound(
    ActionWidget=@n9.ActionNameWidget
    SoundObject=Get SoundObject(self:EUW Meta Sound Cosmetic Sound Name Widget Object Reference=@n9.SoundNameWidget)
   )
   n11 = Set CurrentActionWidget(execute=@n10.DoesNotExist, CurrentActionWidget=@n9.ActionNameWidget, self=@n9.SoundNameWidget)
   > AddChild(self=Get ScrollBox(self=@n11), Content=@n9.SoundNameWidget)
   > PrintString(InString=ADD)
   n12 = Get DiableLoopOption!(self:EUW Meta Sound Cosmetic Action Name Widget Object Reference=@n11) -> (DiableLoopOption:bool)
   > SetIsEnabled(
    self=Get TextButton!(self:EUW Meta Sound Cosmetic Sound Name Widget Object Reference=@n9.SoundNameWidget) -> (TextButton:Editor Utility Button Object Reference)
    bInIsEnabled=NOT Boolean(A=@n12)
   )
   > Branch(Condition=@n12)
   > Update Type Text!(execute, self:EUW Meta Sound Cosmetic Sound Name Widget Object Reference=@n9.SoundNameWidget, ShouldLoop:bool=false) -> (then)
   n13 = OnRemoveWidgetButton!() -> (OutputDelegate, then, SoundNameWidget:EUW Meta Sound Cosmetic Sound Name Widget Object Reference) {is=title K2Node_CustomEvent}
   > Remove Child Widget!(
    execute
    self
    ActionNameWidget:EUW Meta Sound Cosmetic Action Name Widget Object Reference=Get CurrentActionWidget(self=@n13.SoundNameWidget)
    SoundNameWdiget:"EUW Meta Sound Cosmetic Sound Name Widget Object Reference (by ref)"=@n13.SoundNameWidget
   ) -> (then)
   > Set CurrentActionWidget(self=@n13.SoundNameWidget)
   n14 = MoveChildWidget!() -> (
    OutputDelegate
    then
    PreviousActionWidget:EUW Meta Sound Cosmetic Action Name Widget Object Reference
    NewActionWidget:EUW Meta Sound Cosmetic Action Name Widget Object Reference
    SoundNameWidget:EUW Meta Sound Cosmetic Sound Name Widget Object Reference
   ) {is=title K2Node_CustomEvent}
   > Branch(Condition="Not Equal (Object)"(A:Object Reference=@n14.PreviousActionWidget, B:Object Reference=@n14.NewActionWidget) -> (ReturnValue:bool=~) {is=title K2Node_PromotableOperator})
   > RemoveChild(self=Get ScrollBox(self=@n14.PreviousActionWidget), Content=@n14.SoundNameWidget)
   > Add Child Widget(ActionNameWidget=@n14.NewActionWidget, SoundNameWidget=@n14.SoundNameWidget) {is=title K2Node_CallFunction}
   > PrintString(InString=Moved Child)
   n15 = Remove Child Widget!() -> (
    OutputDelegate
    then
    ActionNameWidget:EUW Meta Sound Cosmetic Action Name Widget Object Reference
    SoundNameWdiget:"EUW Meta Sound Cosmetic Sound Name Widget Object Reference (by ref)"
   ) {is=title K2Node_CustomEvent}
   > RemoveChild(self=Get ScrollBox(self=@n15.ActionNameWidget), Content=@n15.SoundNameWdiget)
   > PrintString(InString=Remove)
   ResetActionWidgets() {is=title K2Node_CustomEvent}
   > n16 = Sequence()
   > n17 = For Each Loop(Array:Array of EUW Meta Sound Cosmetic Action Name Widget Object References=Get ActionNameWidgets()) -> (Array Element:EUW Meta Sound Cosmetic Action Name Widget Object Reference)
   Clear(execute=@n17.Completed, TargetArray:Array of EUW Meta Sound Cosmetic Action Name Widget Object References=Get ActionNameWidgets())
   RemoveChild(execute=@n17.LoopBody, self=Get MainBox() -> (MainBox:Editor Utility Scroll Box Object Reference), Content=@n17)
   n18 = For Each Loop(Exec=@n16.then_1, Array:Array of EUW Meta Sound Cosmetic Sound Name Widget Object References=Get SoundNameWidgets()) -> (Array Element:EUW Meta Sound Cosmetic Sound Name Widget Object Reference)
   > MoveChildWidget(
    PreviousActionWidget=Get CurrentActionWidget(self=@n18)
    NewActionWidget="Get (a copy)"(Array:Array of EUW Meta Sound Cosmetic Action Name Widget Object References=Get ActionNameWidgets()) -> (Output:EUW Meta Sound Cosmetic Action Name Widget Object Reference)
    SoundNameWidget=@n18
   ) {is=title K2Node_CallFunction}
   n19 = OnDraggedDetected!() -> (
    OutputDelegate
    then
    ActionNameWidget:EUW Meta Sound Cosmetic Action Name Widget Object Reference
    SoundNameWidget:EUW Meta Sound Cosmetic Sound Name Widget Object Reference
   ) {is=title K2Node_CustomEvent}
   > Remove Child Widget(ActionNameWidget=@n19.ActionNameWidget, SoundNameWdiget=@n19.SoundNameWidget) {is=title K2Node_CallFunction}
   > Set CurrentActionWidget(CurrentActionWidget=@n19.ActionNameWidget, self=@n19.SoundNameWidget)
   n20 = "GetText (Text)"(self=Get PresetText(self=@n19.ActionNameWidget))
   > PrintString(InString=Append(A="Dragging Detected: ", B="To String (Text)"(InText=@n20)))
   n21 = GetSoundNameReferences!() -> (OutputDelegate, then, CosmeticType:String, Names Data Structure:Struct Meta Sound Source Reference Structure, NameOverride:String) {is=title K2Node_CustomEvent}
   n22 = Break Struct Meta Sound Source Reference!(Struct_MetaSoundSourceReference:"Struct Meta Sound Source Reference Structure (by ref)"=@n21.Names Data Structure) -> (
    SkipActionName_27_#1:bool=~
    DiableLoopOption_26_#2:bool=~
    SourceMetaSoundName_22_#3:String
    SourceWavesName_21_#4:Array of Strings
    SameNameIdentifiers_32_#5:Array of Strings
   ) {is=title K2Node_BreakStruct}
   > Set Same Name Identifiers(Same Name Identifiers:Array of Strings=@n22.SameNameIdentifiers_32_#5) -> (Output_Get:Array of Strings)
   > n23 = Branch(Condition=@n22)
   n24 = For Each Loop(Exec=@n23.else, Array:Array of Strings=@n22.SourceWavesName_21_#4) -> (Array Element:String)
   Set CosmeticType(execute=@n24.Completed, CosmeticType:String=@n21.CosmeticType) -> (Output_Get:String)
   > n25 = For Each Loop(Array:Array of EUW Meta Sound Cosmetic Sound Name Widget Object References=Get SoundNameWidgets()) -> (Array Element:EUW Meta Sound Cosmetic Sound Name Widget Object Reference)
   > AttemptToAutomateObjectWidgetToActionWidgets!(execute, self, SoundNameWidget:EUW Meta Sound Cosmetic Sound Name Widget Object Reference=@n25) -> (then) {is=title K2Node_CallFunction}
   SetupWidgetActions!(
    execute=@n24.LoopBody
    self
    ActionName:String=@n24
    SourceMetaSoundName:String=@n22.SourceMetaSoundName_22_#3
    DiableLoopOption:bool=@n22.DiableLoopOption_26_#2
    NameOverride:String=@n21.NameOverride
   ) -> (then)
   n26 = UpdateSoundLists!() -> (OutputDelegate, then, Object:Array of Object References) {is=title K2Node_CustomEvent}
   > n27 = Branch(Condition=Identical!(
     ArrayA:Array of Object References=Get ObjectEntries() -> (ObjectEntries:Array of Object References)
     ArrayB:Array of Object References=@n26.Object
    ) -> (ReturnValue:bool) {is=title K2Node_CallArrayFunction})
   n28 = For Each Loop(Exec=@n27.else, Array=Get SoundNameWidgets())
   Clear(execute=@n28.Completed, TargetArray:Array of EUW Meta Sound Cosmetic Sound Name Widget Object References=Get SoundNameWidgets())
   > Clear(TargetArray:Array of Object References=Get ObjectEntries())
   > n29 = For Each Loop(Array:Array of Object References=@n26.Object) -> (Array Element:Object Reference)
   Set ObjectEntries(execute=@n29.Completed, ObjectEntries:Array of Object References=@n26.Object) -> (Output_Get:Array of Object References)
   > n30 = Get!(execute:Exec) -> (then:Exec, else:Exec, PopUpSoundsList:EUW Meta Sound Cosmetic Selected Sounds Widget Object Reference)
   > Set ObjectEntries!(execute, ObjectEntries=Get ObjectEntries(), self:EUW Meta Sound Cosmetic Selected Sounds Widget Object Reference=@n30) -> (then, Output_Get)
   > CreateSelectedWidgetObject(self:EUW Meta Sound Cosmetic Selected Sounds Widget Object Reference=@n30)
   n31 = Cast To MetaSoundSource!(execute=@n29.LoopBody, Object=@n29) -> (then, CastFailed, AsMetaSound Source:Meta Sound Source Object Reference) {is=title K2Node_DynamicCast}
   TryToAddObjectsToActions!(execute=@n31.CastFailed, self, SoundObject:Object Reference=@n29) -> (then)
   Remove Child Widget(execute=@n28.LoopBody, ActionNameWidget=Get CurrentActionWidget(self=@n28), SoundNameWdiget=@n28)
   > n32 = Get()
   > EmptyList(self:EUW Meta Sound Cosmetic Selected Sounds Widget Object Reference=@n32)
   n33 = Get PopUpSoundsList()
   EmptyList(execute=@n35.LoopBody, self=@n33)
   CreateSelectedWidgetObject(execute=@n36.then, self=@n33)
   n34 = Cast To EUW_MetaSound_Cosmetic_SelectedSoundsWidget!(execute=@n2.then, Object=@n2.ReturnValue) -> (then, CastFailed, AsEUW Meta Sound Cosmetic Selected Sounds Widget:EUW Meta Sound Cosmetic Selected Sounds Widget Object Reference) {is=title K2Node_DynamicCast}
   Assign On Pressed!(
    execute:Exec=@n41.then
    self:Button Object Reference=Get AddSoundsButton!(self:EUW Meta Sound Cosmetic Selected Sounds Widget Object Reference=@n34) -> (AddSoundsButton:Editor Utility Button Object Reference)
    Delegate:"Delegate (by ref)"=Create Event()
   ) -> (then:Exec) {is=title K2Node_AssignDelegate}
   > n35 = For Each Loop(Array:Array of EUW Meta Sound Cosmetic Sound Name Widget Object References=Get SoundNameWidgets()) -> (Array Element:EUW Meta Sound Cosmetic Sound Name Widget Object Reference)
   n36 = Set ObjectEntries(execute=@n35.Completed, ObjectEntries=Get ObjectEntries(), self=@n33)
   n37 = Set PopUpSoundsList(execute=@n34.then, PopUpSoundsList:EUW Meta Sound Cosmetic Selected Sounds Widget Object Reference=@n34) -> (Output_Get:EUW Meta Sound Cosmetic Selected Sounds Widget Object Reference)
   n38 = Get PresetText(self:EUW Meta Sound Cosmetic Selected Sounds Widget Object Reference=@n37)
   n39 = Get PresetText(self:EUW Meta Sound Cosmetic Action Name Widget Object Reference=@n4.ActionWidget)
   n40 = Get Text(self:Text Object Reference=@n39)
   > n41 = Set Text(Text=@n40, self=@n38)
  ]}
 IsProperString {nodes [
   n1 = IsProperString!() -> (then, Substring:String, SearchIn:String) {is=title K2Node_FunctionEntry}
   > Return Node!(execute, ReturnValue:bool=Contains(SearchIn=@n1.SearchIn, Substring=@n1))
   Contains(SearchIn=@n1.SearchIn, Substring=Get CosmeticType() -> (CosmeticType:String))
   AND Boolean(A=false, B=true)
  ]}
 SetOutputText {nodes [
   n1 = SetOutputText!() -> (then, Text:Text) {is=title K2Node_FunctionEntry}
   > Set Text(Text=@n1, self=Get OutputText() -> (OutputText:Text Object Reference))
  ]}
 SetupWidgetActions {nodes [
   n1 = SetupWidgetActions!() -> (then, ActionName:String, SourceMetaSoundName:String, DiableLoopOption:bool=~, NameOverride:String) {is=title K2Node_FunctionEntry}
   > n2 = Construct Object from Class!(
    execute
    Class=&/EUW_MetaSound_Cosmetic_ActionNameWidget.EUW_MetaSound_Cosmetic_ActionNameWidget_C
    self=Get MainBox()
    SourceMetaSoundName:String=@n1.SourceMetaSoundName
    DiableLoopOption:bool=@n1.DiableLoopOption
    SameNameIdentifiers:Array of Strings=Get Same Name Identifiers() -> (Same Name Identifiers:Array of Strings)
   ) -> (then, ReturnValue:EUW Meta Sound Cosmetic Action Name Widget Object Reference)
   > Add Unique(
    TargetArray:Array of EUW Meta Sound Cosmetic Action Name Widget Object References=Get ActionNameWidgets()
    NewItem:"EUW Meta Sound Cosmetic Action Name Widget Object Reference (by ref)"=@n2
   )
   > AddChild(self=Get MainBox(), Content=@n2)
   n3 = "Equal Exactly (String)"(A=@n1.NameOverride)
   n4 = Get ActionName() -> (ActionName:String)
   n5 = Append!(A=@n4, B=" - ", C:String=@n1.NameOverride) -> (ReturnValue)
   > "SetText (Text)"(self=Get PresetText(self=@n2), InText="To Text (String)"(InString=SelectString(A=@n4, B=@n5, bPickA=@n3)))
   > Set BaseMetasoundName!(execute, BaseMetasoundName:String=@n4, self:EUW Meta Sound Cosmetic Action Name Widget Object Reference=@n2) -> (then, Output_Get:String)
   > Add Unique(TargetArray:Array of Strings=Get SameNameIdentifiers(self=@n2), NewItem:"String (by ref)"=Get ActionName())
   > Set OverrideName!(
    execute
    OverrideName:String=Get NameOverride() -> (NameOverride:String)
    self:EUW Meta Sound Cosmetic Action Name Widget Object Reference=@n2
   ) -> (then, Output_Get:String)
   > SetAlwaysShowScrollbar!(execute, self:Scroll Box Object Reference=Get ScrollBox(self=@n2), NewAlwaysShowScrollbar:bool=true) -> (then)
   > Bind Event to On Drag Drop Complete(self:EUW Meta Sound Cosmetic Action Name Widget Object Reference=@n2, Delegate=Create Event()) {is=title K2Node_AddDelegate}
   > n6 = Branch(Condition="Equal Exactly (String)"(A=Get ActionName(), B="Selected (Undefined)"))
   > SetVisibility(self=Get AddSoundsButton(self:EUW Meta Sound Cosmetic Action Name Widget Object Reference=@n2), InVisibility=Collapsed)
   > SetVisibility(self=@n2, InVisibility=Collapsed)
   Bind Event to On Add Sound Button Pressed(execute=@n6.else, self:EUW Meta Sound Cosmetic Action Name Widget Object Reference=@n2, Delegate=Create Event()) {is=title K2Node_AddDelegate}
  ]}
 TryToAddObjectsToActions {nodes [
   TryToAddObjectsToActions!() -> (then, SoundObject:Object Reference) {is=title K2Node_FunctionEntry}
   > n1 = For Each Loop(Array:Array of EUW Meta Sound Cosmetic Action Name Widget Object References=Get ActionNameWidgets()) -> (Array Element:EUW Meta Sound Cosmetic Action Name Widget Object Reference)
   n2 = Get!(execute=@n1.Completed) -> (then, else, ActionWidget:EUW Meta Sound Cosmetic Action Name Widget Object Reference)
   > AddSoundNameToActionWdiget(ActionWidget=@n2, SoundObject=Get SoundObject!() -> (SoundObject))
   n3 = For Each Loop(Exec=@n1.LoopBody, Array:Array of Strings=Get SameNameIdentifiers(self=@n1)) -> (Array Element:String)
   n4 = "To Text (Object)"(InObj=Get SoundObject())
   > Branch(Condition=IsProperString(Substring=@n3, SearchIn="To String (Text)"(InText=@n4)) {is=title K2Node_CallFunction})
   > Set ActionWidget(ActionWidget:EUW Meta Sound Cosmetic Action Name Widget Object Reference=@n1) -> (Output_Get:EUW Meta Sound Cosmetic Action Name Widget Object Reference)
  ]}
]
variables [
 ActionNameWidgets {type="array<EUW_MetaSound_Cosmetic_ActionNameWidget_C>", default []}
 CosmeticType {type=string, default=""}
 ObjectEntries {type="array<Object>", default []}
 PopUpSoundsList {type=EUW_MetaSound_Cosmetic_SelectedSoundsWidget_C, default=null}
 Same Name Identifiers {type="array<string>", default []}
 SoundExist {type=boolean, default=true}
 SoundNameWidgets {type="array<EUW_MetaSound_Cosmetic_SoundNameWidget_C>", default []}
]
widgets {
 info {rootWidgetClass=&CanvasPanel, widgetCount=8}
 widgets [
  CanvasPanel_0 {widgetClassPath=&CanvasPanel, widget=CanvasPanel_0}
  VerticalBox_59 {
   widgetClassPath=&VerticalBox
   parent=CanvasPanel_0
   widget=VerticalBox_59
   slot {class=CanvasPanelSlot, properties {layoutData {alignment=<0 0>, anchors {maximum=<0 0>, minimum=<0 0>}, offsets=<ltrb 0 0 600 600>}}}
  }
  Border_193 {
   widgetClassPath=&Border
   parent=VerticalBox_59
   widget=Border_193
   properties {
    horizontalAlignment=HAlign_Center
    background {
     bIsDynamicallyLoaded=false
     drawAs=RoundedBox
     imageType=NoImage
     mirroring=NoMirror
     resourceName=None
     resourceObject=null
     tiling=NoTile
     imageSize=<32 32>
     margin=<ltrb 0 0 0 0>
     outlineSettings {
      bUseBrushTransparency=false
      roundingType=FixedRadius
      width=1
      color {colorUseRule=UseColor_Specified, specifiedColor=<rgba 0.1198 0.1198 0.1198 0.5>}
      cornerRadii=<0 0 0 0>
     }
     tintColor {colorUseRule=UseColor_Specified, specifiedColor=<rgba 0.0423 0.0423 0.0423 1>}
     uVRegion {bIsValid=false, max=<0 0>, min=<0 0>}
    }
    padding=<ltrb 4 4 4 4>
   }
   slot {class=VerticalBoxSlot}
  }
  SoundListText {
   widgetClassPath=&TextBlock
   parent=Border_193
   widget=SoundListText
   properties {
    text=Cosmetic Action Sounds List
    font {
     bForceMonospaced=false
     bMaterialIsStencil=false
     fontMaterial=null
     fontObject=/LauncherSocial/BrutalType
     letterSpacing=0
     monospacedWidth=1
     size=15
     skewAmount=0
     typefaceFontName=Regular
     outlineSettings {
      bApplyOutlineToDropShadows=false
      bMiteredCorners=false
      bSeparateFillAlpha=false
      outlineMaterial=null
      outlineSize=0
      outlineColor=<rgba 0 0 0 1>
     }
    }
    shadowColorAndOpacity=<rgba 0 0 0 1>
   }
   slot {class=BorderSlot, properties {horizontalAlignment=HAlign_Center, padding=<ltrb 4 4 4 4>}}
  }
  SizeBox_0 {
   widgetClassPath=&SizeBox
   parent=VerticalBox_59
   widget=SizeBox_0
   properties {bOverride_HeightOverride=true, heightOverride=500}
   slot {class=VerticalBoxSlot, properties {size {sizeRule=Fill, value=1}}}
  }
  MainBox {
   widgetClassPath=&EditorUtilityScrollBox
   bIsVariable=true
   parent=SizeBox_0
   widget=MainBox
   properties {alwaysShowScrollbar=true}
   slot {class=SizeBoxSlot}
  }
  EditorUtilityScrollBox_103 {
   widgetClassPath=&EditorUtilityScrollBox
   parent=VerticalBox_59
   widget=EditorUtilityScrollBox_103
   properties {widgetBarStyle {
     thickness=8
     draggedThumbImage {
      bIsDynamicallyLoaded=false
      drawAs=RoundedBox
      imageType=NoImage
      mirroring=NoMirror
      resourceName=None
      resourceObject=null
      tiling=NoTile
      imageSize=<0 0>
      margin=<ltrb 0 0 0 0>
      outlineSettings {
       bUseBrushTransparency=false
       roundingType=FixedRadius
       width=0
       color {colorUseRule=UseColor_Specified, specifiedColor=<rgba 0 0 0 0>}
       cornerRadii=<4 4 4 4>
      }
      tintColor {colorUseRule=UseColor_Specified, specifiedColor=<rgba 0.2159 0.2159 0.2159 1>}
      uVRegion {bIsValid=false, max=<0 0>, min=<0 0>}
     }
     horizontalBackgroundImage {
      bIsDynamicallyLoaded=false
      drawAs=NoDrawType
      imageType=NoImage
      mirroring=NoMirror
      resourceName=None
      resourceObject=null
      tiling=NoTile
      imageSize=<0 0>
      margin=<ltrb 0 0 0 0>
      outlineSettings {
       bUseBrushTransparency=false
       roundingType=HalfHeightRadius
       width=0
       color {colorUseRule=UseColor_Specified, specifiedColor=<rgba 0 0 0 0>}
       cornerRadii=<0 0 0 0>
      }
      tintColor {colorUseRule=UseColor_Specified, specifiedColor=<rgba 1 1 1 1>}
      uVRegion {bIsValid=false, max=<0 0>, min=<0 0>}
     }
     horizontalBottomSlotImage {
      bIsDynamicallyLoaded=false
      drawAs=NoDrawType
      imageType=NoImage
      mirroring=NoMirror
      resourceName=None
      resourceObject=null
      tiling=NoTile
      imageSize=<0 0>
      margin=<ltrb 0 0 0 0>
      outlineSettings {
       bUseBrushTransparency=false
       roundingType=HalfHeightRadius
       width=0
       color {colorUseRule=UseColor_Specified, specifiedColor=<rgba 0 0 0 0>}
       cornerRadii=<0 0 0 0>
      }
      tintColor {colorUseRule=UseColor_Specified, specifiedColor=<rgba 1 1 1 1>}
      uVRegion {bIsValid=false, max=<0 0>, min=<0 0>}
     }
     horizontalTopSlotImage {
      bIsDynamicallyLoaded=false
      drawAs=NoDrawType
      imageType=NoImage
      mirroring=NoMirror
      resourceName=None
      resourceObject=null
      tiling=NoTile
      imageSize=<0 0>
      margin=<ltrb 0 0 0 0>
      outlineSettings {
       bUseBrushTransparency=false
       roundingType=HalfHeightRadius
       width=0
       color {colorUseRule=UseColor_Specified, specifiedColor=<rgba 0 0 0 0>}
       cornerRadii=<0 0 0 0>
      }
      tintColor {colorUseRule=UseColor_Specified, specifiedColor=<rgba 1 1 1 1>}
      uVRegion {bIsValid=false, max=<0 0>, min=<0 0>}
     }
     hoveredThumbImage {
      bIsDynamicallyLoaded=false
      drawAs=RoundedBox
      imageType=NoImage
      mirroring=NoMirror
      resourceName=None
      resourceObject=null
      tiling=NoTile
      imageSize=<0 0>
      margin=<ltrb 0 0 0 0>
      outlineSettings {
       bUseBrushTransparency=false
       roundingType=FixedRadius
       width=0
       color {colorUseRule=UseColor_Specified, specifiedColor=<rgba 0 0 0 0>}
       cornerRadii=<4 4 4 4>
      }
      tintColor {colorUseRule=UseColor_Specified, specifiedColor=<rgba 0.2159 0.2159 0.2159 1>}
      uVRegion {bIsValid=false, max=<0 0>, min=<0 0>}
     }
     normalThumbImage {
      bIsDynamicallyLoaded=false
      drawAs=RoundedBox
      imageType=NoImage
      mirroring=NoMirror
      resourceName=None
      resourceObject=null
      tiling=NoTile
      imageSize=<0 0>
      margin=<ltrb 0 0 0 0>
      outlineSettings {
       bUseBrushTransparency=false
       roundingType=FixedRadius
       width=0
       color {colorUseRule=UseColor_Specified, specifiedColor=<rgba 0 0 0 0>}
       cornerRadii=<4 4 4 4>
      }
      tintColor {colorUseRule=UseColor_Specified, specifiedColor=<rgba 0.0953 0.0953 0.0953 0>}
      uVRegion {bIsValid=false, max=<0 0>, min=<0 0>}
     }
     verticalBackgroundImage {
      bIsDynamicallyLoaded=false
      drawAs=NoDrawType
      imageType=NoImage
      mirroring=NoMirror
      resourceName=None
      resourceObject=null
      tiling=NoTile
      imageSize=<0 0>
      margin=<ltrb 0 0 0 0>
      outlineSettings {
       bUseBrushTransparency=false
       roundingType=HalfHeightRadius
       width=0
       color {colorUseRule=UseColor_Specified, specifiedColor=<rgba 0 0 0 0>}
       cornerRadii=<0 0 0 0>
      }
      tintColor {colorUseRule=UseColor_Specified, specifiedColor=<rgba 1 1 1 1>}
      uVRegion {bIsValid=false, max=<0 0>, min=<0 0>}
     }
     verticalBottomSlotImage {
      bIsDynamicallyLoaded=false
      drawAs=NoDrawType
      imageType=NoImage
      mirroring=NoMirror
      resourceName=None
      resourceObject=null
      tiling=NoTile
      imageSize=<0 0>
      margin=<ltrb 0 0 0 0>
      outlineSettings {
       bUseBrushTransparency=false
       roundingType=HalfHeightRadius
       width=0
       color {colorUseRule=UseColor_Specified, specifiedColor=<rgba 0 0 0 0>}
       cornerRadii=<0 0 0 0>
      }
      tintColor {colorUseRule=UseColor_Specified, specifiedColor=<rgba 1 1 1 1>}
      uVRegion {bIsValid=false, max=<0 0>, min=<0 0>}
     }
     verticalTopSlotImage {
      bIsDynamicallyLoaded=false
      drawAs=NoDrawType
      imageType=NoImage
      mirroring=NoMirror
      resourceName=None
      resourceObject=null
      tiling=NoTile
      imageSize=<0 0>
      margin=<ltrb 0 0 0 0>
      outlineSettings {
       bUseBrushTransparency=false
       roundingType=HalfHeightRadius
       width=0
       color {colorUseRule=UseColor_Specified, specifiedColor=<rgba 0 0 0 0>}
       cornerRadii=<0 0 0 0>
      }
      tintColor {colorUseRule=UseColor_Specified, specifiedColor=<rgba 1 1 1 1>}
      uVRegion {bIsValid=false, max=<0 0>, min=<0 0>}
     }
    }}
   slot {class=VerticalBoxSlot, properties {padding=<ltrb 2 2 2 2>}}
  }
  OutputText {
   widgetClassPath=&TextBlock
   bIsVariable=true
   parent=EditorUtilityScrollBox_103
   widget=OutputText
   properties {
    visibility=SelfHitTestInvisible
    font {
     bForceMonospaced=false
     bMaterialIsStencil=false
     fontMaterial=null
     fontObject=&BrutalType
     letterSpacing=0
     monospacedWidth=1
     size=15
     skewAmount=0
     typefaceFontName=Regular
     outlineSettings {
      bApplyOutlineToDropShadows=false
      bMiteredCorners=false
      bSeparateFillAlpha=false
      outlineMaterial=null
      outlineSize=0
      outlineColor=<rgba 0 0 0 1>
     }
    }
    shadowColorAndOpacity=<rgba 0 0 0 1>
    shadowOffset=<2 1>
   }
   slot {class=ScrollBoxSlot}
  }
 ]
}
