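# Editor utility tile for a cosmetic MetaSound: a button that shows the sound and opens it for editing.
# PreConstruct points a non-interactive SinglePropertyView at the widget itself to display its MsSource property.
# Pressing the button loads the asset at AudPath and opens it through AssetEditorSubsystem, showing an error dialog if opening fails.
# SetAnimText returns the asset name as text, taken after the last dot in AudPath.
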EditorUtilityWidgetBlueprint /Game/Sounds/Fort_Audio_Blueprints/Utilities/EditorUtilityWidgets/Metasounds/Cosmetic/EUW_MetaSound_Cosmetic_MsPreview
graphs [
 EventGraph {nodes [
   "On Pressed (AudButton)"() {is=title K2Node_ComponentBoundEvent}
   > n1 = LoadAsset(AssetPath=Get AudPath() -> (AudPath:String))
   > n2 = OpenEditorForAssets!(
    execute
    self:Asset Editor Subsystem Object Reference=Get AssetEditorSubsystem()
    Assets:Array of Object References=Make Array("[0]":Object Reference=@n1) -> (Array:Array of Object References)
    OpenedMethod:EAssetTypeActivationOpenedMethod Enum=Edit
   ) -> (then, ReturnValue:bool)
   > n3 = Branch(Condition=@n2)
   Show Message Dialog(
    execute=@n3.else
    Title=MakeLiteralText(Value="LOCTEXT(\"Failed To Open\")")
    Message=MakeLiteralText(Value="LOCTEXT(\"Asset failed to load and open. Nagivating to asset in content browser and load manually.\")")
    MessageType=Ok
    DefaultValue=No
    MessageCategory=Error
   )
   Event PreConstruct()
   > SetObject(self=Get SoundView() -> (SoundView:Single Property View Object Reference), NewObject=Self-Reference())
  ]}
 SetAnimText {nodes [
   SetAnimText() {is=title K2Node_FunctionEntry}
   n1 = Get AudPath()
   > Return Node!(execute, ReturnValue:Text="To Text (String)"(InString=Split(SourceString=@n1, InStr=., SearchDir=FromEnd).RightS))
  ]}
]
variables [
 AudPath {type=string, default=None}
 MsSource {type=MetaSoundSource, default=/FortMetasound/Cosmetics/Pickaxe/MSS_Pickaxe_Impact, class=~, properties=~}
]
widgets {
 info {rootWidgetClass=&Overlay, widgetCount=10}
 widgets [
  Overlay_141 {widgetClassPath=&Overlay, widget=Overlay_141}
  Border_1 {
   widgetClassPath=&Border
   parent=Overlay_141
   widget=Border_1
   properties {
    horizontalAlignment=HAlign_Left
    verticalAlignment=VAlign_Center
    background {
     bIsDynamicallyLoaded=false
     drawAs=Image
     imageType=NoImage
     mirroring=NoMirror
     resourceName=None
     resourceObject=null
     tiling=NoTile
     imageSize=<32 32>
     margin=<ltrb 0 0 0 0>
     outlineSettings {
      bUseBrushTransparency=false
      roundingType=HalfHeightRadius
      width=0
      color {colorUseRule=UseColor_Specified, specifiedColor=<rgba 0 0 0 0>}
      cornerRadii=<0 0 0 0>
     }
     tintColor {colorUseRule=UseColor_Specified, specifiedColor=<rgba 0.0208 0.0208 0.0208 1>}
     uVRegion {bIsValid=false, max=<0 0>, min=<0 0>}
    }
    padding=<ltrb 0 0 0 0>
   }
   slot {class=OverlaySlot, properties {horizontalAlignment=HAlign_Fill, verticalAlignment=VAlign_Fill}}
  }
  Border_0 {
   widgetClassPath=&Border
   parent=Border_1
   widget=Border_0
   properties {
    horizontalAlignment=HAlign_Left
    background {
     bIsDynamicallyLoaded=false
     drawAs=Border
     imageType=NoImage
     mirroring=NoMirror
     resourceName=None
     resourceObject=null
     tiling=NoTile
     imageSize=<32 32>
     margin=<ltrb 1 1 1 1>
     outlineSettings {
      bUseBrushTransparency=false
      roundingType=HalfHeightRadius
      width=0
      color {colorUseRule=UseColor_Specified, specifiedColor=<rgba 0 0 0 0>}
      cornerRadii=<0 0 0 0>
     }
     tintColor {colorUseRule=UseColor_Specified, specifiedColor=<rgba 0 0 0 1>}
     uVRegion {bIsValid=false, max=<0 0>, min=<0 0>}
    }
   }
   slot {class=BorderSlot, properties {horizontalAlignment=HAlign_Left, verticalAlignment=VAlign_Center, padding=<ltrb 0 0 0 0>}}
  }
  AudButton {
   widgetClassPath=&EditorUtilityButton
   parent=Border_0
   widget=AudButton
   slot {class=BorderSlot, properties {horizontalAlignment=HAlign_Left}}
  }
  SizeBox_2 {
   widgetClassPath=&SizeBox
   parent=AudButton
   widget=SizeBox_2
   properties {bOverride_WidthOverride=true, heightOverride=1500, widthOverride=500}
   slot {class=ButtonSlot, properties {horizontalAlignment=HAlign_Fill, verticalAlignment=VAlign_Fill, padding=<ltrb 0 0 0 0>}}
  }
  VerticalBox_0 {
   widgetClassPath=&VerticalBox
   parent=SizeBox_2
   widget=VerticalBox_0
   slot {class=SizeBoxSlot, properties {horizontalAlignment=HAlign_Center, verticalAlignment=VAlign_Center, padding=<ltrb 4 2 4 2>}}
  }
  SizeBox_1 {
   widgetClassPath=&SizeBox
   parent=VerticalBox_0
   widget=SizeBox_1
   properties {bOverride_WidthOverride=true, widthOverride=96.1096}
   slot {class=VerticalBoxSlot, properties {horizontalAlignment=HAlign_Center, verticalAlignment=VAlign_Center}}
  }
  SoundView {
   widgetClassPath=&SinglePropertyView
   parent=SizeBox_1
   widget=SoundView
   properties {nameOverride=" ", propertyName=MsSource, visibility=HitTestInvisible}
   slot {class=SizeBoxSlot}
  }
  SizeBox_3 {
   widgetClassPath=&SizeBox
   parent=VerticalBox_0
   widget=SizeBox_3
   slot {class=VerticalBoxSlot, properties {verticalAlignment=VAlign_Center}}
  }
  TextEntry {
   widgetClassPath=&TextBlock
   bIsVariable=true
   parent=SizeBox_3
   widget=TextEntry
   properties {
    autoWrapText=true
    text=Anim
    wrapTextAt=100
    wrappingPolicy=AllowPerCharacterWrapping
    colorAndOpacity {colorUseRule=UseColor_Specified, specifiedColor=<rgba 0.5885 0.5885 1 1>}
    font {
     bForceMonospaced=false
     bMaterialIsStencil=false
     fontMaterial=null
     fontObject=/LauncherSocial/BrutalType
     letterSpacing=0
     monospacedWidth=1
     size=6.75
     skewAmount=0
     typefaceFontName=Regular
     outlineSettings {
      bApplyOutlineToDropShadows=false
      bMiteredCorners=false
      bSeparateFillAlpha=false
      outlineMaterial=null
      outlineSize=0
      outlineColor=<rgba 0 0 0 1>
     }
    }
    shadowColorAndOpacity=<rgba 0 0 0 1>
    shadowOffset=<0.5 0.5>
   }
   slot {class=SizeBoxSlot}
  }
 ]
}
