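# Editor utility widget in the MetaSound audio tools for picking the FortItemDefinition that audio is to be implemented into.
# On PreConstruct, a SinglePropertyView binds to the widget itself, exposing its IDSourceReference variable as an "Item Def" field.
# Edits to that field fire the OnSourcePropertyChanged dispatcher, with an embedded EUW_IDSaveAsset widget laid out below it.
# FnAnim, JunoAnim and SaveAsset variables are declared but unused in this graph.
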EditorUtilityWidgetBlueprint /Game/Sounds/Fort_Audio_Blueprints/Utilities/EditorUtilityWidgets/Metasounds/Tools/EUW_IDReference
graphs [
 EventGraph {nodes [
   "On Property Changed (IDSource)"!() -> (OutputDelegate, then, PropertyName:Name=None) {is=title K2Node_ComponentBoundEvent}
   > Call On Source Property Changed!(execute:Exec, self:EUW IDReference Object Reference) -> (then:Exec) {is=title K2Node_CallDelegate}
   Event PreConstruct()
   > SetObject(self=Get IDSource() -> (IDSource:Single Property View Object Reference), NewObject=Self-Reference())
  ]}
 OnSourcePropertyChanged {nodes [OnSourcePropertyChanged() {is=title K2Node_FunctionEntry}]}
]
variables [
 FnAnim {type=AthenaDanceItemDefinition, default=null}
 IDSourceReference {type=FortItemDefinition, default=null}
 JunoAnim {type=JunoAthenaDanceItemOverrideDefinition, default=null}
 OnSourcePropertyChanged {default=()}
 SaveAsset {type=boolean}
]
widgets {
 info {rootWidgetClass=&VerticalBox, widgetCount=6}
 widgets [
  VerticalBox_0 {widgetClassPath=&VerticalBox, widget=VerticalBox_0}
  CanvasPanel_0 {widgetClassPath=&CanvasPanel, parent=VerticalBox_0, widget=CanvasPanel_0, slot {class=VerticalBoxSlot}}
  GridPanel_0 {
   widgetClassPath=&GridPanel
   parent=CanvasPanel_0
   widget=GridPanel_0
   slot {
    class=CanvasPanelSlot
    properties {bAutoSize=true, layoutData {alignment=<0 0>, anchors {maximum=<0 0>, minimum=<0 0>}, offsets=<ltrb 0 0 500 500>}}
   }
  }
  SizeBox_0 {
   widgetClassPath=&SizeBox
   bIsVariable=true
   parent=GridPanel_0
   widget=SizeBox_0
   properties {widthOverride=400}
   slot {class=GridSlot}
  }
  IDSource {
   widgetClassPath=&SinglePropertyView
   parent=SizeBox_0
   widget=IDSource
   properties {
    nameOverride=Item Def
    propertyName=IDSourceReference
    toolTipText=Item definition that can have audio implemented into corresponding options.
   }
   slot {class=SizeBoxSlot, properties {padding=<ltrb 2 2 2 2>}}
  }
  SaveAssetRef {
   widgetClassPath=&/EUW_IDSaveAsset
   bIsVariable=true
   parent=GridPanel_0
   widget=SaveAssetRef
   slot {class=GridSlot, properties {row=1, padding=<ltrb 0 5 5 5>}}
  }
 ]
}
