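# Editor utility widget for audio tooling that picks the MetaSound Source used as the base for creating new presets.
# A property view bound to its own SourceReference variable gets a struct-driven label, while Set Base and design-time PreConstruct restore DefaultMSS when valid.
# On Construct, it builds AssetPrefix from the first two underscore-separated tokens of the source's name.
# SetSounds relays valid sound references, plus the widget itself, through the SoundForAsset dispatcher.
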EditorUtilityWidgetBlueprint /Game/Sounds/Fort_Audio_Blueprints/Utilities/EditorUtilityWidgets/Metasounds/Tools/EUW_MetaSoundSourceReference
graphs [
 EventGraph {nodes [
   "On Pressed (SetBaseButton)"() {is=title K2Node_ComponentBoundEvent}
   > n1 = Get!(execute:Exec) -> (then:Exec, else:Exec, DefaultMSS:Meta Sound Source Object Reference)
   > Set SourceReference(SourceReference:Meta Sound Source Object Reference=@n1) -> (Output_Get:Meta Sound Source Object Reference)
   n2 = SetSounds!() -> (OutputDelegate, then, Sound:Sound Base Soft Object Reference) {is=title K2Node_CustomEvent}
   > Branch(Condition=IsValidSoftObjectReference(SoftObjectReference=@n2.Sound))
   > Call Sound for Asset!(
    execute:Exec
    self:EUW Meta Sound Source Reference Object Reference
    Sound:Sound Base Soft Object Reference=@n2.Sound
    SourceReference:EUW Meta Sound Source Reference Object Reference=Self-Reference()
   ) -> (then:Exec) {is=title K2Node_CallDelegate}
   UpdateAssetPrefix() {is=title K2Node_CustomEvent}
   > n3 = Get!(execute) -> (then, else, SourceReference:Meta Sound Source Object Reference)
   n4 = ParseIntoArray(SourceString=GetDisplayName(Object=@n3), Delimiter=_)
   > Set AssetPrefix(AssetPrefix:String=Append!(A="Get (a copy)"(Array=@n4), B=_, C:String="Get (a copy)"(Array=@n4, Dimension 1=1), D:String=_) -> (ReturnValue)) -> (Output_Get:String)
   n5 = Event PreConstruct()
   > SetObject(self=Get MetaSoundSource() -> (MetaSoundSource:Single Property View Object Reference), NewObject=Self-Reference())
   n6 = Break Struct Meta Sound Source Reference!(Struct_MetaSoundSourceReference:"Struct Meta Sound Source Reference Structure (by ref)"=Get NamesDataStructure() -> (NamesDataStructure:Struct Meta Sound Source Reference Structure)) -> (
    SkipActionName_27_#1:bool=~
    DiableLoopOption_26_#2:bool=~
    SourceMetaSoundName_22_#3:String
    SourceWavesName_21_#4:Array of Strings
    SameNameIdentifiers_32_#5:Array of Strings
   ) {is=title K2Node_BreakStruct}
   > SetNameOverride!(
    execute
    self:Single Property View Object Reference=Get MetaSoundSource()
    NewPropertyName:Text="To Text (String)"(InString=Append!(A=Source Reference, B="\r\n", C=@n6.SourceMetaSoundName_22_#3) -> (ReturnValue))
   ) -> (then)
   > Branch(Condition=@n5.IsDesignTime)
   > n7 = Get!(execute) -> (then, else, DefaultMSS:Meta Sound Source Object Reference)
   > Set SourceReference(SourceReference=@n7)
   Event Construct()
   > UpdateAssetPrefix() {is=title K2Node_CallFunction}
  ]}
 SoundForAsset {nodes [SoundForAsset!() -> (then, Sound:Sound Base Soft Object Reference, SourceReference:EUW Meta Sound Source Reference Object Reference) {is=title K2Node_FunctionEntry}]}
]
variables [
 AssetPrefix {type=string, default=""}
 AssetSuffix {type=string, default=""}
 DefaultMSS {type=MetaSoundSource, default=null, class=~, properties=~}
 IsArray {type=boolean, default=true}
 LoopParameterID {type=string, default=.Loop.Sounds}
 NameOverride {type=string, default=""}
 NamesDataStructure {
  type=Struct_MetaSoundSourceReference
  default {diable Loop Option=false, skipActionName=false, source MetaSound Name="", same Name Identifiers [], source Waves Name []}
 }
 NonLoopParameterID {type=string, default=.Sounds}
 SoundForAsset {default=()}
 SourceReference {type=MetaSoundSource, default=null, class=~, properties=~}
]
widgets {
 info {rootWidgetClass=&CanvasPanel, widgetCount=5}
 widgets [
  CanvasPanel_0 {widgetClassPath=&CanvasPanel, widget=CanvasPanel_0}
  HorizontalBox_114 {
   widgetClassPath=&HorizontalBox
   parent=CanvasPanel_0
   widget=HorizontalBox_114
   slot {class=CanvasPanelSlot, properties {layoutData {alignment=<0 0>, anchors {maximum=<0 0>, minimum=<0 0>}, offsets=<ltrb 0 0 500 70>}}}
  }
  MetaSoundSource {
   widgetClassPath=&SinglePropertyView
   parent=HorizontalBox_114
   widget=MetaSoundSource
   properties {propertyName=SourceReference, toolTipText=Meta Sound asset that will be used to create new presets.}
   slot {class=HorizontalBoxSlot, properties {size {sizeRule=Fill, value=1}}}
  }
  SetBaseButton {
   widgetClassPath=&EditorUtilityButton
   parent=HorizontalBox_114
   widget=SetBaseButton
   properties {toolTipText=Reset the source reference back to default asset.}
   slot {class=HorizontalBoxSlot, properties {padding=<ltrb 5 5 5 5>}}
  }
  UEFN_TextBlock_C_4 {
   widgetClassPath=&UEFN_TextBlock
   parent=SetBaseButton
   widget=UEFN_TextBlock_C_4
   properties {
    justification=Center
    text=Set Base
    font {
     bForceMonospaced=false
     bMaterialIsStencil=false
     fontMaterial=null
     fontObject=/LauncherSocial/BrutalType
     letterSpacing=0
     monospacedWidth=1
     size=9
     skewAmount=0
     typefaceFontName=Regular
     outlineSettings {
      bApplyOutlineToDropShadows=false
      bMiteredCorners=false
      bSeparateFillAlpha=false
      outlineMaterial=null
      outlineSize=0
      outlineColor=<rgba 0 0 0 1>
     }
    }
    shadowColorAndOpacity=<rgba 0 0 0 1>
   }
   slot {class=ButtonSlot}
  }
 ]
}
