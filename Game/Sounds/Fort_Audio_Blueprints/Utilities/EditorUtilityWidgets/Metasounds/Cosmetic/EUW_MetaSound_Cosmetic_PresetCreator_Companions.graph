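# Editor utility widget for companion cosmetic audio that creates VO, Foley or Emote sound libraries and links them to a companion item definition.
# Data table rows become MetaSound source references, picked through a toggle-button grid or emote options and passed to the parent preset creator.
# Created sounds are added to the library under each row's tag.
# A button can add a lip-sync or floater movement component to the item definition.
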EditorUtilityWidgetBlueprint /Game/Sounds/Fort_Audio_Blueprints/Utilities/EditorUtilityWidgets/Metasounds/Cosmetic/EUW_MetaSound_Cosmetic_PresetCreator_Companions
graphs [
 ClearMoodActions {nodes [
   ClearMoodActions() {is=title K2Node_FunctionEntry}
   > ClearChildren(self=Get MoodGrid() -> (MoodGrid:Uniform Grid Panel Object Reference))
  ]}
 CreateButtonObject {nodes [
   n1 = CreateButtonObject!() -> (then, InputPin:Uniform Grid Panel Object Reference) {is=title K2Node_FunctionEntry}
   > n2 = Construct Object from Class(Class=&EditorUtilityButton, self=@n1) -> (ReturnValue:Editor Utility Button Object Reference)
   > n3 = AddChild(self=@n1, Content=@n2)
   > Return Node!(execute, PanelSlot:Panel Slot Object Reference=@n3, ButtonObject:Editor Utility Button Object Reference=@n2)
  ]}
 CreateButtonText {nodes [
   n1 = CreateButtonText!() -> (then, InputPin:Editor Utility Button Object Reference, InText:Text) {is=title K2Node_FunctionEntry}
   > n2 = Construct Object from Class(Class=/Script/Blutility.EditorUtilityEditableText, self=@n1) -> (ReturnValue:Editor Utility Editable Text Object Reference)
   > "SetText (Editable Text)"!(execute, self:Editable Text Object Reference=@n2, InText:Text=@n1.InText) -> (then)
   > n3 = AddChild(self=@n1, Content=@n2)
   > SetVisibility(self=@n2, InVisibility=HitTestInvisible)
   > Return Node!(execute, PanelSlot=@n3, TextObject:Editor Utility Editable Text Object Reference=@n2)
  ]}
 CreateSourceRef {nodes [
   n1 = CreateSourceRef!() -> (
    then
    TemplateMetasoundSource:EUW Meta Sound Source Reference Object Reference
    OverrideName:String
    CompanionOptions:Struct Metasound Companion Options Structure
   ) {is=title K2Node_FunctionEntry}
   > Set SameName(SameName:Array of Strings=Break Struct Metasound Companion Options!(Struct_Metasound_CompanionOptions:"Struct Metasound Companion Options Structure (by ref)"=@n1.CompanionOptions) -> (Tag_5_#1:Sound Library Tag Structure, SameNameIdentifiers_9_#2:Array of Strings) {is=title K2Node_BreakStruct}.SameNameIdentifiers_9_#2) -> (Output_Get:Array of Strings)
   > Set TemplateRef(TemplateRef:EUW Meta Sound Source Reference Object Reference=@n1) -> (Output_Get:EUW Meta Sound Source Reference Object Reference)
   > Set NameOverride(NameOverride:String=@n1.OverrideName) -> (Output_Get:String)
   > n2 = Create Widget!(
    execute:Exec
    WorldContextObject:Object Reference
    Class:User Widget Class Reference=/Game/Sounds/Fort_Audio_Blueprints/Utilities/EditorUtilityWidgets/Metasounds/Tools/EUW_MetaSoundSourceReference.EUW_MetaSoundSourceReference_C
    OwningPlayer:Player Controller Object Reference
   ) -> (then:Exec, ReturnValue:EUW Meta Sound Source Reference Object Reference)
   > Set SourceWidget(SourceWidget:EUW Meta Sound Source Reference Object Reference=@n2) -> (Output_Get:EUW Meta Sound Source Reference Object Reference)
   > n3 = Sequence()
   > Set SourceReference!(
    execute
    SourceReference:Meta Sound Source Object Reference=Get SourceReference!(self:EUW Meta Sound Source Reference Object Reference=Get TemplateRef() -> (TemplateRef:EUW Meta Sound Source Reference Object Reference)) -> (SourceReference:Meta Sound Source Object Reference)
    self:EUW Meta Sound Source Reference Object Reference=Get SourceWidget() -> (SourceWidget:EUW Meta Sound Source Reference Object Reference)
   ) -> (then, Output_Get:Meta Sound Source Object Reference)
   n4 = Sequence(execute=@n3.then_1)
   > Set AssetPrefix!(
    execute
    AssetPrefix:String=Get AssetPrefix!(self:EUW Meta Sound Source Reference Object Reference=Get TemplateRef()) -> (AssetPrefix:String)
    self:EUW Meta Sound Source Reference Object Reference=Get SourceWidget()
   ) -> (then, Output_Get:String)
   n5 = Sequence(execute=@n4.then_1)
   > Set NamesDataStructure!(
    execute
    NamesDataStructure:Struct Meta Sound Source Reference Structure=Get NamesDataStructure!(self:EUW Meta Sound Source Reference Object Reference=Get TemplateRef()) -> (NamesDataStructure:Struct Meta Sound Source Reference Structure)
    self:EUW Meta Sound Source Reference Object Reference=Get SourceWidget()
   ) -> (then, Output_Get:Struct Meta Sound Source Reference Structure)
   n6 = Sequence(execute=@n5.then_1)
   > Set LoopParameterID!(
    execute
    LoopParameterID:String=Get LoopParameterID!(self:EUW Meta Sound Source Reference Object Reference=Get TemplateRef()) -> (LoopParameterID:String)
    self:EUW Meta Sound Source Reference Object Reference=Get SourceWidget()
   ) -> (then, Output_Get:String)
   n7 = Sequence(execute=@n6.then_1)
   > Set NonLoopParameterID!(
    execute
    NonLoopParameterID:String=Get NonLoopParameterID!(self:EUW Meta Sound Source Reference Object Reference=Get TemplateRef()) -> (NonLoopParameterID:String)
    self:EUW Meta Sound Source Reference Object Reference=Get SourceWidget()
   ) -> (then, Output_Get:String)
   n8 = Sequence(execute=@n7.then_1)
   > Set IsArray!(
    execute
    IsArray:bool=Get IsArray!(self:EUW Meta Sound Source Reference Object Reference=Get TemplateRef()) -> (IsArray:bool)
    self:EUW Meta Sound Source Reference Object Reference=Get SourceWidget()
   ) -> (then, Output_Get:bool)
   n9 = Sequence(execute=@n8.then_1)
   > Set AssetSuffix!(
    execute
    AssetSuffix:String=Get NameOverride() -> (NameOverride:String)
    self:EUW Meta Sound Source Reference Object Reference=Get SourceWidget()
   ) -> (then, Output_Get:String)
   n10 = Sequence(execute=@n9.then_1)
   > Set NameOverride!(execute, NameOverride=Get NameOverride(), self:EUW Meta Sound Source Reference Object Reference=Get SourceWidget()) -> (then, Output_Get)
   n11 = Sequence(execute=@n10.then_1)
   n12 = Get SameName() -> (SameName:Array of Strings)
   n13 = Get NameOverride()
   > Add(TargetArray=@n12, NewItem=@n13) {is=title K2Node_CallArrayFunction}
   n14 = Break Struct Meta Sound Source Reference!(Struct_MetaSoundSourceReference:"Struct Meta Sound Source Reference Structure (by ref)"=Get NamesDataStructure(self=Get TemplateRef())) -> (
    SkipActionName_27_#3:bool=~
    DiableLoopOption_26_#4:bool=~
    SourceMetaSoundName_22_#5:String
    SourceWavesName_21_#6:Array of Strings
    SameNameIdentifiers_32_#7:Array of Strings
   ) {is=title K2Node_BreakStruct}
   > Set NamesDataStructure(
    NamesDataStructure=Make Struct Meta Sound Source Reference!(
     SkipActionName_27_#3:bool=@n14
     DiableLoopOption_26_#4:bool=@n14.DiableLoopOption_26_#4
     SourceMetaSoundName_22_#5:String=@n14.SourceMetaSoundName_22_#5
     SourceWavesName_21_#6:Array of Strings=@n14.SourceWavesName_21_#6
     SameNameIdentifiers_32_#7:Array of Strings=@n12
    ) -> (Struct_MetaSoundSourceReference:Struct Meta Sound Source Reference Structure) {is=title K2Node_MakeStruct}
    self=Get SourceWidget()
   )
   Sequence(execute=@n11.then_1)
   > Return Node!(execute, CreatedSourceWidget:EUW Meta Sound Source Reference Object Reference=Get SourceWidget())
   Make Array("[0]":String=@n13) -> (Array:Array of Strings)
   Get SourceMetaSoundName!(self:EUW Meta Sound Cosmetic Action Name Widget Object Reference) -> (SourceMetaSoundName:String)
  ]}
 EventGraph {nodes [
   "On Clicked (MakeLibraryButton)"() {is=title K2Node_ComponentBoundEvent}
   > n1 = GetCosmeticCreator!(execute, Target:Widget Object Reference=Self-Reference(), __WorldContext:Object Reference) -> (then, AsEUW Meta Sound Cosmetic Preset Creator:EUW Meta Sound Cosmetic Preset Creator Object Reference)
   n2 = Get AssetPath!(self:EUW Meta Sound Cosmetic Preset Creator Object Reference=@n1) -> (AssetPath:EUW Meta Sound Asset Path Object Reference)
   n3 = Get PathInput!(self:EUW Meta Sound Asset Path Object Reference=@n2) -> (PathInput:Editor Utility Editable Text Box Object Reference)
   n4 = "GetText (Text Box)"!(self:Text Box Object Reference=Get AssetNameInput!(self:EUW Meta Sound Asset Path Object Reference=@n2) -> (AssetNameInput:Editor Utility Editable Text Box Object Reference)) -> (ReturnValue:Text)
   n5 = Get LibType() -> (LibType:E_CompanionLibType Enum=NewEnumerator0)
   > n6 = CreateSoundLibrary!(
    execute
    PackagingPath:String="To String (Text)"(InText=Get Text(self=@n3))
    AssetsName:String=Append!(
     A=SL
     B=_
     C:String=Mimosa
     D:String=_
     E:String="To String (Text)"(InText=@n4)
     F:String=_
     G:String=Enum to String!(Enumerator:E_CompanionLibType Enum=@n5) -> (ReturnValue:String)
    ) -> (ReturnValue)
    __WorldContext:Object Reference
   ) -> (then, Sound Library:Sound Library Object Reference, Successful:bool)
   > Branch(Condition=@n6.Successful)
   > n7 = Set SoundLibrary(SoundLibrary:Sound Library Object Reference=@n6) -> (Output_Get:Sound Library Object Reference)
   > SetEditorProperty(
    Object=@n7
    PropertyName=Parent
    PropertyValue:Sound Library Object Reference=Get CompanionLibParent() -> (CompanionLibParent:Sound Library Object Reference)
   )
   > SaveLoadedAsset(AssetToSave=@n6, bOnlyIfIsDirty=true)
   n8 = "On Clicked (SelectAllButton)"() {is=title K2Node_ComponentBoundEvent}
   n9 = Get SelectedButtons() -> (SelectedButtons:Array of Strings)
   Clear(execute=[@n8.then, @n72.then], TargetArray=@n9)
   > n10 = For Each Loop(Array:Array of Texts=Get ButtonTextArray() -> (ButtonTextArray:Array of Texts)) -> (Array Element:Text)
   n11 = Values!(
    execute=@n10.Completed
    TargetMap:Map of Editor Utility Button Object References to EUW Meta Sound Source Reference Object References=Get ButtonArray() -> (ButtonArray:Map of Editor Utility Button Object References to EUW Meta Sound Source Reference Object References)
   ) -> (then, Values:Array of EUW Meta Sound Source Reference Object References)
   > n12 = For Each Loop(Array:Array of EUW Meta Sound Source Reference Object References=@n11) -> (Array Element:EUW Meta Sound Source Reference Object Reference)
   UpdateGridMoodButtons(execute=@n12.Completed)
   Widget Selector!(execute=@n12.LoopBody, self, ShouldAdd:bool=true, Reference:"EUW Meta Sound Source Reference Object Reference (by ref)"=@n12) -> (then)
   Add Unique(execute=@n10.LoopBody, TargetArray=@n9, NewItem="To String (Text)"(InText=@n10))
   n13 = "On Clicked (UnselectAllButton)"() {is=title K2Node_ComponentBoundEvent}
   Clear(execute=[@n13.then, @n46.then], TargetArray=Get SelectedButtons())
   > n14 = Values(TargetMap=Get ButtonArray())
   > n15 = For Each Loop(Array=@n14)
   UpdateGridMoodButtons(execute=@n15.Completed)
   Widget Selector(execute=@n15.LoopBody, ShouldAdd=false, Reference=@n15)
   "On Source Property Changed (ItemDef)"() {is=title K2Node_ComponentBoundEvent}
   > n16 = GetCosmeticCreator(Target=Self-Reference())
   n17 = Create Event()
   > Unbind Event from Finish Setting Sounds(self:EUW Meta Sound Cosmetic Preset Creator Object Reference=@n16, Delegate=@n17) {is=title K2Node_RemoveDelegate}
   > Get!(execute:Exec) -> (then:Exec, else:Exec, ItemDef:EUW IDReference Object Reference)
   > Bind Event to Finish Setting Sounds(self:EUW Meta Sound Cosmetic Preset Creator Object Reference=@n16, Delegate=@n17) {is=title K2Node_AddDelegate}
   "On Pressed (AddComponentButton)"() {is=title K2Node_ComponentBoundEvent}
   > n18 = Get!(execute) -> (then, else, CompanionItemDef:Cosmetic Companion Item Definition Object Reference)
   > n19 = GetItemData!(execute, self:Item Definition Base Object Reference=@n18) -> (Found:Exec, NotFound:Exec, ItemData:"Add Actor Components Structure (by ref)")
   > n20 = Set ComponentsOnCID(ComponentsOnCID:Array of Item Definition Actor Component To Add Structures=Break Add Actor Components!(ItemComponentData_AddActorComponents:"Add Actor Components Structure (by ref)"=@n19) -> (ComponentsToAdd:Array of Item Definition Actor Component To Add Structures) {is=title K2Node_BreakStruct}) -> (Output_Get:Array of Item Definition Actor Component To Add Structures)
   > n21 = For Each Loop with Break!(Exec:Exec, Array:Array of Item Definition Actor Component To Add Structures=@n20, Break:Exec=@n24.then) -> (LoopBody:Exec, Array Element:Item Definition Actor Component To Add Structure, Array Index:int, Completed:Exec)
   n22 = Get!(execute=@n21.Completed) -> (then, else, HasComponent:bool)
   > Show Message Dialog(
    Title=MakeLiteralText(Value="LOCTEXT(\"Warning\")")
    Message=MakeLiteralText(Value="LOCTEXT(\"Companion ID already has the relevant component \")")
    MessageType=Ok
    DefaultValue=No
    MessageCategory=Warning
   )
   AddComponent(execute=@n22.else)
   n23 = "Equal (SoftClassReference)"(
    A:"Object Soft Class Reference (by ref)"=Break Item Definition Actor Component to Add!(ItemDefinitionActorComponentToAdd:"Item Definition Actor Component To Add Structure (by ref)"=@n21) -> (ComponentClass:Actor Component Soft Class Reference, bAddOnServer:bool=~, bAddOnClient:bool=~) {is=title K2Node_BreakStruct}
    B:"Object Soft Class Reference (by ref)"=Get Component Class() -> (Component Class:Actor Component Soft Class Reference)
   ) -> (ReturnValue:bool=~) {is=title K2Node_PromotableOperator}
   Branch(execute=@n21.LoopBody, Condition=@n23)
   > n24 = Set HasComponent(HasComponent:bool=@n23) -> (Output_Get:bool)
   AddComponent(execute=@n19.NotFound)
   Show Message Dialog(
    execute=@n18.else
    Title=MakeLiteralText(Value="LOCTEXT(\"Warning!\")")
    Message=MakeLiteralText(Value="LOCTEXT(\"No Valid item data, component not added to CID.\")")
    MessageType=Ok
    DefaultValue=No
    MessageCategory=Warning
   )
   n25 = "On Selection Changed (LibrarySelectionComboBox)"!() -> (OutputDelegate, then, SelectedItem:String, SelectionType:ESelectInfo Enum) {is=title K2Node_ComponentBoundEvent}
   > n26 = Switch on String!(execute:Exec, Selection:String=@n25.SelectedItem) -> (Default:Exec, VO:Exec, Foley:Exec, Emote:Exec) {pinNames [VO,Foley,Emote]}
   Set CompanionLibParent(
    execute=@n26.Emote
    CompanionLibParent:Sound Library Object Reference=Get EmoteLibraryParent() -> (EmoteLibraryParent:Sound Library Object Reference)
   ) -> (Output_Get:Sound Library Object Reference)
   > n27 = Set LibType(LibType:E_CompanionLibType Enum=NewEnumerator2) -> (Output_Get:E_CompanionLibType Enum=NewEnumerator0)
   SelectView(execute=[@n27.then, @n29.then, @n28.then])
   Set CompanionLibParent(execute=@n26.Foley, CompanionLibParent=Get FoleyLibraryParent() -> (FoleyLibraryParent:Sound Library Object Reference))
   > n28 = Set LibType(LibType=NewEnumerator1) -> (Output_Get=NewEnumerator0)
   Set CompanionLibParent(execute=@n26.VO, CompanionLibParent=Get MoodLibraryParent() -> (MoodLibraryParent:Sound Library Object Reference))
   > n29 = Set LibType(LibType=NewEnumerator0) -> (Output_Get=NewEnumerator0)
   n30 = CreateTableButtons!() -> (OutputDelegate, then, Table:Data Table Object Reference) {is=title K2Node_CustomEvent}
   n31 = Get MsSourceRef() -> (MsSourceRef:Array of EUW Meta Sound Source Reference Object References)
   > n32 = For Each Loop(Array=@n31)
   > Unbind Event from Sound for Asset(self:EUW Meta Sound Source Reference Object Reference=@n32, Delegate=Create Event()) {is=title K2Node_RemoveDelegate}
   Clear(execute=@n32.Completed, TargetArray:Array of EUW Meta Sound Source Reference Object References=@n31)
   > n33 = GetDataTableRowNames(Table=@n30.Table)
   > n34 = For Each Loop(Array:Array of Names=@n33) -> (Array Element:Name)
   > n35 = Get Data Table Row!(
    execute:Exec
    DataTable:Data Table Object Reference=Get CompanionMoodTable() -> (CompanionMoodTable:Data Table Object Reference)
    RowName:Name=@n34
   ) -> (then:Exec, RowNotFound:Exec, ReturnValue:Struct Metasound Companion Options Structure)
   n36 = CreateSourceRef!(
    execute=[@n35.RowNotFound, @n35.then]
    self
    TemplateMetasoundSource:EUW Meta Sound Source Reference Object Reference=Select!(
     NewEnumerator0:EUW Meta Sound Source Reference Object Reference=Get MoodMsReference() -> (MoodMsReference:EUW Meta Sound Source Reference Object Reference)
     NewEnumerator1:EUW Meta Sound Source Reference Object Reference=Get FoleyMsReference() -> (FoleyMsReference:EUW Meta Sound Source Reference Object Reference)
     NewEnumerator2:EUW Meta Sound Source Reference Object Reference
     Index:E_CompanionLibType Enum=Get LibType() -> (LibType=NewEnumerator0)
    ) -> (ReturnValue:EUW Meta Sound Source Reference Object Reference)
    OverrideName:String="To String (Name)"(InName=@n34)
    CompanionOptions:Struct Metasound Companion Options Structure=@n35
   ) -> (then, CreatedSourceWidget:EUW Meta Sound Source Reference Object Reference) {is=title K2Node_CallFunction}
   > Bind Event to Sound for Asset(self:EUW Meta Sound Source Reference Object Reference=@n36, Delegate=Create Event()) {is=title K2Node_AddDelegate}
   > Add(
    TargetArray:Array of EUW Meta Sound Source Reference Object References=Get MsSourceRef()
    NewItem:"EUW Meta Sound Source Reference Object Reference (by ref)"=@n36
   )
   n37 = CreateEmoteRefs() {is=title K2Node_CustomEvent}
   n38 = Sequence!(execute=[@n63.then, @n37.then]) -> (then_0, then_1, then_2:Exec)
   > Get!(execute) -> (then, else, EmoteVo:bool)
   > n39 = Get Data Table Row(DataTable=Get CompanionEmoteTable() -> (CompanionEmoteTable:Data Table Object Reference), RowName=EmoteVo)
   > n40 = CreateSourceRef(
    TemplateMetasoundSource=Get EmoteMsSourceRef() -> (EmoteMsSourceRef:EUW Meta Sound Source Reference Object Reference)
    OverrideName=EmoteVO
    CompanionOptions=@n39
   )
   > Widget Selector(ShouldAdd=true, Reference=@n40)
   n41 = Create Event()
   > Bind Event to Sound for Asset(self=@n40, Delegate=@n41)
   Get!(execute=@n38.then_1) -> (then, else, EmoteFoley:bool)
   > n42 = Get Data Table Row(DataTable=Get CompanionEmoteTable(), RowName=EmoteFoley)
   > n43 = CreateSourceRef(TemplateMetasoundSource=Get EmoteMsSourceRef(), OverrideName=EmoteFoley, CompanionOptions=@n42)
   > Widget Selector(ShouldAdd=true, Reference=@n43)
   > Bind Event to Sound for Asset(self=@n43, Delegate=@n41)
   Get!(execute=@n38.then_2) -> (then, else, EmoteMusic:bool)
   > n44 = Get Data Table Row(DataTable=Get CompanionEmoteTable(), RowName=EmoteMusic)
   > n45 = CreateSourceRef(TemplateMetasoundSource=Get EmoteMsSourceRef(), OverrideName=EmoteMusic, CompanionOptions=@n44)
   > Widget Selector(ShouldAdd=true, Reference=@n45)
   > Bind Event to Sound for Asset(self=@n45, Delegate=@n41)
   n46 = DeselectAllMoods() {is=title K2Node_CustomEvent}
   SelectView() {is=title K2Node_CustomEvent}
   > Clear(TargetArray=Get SoundSourceReferenceWidgets() -> (SoundSourceReferenceWidgets:Array of EUW Meta Sound Source Reference Object References))
   > Set SoundLibrary()
   > DeselectAllMoods() {is=title K2Node_CallFunction}
   > ClearMoodActions() {is=title K2Node_CallFunction}
   > n47 = Switch on E_CompanionLibType!(execute, Selection:E_CompanionLibType Enum=Get LibType() -> (LibType=NewEnumerator0)) -> (NewEnumerator0:Exec, NewEnumerator1:Exec, NewEnumerator2:Exec) {is=title K2Node_SwitchEnum}
   > SetVisibility(self=Get EmoteTypeBox() -> (EmoteTypeBox:Horizontal Box Object Reference), InVisibility=Hidden)
   > SetVisibility(self=Get MoodBorder() -> (MoodBorder:Border Object Reference), InVisibility=Visible)
   > UpdateMoods!(execute, self, ShowGrid:bool=true) -> (then)
   > SelectAllMoods()
   > SetVisibility(self=Get FoleyBox() -> (FoleyBox:Horizontal Box Object Reference), InVisibility=Collapsed)
   > SetVisibility(self=Get AddComponentButton() -> (AddComponentButton:Editor Utility Button Object Reference), InVisibility=Visible)
   > n48 = Set Component Class(Component Class:Actor Component Soft Class Reference=/CosmeticCompanions/Assets/Audio_Shared/Blueprints/Components/BP_PlayerCompanion_LipSync_Component.BP_PlayerCompanion_LipSync_Component_C) -> (Output_Get:Actor Component Soft Class Reference)
   Set ComponentToAdd(
    execute=[@n48.then, @n49.then]
    ComponentToAdd:Item Definition Actor Component To Add Structure=Make Item Definition Actor Component to Add!(ComponentClass:Actor Component Soft Class Reference=Get Component Class(), bAddOnServer:bool=false, bAddOnClient:bool=True) -> (ItemDefinitionActorComponentToAdd:Item Definition Actor Component To Add Structure) {is=title K2Node_MakeStruct}
   ) -> (Output_Get:Item Definition Actor Component To Add Structure)
   SetVisibility(execute=@n47.NewEnumerator1, self=Get EmoteTypeBox(), InVisibility=Hidden)
   > SetVisibility(self=Get MoodBorder(), InVisibility=Visible)
   > UpdateMoods(ShowGrid=true)
   > SelectAllMoods()
   > SetVisibility(self=[Get FoleyBox(), Get AddComponentButton()], InVisibility=Visible)
   > n49 = Set Component Class(Component Class=/CosmeticCompanions/Assets/Audio_Shared/Blueprints/Components/BP_CompanionMovement_Floater.BP_CompanionMovement_Floater_C)
   SetVisibility(execute=@n47.NewEnumerator2, self=Get MoodBorder(), InVisibility=Hidden)
   > SetVisibility(self=Get EmoteTypeBox(), InVisibility=Visible)
   > CreateEmoteRefs() {is=title K2Node_CallFunction}
   > SetVisibility(self=Get FoleyBox(), InVisibility=Collapsed)
   > SetVisibility(self=Get AddComponentButton(), InVisibility=Hidden)
   UpdateMoods!() -> (OutputDelegate, then, ShowGrid:bool=~) {is=title K2Node_CustomEvent}
   > Set MoodGridRow(MoodGridRow:int=0) -> (Output_Get:int=0)
   > Clear!(execute, TargetMap:Map of Editor Utility Button Object References to EUW Meta Sound Source Reference Object References=Get ButtonArray()) -> (then) {is=title K2Node_CallFunction}
   > Clear(TargetArray:Array of Texts=Get ButtonTextArray()) {is=title K2Node_CallArrayFunction}
   > CreateTableButtons!(
    execute
    self
    Table:Data Table Object Reference=Select(
     NewEnumerator0:Data Table Object Reference=Get CompanionMoodTable()
     NewEnumerator1:Data Table Object Reference=Get CompanionFoleyTable() -> (CompanionFoleyTable:Data Table Object Reference)
     NewEnumerator2:Data Table Object Reference
     Index=Get LibType() -> (LibType=NewEnumerator0)
    ) -> (ReturnValue:Data Table Object Reference)
   ) -> (then) {is=title K2Node_CallFunction}
   > n50 = For Each Loop(Array:Array of EUW Meta Sound Source Reference Object References=Get MsSourceRef()) -> (Array Element:EUW Meta Sound Source Reference Object Reference)
   > n51 = CreateButtonObject!(execute, self, InputPin:Uniform Grid Panel Object Reference=Get MoodGrid()) -> (then, PanelSlot:Panel Slot Object Reference, ButtonObject:Editor Utility Button Object Reference) {is=title K2Node_CallFunction}
   n52 = Get NameOverride!(self:EUW Meta Sound Source Reference Object Reference=@n50) -> (NameOverride)
   n53 = IsEmpty(InString=@n52)
   n54 = Break Struct Meta Sound Source Reference!(Struct_MetaSoundSourceReference=Get NamesDataStructure(self=@n50)) -> (SourceMetaSoundName_22_#5)
   > n55 = CreateButtonText!(
    execute
    self
    InputPin:Editor Utility Button Object Reference=@n51.ButtonObject
    InText:Text="To Text (String)"(InString=Select!(Option 0:String=@n52, Option 1:String=@n54, Index:bool=@n53) -> (ReturnValue:String))
   ) -> (then, PanelSlot:Panel Slot Object Reference, TextObject:Editor Utility Editable Text Object Reference) {is=title K2Node_CallFunction}
   > Add!(
    execute
    TargetMap:Map of Editor Utility Button Object References to EUW Meta Sound Source Reference Object References=Get ButtonArray()
    Key:"Editor Utility Button Object Reference (by ref)"=@n51.ButtonObject
    Value:"EUW Meta Sound Source Reference Object Reference (by ref)"=@n50
   ) -> (then) {is=title K2Node_CallFunction}
   > Add Unique(TargetArray:Array of Texts=Get ButtonTextArray(), NewItem:"Text (by ref)"=Get Text(self:Editable Text Object Reference=@n55.TextObject))
   > UpdateGridMoodButtons()
   > n56 = Cast To UniformGridSlot!(execute, Object=@n51) -> (then, CastFailed, AsUniform Grid Slot:Uniform Grid Slot Object Reference) {is=title K2Node_DynamicCast}
   > SetHorizontalAlignment(self:Uniform Grid Slot Object Reference=@n56, InHorizontalAlignment=HAlign_Fill)
   n57 = "% (Integer)"(A=@n50.Array Index, B=4) -> (ReturnValue=0)
   > SetColumn!(execute, self:Uniform Grid Slot Object Reference=@n56, InColumn:int=@n57) -> (then)
   n58 = Get MoodGridRow() -> (MoodGridRow:int=0)
   > SetRow!(execute, self:Uniform Grid Slot Object Reference=@n56, InRow:int=@n58) -> (then)
   > Branch(Condition="Equal (Integer)"(A=@n57, B=3))
   > Increment Int(Value=@n58)
   UpdateGridMoodButtons() {is=title K2Node_CustomEvent}
   > n59 = Keys!(execute, TargetMap:Map of Editor Utility Button Object References to EUW Meta Sound Source Reference Object References=Get ButtonArray()) -> (then, Keys:Array of Editor Utility Button Object References)
   > n60 = For Each Loop(Array:Array of Editor Utility Button Object References=@n59) -> (Array Element:Editor Utility Button Object Reference)
   > Bind Event to On Pressed(self:Button Object Reference=@n60, Delegate=Create Event()) {is=title K2Node_AddDelegate}
   n61 = "Get (a copy)"(Array:Array of Texts=Get ButtonTextArray(), Dimension 1=@n60.Array Index) -> (Output:Text)
   > n62 = Branch(Condition=Contains Item(TargetArray:Array of Strings=Get SelectedButtons(), ItemToFind:"String (by ref)"="To String (Text)"(InText=@n61)))
   > SetBackgroundColor!(
    execute
    self:Button Object Reference=@n60
    InBackgroundColor:Linear Color Structure=Get ActiveColor() -> (ActiveColor:Linear Color Structure)
   ) -> (then)
   SetBackgroundColor(execute=@n62.else, self=@n60, InBackgroundColor=Get InactiveColor() -> (InactiveColor:Linear Color Structure))
   SetEmotes!() -> (OutputDelegate, then, PropertyName:Name) {is=title K2Node_CustomEvent}
   > n63 = Set Current Source Reference Widgets!(
    execute
    self:EUW Meta Sound Cosmetic Preset Creator Object Reference=GetCreatorWidget!() -> (Costmetic Creator:EUW Meta Sound Cosmetic Preset Creator Object Reference)
    CurrentActiveWidget:Widget Object Reference=Self-Reference()
    SourceReferences:Array of EUW Meta Sound Source Reference Object References=Get SoundSourceReferenceWidgets()
    CosmeticType:String=Companion
   ) -> (then) {is=title K2Node_CallFunction}
   OnMoodButtonPressed() {is=title K2Node_CustomEvent}
   > n64 = Keys(TargetMap=Get ButtonArray())
   > n65 = For Each Loop(Array=@n64)
   > Branch(Condition=IsPressed!(self:Button Object Reference=@n65) -> (ReturnValue:bool))
   n66 = "To String (Text)"(InText="Get (a copy)"(Array=Get ButtonTextArray(), Dimension 1=@n65.Array Index))
   n67 = Contains Item(TargetArray=Get SelectedButtons(), ItemToFind=@n66)
   > Widget Selector(
    ShouldAdd=NOT Boolean(A=@n67)
    Reference=Find(
     TargetMap:Map of Editor Utility Button Object References to EUW Meta Sound Source Reference Object References=Get ButtonArray()
     Key:"Editor Utility Button Object Reference (by ref)"=@n65
    ) -> (Value:EUW Meta Sound Source Reference Object Reference)
   )
   > n68 = Branch(Condition=@n67)
   > SetBackgroundColor(self=@n65, InBackgroundColor=Get InactiveColor())
   n69 = Get SelectedButtons()
   > Remove Item!(execute, TargetArray=@n69, Item:"String (by ref)"=@n66) -> (then, ReturnValue:bool)
   SetBackgroundColor(execute=@n68.else, self=@n65, InBackgroundColor=Get ActiveColor())
   > Add Unique(TargetArray:Array of Strings=@n69, NewItem:"String (by ref)"=@n66)
   AddComponent() {is=title K2Node_CustomEvent}
   n70 = Get ComponentsOnCID() -> (ComponentsOnCID:Array of Item Definition Actor Component To Add Structures)
   > Add(
    TargetArray:Array of Item Definition Actor Component To Add Structures=@n70
    NewItem:"Item Definition Actor Component To Add Structure (by ref)"=Get ComponentToAdd() -> (ComponentToAdd:Item Definition Actor Component To Add Structure)
   ) {is=title K2Node_CallArrayFunction}
   n71 = Get CompanionItemDef() -> (CompanionItemDef:Cosmetic Companion Item Definition Object Reference)
   > SetItemData!(
    execute
    self:Item Definition Base Object Reference=@n71
    ItemData:Add Actor Components Structure=Make Add Actor Components!(ComponentsToAdd:Array of Item Definition Actor Component To Add Structures=@n70) -> (ItemComponentData_AddActorComponents:Add Actor Components Structure) {is=title K2Node_MakeStruct}
   ) -> (then)
   > Branch(Condition=Get SaveAsset!(self:EUW IDReference Object Reference=Get ItemDef() -> (ItemDef:EUW IDReference Object Reference)) -> (SaveAsset:bool))
   > SaveLoadedAsset(AssetToSave=@n71, bOnlyIfIsDirty=true)
   > Clear(TargetArray:Array of Item Definition Actor Component To Add Structures=Get ComponentsOnCID())
   n72 = SelectAllMoods() {is=title K2Node_CustomEvent}
   Event PreConstruct()
   n73 = Get IDSource!(self:EUW IDReference Object Reference=Get ItemDef()) -> (IDSource:Single Property View Object Reference)
   > SetPropertyName!(execute, self:Single Property View Object Reference=@n73, NewPropertyName:Name=CompanionItemDef) -> (then)
   n74 = Get IDSource(self=Get SoundLibraryReference() -> (SoundLibraryReference:EUW IDReference Object Reference))
   > SetPropertyName(self=@n74, NewPropertyName=SoundLibrary)
   > SetNameOverride!(execute, self:Single Property View Object Reference=@n74, NewPropertyName:Text="LOCTEXT(\"Sound Library\")") -> (then)
   n75 = Get Emote_Vo() -> (Emote_Vo:Single Property View Object Reference)
   n76 = Get Emote_Music() -> (Emote_Music:Single Property View Object Reference)
   n77 = Get Emote_Foley() -> (Emote_Foley:Single Property View Object Reference)
   > SetObject(self=[@n75, @n76, @n77, @n73, @n74], NewObject=Self-Reference())
   > Bind Event to On Property Changed(self:Property View Base Object Reference=[@n75, @n76, @n77], Delegate=Create Event()) {is=title K2Node_AddDelegate}
   Event Construct()
   > Bind Event to On Source Property Changed(self:EUW IDReference Object Reference=Get SoundLibraryReference(), Delegate=Create Event()) {is=title K2Node_AddDelegate}
   > UpdateMoods(ShowGrid=false) {is=title K2Node_CallFunction}
   > SelectAllMoods() {is=title K2Node_CallFunction}
   > n78 = GetCosmeticCreator(Target=Self-Reference())
   > Bind Event to Finish Setting Sounds(self=@n78, Delegate=Create Event())
   > n79 = For Each Loop(Array:Array of EUW Meta Sound Source Reference Object References=Get CombinedRefs() -> (CombinedRefs:Array of EUW Meta Sound Source Reference Object References)) -> (Array Element:EUW Meta Sound Source Reference Object Reference)
   SelectView(execute=@n79.Completed) {is=title K2Node_CallFunction}
   Bind Event to Sound for Asset(execute=@n79.LoopBody, self=@n79, Delegate=Create Event())
   > SetVisibility(self=Get EmoteTypeBox(), InVisibility=Hidden)
  ]}
 GetCreatorWidget {nodes [
   GetCreatorWidget() {is=title K2Node_FunctionEntry}
   n1 = GetParent!(self:Widget Object Reference) -> (ReturnValue:Panel Widget Object Reference)
   > n2 = Cast To EUW_MetaSound_Cosmetic_PresetCreator!(execute, Object=GetOuterObject!(Object:Object Reference=GetOuterObject(Object=@n1)) -> (ReturnValue:Object Reference)) -> (then, CastFailed, AsEUW Meta Sound Cosmetic Preset Creator:EUW Meta Sound Cosmetic Preset Creator Object Reference) {is=title K2Node_DynamicCast}
   > Return Node!(execute, Costmetic Creator:EUW Meta Sound Cosmetic Preset Creator Object Reference=@n2)
  ]}
 InitializeActionWidgets {nodes [
   InitializeActionWidgets() {is=title K2Node_FunctionEntry}
   > n1 = For Each Loop(Array=Get FoleyMsSourceRef() -> (FoleyMsSourceRef:Array of EUW Meta Sound Source Reference Object References))
   > Widget Selector(ShouldAdd=Get IsFoley() -> (IsFoley:bool), Reference=@n1)
  ]}
 LibrarySet {nodes [LibrarySet() {is=title K2Node_FunctionEntry}]}
 SetMoodActions {nodes [SetMoodActions() {is=title K2Node_FunctionEntry}]}
 SetupLibrary {nodes [
   n1 = SetupLibrary!() -> (then, Sound:Sound Base Soft Object Reference, SourceReference:EUW Meta Sound Source Reference Object Reference) {is=title K2Node_FunctionEntry}
   > n2 = Get Data Table Row(
    DataTable=Select!(
     NewEnumerator0:Data Table Object Reference=Get CompanionMoodTable()
     NewEnumerator1:Data Table Object Reference=Get CompanionFoleyTable()
     NewEnumerator2:Data Table Object Reference=Get CompanionEmoteTable()
     Index:E_CompanionLibType Enum=Get LibType() -> (LibType=NewEnumerator0)
    ) -> (ReturnValue:Data Table Object Reference)
    RowName=String To Name(InString=Get NameOverride(self=@n1.SourceReference))
   )
   n3 = Get SoundLibrary() -> (SoundLibrary:Sound Library Object Reference)
   > Branch(Condition=IsValid(Object=@n3))
   > AddMetasoundToLibrary!(
    execute
    Sound:"Sound Base Soft Object Reference (by ref)"=@n1
    SoundLibraryTag:Sound Library Tag Structure=Break Struct Metasound Companion Options(Struct_Metasound_CompanionOptions=@n2)
    SoundLibrary:Sound Library Object Reference=@n3
    __WorldContext:Object Reference
   ) -> (then, Sounds:Map of Gameplay Tag Structures to Sound Base Soft Object References, Success:bool)
   Show Message Dialog(
    execute=@n2.RowNotFound
    Title=MakeLiteralText(Value="LOCTEXT(\"Cosmetic Option Error\")")
    Message=MakeLiteralText(Value="LOCTEXT(\"Tags not found in Data table. Please reach out to audio producer or audio tech POCs.\")")
    MessageType=Ok
    DefaultValue=No
    MessageCategory=Warning
   )
  ]}
 SoundCreated {nodes [
   n1 = SoundCreated!() -> (then, Sound:Sound Base Soft Object Reference) {is=title K2Node_FunctionEntry}
   > n2 = GetCosmeticOptions!(
    execute
    DataTable:Data Table Object Reference=Get DataTable() -> (DataTable:Data Table Object Reference)
    RowName:Name=None
    __WorldContext:Object Reference
   ) -> (
    then
    RowFound:bool
    CosmeticOption:Enum_MetaSound_Creator_CosmeticOptions Enum=NewEnumerator1
    Stat Manager Tag:Fort Stat Manager Tag Structure
    SoundLibraryTag:Sound Library Tag Structure
   )
   > Branch(Condition=@n2)
   > AddMetasoundToLibrary(Sound=@n1, SoundLibraryTag=@n2.SoundLibraryTag, SoundLibrary=Get SoundLibrary())
  ]}
 SoundCreationFinished {nodes [
   SoundCreationFinished() {is=title K2Node_FunctionEntry}
   > n1 = Sequence!(execute) -> (then_0, then_1)
   > n2 = Get!(execute) -> (then, else, CompanionItemDef:Cosmetic Companion Item Definition Object Reference)
   > n3 = GetEditorProperty!(execute:Exec, Object:Object Reference=@n2, PropertyName:Name=CosmeticSoundLibraries) -> (then:Exec, PropertyValue:Array of Sound Library Object References, ReturnValue:bool)
   > Branch(Condition=@n3.ReturnValue)
   > n4 = Set ExistingSoundLibs(ExistingSoundLibs:Array of Sound Library Object References=@n3) -> (Output_Get:Array of Sound Library Object References)
   > Add(TargetArray:Array of Sound Library Object References=@n4, NewItem:"Sound Library Object Reference (by ref)"=Get SoundLibrary())
   > n5 = SetEditorProperty(Object=@n2, PropertyName=CosmeticSoundLibraries, PropertyValue:Array of Sound Library Object References=@n4)
   > Branch(Condition=@n5)
   n6 = GetDisplayName!(self:Item Definition Base Object Reference=@n2) -> (ReturnValue:Text)
   > PrintString(InString=Append!(A="Success! Library set on ", B="To String (Text)"(InText=@n6)) -> (ReturnValue))
   n7 = Get!(execute=@n1.then_1) -> (then, else, SoundLibrary:Sound Library Object Reference)
   n8 = Get SaveAssetRef!(self:EUW IDReference Object Reference=Get SoundLibraryReference()) -> (SaveAssetRef:EUW IDSave Asset Object Reference)
   > Get!(execute, self:EUW IDSave Asset Object Reference=@n8) -> (then, else, SaveAsset:bool)
   > SaveIdAsset!(execute, self:EUW IDSave Asset Object Reference=@n8, AssetToSave:Object Reference=@n7) -> (then)
   Get Object Path String(Object=@n7)
  ]}
 WaveNameManagement {nodes [
   n1 = Inputs!() -> (self:EUW Meta Sound Source Reference Object Reference)
   n2 = "Get (a copy)"(Array:Array of Strings=Break Struct Meta Sound Source Reference!(Struct_MetaSoundSourceReference=Get NamesDataStructure(self=@n1)) -> (SourceWavesName_21_#6:Array of Strings)) -> (Output:String)
   Outputs!(AppendedName:Name=String To Name(InString=Append(A=Companion, B=@n2)))
  ]}
 Widget Selector {nodes [
   n1 = Widget Selector!() -> (then, ShouldAdd:bool=~, Reference:"EUW Meta Sound Source Reference Object Reference (by ref)") {is=title K2Node_FunctionEntry}
   > n2 = Branch(Condition=@n1)
   n3 = Get SoundSourceReferenceWidgets()
   > n4 = Add Unique(
    TargetArray:Array of EUW Meta Sound Source Reference Object References=@n3
    NewItem:"EUW Meta Sound Source Reference Object Reference (by ref)"=@n1.Reference
   )
   Set Current Source Reference Widgets(
    execute=[@n5.then, @n4.then]
    self=GetCreatorWidget() {is=title K2Node_CallFunction}
    CurrentActiveWidget=Self-Reference()
    SourceReferences=Get SoundSourceReferenceWidgets()
    CosmeticType=Companion
   )
   n5 = Remove Item(
    execute=@n2.else
    TargetArray:Array of EUW Meta Sound Source Reference Object References=@n3
    Item:"EUW Meta Sound Source Reference Object Reference (by ref)"=@n1.Reference
   )
  ]}
]
variables [
 ActiveColor {type=LinearColor, default=<rgba 0 0.4657 1 1>}
 Add Component Button {type=EditorUtilityButton, default=null}
 bAddComponent {type=boolean}
 ButtonArray {default {}}
 ButtonTextArray {type="array<string>", default []}
 CombinedRefs {type="array<EUW_MetaSoundSourceReference_C>", default []}
 CompanionEmoteTable {
  type=DataTable
  default=/Game/Sounds/Fort_Audio_Blueprints/Utilities/EditorUtilityWidgets/Metasounds/Cosmetic/Companion/DataTable_Metasound_CompanionEMote
 }
 CompanionFoleyTable {
  type=DataTable
  default=/Game/Sounds/Fort_Audio_Blueprints/Utilities/EditorUtilityWidgets/Metasounds/Cosmetic/Companion/DataTable_Metasound_CompanionFoley
 }
 CompanionItemDef {type=CosmeticCompanionItemDefinition, default=null}
 CompanionLibParent {type=SoundLibrary, default=/Game/Sounds/SoundLibrary/Parents/Cosmetic/Companion/SL_Companion_VO}
 CompanionMoodTable {
  type=DataTable
  default=/Game/Sounds/Fort_Audio_Blueprints/Utilities/EditorUtilityWidgets/Metasounds/Cosmetic/Companion/DataTable_Metasound_CompanionMoods
 }
 Component Class {
  type=Class
  default=&BP_PlayerCompanion_LipSync_Component_C
 }
 ComponentsOnCID {type="array<ItemDefinitionActorComponentToAdd>", default []}
 ComponentToAdd {
  type=ItemDefinitionActorComponentToAdd
  default {
   bAddOnClient=true
   bAddOnServer=false
   componentClass=&BP_PlayerCompanion_LipSync_Component_C
  }
 }
 Created Source Widget {type=EUW_MetaSoundSourceReference_C, default=null}
 DataTable {
  type=DataTable
  default=&/DataTable_MetaSound_CosmeticOptions
 }
 EmoteFoley {type=boolean, default=true}
 EmoteLibraryParent {type=SoundLibrary, default=/Game/Sounds/SoundLibrary/Parents/Cosmetic/Companion/SL_Companion_Emote}
 EmoteMsSourceRef {
  type=EUW_MetaSoundSourceReference_C
  default=&/EUW_MetaSound_Cosmetic_PresetCreator_Companions.Default__EUW_MetaSound_Cosmetic_PresetCreator_Companions_C:EUW_MetaSoundSourceReference_C_3
 }
 EmoteMusic {type=boolean, default=true}
 EmoteVo {type=boolean, default=true}
 ExistingSoundLibs {type="array<SoundLibrary>", default []}
 FoleyLibraryParent {type=SoundLibrary, default=/Game/Sounds/SoundLibrary/Parents/Cosmetic/Companion/SL_Companion_Quadraped_Foley}
 FoleyMsReference {
  type=EUW_MetaSoundSourceReference_C
  default=&Default__EUW_MetaSound_Cosmetic_PresetCreator_Companions_C:EUW_MetaSoundSourceReference_C_1
 }
 FoleyMsSourceRef {type="array<EUW_MetaSoundSourceReference_C>", default []}
 HasComponent {type=boolean}
 In Column {type=integer}
 InactiveColor {type=LinearColor, default=<rgba 1 1 1 1>}
 IsFoley {type=boolean, default=true}
 IsHardStep {type=boolean, default=true}
 IsMood {type=boolean, default=true}
 LibType {type=E_CompanionLibType, default=VO}
 MoodGridRow {type=integer}
 MoodLibraryParent {type=SoundLibrary, default=&SL_Companion_VO}
 MoodMsReference {
  type=EUW_MetaSoundSourceReference_C
  default=&Default__EUW_MetaSound_Cosmetic_PresetCreator_Companions_C:EUW_MetaSoundSourceReference_C_2
 }
 MsSourceRef {type="array<EUW_MetaSoundSourceReference_C>", default []}
 SelectedButtons {type="array<string>", default []}
 SoundLibrary {type=SoundLibrary, default=null}
 SoundSourceReferenceWidgets {type="array<EUW_MetaSoundSourceReference_C>", default []}
]
widgets {
 info {rootWidgetClass=&CanvasPanel, widgetCount=28}
 widgets [
  CanvasPanel_0 {widgetClassPath=&CanvasPanel, widget=CanvasPanel_0}
  GridPanel_0 {
   widgetClassPath=&GridPanel
   parent=CanvasPanel_0
   widget=GridPanel_0
   slot {
    class=CanvasPanelSlot
    properties {bAutoSize=true, layoutData {alignment=<0 0>, anchors {maximum=<0 0>, minimum=<0 0>}, offsets=<ltrb 0 0 500 300>}}
   }
  }
  GridPanel_183 {
   widgetClassPath=&GridPanel
   parent=GridPanel_0
   widget=GridPanel_183
   slot {class=GridSlot, properties {horizontalAlignment=HAlign_Center, verticalAlignment=VAlign_Center, nudge=<0 10>}}
  }
  VerticalBox_61 {widgetClassPath=&VerticalBox, parent=GridPanel_183, widget=VerticalBox_61, slot {class=GridSlot}}
  HorizontalBox_0 {widgetClassPath=&HorizontalBox, parent=VerticalBox_61, widget=HorizontalBox_0, slot {class=VerticalBoxSlot}}
  SizeBox_107 {
   widgetClassPath=&SizeBox
   parent=HorizontalBox_0
   widget=SizeBox_107
   properties {
    bOverride_MaxDesiredWidth=true
    bOverride_MinDesiredWidth=true
    bOverride_WidthOverride=true
    maxDesiredWidth=200
    minDesiredWidth=150
    widthOverride=150
   }
   slot {class=HorizontalBoxSlot, properties {size {sizeRule=Fill, value=1}}}
  }
  LibrarySelectionComboBox {
   widgetClassPath=/Script/Blutility.EditorUtilityComboBoxString
   bIsVariable=true
   parent=SizeBox_107
   widget=LibrarySelectionComboBox
   properties {
    selectedOption=VO
    defaultOptions [VO,Foley,Emote]
    font {
     bForceMonospaced=false
     bMaterialIsStencil=false
     fontMaterial=null
     fontObject=/LauncherSocial/BrutalType
     letterSpacing=0
     monospacedWidth=1
     size=9
     skewAmount=0
     typefaceFontName=Regular
     outlineSettings {
      bApplyOutlineToDropShadows=false
      bMiteredCorners=false
      bSeparateFillAlpha=false
      outlineMaterial=null
      outlineSize=0
      outlineColor=<rgba 0 0 0 1>
     }
    }
    foregroundColor {colorUseRule=UseColor_Specified, specifiedColor=<rgba 0.5271 0.5271 0.5271 1>}
   }
   slot {class=SizeBoxSlot, properties {verticalAlignment=VAlign_Center, padding=<ltrb 10 10 10 10>}}
  }
  MoodBorder {
   widgetClassPath=&Border
   bIsVariable=true
   parent=VerticalBox_61
   widget=MoodBorder
   properties {background {
     bIsDynamicallyLoaded=false
     drawAs=Image
     imageType=NoImage
     mirroring=NoMirror
     resourceName=None
     resourceObject=null
     tiling=NoTile
     imageSize=<32 32>
     margin=<ltrb 0 0 0 0>
     outlineSettings {
      bUseBrushTransparency=false
      roundingType=HalfHeightRadius
      width=0
      color {colorUseRule=UseColor_Specified, specifiedColor=<rgba 0 0 0 0>}
      cornerRadii=<0 0 0 0>
     }
     tintColor {colorUseRule=UseColor_Specified, specifiedColor=<rgba 0.0208 0.0208 0.0208 1>}
     uVRegion {bIsValid=false, max=<0 0>, min=<0 0>}
    }}
   slot {class=VerticalBoxSlot}
  }
  VerticalBox_1 {widgetClassPath=&VerticalBox, parent=MoodBorder, widget=VerticalBox_1, slot {class=BorderSlot}}
  HorizontalBox_421 {widgetClassPath=&HorizontalBox, parent=VerticalBox_1, widget=HorizontalBox_421, slot {class=VerticalBoxSlot}}
  SelectAllButton {
   widgetClassPath=&EditorUtilityButton
   parent=HorizontalBox_421
   widget=SelectAllButton
   slot {class=HorizontalBoxSlot, properties {size {sizeRule=Fill, value=1}}}
  }
  UEFN_TextBlock_C_106 {
   widgetClassPath=&UEFN_TextBlock
   parent=SelectAllButton
   widget=UEFN_TextBlock_C_106
   properties {
    justification=Center
    text=Select All
    font {
     bForceMonospaced=false
     bMaterialIsStencil=false
     fontMaterial=null
     fontObject=&Roboto
     letterSpacing=0
     monospacedWidth=1
     size=9
     skewAmount=0
     typefaceFontName=Regular
     outlineSettings {
      bApplyOutlineToDropShadows=false
      bMiteredCorners=false
      bSeparateFillAlpha=false
      outlineMaterial=null
      outlineSize=0
      outlineColor=<rgba 0 0 0 1>
     }
    }
    shadowColorAndOpacity=<rgba 0 0 0 1>
   }
   slot {class=ButtonSlot}
  }
  UnselectAllButton {
   widgetClassPath=&EditorUtilityButton
   parent=HorizontalBox_421
   widget=UnselectAllButton
   slot {class=HorizontalBoxSlot, properties {size {sizeRule=Fill, value=1}}}
  }
  UEFN_TextBlock_C_209 {
   widgetClassPath=&UEFN_TextBlock
   parent=UnselectAllButton
   widget=UEFN_TextBlock_C_209
   properties {
    justification=Center
    text=Unselect All
    font {
     bForceMonospaced=false
     bMaterialIsStencil=false
     fontMaterial=null
     fontObject=&Roboto
     letterSpacing=0
     monospacedWidth=1
     size=9
     skewAmount=0
     typefaceFontName=Regular
     outlineSettings {
      bApplyOutlineToDropShadows=false
      bMiteredCorners=false
      bSeparateFillAlpha=false
      outlineMaterial=null
      outlineSize=0
      outlineColor=<rgba 0 0 0 1>
     }
    }
    shadowColorAndOpacity=<rgba 0 0 0 1>
   }
   slot {class=ButtonSlot}
  }
  MoodGrid {widgetClassPath=/Script/UMG.UniformGridPanel, bIsVariable=true, parent=VerticalBox_1, widget=MoodGrid, slot {class=VerticalBoxSlot}}
  FoleyBox {
   widgetClassPath=&HorizontalBox
   bIsVariable=true
   parent=VerticalBox_1
   widget=FoleyBox
   properties {visibility=Collapsed}
   slot {class=VerticalBoxSlot, properties {horizontalAlignment=HAlign_Center}}
  }
  EmoteTypeBox {widgetClassPath=&HorizontalBox, bIsVariable=true, parent=VerticalBox_61, widget=EmoteTypeBox, slot {class=VerticalBoxSlot}}
  Emote_Foley {
   widgetClassPath=&SinglePropertyView
   parent=EmoteTypeBox
   widget=Emote_Foley
   properties {propertyName=EmoteFoley}
   slot {class=HorizontalBoxSlot, properties {size {sizeRule=Fill, value=1}}}
  }
  Emote_Music {
   widgetClassPath=&SinglePropertyView
   parent=EmoteTypeBox
   widget=Emote_Music
   properties {propertyName=EmoteVo}
   slot {class=HorizontalBoxSlot, properties {size {sizeRule=Fill, value=1}}}
  }
  Emote_Vo {
   widgetClassPath=&SinglePropertyView
   parent=EmoteTypeBox
   widget=Emote_Vo
   properties {propertyName=EmoteMusic}
   slot {class=HorizontalBoxSlot}
  }
  VerticalBox_0 {
   widgetClassPath=&VerticalBox
   parent=VerticalBox_61
   widget=VerticalBox_0
   properties {bOverride_Cursor=true}
   slot {class=VerticalBoxSlot, properties {horizontalAlignment=HAlign_Center, verticalAlignment=VAlign_Center}}
  }
  MakeLibraryButton {
   widgetClassPath=&EditorUtilityButton
   parent=VerticalBox_0
   widget=MakeLibraryButton
   slot {class=VerticalBoxSlot, properties {padding=<ltrb 5 10 5 5>, size {sizeRule=Fill, value=1}}}
  }
  PickaxeText {
   widgetClassPath=&UEFN_TextBlock
   bIsVariable=true
   parent=MakeLibraryButton
   widget=PickaxeText
   properties {
    justification=Center
    text=Create SoundLibrary
    font {
     bForceMonospaced=false
     bMaterialIsStencil=false
     fontMaterial=null
     fontObject=&Roboto
     letterSpacing=0
     monospacedWidth=1
     size=9
     skewAmount=0
     typefaceFontName=Regular
     outlineSettings {
      bApplyOutlineToDropShadows=false
      bMiteredCorners=false
      bSeparateFillAlpha=false
      outlineMaterial=null
      outlineSize=0
      outlineColor=<rgba 0 0 0 1>
     }
    }
    shadowColorAndOpacity=<rgba 0 0 0 1>
   }
   slot {class=ButtonSlot}
  }
  SoundLibraryReference {
   widgetClassPath=/Game/Sounds/Fort_Audio_Blueprints/Utilities/EditorUtilityWidgets/Metasounds/Tools/EUW_IDReference
   bIsVariable=true
   parent=VerticalBox_0
   widget=SoundLibraryReference
   slot {class=VerticalBoxSlot}
  }
  AddComponentButton {
   widgetClassPath=&EditorUtilityButton
   parent=VerticalBox_0
   widget=AddComponentButton
   slot {class=VerticalBoxSlot, properties {padding=<ltrb 5 10 5 5>, size {sizeRule=Fill, value=1}}}
  }
  TextBlock_55 {
   widgetClassPath=&TextBlock
   parent=AddComponentButton
   widget=TextBlock_55
   properties {
    justification=Center
    text=Add Component
    font {
     bForceMonospaced=false
     bMaterialIsStencil=false
     fontMaterial=null
     fontObject=&Roboto
     letterSpacing=0
     monospacedWidth=1
     size=9
     skewAmount=0
     typefaceFontName=Regular
     outlineSettings {
      bApplyOutlineToDropShadows=false
      bMiteredCorners=false
      bSeparateFillAlpha=false
      outlineMaterial=null
      outlineSize=0
      outlineColor=<rgba 0 0 0 1>
     }
    }
    shadowColorAndOpacity=<rgba 0 0 0 1>
   }
   slot {class=ButtonSlot}
  }
  ItemDef {
   widgetClassPath=&EUW_IDReference
   bIsVariable=true
   parent=VerticalBox_0
   widget=ItemDef
   slot {class=VerticalBoxSlot}
  }
  SelectAllBox {widgetClassPath=&HorizontalBox, bIsVariable=true, parent=VerticalBox_0, widget=SelectAllBox, slot {class=VerticalBoxSlot}}
 ]
}
