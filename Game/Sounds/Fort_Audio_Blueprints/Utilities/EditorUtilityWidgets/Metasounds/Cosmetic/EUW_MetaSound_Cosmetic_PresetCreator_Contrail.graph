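# Editor utility widget, the contrail panel of the MetaSound cosmetic preset creator, that wires built sounds into skydive contrail cosmetics.
# When its source reference reports a sound, it adds music-tagged audio user data to it.
# It reuses or creates a Sound Library, maps the sound to the Skydive tag, and sets parent and play requirements.
# Any selected contrail item definition then gets that library as its sole cosmetic library and is saved.
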EditorUtilityWidgetBlueprint /Game/Sounds/Fort_Audio_Blueprints/Utilities/EditorUtilityWidgets/Metasounds/Cosmetic/EUW_MetaSound_Cosmetic_PresetCreator_Contrail
graphs [
 AddAudioUserData {nodes [
   n1 = AddAudioUserData!() -> (then, SoundBase:Sound Base Soft Object Reference) {is=title K2Node_FunctionEntry}
   n2 = Resolve Soft Reference!(Input:Sound Base Soft Object Reference=@n1) -> (Output:Sound Base Object Reference)
   > n3 = HasAssetUserDataOfClass!(
    execute
    self:Interface Asset User Data Interface=@n2
    InUserDataClass:Asset User Data Class Reference=Get AudioUserData() -> (AudioUserData:Asset User Data Class Reference)
   ) -> (then, ReturnValue:bool)
   > n4 = Branch(Condition=@n3)
   n5 = GetAssetUserDataOfClass!(
    execute=[@n4.then, @n8.then]
    self:Interface Asset User Data Interface=@n2
    InUserDataClass:Asset User Data Class Reference=Get AudioUserData()
   ) -> (then, ReturnValue:Asset User Data Object Reference)
   > n6 = Cast To AudioAssetUserData!(execute, Object=@n5) -> (then, CastFailed, AsAudio Asset User Data:Audio Asset User Data Object Reference) {is=title K2Node_DynamicCast}
   > AddTag!(
    execute
    Container:"Gameplay Tag Container Structure (by ref)"=Get MetadataTags!(self:Audio Asset User Data Object Reference=@n6) -> (MetadataTags:Gameplay Tag Container Structure)
    GameplayTag:"Gameplay Tag Structure (by ref)"=Get ContrailAudioUserDataTag() -> (ContrailAudioUserDataTag:Gameplay Tag Structure)
   ) -> (then)
   n7 = AddAssetUserDataOfClass!(execute=@n4.else, self:Interface Asset User Data Interface=@n2, InUserDataClass:Asset User Data Class Reference=Get AudioUserData()) -> (then, ReturnValue:bool)
   > n8 = Branch(Condition=@n7)
  ]}
 CreateSoundLibrary {nodes [
   n1 = CreateSoundLibrary!() -> (then, AssetName:String, PackagePath:String) {is=title K2Node_FunctionEntry}
   > n2 = GetAssetTools()
   > n3 = CreateAsset(self=@n2, AssetName=@n1, PackagePath=@n1.PackagePath, AssetClass=/Script/SoundLibrary, CallingContext=Cosmetic)
   > n4 = SaveLoadedAsset(AssetToSave=@n3, bOnlyIfIsDirty=true)
   > PrintString(InString=Append(A=GetDisplayName(Object=@n3), B=SelectString(A=" Saved!", B=" Failed to Save!", bPickA=@n4)))
   > n5 = Cast To SoundLibrary!(execute, Object=@n3) -> (then, CastFailed, AsSound Library:Sound Library Object Reference) {is=title K2Node_DynamicCast}
   > Return Node!(execute, AsSound Library:Sound Library Object Reference=@n5)
   Return Node(execute=@n5.CastFailed)
  ]}
 EventGraph {nodes [
   n1 = OnSoundBuilt!() -> (OutputDelegate, then, Sound:Sound Base Soft Object Reference, SourceReference:EUW Meta Sound Source Reference Object Reference) {is=title K2Node_CustomEvent}
   > AddAudioUserData!(execute, self, SoundBase:Sound Base Soft Object Reference=@n1.Sound) -> (then) {is=title K2Node_CallFunction}
   > n2 = Get!(execute:Exec) -> (then:Exec, else:Exec, SoundLibrary:Sound Library Object Reference)
   > n3 = Set SoundLibrarySounds(SoundLibrarySounds:Map of Gameplay Tag Structures to Sound Base Soft Object References=Get Sounds!(self:Sound Library Object Reference=@n2) -> (Sounds:Map of Gameplay Tag Structures to Sound Base Soft Object References)) -> (Output_Get:Map of Gameplay Tag Structures to Sound Base Soft Object References)
   > n4 = Add!(
    execute
    TargetMap:Map of Gameplay Tag Structures to Sound Base Soft Object References=@n3
    Key:"Gameplay Tag Structure (by ref)"=(TagName="SoundLibrary.Skydive")
    Value:"Sound Base Soft Object Reference (by ref)"=@n1.Sound
   ) -> (then) {is=title K2Node_CallFunction}
   n5 = Get SoundLibrary() -> (SoundLibrary:Sound Library Object Reference)
   SetEditorProperty(
    execute=[@n11.then, @n4.then]
    Object=@n5
    PropertyName=Sounds
    PropertyValue:Map of Gameplay Tag Structures to Sound Base Soft Object References=Get SoundLibrarySounds() -> (SoundLibrarySounds:Map of Gameplay Tag Structures to Sound Base Soft Object References)
   )
   > n6 = Set ParentSoundLibrary(ParentSoundLibrary:Sound Library Object Reference=/BRCosmetics/Sounds/Contrails/SL_Contrail_Parent) -> (Output_Get:Sound Library Object Reference)
   > SetEditorProperty(Object=@n5, PropertyName=Parent, PropertyValue:Sound Library Object Reference=@n6)
   > Set LibraryRequirements!(
    execute
    LibraryRequirements:Audio Gameplay Requirements Structure=Make Audio Gameplay Requirements!(
     Preset:Audio Requirement Preset Object Reference=Get SoundLibraryRequirements() -> (SoundLibraryRequirements:Audio Requirement Preset Object Reference)
     Custom:Gameplay Tag Query Structure=(TokenStreamVersion=0,TagDictionary=,QueryTokenStream=,UserDescription="",AutoDescription="")
    ) -> (AudioGameplayRequirements:Audio Gameplay Requirements Structure) {is=title K2Node_MakeStruct}
    self:Sound Library Object Reference=@n5
   ) -> (then, Output_Get:Audio Gameplay Requirements Structure)
   > Clear!(execute, TargetMap:Map of Gameplay Tag Structures to Sound Base Soft Object References=Get SoundLibrarySounds()) -> (then) {is=title K2Node_CallFunction}
   > ImplementID()
   n7 = Get AssetPath!(self:EUW Meta Sound Cosmetic Preset Creator Object Reference=GetCreatorWidget!() -> (Costmetic Creator:EUW Meta Sound Cosmetic Preset Creator Object Reference)) -> (AssetPath:EUW Meta Sound Asset Path Object Reference)
   n8 = Get PathInput!(self:EUW Meta Sound Asset Path Object Reference=@n7) -> (PathInput:Editor Utility Editable Text Box Object Reference)
   n9 = "GetText (Text Box)"!(self:Text Box Object Reference=Get AssetNameInput!(self:EUW Meta Sound Asset Path Object Reference=@n7) -> (AssetNameInput:Editor Utility Editable Text Box Object Reference)) -> (ReturnValue:Text)
   n10 = CreateSoundLibrary!(
    execute=@n2.else
    self
    AssetName:String=Append!(A=SL_Contrail, B=_, C:String="To String (Text)"(InText=@n9)) -> (ReturnValue)
    PackagePath:String="To String (Text)"(InText=Get Text(self=@n8))
   ) -> (then, AsSound Library:Sound Library Object Reference) {is=title K2Node_CallFunction}
   > Set SoundLibrary(SoundLibrary:Sound Library Object Reference=@n10) -> (Output_Get:Sound Library Object Reference)
   > n11 = Add(TargetMap=Get SoundLibrarySounds(), Key=(TagName="SoundLibrary.Skydive"), Value=@n1.Sound)
   ImplementID() {is=title K2Node_CustomEvent}
   > n12 = Get!(execute) -> (then, else, Contrail Item Definition:Athena Sky Dive Contrail Item Definition Object Reference)
   > n13 = Get!(execute) -> (then, else, SoundLibrary:Sound Library Object Reference)
   > SetEditorProperty(
    Object=@n12
    PropertyName=CosmeticSoundLibraries
    PropertyValue:Array of Sound Library Object References=Make Array("[0]":Sound Library Object Reference=@n13) -> (Array:Array of Sound Library Object References)
   )
   > SaveLoadedAsset(AssetToSave=@n12, bOnlyIfIsDirty=true)
   Event PreConstruct()
   n14 = Get IDSource!(self:EUW IDReference Object Reference=Get IDReference() -> (IDReference:EUW IDReference Object Reference)) -> (IDSource:Single Property View Object Reference)
   > SetPropertyName!(execute, self:Single Property View Object Reference=@n14, NewPropertyName:Name=Contrail Item Definition) -> (then)
   > SetObject(self=@n14, NewObject=Self-Reference())
   > SetObject(self=Get SoundLibraryView() -> (SoundLibraryView:Single Property View Object Reference), NewObject=Self-Reference())
   Event Construct()
   n15 = Get SoundSourceReferenceContrail() -> (SoundSourceReferenceContrail:EUW Meta Sound Source Reference Object Reference)
   > Add Unique(
    TargetArray:Array of EUW Meta Sound Source Reference Object References=Get SoundSourceReferenceWidgets() -> (SoundSourceReferenceWidgets:Array of EUW Meta Sound Source Reference Object References)
    NewItem:"EUW Meta Sound Source Reference Object Reference (by ref)"=@n15
   )
   > Bind Event to Sound for Asset(self:EUW Meta Sound Source Reference Object Reference=@n15, Delegate=Create Event()) {is=title K2Node_AddDelegate}
  ]}
 GetCreatorWidget {nodes [
   GetCreatorWidget() {is=title K2Node_FunctionEntry}
   n1 = GetParent!(self:Widget Object Reference) -> (ReturnValue:Panel Widget Object Reference)
   > n2 = Cast To EUW_MetaSound_Cosmetic_PresetCreator!(execute, Object=GetOuterObject!(Object:Object Reference=GetOuterObject(Object=@n1)) -> (ReturnValue:Object Reference)) -> (then, CastFailed, AsEUW Meta Sound Cosmetic Preset Creator:EUW Meta Sound Cosmetic Preset Creator Object Reference) {is=title K2Node_DynamicCast}
   > Return Node!(execute, Costmetic Creator:EUW Meta Sound Cosmetic Preset Creator Object Reference=@n2)
  ]}
]
variables [
 AudioUserData {type=Class, default=/Script/AudioGameplay.AudioAssetUserData}
 Contrail Item Definition {type=AthenaSkyDiveContrailItemDefinition, default=null}
 ContrailAudioUserDataTag {type=GameplayTag, default {tagName=Audio.Metadata.Music}}
 ParentSoundLibrary {type=SoundLibrary, default=null}
 SoundLibrary {type=SoundLibrary, default=null}
 SoundLibraryRequirements {type=AudioRequirementPreset, default=/Game/Sounds/SoundLibrary/Contexts/Skydive/SLR_Contrail_CanPlay}
 SoundLibrarySounds {default {}}
 SoundSourceReferenceContrail {
  type=EUW_MetaSoundSourceReference_C
  default=&/EUW_MetaSound_Cosmetic_PresetCreator_Contrail.Default__EUW_MetaSound_Cosmetic_PresetCreator_Contrail_C:EUW_MetaSoundSourceReference_C_3
 }
 SoundSourceReferenceWidgets {type="array<EUW_MetaSoundSourceReference_C>", default []}
]
widgets {
 info {rootWidgetClass=&CanvasPanel, widgetCount=5}
 widgets [
  CanvasPanel_0 {widgetClassPath=&CanvasPanel, widget=CanvasPanel_0}
  GridPanel_0 {
   widgetClassPath=&GridPanel
   parent=CanvasPanel_0
   widget=GridPanel_0
   slot {
    class=CanvasPanelSlot
    properties {bAutoSize=true, layoutData {alignment=<0 0>, anchors {maximum=<0 0>, minimum=<0 0>}, offsets=<ltrb 0 0 500 300>}}
   }
  }
  GridPanel_183 {
   widgetClassPath=&GridPanel
   parent=GridPanel_0
   widget=GridPanel_183
   slot {class=GridSlot, properties {horizontalAlignment=HAlign_Center, verticalAlignment=VAlign_Center, nudge=<0 10>}}
  }
  IDReference {
   widgetClassPath=/Game/Sounds/Fort_Audio_Blueprints/Utilities/EditorUtilityWidgets/Metasounds/Tools/EUW_IDReference
   bIsVariable=true
   parent=GridPanel_183
   widget=IDReference
   slot {class=GridSlot, properties {columnSpan=5, horizontalAlignment=HAlign_Center, row=2, verticalAlignment=VAlign_Center, nudge=<0 25>}}
  }
  SoundLibraryView {
   widgetClassPath=&SinglePropertyView
   parent=GridPanel_0
   widget=SoundLibraryView
   properties {
    propertyName=SoundLibrary
    toolTipText=Contrail Sound Library. If left empty a new one will be created
    renderTransform {angle=0, scale=<1 1>, shear=<0 0>, translation=<2 130>}
   }
   slot {class=GridSlot}
  }
 ]
}
