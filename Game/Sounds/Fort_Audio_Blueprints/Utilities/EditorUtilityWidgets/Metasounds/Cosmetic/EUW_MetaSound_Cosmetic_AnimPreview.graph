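# Editor utility widget for MetaSound cosmetic audio tooling that displays an emote montage as a button and can add its emote sound notify.
# Pressing the button loads AnimPath and opens it in the asset editor, with a dialog on failure.
# AddEmoteTrack recreates a named notify track with a full-length FortAnimNotifyState_EmoteSound, setting 1P/3P sounds and the primary flag through editor properties.
# Non-primary sounds also get SoundName from the 1P asset's name.
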EditorUtilityWidgetBlueprint /Game/Sounds/Fort_Audio_Blueprints/Utilities/EditorUtilityWidgets/Metasounds/Cosmetic/EUW_MetaSound_Cosmetic_AnimPreview
graphs [
 AddEmoteTrack {nodes [
   n1 = AddEmoteTrack!() -> (then, TrackName:Name, 1pAudio:Sound Base Object Reference, 3pAudio:Sound Base Object Reference, PrimarySound:bool=~) {is=title K2Node_FunctionEntry}
   > Set 1pSound(1pSound:Sound Base Object Reference=@n1.1pAudio) -> (Output_Get:Sound Base Object Reference)
   > Set 3pSound(3pSound:Sound Base Object Reference=@n1.3pAudio) -> (Output_Get:Sound Base Object Reference)
   > Set bPrimarySound(bPrimarySound:bool=@n1.PrimarySound) -> (Output_Get:bool)
   n2 = Get Animation() -> (Animation:Anim Montage Soft Object Reference)
   > n3 = LoadAsset_Blocking(Asset=@n2)
   > n4 = Cast To AnimMontage!(execute, Object=@n3) -> (then, CastFailed, AsAnim Montage:Anim Montage Object Reference) {is=title K2Node_DynamicCast}
   > n5 = Set AnimMontage(AnimMontage:Anim Montage Object Reference=@n4) -> (Output_Get:Anim Montage Object Reference)
   > n6 = Branch(Condition=IsValidAnimNotifyTrackName!(AnimationSequenceBase:Anim Sequence Base Object Reference=@n5, NotifyTrackName:Name=@n1) -> (ReturnValue:bool))
   > n7 = RemoveAnimationNotifyTrack!(execute, AnimationSequenceBase:Anim Sequence Base Object Reference=@n5, NotifyTrackName:Name=@n1) -> (then)
   AddAnimationNotifyTrack!(
    execute=[@n7.then, @n6.else]
    AnimationSequenceBase:Anim Sequence Base Object Reference=@n5
    NotifyTrackName:Name=@n1
    TrackColor:Linear Color Structure=(R=1,G=1,B=1,A=1)
   ) -> (then)
   > n8 = AddAnimationNotifyStateEvent!(
    execute
    AnimationSequenceBase:Anim Sequence Base Object Reference=@n5
    NotifyTrackName:Name=@n1
    StartTime:float=0
    Duration:float=Get SequenceLength!(self:Anim Sequence Base Object Reference=@n5) -> (SequenceLength:float)
    NotifyStateClass:Anim Notify State Class Reference=/Script/FortniteGame.FortAnimNotifyState_EmoteSound
   ) -> (then, ReturnValue:Anim Notify State Object Reference)
   > n9 = Is Valid(InputObject=@n8)
   ErrorLog!(execute=@n9.Is Not Valid, self, Message:"String (by ref)"=MakeLiteralString(Value=Unable to add notify)) -> (then)
   n10 = Get 1pSound() -> (1pSound:Sound Base Object Reference)
   SetEditorProperty(execute=@n9.Is Valid, Object=@n8, PropertyName=EmoteSound1P, PropertyValue:Sound Base Object Reference=@n10)
   > SetEditorProperty(Object=@n8, PropertyName=EmoteSound3P, PropertyValue=Get 3pSound() -> (3pSound:Sound Base Object Reference))
   n11 = Get bPrimarySound() -> (bPrimarySound:bool)
   > SetEditorProperty(Object=@n8, PropertyName=bPrimarySound, PropertyValue:bool=@n11)
   > n12 = Branch(Condition=@n11)
   SetEditorProperty(execute=@n12.else, Object=@n8, PropertyName=SoundName, PropertyValue:Name=String To Name(InString=GetDisplayName(Object=@n10)))
   ErrorLog(execute=@n4.CastFailed, Message=MakeLiteralString(Value=Unable to load animation))
   Resolve Soft Reference!(Input:Anim Montage Soft Object Reference=@n2) -> (Output:Anim Montage Object Reference)
  ]}
 ErrorLog {nodes [
   n1 = ErrorLog!() -> (then, Message:"String (by ref)") {is=title K2Node_FunctionEntry}
   > Show Message Dialog(
    Title=MakeLiteralText(Value="LOCTEXT(\"Error creating animation notify\")")
    Message="To Text (String)"(InString=Append(A=@n1, B=" - Contact tech audio for assistance"))
    MessageType=Ok
    DefaultValue=No
    MessageCategory=Warning
   )
  ]}
 EventGraph {nodes [
   "On Pressed (AnimButton)"() {is=title K2Node_ComponentBoundEvent}
   > n1 = LoadAsset(AssetPath=Get AnimPath() -> (AnimPath:String))
   > n2 = OpenEditorForAssets!(
    execute
    self:Asset Editor Subsystem Object Reference=Get AssetEditorSubsystem()
    Assets:Array of Object References=Make Array("[0]":Object Reference=@n1) -> (Array:Array of Object References)
    OpenedMethod:EAssetTypeActivationOpenedMethod Enum=Edit
   ) -> (then, ReturnValue:bool)
   > n3 = Branch(Condition=@n2)
   Show Message Dialog(
    execute=@n3.else
    Title=MakeLiteralText(Value="LOCTEXT(\"Failed To Open\")")
    Message=MakeLiteralText(Value="LOCTEXT(\"Asset failed to load and open. Nagivating to asset in content browser and load manually.\")")
    MessageType=Ok
    DefaultValue=No
    MessageCategory=Error
   )
   Event PreConstruct()
   > SetObject(self=Get AnimationView() -> (AnimationView:Single Property View Object Reference), NewObject=Self-Reference())
  ]}
 SetAnimText {nodes [
   SetAnimText() {is=title K2Node_FunctionEntry}
   n1 = Get AnimPath()
   > Return Node!(execute, ReturnValue:Text="To Text (String)"(InString=Split(SourceString=@n1, InStr=., SearchDir=FromEnd).RightS))
  ]}
]
variables [
 Animation {type=AnimMontage, default=/Game/Animation/Game/MainPlayer/Montages/Gadgets/Ability_C4_Detonator_Activate}
 AnimMontage {type=AnimMontage, default=null}
 AnimPath {type=string, default=None}
 As Anim Sequence Base {type=AnimSequenceBase, default=null}
 EmoteSettings {type=FortAnimNotifyState_EmoteSound, default=null}
 Notify Track Name {type=string, default=AudioTrack}
 SlContext {type=Class, default=null}
 SlEvent {type=GameplayTag, default {tagName=None}}
 SoundLibrarySettings {type=boolean}
]
widgets {
 info {rootWidgetClass=&Overlay, widgetCount=10}
 widgets [
  Overlay_141 {widgetClassPath=&Overlay, widget=Overlay_141}
  Border_1 {
   widgetClassPath=&Border
   parent=Overlay_141
   widget=Border_1
   properties {
    horizontalAlignment=HAlign_Left
    verticalAlignment=VAlign_Center
    background {
     bIsDynamicallyLoaded=false
     drawAs=Image
     imageType=NoImage
     mirroring=NoMirror
     resourceName=None
     resourceObject=null
     tiling=NoTile
     imageSize=<32 32>
     margin=<ltrb 0 0 0 0>
     outlineSettings {
      bUseBrushTransparency=false
      roundingType=HalfHeightRadius
      width=0
      color {colorUseRule=UseColor_Specified, specifiedColor=<rgba 0 0 0 0>}
      cornerRadii=<0 0 0 0>
     }
     tintColor {colorUseRule=UseColor_Specified, specifiedColor=<rgba 0.0208 0.0208 0.0208 1>}
     uVRegion {bIsValid=false, max=<0 0>, min=<0 0>}
    }
    padding=<ltrb 0 0 0 0>
   }
   slot {class=OverlaySlot, properties {horizontalAlignment=HAlign_Fill, verticalAlignment=VAlign_Fill}}
  }
  Border_0 {
   widgetClassPath=&Border
   parent=Border_1
   widget=Border_0
   properties {
    horizontalAlignment=HAlign_Left
    background {
     bIsDynamicallyLoaded=false
     drawAs=Border
     imageType=NoImage
     mirroring=NoMirror
     resourceName=None
     resourceObject=null
     tiling=NoTile
     imageSize=<32 32>
     margin=<ltrb 1 1 1 1>
     outlineSettings {
      bUseBrushTransparency=false
      roundingType=HalfHeightRadius
      width=0
      color {colorUseRule=UseColor_Specified, specifiedColor=<rgba 0 0 0 0>}
      cornerRadii=<0 0 0 0>
     }
     tintColor {colorUseRule=UseColor_Specified, specifiedColor=<rgba 0 0 0 1>}
     uVRegion {bIsValid=false, max=<0 0>, min=<0 0>}
    }
   }
   slot {class=BorderSlot, properties {horizontalAlignment=HAlign_Left, verticalAlignment=VAlign_Center, padding=<ltrb 0 0 0 0>}}
  }
  AnimButton {
   widgetClassPath=&EditorUtilityButton
   parent=Border_0
   widget=AnimButton
   slot {class=BorderSlot, properties {horizontalAlignment=HAlign_Left}}
  }
  SizeBox_2 {
   widgetClassPath=&SizeBox
   parent=AnimButton
   widget=SizeBox_2
   properties {bOverride_WidthOverride=true, heightOverride=1500, widthOverride=200}
   slot {class=ButtonSlot, properties {horizontalAlignment=HAlign_Fill, verticalAlignment=VAlign_Fill, padding=<ltrb 0 0 0 0>}}
  }
  VerticalBox_0 {
   widgetClassPath=&VerticalBox
   parent=SizeBox_2
   widget=VerticalBox_0
   slot {class=SizeBoxSlot, properties {horizontalAlignment=HAlign_Center, verticalAlignment=VAlign_Center, padding=<ltrb 4 2 4 2>}}
  }
  SizeBox_1 {
   widgetClassPath=&SizeBox
   parent=VerticalBox_0
   widget=SizeBox_1
   properties {bOverride_WidthOverride=true, widthOverride=96.1096}
   slot {class=VerticalBoxSlot, properties {horizontalAlignment=HAlign_Center, verticalAlignment=VAlign_Center}}
  }
  AnimationView {
   widgetClassPath=&SinglePropertyView
   parent=SizeBox_1
   widget=AnimationView
   properties {nameOverride=" ", propertyName=Animation, visibility=HitTestInvisible}
   slot {class=SizeBoxSlot}
  }
  SizeBox_3 {
   widgetClassPath=&SizeBox
   parent=VerticalBox_0
   widget=SizeBox_3
   slot {class=VerticalBoxSlot, properties {verticalAlignment=VAlign_Center}}
  }
  TextEntry {
   widgetClassPath=&TextBlock
   bIsVariable=true
   parent=SizeBox_3
   widget=TextEntry
   properties {
    autoWrapText=true
    text=Anim
    wrapTextAt=100
    wrappingPolicy=AllowPerCharacterWrapping
    colorAndOpacity {colorUseRule=UseColor_Specified, specifiedColor=<rgba 0.5885 0.5885 1 1>}
    font {
     bForceMonospaced=false
     bMaterialIsStencil=false
     fontMaterial=null
     fontObject=/LauncherSocial/BrutalType
     letterSpacing=0
     monospacedWidth=1
     size=6.75
     skewAmount=0
     typefaceFontName=Regular
     outlineSettings {
      bApplyOutlineToDropShadows=false
      bMiteredCorners=false
      bSeparateFillAlpha=false
      outlineMaterial=null
      outlineSize=0
      outlineColor=<rgba 0 0 0 1>
     }
    }
    shadowColorAndOpacity=<rgba 0 0 0 1>
    shadowOffset=<0.5 0.5>
   }
   slot {class=SizeBoxSlot}
  }
 ]
}
