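# Editor utility widget serving as the Emote panel of the cosmetic MetaSound preset creator.
# Toggles for 1P/3P music, foley and traversal register source references with the parent creator; callbacks apply sound class and concurrency to resulting sounds.
# Athena/Juno dance definitions supply montages, partly read by property name, for a preview grid.
# When the creator finishes, auto-add places sound tracks on previewed montages; a MetaSound builder sets the Looping input default.
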EditorUtilityWidgetBlueprint /Game/Sounds/Fort_Audio_Blueprints/Utilities/EditorUtilityWidgets/Metasounds/Cosmetic/EUW_MetaSound_Cosmetic_PresetCreator_Emote
graphs [
 AddAnimPreviewWidget {nodes [
   n1 = AddAnimPreviewWidget!() -> (then, Animation:Anim Montage Soft Object Reference, AnimCount:int) {is=title K2Node_FunctionEntry}
   > n2 = Construct Object from Class!(
    execute
    Class=&/EUW_MetaSound_Cosmetic_AnimPreview.EUW_MetaSound_Cosmetic_AnimPreview_C
    self=Get AnimationBox() -> (AnimationBox:Uniform Grid Panel Object Reference)
    AnimPath:String="To String (SoftObjectReference)"!(SoftObjectReference:"Soft Object Reference (by ref)"=@n1) -> (ReturnValue:String)
    Animation:Anim Montage Soft Object Reference=@n1
   ) -> (then, ReturnValue:EUW Meta Sound Cosmetic Anim Preview Object Reference)
   > n3 = AddChild(self=Get AnimationBox(), Content=@n2)
   > SetNameOverride!(
    execute
    self:Single Property View Object Reference=Get AnimationView!(self:EUW Meta Sound Cosmetic Anim Preview Object Reference=@n2) -> (AnimationView:Single Property View Object Reference)
    NewPropertyName:Text="LOCTEXT(\" \")"
   ) -> (then)
   > n4 = Cast To UniformGridSlot!(execute, Object=@n3) -> (then, CastFailed, AsUniform Grid Slot:Uniform Grid Slot Object Reference) {is=title K2Node_DynamicCast}
   > SetHorizontalAlignment(self:Uniform Grid Slot Object Reference=@n4, InHorizontalAlignment=HAlign_Fill)
   n5 = "% (Integer)"(A=@n1.AnimCount, B=3) -> (ReturnValue=0)
   > SetColumn!(execute, self:Uniform Grid Slot Object Reference=@n4, InColumn:int=@n5) -> (then)
   n6 = Get AnimGridRow() -> (AnimGridRow:int=0)
   > SetRow!(execute, self:Uniform Grid Slot Object Reference=@n4, InRow:int=@n6) -> (then)
   > Branch(Condition="Equal (Integer)"(A=@n5, B=2))
   > Increment Int(Value=@n6)
   "To String (SoftObjectReference)"()
  ]}
 AddExtraAnimation {nodes [
   AddExtraAnimation!() -> (then, PropertyName:Name) {is=title K2Node_FunctionEntry}
   n1 = Get AdditionalAnimMon() -> (AdditionalAnimMon:Anim Montage Soft Object Reference)
   > Branch(Condition=IsValidSoftObjectReference(SoftObjectReference=@n1))
   > n2 = For Each Loop with Break!(Exec:Exec, Array:Array of Widget Object References=GetAllChildren(self=Get AnimationBox()), Break:Exec=@n8.then) -> (LoopBody:Exec, Array Element:Widget Object Reference, Array Index:int, Completed:Exec)
   n3 = Get!(execute:Exec=@n2.Completed) -> (then:Exec, else:Exec, InList:bool)
   Set AdditionalAnimMon(execute=[@n5.then, @n3.then], AdditionalAnimMon:Anim Montage Soft Object Reference) -> (Output_Get:Anim Montage Soft Object Reference)
   n4 = Get AnimsArray() -> (AnimsArray:Array of Anim Montage Soft Object References)
   AddAnimPreviewWidget!(
    execute=@n3.else
    self
    Animation:Anim Montage Soft Object Reference=@n1
    AnimCount:int=Length(TargetArray:Array of Anim Montage Soft Object References=@n4)
   ) -> (then) {is=title K2Node_CallFunction}
   > n5 = Add(TargetArray:Array of Anim Montage Soft Object References=@n4, NewItem:"Anim Montage Soft Object Reference (by ref)"=@n1) {is=title K2Node_CallArrayFunction}
   n6 = Cast To EUW_MetaSound_Cosmetic_AnimPreview!(execute=@n2.LoopBody, Object=@n2) -> (then, CastFailed, AsEUW Meta Sound Cosmetic Anim Preview:EUW Meta Sound Cosmetic Anim Preview Object Reference) {is=title K2Node_DynamicCast}
   n7 = "Equal (SoftObjectReference)"(
    A:"Soft Object Reference (by ref)"=Get Animation!(self:EUW Meta Sound Cosmetic Anim Preview Object Reference=@n6) -> (Animation:Anim Montage Soft Object Reference)
    B:"Soft Object Reference (by ref)"=@n1
   ) -> (ReturnValue:bool=~) {is=title K2Node_PromotableOperator}
   > Branch(Condition=@n7)
   > n8 = Set InList(InList:bool=@n7) -> (Output_Get:bool)
  ]}
 EventGraph {nodes [
   n1 = "On Check State Changed (AthenaCheck)"!() -> (OutputDelegate, then, bIsChecked:bool) {is=title K2Node_ComponentBoundEvent}
   n2 = Branch(execute=[@n1.then, @n43.then], Condition=IsChecked(self=Get AthenaCheck() -> (AthenaCheck:Editor Utility Check Box Object Reference)))
   > n3 = Create Widget!(
    execute:Exec
    WorldContextObject:Object Reference
    Class:User Widget Class Reference=/Game/Sounds/Fort_Audio_Blueprints/Utilities/EditorUtilityWidgets/Metasounds/Tools/EUW_IDReference.EUW_IDReference_C
    OwningPlayer:Player Controller Object Reference
    IDSourceReference:Fort Item Definition Object Reference
   ) -> (then:Exec, ReturnValue:EUW IDReference Object Reference)
   > n4 = Set AthenaRef(AthenaRef:EUW IDReference Object Reference=@n3) -> (Output_Get:EUW IDReference Object Reference)
   > AddChild(self=Get AnimSelectionBox() -> (AnimSelectionBox:Horizontal Box Object Reference), Content=@n4)
   n5 = Get AthenaRef() -> (AthenaRef:EUW IDReference Object Reference)
   n6 = Get IDSource!(self:EUW IDReference Object Reference=@n5) -> (IDSource:Single Property View Object Reference)
   > SetNameOverride(self=@n6, NewPropertyName="LOCTEXT(\"Athena Emote\")")
   > Set Grid Slot Alignment!(
    execute:Exec
    Widget:Widget Object Reference=Get SizeBox_0!(self:EUW IDReference Object Reference=@n5) -> (SizeBox_0:Size Box Object Reference)
    Horizontal:EHorizontalAlignment Enum=HAlign_Left
    Vertical:EVerticalAlignment Enum=VAlign_Fill
   ) -> (then:Exec) {is=title K2Node_MacroInstance}
   > SetNameOverride(self=@n6, NewPropertyName="LOCTEXT(\"Athena\")")
   > SetPropertyName!(execute, self:Single Property View Object Reference=@n6, NewPropertyName:Name=AthenaDanceItemDefinition) -> (then)
   > SetObject(self=@n6, NewObject=Self-Reference())
   > BindSetup!(execute, self, Target:EUW IDReference Object Reference=Get AthenaRef(), Bind:bool=true) -> (then)
   n7 = Get!(execute=@n2.else) -> (then, else, AthenaRef:EUW IDReference Object Reference)
   > RemoveChild(self=Get AnimSelectionBox(), Content=@n7)
   > Set AthenaDanceItemDefinition(AthenaDanceItemDefinition:Athena Dance Item Definition Object Reference) -> (Output_Get:Athena Dance Item Definition Object Reference)
   > BindSetup(Target=@n7, Bind=false)
   n8 = "On Check State Changed (JunoCheck)"!() -> (OutputDelegate, then, bIsChecked:bool) {is=title K2Node_ComponentBoundEvent}
   > n9 = Branch(Condition=@n8.bIsChecked)
   > n10 = Create Widget(Class=&EUW_IDReference_C)
   > n11 = Set JunoRef(JunoRef:EUW IDReference Object Reference=@n10) -> (Output_Get:EUW IDReference Object Reference)
   > AddChild(self=Get AnimSelectionBox(), Content=@n11)
   n12 = Get IDSource(self=Get JunoRef() -> (JunoRef:EUW IDReference Object Reference))
   > SetNameOverride(self=@n12, NewPropertyName="LOCTEXT(\"Juno Emote\")")
   > SetNameOverride(self=@n12, NewPropertyName="LOCTEXT(\"Juno\")")
   > SetPropertyName(self=@n12, NewPropertyName=JunoDanceItemDefenition)
   > SetObject(self=@n12, NewObject=Self-Reference())
   > BindSetup(Target=Get JunoRef(), Bind=true)
   n13 = Get!(execute=@n9.else) -> (then, else, JunoRef:EUW IDReference Object Reference)
   > RemoveChild(self=Get AnimSelectionBox(), Content=@n13)
   > Set JunoDanceItemDefenition(JunoDanceItemDefenition:Juno Athena Dance Item Override Definition Object Reference) -> (Output_Get:Juno Athena Dance Item Override Definition Object Reference)
   > BindSetup(Target=@n13, Bind=false)
   "On Property Changed (Music1pView)"!() -> (OutputDelegate, then, PropertyName:Name=None) {is=title K2Node_ComponentBoundEvent}
   > SetMusic1p()
   "On Property Changed (Music3pView)"!() -> (OutputDelegate, then, PropertyName:Name=None) {is=title K2Node_ComponentBoundEvent}
   > SetMusic3p()
   "On Property Changed (FoleyView)"!() -> (OutputDelegate, then, PropertyName:Name=None) {is=title K2Node_ComponentBoundEvent}
   > SetFoley()
   "On Property Changed (TraversalView)"!() -> (OutputDelegate, then, PropertyName:Name=None) {is=title K2Node_ComponentBoundEvent}
   > SetTraversal()
   OnIDSourcePropertyChanged() {is=title K2Node_CustomEvent}
   > ClearChildren(self=Get AnimationBox())
   > Clear(TargetArray:Array of Anim Montage Soft Object References=Get AnimsArray())
   > Set AnimGridRow(AnimGridRow:int=0) -> (Output_Get:int=0)
   > Set AnimCount(AnimCount:int=0) -> (Output_Get:int=0)
   > n14 = Get!(execute) -> (then, else, AthenaDanceItemDefinition:Athena Dance Item Definition Object Reference)
   > n15 = For Each Loop(Array:Array of Names=Get Property Name() -> (Property Name:Array of Names)) -> (Array Element:Name)
   n16 = Get!(execute=[@n15.Completed, @n14.else]) -> (then, else, JunoDanceItemDefenition:Juno Athena Dance Item Override Definition Object Reference)
   > n17 = JunoAnim!(execute, self, Target:Juno Athena Dance Item Override Definition Object Reference=@n16) -> (then)
   n18 = Get AnimsArray()
   n19 = For Each Loop(Exec=[@n17.then, @n45.else, @n16.else], Array:Array of Anim Montage Soft Object References=@n18) -> (Array Element:Anim Montage Soft Object Reference)
   > AddAnimPreviewWidget(Animation=@n19, AnimCount=@n19.Array Index)
   n20 = GetEditorProperty!(execute:Exec=@n15.LoopBody, Object:Object Reference=@n14, PropertyName:Name=@n15) -> (then:Exec, PropertyValue:"Anim Montage Soft Object Reference (by ref)", ReturnValue:bool)
   > n21 = Branch(Condition=@n20.ReturnValue)
   > Set AnimCount(AnimCount=@n15.Array Index) -> (Output_Get=0)
   > Branch(Condition=IsValidSoftObjectReference(SoftObjectReference=@n20))
   > Add(TargetArray=Get AnimsArray(), NewItem=@n20)
   n22 = GetEditorProperty(execute=@n21.else, Object=@n14, PropertyName=@n15) -> (PropertyValue:Array of Fort Emote Mapping Structures)
   > Branch(Condition=@n22.ReturnValue)
   > n23 = For Each Loop(Array:Array of Fort Emote Mapping Structures=@n22) -> (Array Element:Fort Emote Mapping Structure)
   n24 = Break Fort Emote Mapping!(FortEmoteMapping:"Fort Emote Mapping Structure (by ref)"=@n23) -> (BodyType:EFortCustomBodyType Enum, Gender:EFortCustomGender Enum, EmoteMontage:Anim Montage Soft Object Reference) {is=title K2Node_BreakStruct}
   > Branch(Condition=IsValidSoftObjectReference(SoftObjectReference=@n24.EmoteMontage))
   > Add(TargetArray=Get AnimsArray(), NewItem=@n24.EmoteMontage)
   n25 = EmoteMusic1p!() -> (OutputDelegate, then, Sound:Sound Base Soft Object Reference, SourceReference:EUW Meta Sound Source Reference Object Reference) {is=title K2Node_CustomEvent}
   > Set Sound Properties!(
    execute
    self
    Sound:Sound Base Soft Object Reference=@n25.Sound
    SoundClass:Sound Class Object Reference=Get Music1pSoundClass() -> (Music1pSoundClass:Sound Class Object Reference)
    ConcurrencySet:Set of Sound Concurrency Object References=Get Music1pSoundConcurrencySet() -> (Music1pSoundConcurrencySet:Set of Sound Concurrency Object References)
   ) -> (then) {is=title K2Node_CallFunction}
   > Set 1pMusic(1pMusic:Sound Base Soft Object Reference=@n25.Sound) -> (Output_Get:Sound Base Soft Object Reference)
   n26 = OnCheckStateChanged_Event!() -> (OutputDelegate, then, bIsChecked:bool=~) {is=title K2Node_CustomEvent}
   > SetVisibility(
    self=Get AdditionalAnim() -> (AdditionalAnim:Single Property View Object Reference)
    InVisibility=Select(Option 0=Hidden, Option 1=Visible, Index=@n26.bIsChecked)
   )
   n27 = MetasoundsFinished() {is=title K2Node_CustomEvent}
   > Branch(Condition=Get bAutoAddState() -> (bAutoAddState:bool))
   > SetAnimNotifies()
   n28 = EmoteMusic3p!() -> (OutputDelegate, then, Sound:Sound Base Soft Object Reference, SourceReference:EUW Meta Sound Source Reference Object Reference) {is=title K2Node_CustomEvent}
   > Set Sound Properties(
    Sound=@n28.Sound
    SoundClass=Get Music3pSoundClass() -> (Music3pSoundClass:Sound Class Object Reference)
    ConcurrencySet=Get Music3pSoundConcurrencySet() -> (Music3pSoundConcurrencySet:Set of Sound Concurrency Object References)
   )
   > Set 3pMusic(3pMusic:Sound Base Soft Object Reference=@n28.Sound) -> (Output_Get:Sound Base Soft Object Reference)
   n29 = EmoteFoley!() -> (OutputDelegate, then, Sound:Sound Base Soft Object Reference, SourceReference:EUW Meta Sound Source Reference Object Reference) {is=title K2Node_CustomEvent}
   > Set Sound Properties(
    Sound=@n29.Sound
    SoundClass=Get FoleySoundClass() -> (FoleySoundClass:Sound Class Object Reference)
    ConcurrencySet=Get MusicFoleySoundConcurrencySet() -> (MusicFoleySoundConcurrencySet:Set of Sound Concurrency Object References)
   )
   > Set FoleyAudio(FoleyAudio:Sound Base Soft Object Reference=@n29.Sound) -> (Output_Get:Sound Base Soft Object Reference)
   n30 = EmoteTravseral!() -> (OutputDelegate, then, Sound:Sound Base Soft Object Reference, SourceReference:EUW Meta Sound Source Reference Object Reference) {is=title K2Node_CustomEvent}
   > Set Sound Properties(
    Sound=@n30.Sound
    SoundClass=Get TraversalSoundClass() -> (TraversalSoundClass:Sound Class Object Reference)
    ConcurrencySet=Get TraversalSoundConcurrencySet() -> (TraversalSoundConcurrencySet:Set of Sound Concurrency Object References)
   )
   > Set TraversalMusic(TraversalMusic:Sound Base Soft Object Reference=@n30.Sound) -> (Output_Get:Sound Base Soft Object Reference)
   n31 = EmoteSoundForAsset!() -> (OutputDelegate, then, Sound:Sound Base Soft Object Reference, SourceReference:EUW Meta Sound Source Reference Object Reference) {is=title K2Node_CustomEvent}
   > Set Is Looping(Is Looping:bool=false) -> (Output_Get:bool)
   > n32 = Cast To MetaSoundSource!(execute, Object=Resolve Soft Reference!(Input:Sound Base Soft Object Reference=@n31.Sound) -> (Output:Sound Base Object Reference)) -> (then, CastFailed, AsMetaSound Source:Meta Sound Source Object Reference) {is=title K2Node_DynamicCast}
   > n33 = For Each Loop(Array:Array of EUW Meta Sound Cosmetic Action Name Widget Object References=Get ActionWidgets!(self:EUW Meta Sound Cosmetic Preset Creator Object Reference=Get Creator Widget!() -> (Costmetic Creator:EUW Meta Sound Cosmetic Preset Creator Object Reference) {is=title K2Node_CallFunction}) -> (ActionWidgets:Array of EUW Meta Sound Cosmetic Action Name Widget Object References)) -> (Array Element:EUW Meta Sound Cosmetic Action Name Widget Object Reference)
   > n34 = For Each Loop(Array:Array of Widget Object References=GetAllChildren(self=Get ScrollBox!(self:EUW Meta Sound Cosmetic Action Name Widget Object Reference=@n33) -> (ScrollBox:Editor Utility Scroll Box Object Reference))) -> (Array Element:Widget Object Reference)
   n35 = Create MetaSound Literal from Boolean!(execute=@n34.Completed, Value:bool=Get Is Looping() -> (Is Looping:bool)) -> (then, ReturnValue:Meta Sound Literal Structure)
   > SetGraphInputDefault!(
    execute
    self:Meta Sound Builder Base Object Reference=Find Or Begin Building MetaSound Asset!(
     self:Meta Sound Editor Subsystem Object Reference=Get MetaSoundEditorSubsystem() -> (ReturnValue:Meta Sound Editor Subsystem Object Reference) {is=title K2Node_GetEditorSubsystem}
     MetaSound:Meta Sound Document Interface Interface=@n32
    ) -> (ReturnValue:Meta Sound Builder Base Object Reference)
    InputName:Name=Looping
    Literal:"Meta Sound Literal Structure (by ref)"=@n35
    PageName:Name
   ) -> (Succeeded:Exec, Failed:Exec)
   n36 = Cast To EUW_MetaSound_Cosmetic_SoundNameWidget!(execute=@n34.LoopBody, Object=@n34) -> (then, CastFailed, AsEUW Meta Sound Cosmetic Sound Name Widget:EUW Meta Sound Cosmetic Sound Name Widget Object Reference) {is=title K2Node_DynamicCast}
   n37 = Get IsLooping!(self:EUW Meta Sound Cosmetic Sound Name Widget Object Reference=@n36) -> (IsLooping:bool)
   > Branch(Condition=@n37)
   > Set Is Looping(Is Looping=@n37)
   n38 = JunoAnim!() -> (OutputDelegate, then, Target:Juno Athena Dance Item Override Definition Object Reference) {is=title K2Node_CustomEvent}
   > n39 = For Each Loop(Array:Array of Anim Montage Soft Object References=Make Array!(
     "[0]":Anim Montage Soft Object Reference=Get Animation(self:Juno Athena Dance Item Override Definition Object Reference=@n38.Target)
     "[1]":Anim Montage Soft Object Reference=Get AnimationFemaleOverride!(self:Juno Athena Dance Item Override Definition Object Reference=@n38.Target) -> (AnimationFemaleOverride:Anim Montage Soft Object Reference)
    ) -> (Array:Array of Anim Montage Soft Object References)) -> (Array Element:Anim Montage Soft Object Reference)
   > Branch(Condition=IsValidSoftObjectReference(SoftObjectReference=@n39))
   > Add(TargetArray=Get AnimsArray(), NewItem=@n39)
   n40 = BindSetup!() -> (OutputDelegate, then, Target:EUW IDReference Object Reference, Bind:bool=~) {is=title K2Node_CustomEvent}
   > n41 = Branch(Condition=@n40.Bind)
   n42 = Create Event()
   > Assign On Source Property Changed!(execute:Exec, self:EUW IDReference Object Reference=@n40.Target, Delegate:"Delegate (by ref)"=@n42) -> (then:Exec) {is=title K2Node_AssignDelegate}
   > SetVisibility(
    self=Get SaveAssetRef!(self:EUW IDReference Object Reference=@n40.Target) -> (SaveAssetRef:EUW IDSave Asset Object Reference)
    InVisibility=Collapsed
   )
   Unbind Event from On Source Property Changed(execute=@n41.else, self:EUW IDReference Object Reference=@n40.Target, Delegate=@n42) {is=title K2Node_RemoveDelegate}
   > OnIDSourcePropertyChanged() {is=title K2Node_CallFunction}
   n43 = InitialCheck() {is=title K2Node_CustomEvent}
   Event PreConstruct()
   > SetObject(
    self=[
     Get AutoStateVal() -> (AutoStateVal:Single Property View Object Reference)
     Get AdditionalAnim()
     Get Music3pDetailView() -> (Music3pDetailView:Details View Object Reference)
     Get FoleyDetailView() -> (FoleyDetailView:Details View Object Reference)
     Get TraversalDetailView() -> (TraversalDetailView:Details View Object Reference)
     Get Music1pDetailView() -> (Music1pDetailView:Details View Object Reference)
    ]
    NewObject=Self-Reference()
   )
   > SetObject(
    self=[
     Get Music3pView() -> (Music3pView:Single Property View Object Reference)
     Get AutoStateVal()
     Get TraversalView() -> (TraversalView:Single Property View Object Reference)
     Get FoleyView() -> (FoleyView:Single Property View Object Reference)
     Get Music1pView() -> (Music1pView:Single Property View Object Reference)
    ]
    NewObject=Self-Reference()
   )
   > Set CurrentCount(CurrentCount:int=0) -> (Output_Get:int=0)
   Event Construct()
   > Assign Sound for Asset!(
    execute:Exec
    self:EUW Meta Sound Source Reference Object Reference=Get SoundSourceReferenceEmote1p() -> (SoundSourceReferenceEmote1p:EUW Meta Sound Source Reference Object Reference)
    Delegate:"Delegate (by ref)"=Create Event()
   ) -> (then:Exec) {is=title K2Node_AssignDelegate}
   > Bind Event to Sound for Asset(
    self:EUW Meta Sound Source Reference Object Reference=Get SoundSourceReferenceEmote3p() -> (SoundSourceReferenceEmote3p:EUW Meta Sound Source Reference Object Reference)
    Delegate=Create Event()
   ) {is=title K2Node_AddDelegate}
   > Bind Event to Sound for Asset(
    self=Get SoundSourceReferenceEmoteFoley() -> (SoundSourceReferenceEmoteFoley:EUW Meta Sound Source Reference Object Reference)
    Delegate=Create Event()
   )
   > Bind Event to Sound for Asset(
    self=Get SoundSourceReferenceEmoteTraversal() -> (SoundSourceReferenceEmoteTraversal:EUW Meta Sound Source Reference Object Reference)
    Delegate=Create Event()
   )
   > Setup Visibility Views()
   > InitializeActionWidgets()
   > Assign Sound for Asset(self=Get SoundSourceReferenceEmoteFoley(), Delegate=Create Event())
   > InitialCheck() {is=title K2Node_CallFunction}
   > n44 = GetCosmeticCreator!(execute, Target:Widget Object Reference=Self-Reference(), __WorldContext:Object Reference) -> (then, AsEUW Meta Sound Cosmetic Preset Creator:EUW Meta Sound Cosmetic Preset Creator Object Reference)
   > Bind Event to Finish Setting Sounds(self:EUW Meta Sound Cosmetic Preset Creator Object Reference=@n44, Delegate=@n27) {is=title K2Node_AddDelegate}
   > Bind Event to On Property Changed(self:Property View Base Object Reference=Get AdditionalAnim(), Delegate=Create Event()) {is=title K2Node_AddDelegate}
   > Bind Event to On Check State Changed(self:Check Box Object Reference=Get ExtraCheck() -> (ExtraCheck:Editor Utility Check Box Object Reference), Delegate=Create Event()) {is=title K2Node_AddDelegate}
   n45 = Branch(Condition=Is Empty!(TargetArray:Array of Anim Montage Soft Object References=@n18) -> (ReturnValue:bool))
   SetCheckedState!(execute, self:Check Box Object Reference=Get AthenaCheck(), InCheckedState:ECheckBoxState Enum=Checked) -> (then)
   Get VerticalBox_86() -> (VerticalBox_86:Vertical Box Object Reference)
   Get VerticalBox_86()
  ]}
 Get Concurrency Set {nodes [
   n1 = Get Concurrency Set!() -> (then, SoundSource:EUW Meta Sound Source Reference Object Reference) {is=title K2Node_FunctionEntry}
   > Return Node!(
    execute
    ConcurrencySet:Set of Sound Concurrency Object References=Get ConcurrencySet!(self:Sound Base Object Reference=Get SourceReference!(self:EUW Meta Sound Source Reference Object Reference=@n1) -> (SourceReference:Meta Sound Source Object Reference)) -> (ConcurrencySet:Set of Sound Concurrency Object References)
   )
  ]}
 Get Creator Widget {nodes [
   Get Creator Widget() {is=title K2Node_FunctionEntry}
   n1 = GetParent!(self:Widget Object Reference) -> (ReturnValue:Panel Widget Object Reference)
   > n2 = Cast To EUW_MetaSound_Cosmetic_PresetCreator!(execute, Object=GetOuterObject!(Object:Object Reference=GetOuterObject(Object=@n1)) -> (ReturnValue:Object Reference)) -> (then, CastFailed, AsEUW Meta Sound Cosmetic Preset Creator:EUW Meta Sound Cosmetic Preset Creator Object Reference) {is=title K2Node_DynamicCast}
   > Return Node!(execute, Costmetic Creator:EUW Meta Sound Cosmetic Preset Creator Object Reference=@n2)
  ]}
 Get Sound Class {nodes [
   n1 = Get Sound Class!() -> (then, SoundSource:EUW Meta Sound Source Reference Object Reference) {is=title K2Node_FunctionEntry}
   > Return Node!(
    execute
    SoundClassObject:Sound Class Object Reference=Get SoundClassObject!(self:Sound Base Object Reference=Get SourceReference(self=@n1)) -> (SoundClassObject:Sound Class Object Reference)
   )
  ]}
 InitializeActionWidgets {nodes [
   InitializeActionWidgets() {is=title K2Node_FunctionEntry}
   > SetMusic1p()
   > SetMusic3p()
   > SetFoley()
   > SetTraversal()
  ]}
 Set Sound Properties {nodes [
   n1 = Set Sound Properties!() -> (
    then
    Sound:Sound Base Soft Object Reference
    SoundClass:Sound Class Object Reference
    ConcurrencySet:Set of Sound Concurrency Object References
   ) {is=title K2Node_FunctionEntry}
   n2 = Resolve Soft Reference(Input=@n1)
   > SetEditorProperty(Object=@n2, PropertyName=SoundClassObject, PropertyValue:Sound Class Object Reference=@n1.SoundClass)
   > SetEditorProperty(Object=@n2, PropertyName=ConcurrencySet, PropertyValue:Set of Sound Concurrency Object References=@n1.ConcurrencySet)
  ]}
 SetAnimNotifies {nodes [
   SetAnimNotifies() {is=title K2Node_FunctionEntry}
   > n1 = For Each Loop(Array:Array of Widget Object References=GetAllChildren(self=Get AnimationBox())) -> (Array Element:Widget Object Reference)
   > n2 = Cast To EUW_MetaSound_Cosmetic_AnimPreview(Object=@n1)
   > Set AnimPreview(AnimPreview:EUW Meta Sound Cosmetic Anim Preview Object Reference=@n2) -> (Output_Get:EUW Meta Sound Cosmetic Anim Preview Object Reference)
   > n3 = Sequence()
   n4 = Get 1pMusic() -> (1pMusic:Sound Base Soft Object Reference)
   n5 = Get 3pMusic() -> (3pMusic:Sound Base Soft Object Reference)
   > Branch(Condition=AND Boolean(A=IsValidSoftObjectReference(SoftObjectReference=@n4), B=IsValidSoftObjectReference(SoftObjectReference=@n5)))
   > AddEmoteTrack!(
    execute
    self:EUW Meta Sound Cosmetic Anim Preview Object Reference=Get AnimPreview() -> (AnimPreview:EUW Meta Sound Cosmetic Anim Preview Object Reference)
    TrackName:Name=EmoteMusic
    1pAudio:Sound Base Object Reference=Resolve Soft Reference(Input=@n4)
    3pAudio:Sound Base Object Reference=Resolve Soft Reference(Input=@n5)
    PrimarySound:bool=true
   ) -> (then)
   n6 = Sequence(execute=@n3.then_1)
   n7 = Get FoleyAudio() -> (FoleyAudio:Sound Base Soft Object Reference)
   > Branch(Condition=IsValidSoftObjectReference(SoftObjectReference=@n7))
   n8 = Resolve Soft Reference(Input=@n7)
   > AddEmoteTrack(self=Get AnimPreview(), TrackName=EmoteFoley, 1pAudio=@n8, 3pAudio=@n8, PrimarySound=false)
   Sequence(execute=@n6.then_1)
   n9 = Get TraversalMusic() -> (TraversalMusic:Sound Base Soft Object Reference)
   > Branch(Condition=IsValidSoftObjectReference(SoftObjectReference=@n9))
   n10 = Resolve Soft Reference(Input=@n9)
   > AddEmoteTrack(self=Get AnimPreview(), TrackName=EmoteTraversal, 1pAudio=@n10, 3pAudio=@n10, PrimarySound=false)
  ]}
 SetFoley {nodes [
   SetFoley() {is=title K2Node_FunctionEntry}
   > Widget Selector!(
    execute
    self
    ShouldAdd:bool=Get IsFoley() -> (IsFoley:bool)
    Reference:"EUW Meta Sound Source Reference Object Reference (by ref)"=Get SoundSourceReferenceEmoteFoley()
   ) -> (then)
   > Set FoleySoundClass(FoleySoundClass:Sound Class Object Reference=Get Sound Class!(SoundSource:EUW Meta Sound Source Reference Object Reference=Get SoundSourceReferenceEmoteFoley()) -> (SoundClassObject:Sound Class Object Reference) {is=title K2Node_CallFunction}) -> (Output_Get:Sound Class Object Reference)
   > Set MusicFoleySoundConcurrencySet(MusicFoleySoundConcurrencySet:Set of Sound Concurrency Object References=Get Concurrency Set!(SoundSource:EUW Meta Sound Source Reference Object Reference=Get SoundSourceReferenceEmoteFoley()) -> (ConcurrencySet:Set of Sound Concurrency Object References) {is=title K2Node_CallFunction}) -> (Output_Get:Set of Sound Concurrency Object References)
  ]}
 SetMusic1p {nodes [
   SetMusic1p() {is=title K2Node_FunctionEntry}
   > Widget Selector(ShouldAdd=Get IsMusic1p() -> (IsMusic1p:bool), Reference=Get SoundSourceReferenceEmote1p())
   > Set Music1pSoundClass(Music1pSoundClass:Sound Class Object Reference=Get Sound Class(SoundSource=Get SoundSourceReferenceEmote1p())) -> (Output_Get:Sound Class Object Reference)
   > Set Music1pSoundConcurrencySet(Music1pSoundConcurrencySet:Set of Sound Concurrency Object References=Get Concurrency Set(SoundSource=Get SoundSourceReferenceEmote1p())) -> (Output_Get:Set of Sound Concurrency Object References)
  ]}
 SetMusic3p {nodes [
   SetMusic3p() {is=title K2Node_FunctionEntry}
   > Widget Selector(ShouldAdd=Get IsMusic3p() -> (IsMusic3p:bool), Reference=Get SoundSourceReferenceEmote3p())
   > Set Music3pSoundClass(Music3pSoundClass:Sound Class Object Reference=Get Sound Class(SoundSource=Get SoundSourceReferenceEmote3p())) -> (Output_Get:Sound Class Object Reference)
   > Set Music3pSoundConcurrencySet(Music3pSoundConcurrencySet:Set of Sound Concurrency Object References=Get Concurrency Set(SoundSource=Get SoundSourceReferenceEmote3p())) -> (Output_Get:Set of Sound Concurrency Object References)
  ]}
 SetTraversal {nodes [
   SetTraversal() {is=title K2Node_FunctionEntry}
   > Widget Selector(ShouldAdd=Get IsTraversal() -> (IsTraversal:bool), Reference=Get SoundSourceReferenceEmoteTraversal())
   > Set TraversalSoundClass(TraversalSoundClass:Sound Class Object Reference=Get Sound Class(SoundSource=Get SoundSourceReferenceEmoteTraversal())) -> (Output_Get:Sound Class Object Reference)
   > Set TraversalSoundConcurrencySet(TraversalSoundConcurrencySet:Set of Sound Concurrency Object References=Get Concurrency Set(SoundSource=Get SoundSourceReferenceEmoteTraversal())) -> (Output_Get:Set of Sound Concurrency Object References)
  ]}
 Setup Visibility Views {nodes [
   Setup Visibility Views() {is=title K2Node_FunctionEntry}
   > SetVisibility(self=Get Music1pDetailView(), InVisibility=Select(Option 0=Collapsed, Option 1=Visible, Index=Get IsMusic1p()))
   > SetVisibility(self=Get Music3pDetailView(), InVisibility=Select(Option 0=Collapsed, Option 1=Visible, Index=Get IsMusic3p()))
   > SetVisibility(self=Get FoleyDetailView(), InVisibility=Select(Option 0=Collapsed, Option 1=Visible, Index=Get IsFoley()))
   > SetVisibility(self=Get TraversalDetailView(), InVisibility=Select(Option 0=Collapsed, Option 1=Visible, Index=Get IsTraversal()))
  ]}
 TriggerNotifySetUp {nodes [TriggerNotifySetUp() {is=title K2Node_FunctionEntry}]}
 Widget Selector {nodes [
   n1 = Widget Selector!() -> (then, ShouldAdd:bool=~, Reference:"EUW Meta Sound Source Reference Object Reference (by ref)") {is=title K2Node_FunctionEntry}
   > n2 = Branch(Condition=@n1)
   n3 = Get SoundSourceReferenceWidgets() -> (SoundSourceReferenceWidgets:Array of EUW Meta Sound Source Reference Object References)
   > n4 = Add Unique(
    TargetArray:Array of EUW Meta Sound Source Reference Object References=@n3
    NewItem:"EUW Meta Sound Source Reference Object Reference (by ref)"=@n1.Reference
   )
   Set members in Struct Meta Sound Source Reference!(
    execute:Exec=[@n5.then, @n4.then]
    StructRef:"Struct Meta Sound Source Reference Structure (by ref)"=Get NamesDataStructure!(self:EUW Meta Sound Source Reference Object Reference=Get SoundSourceReferenceEmote3p()) -> (NamesDataStructure:Struct Meta Sound Source Reference Structure)
    SkipActionName_27_#1:bool=Get IsMusic1p()
   ) -> (then:Exec, StructOut:"Struct Meta Sound Source Reference Structure (by ref)") {is=title K2Node_SetFieldsInStruct}
   > Set Current Source Reference Widgets!(
    execute
    self:EUW Meta Sound Cosmetic Preset Creator Object Reference=Get Creator Widget() {is=title K2Node_CallFunction}
    CurrentActiveWidget:Widget Object Reference=Self-Reference()
    SourceReferences:Array of EUW Meta Sound Source Reference Object References=Get SoundSourceReferenceWidgets()
    CosmeticType:String=Emote
   ) -> (then) {is=title K2Node_CallFunction}
   > Setup Visibility Views() {is=title K2Node_CallFunction}
   n5 = Remove Item!(
    execute=@n2.else
    TargetArray:Array of EUW Meta Sound Source Reference Object References=@n3
    Item:"EUW Meta Sound Source Reference Object Reference (by ref)"=@n1.Reference
   ) -> (then, ReturnValue:bool)
  ]}
]
variables [
 1pMusic {type=SoundBase, default=null}
 3pMusic {type=SoundBase, default=null}
 Accumilate {type=boolean}
 AdditionalAnimMon {type=AnimMontage, default=null}
 Animations {type="map<AnimMontage>", default {}}
 AnimCount {type=integer}
 AnimGridRow {type=integer}
 AnimsArray {type="array<AnimMontage>", default []}
 AthenaDanceItemDefinition {type=AthenaDanceItemDefinition, default=null}
 AthenaRef {type=EUW_IDReference_C, default=null}
 bAutoAddState {type=boolean, default=true}
 Compare With {type=integer}
 CurrentCount {type=integer}
 FoleyAudio {type=SoundBase, default=null}
 FoleySoundClass {type=SoundClass, default=&Fort_Music_Emote, class=~, graph_type=~, children=~, properties=~}
 Is Looping {type=boolean}
 IsFoley {type=boolean}
 IsMusic1p {type=boolean, default=true}
 IsMusic3p {type=boolean, default=true}
 IsTraversal {type=boolean}
 JunoDanceItemDefenition {type=JunoAthenaDanceItemOverrideDefinition, default=null}
 JunoRef {type=EUW_IDReference_C, default=null}
 Music1pSoundClass {type=SoundClass, default=/Game/Sounds/FortSoundClassesAndModes/Fort_Music_Emote_1P_PSM, class=~, graph_type=~, children=~, properties=~}
 Music1pSoundConcurrencySet {type="array<SoundConcurrency>", default [/Game/Sounds/Concurrency_Rules/Athena_EmoteMusic_1P_Conc]}
 Music3pSoundClass {type=SoundClass, default=/Game/Sounds/FortSoundClassesAndModes/Fort_Music_Emote_3P, class=~, graph_type=~, children=~, properties=~}
 Music3pSoundConcurrencySet {type="array<SoundConcurrency>", default [/Game/Sounds/Concurrency_Rules/Athena_EmoteMusic_Conc]}
 MusicFoleySoundConcurrencySet {type="array<SoundConcurrency>", default [/Game/Sounds/Concurrency_Rules/Athena_EmoteFoley_Conc]}
 NewVar {type=VariantSwapMontageActiveSwapsRequirement, default {item=null, variantSelections []}}
 Property Name {
  type="array<string>"
  default [Animation, AnimationFemaleOverride, AnimationOverrides, FrontEndAnimation, FrontEndAnimationFemaleOverride, FrontEndAnimationOverrides]
 }
 SoundSourceReferenceEmote1p {
  type=EUW_MetaSoundSourceReference_C
  default=&/EUW_MetaSound_Cosmetic_PresetCreator_Emote.Default__EUW_MetaSound_Cosmetic_PresetCreator_Emote_C:EUW_MetaSoundSourceReference_C_0
 }
 SoundSourceReferenceEmote3p {
  type=EUW_MetaSoundSourceReference_C
  default=&Default__EUW_MetaSound_Cosmetic_PresetCreator_Emote_C:EUW_MetaSoundSourceReference_C_1
 }
 SoundSourceReferenceEmoteFoley {
  type=EUW_MetaSoundSourceReference_C
  default=&Default__EUW_MetaSound_Cosmetic_PresetCreator_Emote_C:EUW_MetaSoundSourceReference_C_4
 }
 SoundSourceReferenceEmoteTraversal {
  type=EUW_MetaSoundSourceReference_C
  default=&Default__EUW_MetaSound_Cosmetic_PresetCreator_Emote_C:EUW_MetaSoundSourceReference_C_2
 }
 SoundSourceReferenceWidgets {type="array<EUW_MetaSoundSourceReference_C>", default []}
 TraversalMusic {type=SoundBase, default=null}
 TraversalSoundClass {type=SoundClass, default=&Fort_Music_Emote, class=~, graph_type=~, children=~, properties=~}
 TraversalSoundConcurrencySet {type="array<SoundConcurrency>", default [&Athena_EmoteFoley_Conc]}
]
widgets {
 info {rootWidgetClass=&CanvasPanel, widgetCount=38}
 widgets [
  CanvasPanel_0 {widgetClassPath=&CanvasPanel, widget=CanvasPanel_0, properties {bOverrideAccessibleDefaults=true}}
  GridPanel_0 {
   widgetClassPath=&GridPanel
   parent=CanvasPanel_0
   widget=GridPanel_0
   slot {
    class=CanvasPanelSlot
    properties {bAutoSize=true, layoutData {alignment=<0 0>, anchors {maximum=<0 0>, minimum=<0 0>}, offsets=<ltrb 0 0 500 300>}}
   }
  }
  GridPanel_183 {widgetClassPath=&GridPanel, parent=GridPanel_0, widget=GridPanel_183, slot {class=GridSlot}}
  HorizontalBox_116 {
   widgetClassPath=&HorizontalBox
   parent=GridPanel_183
   widget=HorizontalBox_116
   properties {toolTipText=Action for Metasound data.}
   slot {class=GridSlot, properties {horizontalAlignment=HAlign_Left, verticalAlignment=VAlign_Top, nudge=<0 1>, padding=<ltrb 2 2 2 5>}}
  }
  Music1pView {
   widgetClassPath=&SinglePropertyView
   parent=HorizontalBox_116
   widget=Music1pView
   properties {nameOverride=Music 1p:, propertyName=IsMusic1p}
   slot {class=HorizontalBoxSlot}
  }
  Music3pView {
   widgetClassPath=&SinglePropertyView
   parent=HorizontalBox_116
   widget=Music3pView
   properties {nameOverride=Music 3p:, propertyName=IsMusic3p}
   slot {class=HorizontalBoxSlot}
  }
  FoleyView {
   widgetClassPath=&SinglePropertyView
   parent=HorizontalBox_116
   widget=FoleyView
   properties {nameOverride=Foley:, propertyName=IsFoley}
   slot {class=HorizontalBoxSlot}
  }
  TraversalView {
   widgetClassPath=&SinglePropertyView
   parent=HorizontalBox_116
   widget=TraversalView
   properties {nameOverride=Traversal:, propertyName=IsTraversal}
   slot {class=HorizontalBoxSlot}
  }
  SizeBox_0 {
   widgetClassPath=&SizeBox
   parent=GridPanel_183
   widget=SizeBox_0
   properties {bOverride_MaxDesiredHeight=true, bOverride_MinDesiredHeight=true, heightOverride=200, maxDesiredHeight=350, minDesiredHeight=200}
   slot {class=GridSlot, properties {row=2}}
  }
  EditorUtilityScrollBox_71 {widgetClassPath=&EditorUtilityScrollBox, parent=SizeBox_0, widget=EditorUtilityScrollBox_71, slot {class=SizeBoxSlot}}
  Music1pDetailView {
   widgetClassPath=&DetailsView
   parent=EditorUtilityScrollBox_71
   widget=Music1pDetailView
   properties {categoriesToShow [Music 1p]}
   slot {class=ScrollBoxSlot}
  }
  Music3pDetailView {
   widgetClassPath=&DetailsView
   parent=EditorUtilityScrollBox_71
   widget=Music3pDetailView
   properties {categoriesToShow [Music 3p]}
   slot {class=ScrollBoxSlot}
  }
  FoleyDetailView {
   widgetClassPath=&DetailsView
   parent=EditorUtilityScrollBox_71
   widget=FoleyDetailView
   properties {categoriesToShow [Foley]}
   slot {class=ScrollBoxSlot}
  }
  TraversalDetailView {
   widgetClassPath=&DetailsView
   parent=EditorUtilityScrollBox_71
   widget=TraversalDetailView
   properties {categoriesToShow [Traversal]}
   slot {class=ScrollBoxSlot}
  }
  SizeBox {
   widgetClassPath=&SizeBox
   parent=GridPanel_183
   widget=SizeBox
   properties {bOverride_HeightOverride=true, bOverride_WidthOverride=true, heightOverride=36, minDesiredHeight=200}
   slot {class=GridSlot, properties {row=1}}
  }
  EditorUtilityMultiLineEditableText_0 {
   widgetClassPath=/Script/Blutility.EditorUtilityMultiLineEditableText
   bIsVariable=true
   parent=SizeBox
   widget=EditorUtilityMultiLineEditableText_0
   properties {text=Intro and Looping sounds should be split into separate assets. Run tool multiple times to get each asset type.}
   slot {class=SizeBoxSlot, properties {padding=<ltrb 2 2 2 2>}}
  }
  VerticalBox_0 {
   widgetClassPath=&VerticalBox
   bIsVariable=true
   parent=GridPanel_0
   widget=VerticalBox_0
   slot {class=GridSlot, properties {row=2}}
  }
  Border_0 {
   widgetClassPath=&Border
   parent=VerticalBox_0
   widget=Border_0
   properties {brushColor=<rgba 0.0087 0.0087 0.0087 1>}
   slot {class=VerticalBoxSlot, properties {horizontalAlignment=HAlign_Left}}
  }
  HorizontalBox_1 {widgetClassPath=&HorizontalBox, parent=Border_0, widget=HorizontalBox_1, slot {class=BorderSlot}}
  Border_212 {
   widgetClassPath=&Border
   parent=HorizontalBox_1
   widget=Border_212
   properties {brushColor=<rgba 0.0348 0.0348 0.0348 1>}
   slot {class=HorizontalBoxSlot, properties {padding=<ltrb 5 0 0 0>}}
  }
  HorizontalBox_2 {
   widgetClassPath=&HorizontalBox
   parent=Border_212
   widget=HorizontalBox_2
   properties {toolTipText=Show athena anim previews}
   slot {class=BorderSlot}
  }
  EditorUtilityEditableText_252 {
   widgetClassPath=/Script/Blutility.EditorUtilityEditableText
   bIsVariable=true
   parent=HorizontalBox_2
   widget=EditorUtilityEditableText_252
   properties {
    text="Athena Emotes "
    widgetStyle {
     backgroundImageComposing {
      bIsDynamicallyLoaded=false
      drawAs=Border
      imageType=FullColor
      mirroring=NoMirror
      resourceName=../../../Engine/Content/Slate/Old/HyperlinkDotted.png
      resourceObject=null
      tiling=Both
      imageSize=<0 0>
      margin=<ltrb 0 0 0 0.1875>
      outlineSettings {
       bUseBrushTransparency=false
       roundingType=HalfHeightRadius
       width=0
       color {colorUseRule=UseColor_Specified, specifiedColor=<rgba 0 0 0 0>}
       cornerRadii=<0 0 0 0>
      }
      tintColor {colorUseRule=UseColor_Specified, specifiedColor=<rgba 1 1 1 1>}
      uVRegion {bIsValid=false, max=<0 0>, min=<0 0>}
     }
     backgroundImageSelected {
      bIsDynamicallyLoaded=false
      drawAs=Image
      imageType=NoImage
      mirroring=NoMirror
      resourceName=None
      resourceObject=null
      tiling=NoTile
      imageSize=<0 0>
      margin=<ltrb 0 0 0 0>
      outlineSettings {
       bUseBrushTransparency=false
       roundingType=HalfHeightRadius
       width=0
       color {colorUseRule=UseColor_Specified, specifiedColor=<rgba 0 0 0 0>}
       cornerRadii=<0 0 0 0>
      }
      tintColor {colorUseRule=UseColor_Specified, specifiedColor=<rgba 0 0.162 0.7454 1>}
      uVRegion {bIsValid=false, max=<0 0>, min=<0 0>}
     }
     caretImage {
      bIsDynamicallyLoaded=false
      drawAs=Image
      imageType=NoImage
      mirroring=NoMirror
      resourceName=None
      resourceObject=null
      tiling=NoTile
      imageSize=<0 0>
      margin=<ltrb 0 0 0 0>
      outlineSettings {
       bUseBrushTransparency=false
       roundingType=HalfHeightRadius
       width=0
       color {colorUseRule=UseColor_Specified, specifiedColor=<rgba 0 0 0 0>}
       cornerRadii=<0 0 0 0>
      }
      tintColor {colorUseRule=UseColor_Specified, specifiedColor=<rgba 0.0636 0 1 1>}
      uVRegion {bIsValid=false, max=<0 0>, min=<0 0>}
     }
     colorAndOpacity {colorUseRule=UseColor_Foreground, specifiedColor=<rgba 1 0 1 1>}
     font {
      bForceMonospaced=false
      bMaterialIsStencil=false
      fontMaterial=null
      fontObject=null
      letterSpacing=0
      monospacedWidth=1
      size=9
      skewAmount=0
      typefaceFontName=Regular
      outlineSettings {
       bApplyOutlineToDropShadows=false
       bMiteredCorners=false
       bSeparateFillAlpha=false
       outlineMaterial=null
       outlineSize=0
       outlineColor=<rgba 0 0 0 1>
      }
     }
     hintColorAndOpacity {colorUseRule=UseColor_Foreground, specifiedColor=<rgba 1 0 1 1>}
     selectedBackgroundColor {colorUseRule=UseColor_Foreground, specifiedColor=<rgba 1 0 1 1>}
    }
   }
   slot {class=HorizontalBoxSlot, properties {verticalAlignment=VAlign_Center}}
  }
  AthenaCheck {
   widgetClassPath=&EditorUtilityCheckBox
   bIsVariable=true
   parent=HorizontalBox_2
   widget=AthenaCheck
   properties {bOverrideAccessibleDefaults=true, checkedState=Checked}
   slot {class=HorizontalBoxSlot}
  }
  Border {
   widgetClassPath=&Border
   parent=HorizontalBox_1
   widget=Border
   properties {brushColor=<rgba 0.0348 0.0348 0.0348 1>}
   slot {class=HorizontalBoxSlot, properties {padding=<ltrb 5 0 5 0>}}
  }
  HorizontalBox_3 {
   widgetClassPath=&HorizontalBox
   parent=Border
   widget=HorizontalBox_3
   properties {toolTipText=Show juno anim previews}
   slot {class=BorderSlot}
  }
  EditorUtilityEditableText {
   widgetClassPath=&EditorUtilityEditableText
   bIsVariable=true
   parent=HorizontalBox_3
   widget=EditorUtilityEditableText
   properties {text="Juno Emotes "}
   slot {class=HorizontalBoxSlot, properties {verticalAlignment=VAlign_Center}}
  }
  JunoCheck {
   widgetClassPath=&EditorUtilityCheckBox
   bIsVariable=true
   parent=HorizontalBox_3
   widget=JunoCheck
   slot {class=HorizontalBoxSlot}
  }
  Border_1 {
   widgetClassPath=&Border
   parent=HorizontalBox_1
   widget=Border_1
   properties {brushColor=<rgba 0.0348 0.0348 0.0348 1>}
   slot {class=HorizontalBoxSlot, properties {padding=<ltrb 5 0 0 0>}}
  }
  HorizontalBox {
   widgetClassPath=&HorizontalBox
   parent=Border_1
   widget=HorizontalBox
   properties {toolTipText=Show athena anim previews}
   slot {class=BorderSlot}
  }
  EditorUtilityEditableText_1 {
   widgetClassPath=&EditorUtilityEditableText
   bIsVariable=true
   parent=HorizontalBox
   widget=EditorUtilityEditableText_1
   properties {
    text="Extra Emotes "
    widgetStyle {
     backgroundImageComposing {
      bIsDynamicallyLoaded=false
      drawAs=Border
      imageType=FullColor
      mirroring=NoMirror
      resourceName=../../../Engine/Content/Slate/Old/HyperlinkDotted.png
      resourceObject=null
      tiling=Both
      imageSize=<0 0>
      margin=<ltrb 0 0 0 0.1875>
      outlineSettings {
       bUseBrushTransparency=false
       roundingType=HalfHeightRadius
       width=0
       color {colorUseRule=UseColor_Specified, specifiedColor=<rgba 0 0 0 0>}
       cornerRadii=<0 0 0 0>
      }
      tintColor {colorUseRule=UseColor_Specified, specifiedColor=<rgba 1 1 1 1>}
      uVRegion {bIsValid=false, max=<0 0>, min=<0 0>}
     }
     backgroundImageSelected {
      bIsDynamicallyLoaded=false
      drawAs=Image
      imageType=NoImage
      mirroring=NoMirror
      resourceName=None
      resourceObject=null
      tiling=NoTile
      imageSize=<0 0>
      margin=<ltrb 0 0 0 0>
      outlineSettings {
       bUseBrushTransparency=false
       roundingType=HalfHeightRadius
       width=0
       color {colorUseRule=UseColor_Specified, specifiedColor=<rgba 0 0 0 0>}
       cornerRadii=<0 0 0 0>
      }
      tintColor {colorUseRule=UseColor_Specified, specifiedColor=<rgba 0 0.162 0.7454 1>}
      uVRegion {bIsValid=false, max=<0 0>, min=<0 0>}
     }
     caretImage {
      bIsDynamicallyLoaded=false
      drawAs=Image
      imageType=NoImage
      mirroring=NoMirror
      resourceName=None
      resourceObject=null
      tiling=NoTile
      imageSize=<0 0>
      margin=<ltrb 0 0 0 0>
      outlineSettings {
       bUseBrushTransparency=false
       roundingType=HalfHeightRadius
       width=0
       color {colorUseRule=UseColor_Specified, specifiedColor=<rgba 0 0 0 0>}
       cornerRadii=<0 0 0 0>
      }
      tintColor {colorUseRule=UseColor_Specified, specifiedColor=<rgba 0.0636 0 1 1>}
      uVRegion {bIsValid=false, max=<0 0>, min=<0 0>}
     }
     colorAndOpacity {colorUseRule=UseColor_Foreground, specifiedColor=<rgba 1 0 1 1>}
     font {
      bForceMonospaced=false
      bMaterialIsStencil=false
      fontMaterial=null
      fontObject=null
      letterSpacing=0
      monospacedWidth=1
      size=9
      skewAmount=0
      typefaceFontName=Regular
      outlineSettings {
       bApplyOutlineToDropShadows=false
       bMiteredCorners=false
       bSeparateFillAlpha=false
       outlineMaterial=null
       outlineSize=0
       outlineColor=<rgba 0 0 0 1>
      }
     }
     hintColorAndOpacity {colorUseRule=UseColor_Foreground, specifiedColor=<rgba 1 0 1 1>}
     selectedBackgroundColor {colorUseRule=UseColor_Foreground, specifiedColor=<rgba 1 0 1 1>}
    }
   }
   slot {class=HorizontalBoxSlot, properties {verticalAlignment=VAlign_Center}}
  }
  ExtraCheck {
   widgetClassPath=&EditorUtilityCheckBox
   bIsVariable=true
   parent=HorizontalBox
   widget=ExtraCheck
   properties {bOverrideAccessibleDefaults=true}
   slot {class=HorizontalBoxSlot}
  }
  AutoStateVal {
   widgetClassPath=&SinglePropertyView
   parent=HorizontalBox_1
   widget=AutoStateVal
   properties {propertyName=bAutoAddState}
   slot {class=HorizontalBoxSlot}
  }
  VerticalBox_86 {widgetClassPath=&VerticalBox, bIsVariable=true, parent=VerticalBox_0, widget=VerticalBox_86, slot {class=VerticalBoxSlot}}
  SizeBox_1 {
   widgetClassPath=&SizeBox
   parent=VerticalBox_0
   widget=SizeBox_1
   properties {
    bOverride_MaxDesiredHeight=true
    bOverride_MinDesiredHeight=true
    bOverride_WidthOverride=true
    heightOverride=150
    maxDesiredHeight=300
    minDesiredHeight=150
    widthOverride=400
   }
   slot {class=VerticalBoxSlot}
  }
  EditorUtilityScrollBox_0 {widgetClassPath=&EditorUtilityScrollBox, parent=SizeBox_1, widget=EditorUtilityScrollBox_0, slot {class=SizeBoxSlot}}
  AnimSelectionBox {
   widgetClassPath=&HorizontalBox
   bIsVariable=true
   parent=EditorUtilityScrollBox_0
   widget=AnimSelectionBox
   slot {class=ScrollBoxSlot}
  }
  AdditionalAnim {
   widgetClassPath=&SinglePropertyView
   parent=EditorUtilityScrollBox_0
   widget=AdditionalAnim
   properties {nameOverride=Extra, propertyName=AdditionalAnimMon, visibility=Hidden}
   slot {class=ScrollBoxSlot}
  }
  AnimationBox {
   widgetClassPath=/Script/UMG.UniformGridPanel
   bIsVariable=true
   parent=EditorUtilityScrollBox_0
   widget=AnimationBox
   slot {class=ScrollBoxSlot}
  }
 ]
}
