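# Glider panel of the MetaSound cosmetic preset creator, an editor utility widget for assigning open, close and thrust-loop sounds to glider item definitions.
# Checkboxes choose which source reference widgets register with the parent creator, found through the widget's outer chain.
# Sound delegates write into the selected glider, and a retriggerable delay batches saving.
# A legacy check reads ThrustSoundParams via editor properties and offers to clear outdated entries.
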EditorUtilityWidgetBlueprint /Game/Sounds/Fort_Audio_Blueprints/Utilities/EditorUtilityWidgets/Metasounds/Cosmetic/EUW_MetaSound_Cosmetic_PresetCreator_Glider
graphs [
 EventGraph {nodes [
   "On Property Changed (OpenView)"!() -> (OutputDelegate, then, PropertyName:Name=None) {is=title K2Node_ComponentBoundEvent}
   > Widget Selector!(
    execute
    self
    ShouldAdd:bool=Get IsOpen() -> (IsOpen:bool)
    Reference:"EUW Meta Sound Source Reference Object Reference (by ref)"=Get SoundSourceReferenceOpen() -> (SoundSourceReferenceOpen:EUW Meta Sound Source Reference Object Reference)
   ) -> (then)
   "On Property Changed (CloseView)"!() -> (OutputDelegate, then, PropertyName:Name=None) {is=title K2Node_ComponentBoundEvent}
   > Widget Selector(
    ShouldAdd=Get IsClose() -> (IsClose:bool)
    Reference=Get SoundSourceReferenceClose() -> (SoundSourceReferenceClose:EUW Meta Sound Source Reference Object Reference)
   )
   "On Property Changed (ThrustView)"!() -> (OutputDelegate, then, PropertyName:Name=None) {is=title K2Node_ComponentBoundEvent}
   > Widget Selector(
    ShouldAdd=Get IsThrust() -> (IsThrust:bool)
    Reference=Get SoundSourceReferenceEmoteThrust() -> (SoundSourceReferenceEmoteThrust:EUW Meta Sound Source Reference Object Reference)
   )
   CheckForOldImp() {is=title K2Node_CustomEvent}
   > Clear(TargetArray=Get ThustStringArray() -> (ThustStringArray:Array of Strings))
   > n1 = Get!(execute:Exec) -> (then:Exec, else:Exec, Glider Item Definition:Athena Glider Item Definition Object Reference)
   > n2 = GetEditorProperty!(execute:Exec, Object:Object Reference=@n1, PropertyName:Name=ThrustSoundParams) -> (then:Exec, PropertyValue:Map of ELayeredAudioTriggerDirection Enums to Fort Glider Layered Audio Float Param Structures, ReturnValue:bool)
   > Branch(Condition=@n2.ReturnValue)
   > n3 = Set ThrustParams(ThrustParams:Map of ELayeredAudioTriggerDirection Enums to Fort Glider Layered Audio Float Param Structures=@n2) -> (Output_Get:Map of ELayeredAudioTriggerDirection Enums to Fort Glider Layered Audio Float Param Structures=AnyDirection)
   > n4 = Branch(Condition=Is Empty!(TargetMap:Map of ELayeredAudioTriggerDirection Enums to Fort Glider Layered Audio Float Param Structures=@n3) -> (ReturnValue:bool) {is=title K2Node_CallFunction})
   n5 = Keys!(execute=@n4.else, TargetMap:Map of ELayeredAudioTriggerDirection Enums to Fort Glider Layered Audio Float Param Structures=@n3) -> (then, Keys:Array of ELayeredAudioTriggerDirection Enums)
   > n6 = For Each Loop(Array:Array of ELayeredAudioTriggerDirection Enums=@n5) -> (Array Element:ELayeredAudioTriggerDirection Enum)
   n7 = Append!(
    A="Thrust params have been detected on "
    B=GetDisplayName(Object=@n1)
    C:String=", this glider has the following settings:"
    D:String="\r\n\r\n"
    E:String=JoinStringArray(SourceArray=Get ThustStringArray(), Separator="\r\n")
    F:String="\r\n\r\n"
    G:String="These parameters are for an older implementation standard, and may no longer be needed. Would you like to automatically remove them?"
   ) -> (ReturnValue)
   n8 = Show Message Dialog(
    execute=@n6.Completed
    Title=MakeLiteralText(Value="LOCTEXT(\"Old Glider Thrust Implementation Detected\")")
    Message=MakeLiteralText(Value="To Text (String)"(InString=@n7))
    MessageType=YesNo
    DefaultValue=Yes
    MessageCategory=Warning
   )
   > n9 = Switch on EAppReturnType(Selection=@n8)
   n10 = Get ThrustParams() -> (ThrustParams:Map of ELayeredAudioTriggerDirection Enums to Fort Glider Layered Audio Float Param Structures=AnyDirection)
   Clear!(execute=@n9.Yes, TargetMap:Map of ELayeredAudioTriggerDirection Enums to Fort Glider Layered Audio Float Param Structures=@n10) -> (then) {is=title K2Node_CallFunction}
   n11 = Get Glider Item Definition() -> (Glider Item Definition:Athena Glider Item Definition Object Reference)
   > SetEditorProperty(
    Object=@n11
    PropertyName=ThrustSoundParams
    PropertyValue:Map of ELayeredAudioTriggerDirection Enums to Fort Glider Layered Audio Float Param Structures=@n10
   )
   > SaveIdAsset!(
    execute
    self:EUW IDSave Asset Object Reference=Get SaveAssetRef!(self:EUW IDReference Object Reference=Get IDReference() -> (IDReference:EUW IDReference Object Reference)) -> (SaveAssetRef:EUW IDSave Asset Object Reference)
    AssetToSave:Object Reference=@n11
   ) -> (then)
   Add(
    execute=@n6.LoopBody
    TargetArray=Get ThustStringArray()
    NewItem=Append!(A="- ", B=Enum to String!(Enumerator:ELayeredAudioTriggerDirection Enum=@n6) -> (ReturnValue:String)) -> (ReturnValue)
   ) {is=title K2Node_CallArrayFunction}
   n12 = CloseSoundForAsset!() -> (OutputDelegate, then, Sound:Sound Base Soft Object Reference, SourceReference:EUW Meta Sound Source Reference Object Reference) {is=title K2Node_CustomEvent}
   > n13 = Set CloseSound!(
    execute
    CloseSound:Sound Base Soft Object Reference=@n12.Sound
    self:Athena Glider Item Definition Object Reference=Get Glider Item Definition()
   ) -> (then, Output_Get:Sound Base Soft Object Reference)
   RetriggerableDelay!(execute=[@n13.then, @n16.then, @n18.then], WorldContextObject:Object Reference, Duration=1) -> (then)
   > n14 = Get()
   > SaveIdAsset(self=Get SaveAssetRef(self=Get IDReference()), AssetToSave=@n14)
   n15 = ThrustLoopSoundForAsset!() -> (OutputDelegate, then, Sound:Sound Base Soft Object Reference, SourceReference:EUW Meta Sound Source Reference Object Reference) {is=title K2Node_CustomEvent}
   > n16 = Set ThrustLoopSound!(
    execute
    ThrustLoopSound:Sound Base Soft Object Reference=@n15.Sound
    self:Athena Glider Item Definition Object Reference=Get Glider Item Definition()
   ) -> (then, Output_Get:Sound Base Soft Object Reference)
   n17 = OpenSoundForAsset!() -> (OutputDelegate, then, Sound:Sound Base Soft Object Reference, SourceReference:EUW Meta Sound Source Reference Object Reference) {is=title K2Node_CustomEvent}
   > n18 = Set OpenSound!(
    execute
    OpenSound:Sound Base Soft Object Reference=@n17.Sound
    self:Athena Glider Item Definition Object Reference=Get Glider Item Definition()
   ) -> (then, Output_Get:Sound Base Soft Object Reference)
   OnSourcePropertyChanged() {is=title K2Node_CustomEvent}
   > n19 = Get()
   > Assign Sound for Asset!(
    execute:Exec
    self:EUW Meta Sound Source Reference Object Reference=Get SoundSourceReferenceOpen()
    Delegate:"Delegate (by ref)"=Create Event()
   ) -> (then:Exec) {is=title K2Node_AssignDelegate}
   > Bind Event to Sound for Asset(self:EUW Meta Sound Source Reference Object Reference=Get SoundSourceReferenceClose(), Delegate=Create Event()) {is=title K2Node_AddDelegate}
   > Bind Event to Sound for Asset(self=Get SoundSourceReferenceEmoteThrust(), Delegate=Create Event())
   PrintString(execute=@n19.else, InString="Not Valid Glider Selected!")
   Event PreConstruct()
   n20 = Get IDSource!(self:EUW IDReference Object Reference=Get IDReference()) -> (IDSource:Single Property View Object Reference)
   > SetPropertyName!(execute, self:Single Property View Object Reference=@n20, NewPropertyName:Name=Glider Item Definition) -> (then)
   > SetObject(
    self=[
     Get ThrustView() -> (ThrustView:Single Property View Object Reference)
     Get CloseView() -> (CloseView:Single Property View Object Reference)
     Get OpenView() -> (OpenView:Single Property View Object Reference)
     @n20
    ]
    NewObject=Self-Reference()
   )
   Event Construct()
   > Assign On Source Property Changed!(execute:Exec, self:EUW IDReference Object Reference=Get IDReference(), Delegate:"Delegate (by ref)"=Create Event()) -> (then:Exec) {is=title K2Node_AssignDelegate}
   > InitializeActionWidgets()
   > n21 = GetCosmeticCreator!(execute, Target:Widget Object Reference=Self-Reference(), __WorldContext:Object Reference) -> (then, AsEUW Meta Sound Cosmetic Preset Creator:EUW Meta Sound Cosmetic Preset Creator Object Reference)
   > Bind Event to Finish Setting Sounds(self:EUW Meta Sound Cosmetic Preset Creator Object Reference=@n21, Delegate=Create Event()) {is=title K2Node_AddDelegate}
  ]}
 GetCreatorWidget {nodes [
   GetCreatorWidget() {is=title K2Node_FunctionEntry}
   n1 = GetParent!(self:Widget Object Reference) -> (ReturnValue:Panel Widget Object Reference)
   > n2 = Cast To EUW_MetaSound_Cosmetic_PresetCreator!(execute, Object=GetOuterObject!(Object:Object Reference=GetOuterObject(Object=@n1)) -> (ReturnValue:Object Reference)) -> (then, CastFailed, AsEUW Meta Sound Cosmetic Preset Creator:EUW Meta Sound Cosmetic Preset Creator Object Reference) {is=title K2Node_DynamicCast}
   > Return Node!(execute, Costmetic Creator:EUW Meta Sound Cosmetic Preset Creator Object Reference=@n2)
  ]}
 InitializeActionWidgets {nodes [
   InitializeActionWidgets() {is=title K2Node_FunctionEntry}
   > Widget Selector(ShouldAdd=Get IsOpen(), Reference=Get SoundSourceReferenceOpen())
   > Widget Selector(ShouldAdd=Get IsClose(), Reference=Get SoundSourceReferenceClose())
   > Widget Selector(ShouldAdd=Get IsThrust(), Reference=Get SoundSourceReferenceEmoteThrust())
  ]}
 Widget Selector {nodes [
   n1 = Widget Selector!() -> (then, ShouldAdd:bool=~, Reference:"EUW Meta Sound Source Reference Object Reference (by ref)") {is=title K2Node_FunctionEntry}
   > n2 = Branch(Condition=@n1)
   n3 = Get SoundSourceReferenceWidgets() -> (SoundSourceReferenceWidgets:Array of EUW Meta Sound Source Reference Object References)
   > n4 = Add Unique(
    TargetArray:Array of EUW Meta Sound Source Reference Object References=@n3
    NewItem:"EUW Meta Sound Source Reference Object Reference (by ref)"=@n1.Reference
   )
   Set Current Source Reference Widgets!(
    execute=[@n5.then, @n4.then]
    self:EUW Meta Sound Cosmetic Preset Creator Object Reference=GetCreatorWidget!() -> (Costmetic Creator:EUW Meta Sound Cosmetic Preset Creator Object Reference) {is=title K2Node_CallFunction}
    CurrentActiveWidget:Widget Object Reference=Self-Reference()
    SourceReferences:Array of EUW Meta Sound Source Reference Object References=Get SoundSourceReferenceWidgets()
    CosmeticType:String=Glider
   ) -> (then) {is=title K2Node_CallFunction}
   n5 = Remove Item!(
    execute=@n2.else
    TargetArray:Array of EUW Meta Sound Source Reference Object References=@n3
    Item:"EUW Meta Sound Source Reference Object Reference (by ref)"=@n1.Reference
   ) -> (then, ReturnValue:bool)
  ]}
]
variables [
 Glider Item Definition {type=AthenaGliderItemDefinition, default=null}
 IsClose {type=boolean, default=true}
 IsOpen {type=boolean, default=true}
 IsThrust {type=boolean, default=true}
 SoundSourceReferenceClose {
  type=EUW_MetaSoundSourceReference_C
  default=&/EUW_MetaSound_Cosmetic_PresetCreator_Glider.Default__EUW_MetaSound_Cosmetic_PresetCreator_Glider_C:EUW_MetaSoundSourceReference_C_1
 }
 SoundSourceReferenceEmoteThrust {
  type=EUW_MetaSoundSourceReference_C
  default=&Default__EUW_MetaSound_Cosmetic_PresetCreator_Glider_C:EUW_MetaSoundSourceReference_C_2
 }
 SoundSourceReferenceOpen {
  type=EUW_MetaSoundSourceReference_C
  default=&Default__EUW_MetaSound_Cosmetic_PresetCreator_Glider_C:EUW_MetaSoundSourceReference_C_0
 }
 SoundSourceReferenceWidgets {type="array<EUW_MetaSoundSourceReference_C>", default []}
 ThrustParams {default {anyDirection {attackSpeed=3, curve=null, interpType=Linear, releaseSpeed=2, oneshots []}}}
 ThustStringArray {type="array<string>", default []}
]
widgets {
 info {rootWidgetClass=&CanvasPanel, widgetCount=10}
 widgets [
  CanvasPanel_0 {widgetClassPath=&CanvasPanel, widget=CanvasPanel_0}
  GridPanel_0 {
   widgetClassPath=&GridPanel
   parent=CanvasPanel_0
   widget=GridPanel_0
   slot {
    class=CanvasPanelSlot
    properties {bAutoSize=true, layoutData {alignment=<0 0>, anchors {maximum=<0 0>, minimum=<0 0>}, offsets=<ltrb 0 0 500 300>}}
   }
  }
  GridPanel_183 {
   widgetClassPath=&GridPanel
   parent=GridPanel_0
   widget=GridPanel_183
   slot {class=GridSlot, properties {horizontalAlignment=HAlign_Center, verticalAlignment=VAlign_Center, nudge=<0 10>}}
  }
  HorizontalBox_116 {
   widgetClassPath=&HorizontalBox
   parent=GridPanel_183
   widget=HorizontalBox_116
   properties {toolTipText=Action for Metasound data.}
   slot {class=GridSlot, properties {horizontalAlignment=HAlign_Left, verticalAlignment=VAlign_Top}}
  }
  OpenView {
   widgetClassPath=&SinglePropertyView
   parent=HorizontalBox_116
   widget=OpenView
   properties {nameOverride=Open:, propertyName=IsOpen}
   slot {class=HorizontalBoxSlot}
  }
  HorizontalBox {
   widgetClassPath=&HorizontalBox
   parent=GridPanel_183
   widget=HorizontalBox
   properties {toolTipText=Action for Metasound data.}
   slot {class=GridSlot, properties {column=1, horizontalAlignment=HAlign_Left, verticalAlignment=VAlign_Top}}
  }
  CloseView {
   widgetClassPath=&SinglePropertyView
   parent=HorizontalBox
   widget=CloseView
   properties {nameOverride=Close:, propertyName=IsClose}
   slot {class=HorizontalBoxSlot}
  }
  HorizontalBox_1 {
   widgetClassPath=&HorizontalBox
   parent=GridPanel_183
   widget=HorizontalBox_1
   properties {toolTipText=Action for Metasound data.}
   slot {class=GridSlot, properties {column=2, horizontalAlignment=HAlign_Left, verticalAlignment=VAlign_Top}}
  }
  ThrustView {
   widgetClassPath=&SinglePropertyView
   parent=HorizontalBox_1
   widget=ThrustView
   properties {nameOverride=Thrust:, propertyName=IsThrust}
   slot {class=HorizontalBoxSlot}
  }
  IDReference {
   widgetClassPath=/Game/Sounds/Fort_Audio_Blueprints/Utilities/EditorUtilityWidgets/Metasounds/Tools/EUW_IDReference
   bIsVariable=true
   parent=GridPanel_183
   widget=IDReference
   slot {class=GridSlot, properties {columnSpan=5, horizontalAlignment=HAlign_Center, row=2, verticalAlignment=VAlign_Center, nudge=<0 25>}}
  }
 ]
}
