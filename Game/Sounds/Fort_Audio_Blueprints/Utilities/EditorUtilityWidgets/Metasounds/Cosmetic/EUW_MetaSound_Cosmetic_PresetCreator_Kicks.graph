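# Editor utility widget for the MetaSound Cosmetic Preset Creator that builds Sound Libraries for additive Kicks (shoe) cosmetics.
# Its button creates a named library parented to SL_Kicks_Additive_Cosmetic.
# Source-reference widgets registered with the creator feed sounds into step, jump/land and preview tags; finishing saves library and shoe, then opens the shoe.
# Reactive StatWatcher component functions, using subobject editing on part modifier Blueprints, are present but unused by the event graph.
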EditorUtilityWidgetBlueprint /Game/Sounds/Fort_Audio_Blueprints/Utilities/EditorUtilityWidgets/Metasounds/Cosmetic/EUW_MetaSound_Cosmetic_PresetCreator_Kicks
graphs [
 Add MetaSounds To Library {nodes [
   n1 = Add MetaSounds To Library!() -> (then, Sounds:"Sound Base Soft Object Reference (by ref)", Array:Array of Sound Library Tag Structures) {is=title K2Node_FunctionEntry}
   > n2 = For Each Loop(Array:Array of Sound Library Tag Structures=@n1.Array) -> (Array Element:Sound Library Tag Structure)
   Add Sound Library to Character Part(execute=@n2.Completed)
   n3 = AddMetasoundToLibrary!(
    execute=@n2.LoopBody
    Sound:"Sound Base Soft Object Reference (by ref)"=@n1
    SoundLibraryTag:Sound Library Tag Structure=@n2
    SoundLibrary:Sound Library Object Reference=Get SoundLibrary() -> (SoundLibrary:Sound Library Object Reference)
    __WorldContext:Object Reference
   ) -> (then, Sounds:Map of Gameplay Tag Structures to Sound Base Soft Object References, Success:bool)
   > Set LibrarySounds(LibrarySounds:Map of Gameplay Tag Structures to Sound Base Soft Object References=@n3) -> (Output_Get:Map of Gameplay Tag Structures to Sound Base Soft Object References)
   Add!(
    execute
    TargetMap:Map of Gameplay Tag Structures to Sound Base Soft Object References=Get Sounds!(self:Sound Library Object Reference=Get SoundLibrary()) -> (Sounds:Map of Gameplay Tag Structures to Sound Base Soft Object References)
    Key:"Gameplay Tag Structure (by ref)"
    Value:"Sound Base Soft Object Reference (by ref)"
   ) -> (then) {is=title K2Node_CallFunction}
   Get SoundSourceReferenceWidgetsTags() -> (SoundSourceReferenceWidgetsTags:Array of S Additive Kicks Sound Library Groups Structures)
  ]}
 Add Reactive Object to Blueprint {nodes [
   n1 = Add Reactive Object to Blueprint!() -> (then, SubobjectHandle:Array of Subobject Data Handle Structures) {is=title K2Node_FunctionEntry}
   > n2 = Get Cosmetic Options!(execute:Exec) -> (
    Success:Exec
    Failed:Exec
    CosmeticOptions:Enum_MetaSound_Creator_CosmeticOptions Enum=NewEnumerator1
    Stat Manager Tag:Fort Stat Manager Tag Structure
    SoundLibraryTag:Sound Library Tag Structure
   ) {is=title K2Node_MacroInstance}
   n3 = Get SubobjectDataSubsystem()
   > Is Valid(InputObject=@n3)
   n4 = "Get (a ref)"(Array:Array of Subobject Data Handle Structures=@n1, Dimension 1=0) -> (Output:"Subobject Data Handle Structure (by ref)")
   > Branch(Condition=IsHandleValid!(DataHandle:"Subobject Data Handle Structure (by ref)"=@n4) -> (ReturnValue:bool))
   > n5 = AddNewSubobject!(
    execute
    self:Subobject Data Subsystem Object Reference=@n3
    Params_ParentHandle:Subobject Data Handle Structure=@n4
    Params_NewClass:Object Class Reference=/Game/Sounds/Fort_Audio_Blueprints/Cosmetics/B_Reactive_StatWatcher_Component.B_Reactive_StatWatcher_Component_C
    Params_BlueprintContext:Blueprint Object Reference=GetBlueprint!(Data:"Subobject Data Structure (by ref)"=GetData(DataHandle=@n4)) -> (ReturnValue:Blueprint Object Reference)
    Params_bSkipMarkBlueprintModified:bool=False
    Params_bConformTransformToParent:bool=false
   ) -> (then, FailReason:Text, ReturnValue:Subobject Data Handle Structure)
   n6 = Split(
    SourceString=Get Class Display Name(Class=&B_Reactive_StatWatcher_Component_C)
    InStr=_
    SearchDir=FromEnd
   )
   n7 = Append(A=_, B=GetCosmeticOption!() -> (ReturnValue:String))
   > RenameSubobject!(
    execute
    self:Subobject Data Subsystem Object Reference=@n3
    Handle:"Subobject Data Handle Structure (by ref)"=@n5.ReturnValue
    InNewName:"Text (by ref)"="To Text (String)"(InString=Append(A=@n6, B=@n7))
   ) -> (then, ReturnValue:bool)
   > n8 = Branch(Condition=IsHandleValid(DataHandle=@n5.ReturnValue))
   n9 = GetVariableName!(Data:"Subobject Data Structure (by ref)"=GetData(DataHandle=@n5.ReturnValue)) -> (ReturnValue:Name=None)
   > PrintString(InString=Append(A="Component Added: ", B="To String (Name)"(InName=@n9)))
   > n10 = Find Stat Watch Component!(exec, Handle:"Subobject Data Handle Structure (by ref)"=@n5.ReturnValue) -> (Success:Exec, Failed:Exec, AsB Reactive Stat Watcher Component:B Reactive Stat Watcher Component Object Reference) {is=title K2Node_MacroInstance}
   n11 = Make Struct Meta Sound Reactive Component Data!(
    ReactiveComponent_35_#1:B Reactive Stat Watcher Component Class Reference=&B_Reactive_StatWatcher_Component_C
    WatcherStatTag_12_#2:Fort Stat Manager Tag Structure=@n2.Stat Manager Tag
    DefaultValue_14_#3:int=-1
    UpdateAudioParameter_30_#4:bool=False
    BroadcastInitialValue_32_#5:bool=False
    InitialValueBroadcastDelay_23_#6:double=-1
    ConsumingWaitingNotifications_27_#7:bool=True
    Object_43_#8:B Reactive Stat Watcher Component Object Reference=@n10
   ) -> (Struct_MetaSound_Reactive_Component_Data:Struct Meta Sound Reactive Component Data Structure) {is=title K2Node_MakeStruct}
   > Set Array Elem!(
    execute
    TargetArray:Array of Struct Meta Sound Reactive Component Data Structures=Get ReactiveComponents() -> (ReactiveComponents:Array of Struct Meta Sound Reactive Component Data Structures)
    Index:int=0
    Item:"Struct Meta Sound Reactive Component Data Structure (by ref)"=@n11
    bSizeToFit:bool=true
   ) -> (then)
   n12 = Break Struct Meta Sound Reactive Component Data!(Struct_MetaSound_Reactive_Component_Data:"Struct Meta Sound Reactive Component Data Structure (by ref)"=@n11) -> (
    ReactiveComponent_35_#1:B Reactive Stat Watcher Component Class Reference
    WatcherStatTag_12_#2:Fort Stat Manager Tag Structure
    DefaultValue_14_#3:int
    UpdateAudioParameter_30_#4:bool=~
    BroadcastInitialValue_32_#5:bool=~
    InitialValueBroadcastDelay_23_#6:double=~
    ConsumingWaitingNotifications_27_#7:bool=~
    Object_43_#8:B Reactive Stat Watcher Component Object Reference
   ) {is=title K2Node_BreakStruct}
   > Set WatcherStatTag!(
    execute
    WatcherStatTag:Fort Stat Manager Tag Structure=@n12.WatcherStatTag_12_#2
    self:B Reactive Stat Watcher Component Object Reference=@n10
   ) -> (then, Output_Get:Fort Stat Manager Tag Structure)
   > Set Default Value!(
    execute
    Default Value:int=@n12.DefaultValue_14_#3
    self:B Reactive Stat Watcher Component Object Reference=@n10
   ) -> (then, Output_Get:int=0)
   > Set UpdateAudioParameter!(
    execute
    UpdateAudioParameter:bool=@n12.UpdateAudioParameter_30_#4
    self:B Reactive Stat Watcher Component Object Reference=@n10
   ) -> (then, Output_Get:bool)
   > Set Broadcast Initial Value!(
    execute
    Broadcast Initial Value:bool=@n12.BroadcastInitialValue_32_#5
    self:B Reactive Stat Watcher Component Object Reference=@n10
   ) -> (then, Output_Get:bool)
   > Set Initial Value Broadcast Delay!(
    execute
    Initial Value Broadcast Delay:float=@n12.InitialValueBroadcastDelay_23_#6
    self:B Reactive Stat Watcher Component Object Reference=@n10
   ) -> (then, Output_Get:float)
   > Set Consume Waiting Notifications!(
    execute
    Consume Waiting Notifications:bool=@n12.ConsumingWaitingNotifications_27_#7
    self:B Reactive Stat Watcher Component Object Reference=@n10
   ) -> (then, Output_Get:bool)
   > Set SoundLibraryTag!(execute, SoundLibraryTag:Sound Library Tag Structure=@n2.SoundLibraryTag, self:B Reactive Stat Watcher Component Object Reference=@n10) -> (then, Output_Get:Sound Library Tag Structure)
   PrintString(execute=@n8.else, InString="To String (Text)"(InText=@n5))
  ]}
 Add Sound Library to Character Part {nodes [
   n1 = Get!(execute:Exec=@n2.then) -> (then:Exec, else:Exec, SoundLibrary:Sound Library Object Reference)
   Add Unique(
    TargetArray:Array of Sound Library Object References=Get SoundLibraries!(self:Custom Character Part Object Reference) -> (SoundLibraries:Array of Sound Library Object References)
    NewItem:"Sound Library Object Reference (by ref)"=@n1
   )
   > Set Current Sound Library(Current Sound Library:Sound Library Object Reference=@n1) -> (Output_Get:Sound Library Object Reference)
   n2 = Add Sound Library to Character Part() {is=title K2Node_FunctionEntry}
   Remove Item!(
    execute
    TargetArray:Array of Sound Library Object References=Get SoundLibraries()
    Item:"Sound Library Object Reference (by ref)"=Get Current Sound Library() -> (Current Sound Library:Sound Library Object Reference)
   ) -> (then, ReturnValue:bool)
   Get!(execute=@n3.then) -> (then, else, ShoeID:Cosmetic Shoes Item Definition Object Reference)
   Get(execute=@n5.then)
   n3 = Clear!(
    execute=@n1.else
    TargetMap:Map of Gameplay Tag Structures to Sound Base Soft Object References=Get LibrarySounds() -> (LibrarySounds:Map of Gameplay Tag Structures to Sound Base Soft Object References)
   ) -> (then) {is=title K2Node_CallFunction}
   n4 = Get Sounds(self=@n1)
   n5 = Set LibrarySounds(execute=@n1.then, LibrarySounds=@n4)
  ]}
 Add Widget Sound Libaries to Sound Library Asset {nodes [
   Add Widget Sound Libaries to Sound Library Asset() {is=title K2Node_FunctionEntry}
   > SetEditorProperty(
    Object=Get SoundLibrary()
    PropertyName=Sounds
    PropertyValue:Map of Gameplay Tag Structures to Sound Base Soft Object References=Get LibrarySounds()
   )
  ]}
 Attempt To Create Reactive Component {nodes [
   n1 = Attempt To Create Reactive Component!() -> (then, InputPin:Fort Stat Manager Tag Structure) {is=title K2Node_FunctionEntry}
   > n2 = Part Modifier Parent Check!(execute:Exec) -> (Success:Exec, Fail:Exec) {is=title K2Node_MacroInstance}
   > Set TagAlreadyApplied(TagAlreadyApplied:bool=false) -> (Output_Get:bool)
   > n3 = Get!(execute) -> (then, else, As B Athena Part Modifier Generic:B Athena Part Modifier Generic Object Reference)
   n4 = Get SubobjectDataSubsystem()
   > Is Valid(InputObject=@n4)
   > n5 = Gather Subobject Data For Instance(self=@n4, Context=@n3)
   > n6 = For Each Loop(Array:Array of Subobject Data Handle Structures=@n5) -> (Array Element:Subobject Data Handle Structure)
   n7 = GetObject!(
    Data:"Subobject Data Structure (by ref)"=FindSubobjectDataFromHandle!(self:Subobject Data Subsystem Object Reference=@n4, Handle:"Subobject Data Handle Structure (by ref)"=@n6) -> (OutData:Subobject Data Structure, ReturnValue:bool)
    bEvenIfPendingKill:bool=false
   ) -> (ReturnValue:Object Reference)
   > Cast To B_Reactive_StatWatcher_Component Class!(execute:Exec, Class:Object Class Reference=Get Class(Object=@n7) -> (ReturnValue:Object Class Reference)) -> (then:Exec, CastFailed:Exec, AsB Reactive Stat Watcher Component:B Reactive Stat Watcher Component Class Reference) {is=title K2Node_ClassDynamicCast}
   > n8 = Cast To B_Reactive_StatWatcher_Component!(execute, Object=@n7) -> (then, CastFailed, AsB Reactive Stat Watcher Component:B Reactive Stat Watcher Component Object Reference) {is=title K2Node_DynamicCast}
   > Branch(Condition=MatchesTag!(
     TagOne:Gameplay Tag Structure=Get WatcherStatTag!(self:B Reactive Stat Watcher Component Object Reference=@n8) -> (WatcherStatTag:Fort Stat Manager Tag Structure)
     TagTwo:Gameplay Tag Structure=@n1
     bExactMatch:bool=true
    ) -> (ReturnValue:bool))
   > Set TagAlreadyApplied(TagAlreadyApplied=true)
   n9 = Branch(execute=@n6.Completed, Condition=Get TagAlreadyApplied() -> (TagAlreadyApplied:bool))
   > PrintString(InString=Reactive Component already exists with same tag on character parts. No component will be created.)
   Add Reactive Object to Blueprint!(execute=@n9.else, self, SubobjectHandle:Array of Subobject Data Handle Structures=@n5) -> (then) {is=title K2Node_CallFunction}
   Show Message Dialog(
    execute=@n2.Fail
    Title=MakeLiteralText(Value="LOCTEXT(\"Part Modifier Blueprint Component Failed!\")")
    Message=MakeLiteralText(Value="LOCTEXT(\"Part Modifier Blueprint can not be \\\"B_Athena_PartModifier_Generic\\\". \\r\\n\\r\\nNo component will be added to this bluerpint.\\r\\n\\r\\nPlease add the proper blueprint or reach out to audio production or design to get acces to the proper assets. \\r\\n\")")
    MessageType=Ok
    DefaultValue=Continue
    MessageCategory=Error
   )
  ]}
 Create Sound Library {nodes [
   n1 = Create Sound Library!() -> (then, PackagePath:String, AssetName:String) {is=title K2Node_FunctionEntry}
   > n2 = CreateSoundLibrary!(execute, PackagingPath:String=@n1, AssetsName:String=@n1.AssetName, __WorldContext:Object Reference) -> (then, Sound Library:Sound Library Object Reference, Successful:bool)
   > Branch(Condition=@n2.Successful)
   > SaveLoadedAsset(AssetToSave=@n2, bOnlyIfIsDirty=true)
   > Return Node!(execute, AsSound Library:Sound Library Object Reference=@n2)
  ]}
 EventGraph {nodes [
   "On Pressed (SoundLibraryButton)"() {is=title K2Node_ComponentBoundEvent}
   n1 = Get PathInput!(self:EUW Meta Sound Asset Path Object Reference=Get AssetPath!(self:EUW Meta Sound Cosmetic Preset Creator Object Reference=GetCreatorWidget!() -> (Costmetic Creator:EUW Meta Sound Cosmetic Preset Creator Object Reference)) -> (AssetPath:EUW Meta Sound Asset Path Object Reference)) -> (PathInput:Editor Utility Editable Text Box Object Reference)
   n2 = GetCreatorWidget()
   n3 = "GetText (Text Box)"!(self:Text Box Object Reference=Get AssetNameInput!(self:EUW Meta Sound Asset Path Object Reference=Get AssetPath(self=@n2)) -> (AssetNameInput:Editor Utility Editable Text Box Object Reference)) -> (ReturnValue:Text)
   > n4 = Create Sound Library!(
    execute
    self
    PackagePath:String="To String (Text)"(InText=Get Text(self=@n1))
    AssetName:String=Append!(A=SL, B=_, C:String=AdditiveKicks, D:String=_, E:String="To String (Text)"(InText=@n3)) -> (ReturnValue)
   ) -> (then, AsSound Library:Sound Library Object Reference) {is=title K2Node_CallFunction}
   > n5 = Set SoundLibrary(SoundLibrary:Sound Library Object Reference=@n4) -> (Output_Get:Sound Library Object Reference)
   > Set LibrarySounds(LibrarySounds=Get Sounds(self=@n5))
   > SetEditorProperty(
    Object=@n5
    PropertyName=Parent
    PropertyValue:Sound Library Object Reference=Get LibReactiveParent() -> (LibReactiveParent:Sound Library Object Reference)
   )
   "On Property Changed (LibrarySoundsDetails)"!() -> (OutputDelegate, then, PropertyName:Name=None) {is=title K2Node_ComponentBoundEvent}
   > Add Widget Sound Libaries to Sound Library Asset() {is=title K2Node_CallFunction}
   "On Property Changed (SoundLibraryView)"!() -> (OutputDelegate, then, PropertyName:Name=None) {is=title K2Node_ComponentBoundEvent}
   > n6 = Get!(execute) -> (then, else, SoundLibrary:Sound Library Object Reference)
   > Set LibrarySounds(LibrarySounds=Get Sounds(self=@n6))
   Clear(execute=@n6.else, TargetMap=Get LibrarySounds())
   n7 = Add Sound to Assets!() -> (OutputDelegate, then, Sound:Sound Base Soft Object Reference, SourceReference:EUW Meta Sound Source Reference Object Reference) {is=title K2Node_CustomEvent}
   > n8 = For Each Loop with Break!(Exec:Exec, Array:Array of S Additive Kicks Sound Library Groups Structures=Get SoundSourceReferenceWidgetsTags(), Break:Exec=@n10.then) -> (LoopBody:Exec, Array Element:S Additive Kicks Sound Library Groups Structure, Array Index:int, Completed:Exec)
   n9 = Break S Additive Kicks Sound Library Groups!(S_AdditiveKicksSoundLibraryGroups:"S Additive Kicks Sound Library Groups Structure (by ref)"=@n8) -> (
    SourceRef_10_#9:EUW Meta Sound Source Reference Object Reference
    SoundLibraryTags_6_#10:Array of Sound Library Tag Structures
   ) {is=title K2Node_BreakStruct}
   > Branch(Condition="Equal (Object)"(A=@n9, B=@n7.SourceReference))
   > n10 = Add MetaSounds To Library!(
    execute
    self
    Sounds:"Sound Base Soft Object Reference (by ref)"=@n7.Sound
    Array:Array of Sound Library Tag Structures=@n9.SoundLibraryTags_6_#10
   ) -> (then) {is=title K2Node_CallFunction}
   n11 = SoundsFinihsed() {is=title K2Node_CustomEvent}
   > SaveAllAssets()
   n12 = Get ShoeID() -> (ShoeID:Cosmetic Shoes Item Definition Object Reference)
   > Is Valid(InputObject=@n12)
   > n13 = LoadAsset(AssetPath=Get Object Path String(Object=@n12))
   > OpenEditorForAssets!(
    execute
    self:Asset Editor Subsystem Object Reference=Get AssetEditorSubsystem()
    Assets:Array of Object References=Make Array("[0]":Object Reference=@n13) -> (Array:Array of Object References)
    OpenedMethod:EAssetTypeActivationOpenedMethod Enum=Edit
   ) -> (then, ReturnValue:bool)
   Event PreConstruct()
   > SetObject(
    self=[
     Get SoundLibraryView() -> (SoundLibraryView:Single Property View Object Reference)
     Get LibrarySoundsDetails() -> (LibrarySoundsDetails:Details View Object Reference)
     Get CosmeticShoes() -> (CosmeticShoes:Single Property View Object Reference)
    ]
    NewObject=Self-Reference()
   )
   Event Construct()
   > n14 = For Each Loop(Array:Array of S Additive Kicks Sound Library Groups Structures=Get SoundSourceReferenceWidgetsTags()) -> (Array Element:S Additive Kicks Sound Library Groups Structure)
   n15 = Break S Additive Kicks Sound Library Groups(S_AdditiveKicksSoundLibraryGroups=@n14)
   > Widget Selector!(execute, self, ShouldAdd:bool=true, Reference:"EUW Meta Sound Source Reference Object Reference (by ref)"=@n15) -> (then)
   > Assign Sound for Asset!(execute:Exec, self:EUW Meta Sound Source Reference Object Reference=@n15, Delegate:"Delegate (by ref)"=@n7) -> (then:Exec) {is=title K2Node_AssignDelegate}
   n16 = GetCosmeticCreator!(execute=@n14.Completed, Target:Widget Object Reference=Self-Reference(), __WorldContext:Object Reference) -> (then, AsEUW Meta Sound Cosmetic Preset Creator:EUW Meta Sound Cosmetic Preset Creator Object Reference)
   > Bind Event to Finish Setting Sounds(self:EUW Meta Sound Cosmetic Preset Creator Object Reference=@n16, Delegate=@n11) {is=title K2Node_AddDelegate}
   "On Property Changed (CosmeticShoes)"!() -> (OutputDelegate, then, PropertyName:Name=None) {is=title K2Node_ComponentBoundEvent}
  ]}
 FindStatWatchComponent {nodes [
   n1 = Inputs!() -> (exec:Exec, Handle:"Subobject Data Handle Structure (by ref)")
   n2 = Get SubobjectDataSubsystem()
   > n3 = Is Valid(InputObject=@n2)
   > n4 = Cast To B_Reactive_StatWatcher_Component(Object=GetObject(Data=FindSubobjectDataFromHandle(self=@n2, Handle=@n1), bEvenIfPendingKill=true))
   > Outputs!(
    Success:Exec
    Failed:Exec=[@n4.CastFailed, @n3.Is Not Valid]
    AsB Reactive Stat Watcher Component:B Reactive Stat Watcher Component Object Reference=@n4
   )
   Show Message Dialog(
    Title=MakeLiteralText(Value="LOCTEXT(\"Warning: Stat Watcher Component Missing\")")
    Message=MakeLiteralText(Value="LOCTEXT(\"Please verify this asset has a reactive stat watch component\")")
    MessageType=Ok
    DefaultValue=No
    MessageCategory=Warning
   )
  ]}
 GetCosmeticOption {nodes [
   GetCosmeticOption() {is=title K2Node_FunctionEntry}
   n1 = Get CosmeticOption() -> (CosmeticOption:Enum_MetaSound_Creator_CosmeticOptions Enum=NewEnumerator1)
   > Return Node!(
    execute
    ReturnValue:String=Replace(SourceString=Enum to String!(Enumerator:Enum_MetaSound_Creator_CosmeticOptions Enum=@n1) -> (ReturnValue:String), From=" ")
   )
  ]}
 GetCosmeticOptions {nodes [
   n1 = String To Name(InString=GetCosmeticOption() {is=title K2Node_CallFunction})
   n2 = Get CosmeticOptionsTable() -> (CosmeticOptionsTable:Data Table Object Reference)
   n3 = Get Data Table Row!(execute:Exec, DataTable:Data Table Object Reference=@n2, RowName:Name=@n1) -> (then:Exec, RowNotFound:Exec, ReturnValue:Struct Meta Sound Cosmetic Options Structure)
   n4 = GetCosmeticOptions!(execute=@n5.execute, DataTable:Data Table Object Reference=@n2, RowName:Name=@n1, __WorldContext:Object Reference) -> (
    then
    RowFound:bool
    CosmeticOption:Enum_MetaSound_Creator_CosmeticOptions Enum=NewEnumerator1
    Stat Manager Tag:Fort Stat Manager Tag Structure
    SoundLibraryTag:Sound Library Tag Structure
   )
   Outputs!(
    Success=[@n3.then, @n6.then]
    Failed=[@n3.RowNotFound, @n6.else]
    CosmeticOptions:Enum_MetaSound_Creator_CosmeticOptions Enum=@n4.CosmeticOption
    Stat Manager Tag:Fort Stat Manager Tag Structure=@n4.Stat Manager Tag
    SoundLibraryTag:Sound Library Tag Structure=@n4.SoundLibraryTag
   )
   n5 = Inputs!() -> (execute:Exec)
   Break Struct Meta Sound Cosmetic Options!(Struct_MetaSound_CosmeticOptions:"Struct Meta Sound Cosmetic Options Structure (by ref)"=@n3) -> (
    CosmeticOption_2_#11:Enum_MetaSound_Creator_CosmeticOptions Enum
    StatManagerTag_7_#12:Fort Stat Manager Tag Structure
    SoundLibraryTag_8_#13:Sound Library Tag Structure
   ) {is=title K2Node_BreakStruct}
   n6 = Branch(execute=@n4.then, Condition=@n4)
  ]}
 GetCreatorWidget {nodes [
   GetCreatorWidget() {is=title K2Node_FunctionEntry}
   > n1 = GetCosmeticCreator(Target=Self-Reference())
   > Return Node!(execute, Costmetic Creator:EUW Meta Sound Cosmetic Preset Creator Object Reference=@n1)
  ]}
 PartModifierParentCheck {nodes [
   Inputs()
   n1 = Get PartModifierBlueprint() -> (PartModifierBlueprint:Actor Soft Class Reference)
   > n2 = Branch(Condition="Equal Exactly (String)"(
     A="To String (SoftClassReference)"!(SoftClassReference:"Object Soft Class Reference (by ref)"=@n1) -> (ReturnValue:String)
     B=/Game/Athena/Cosmetics/Blueprints/B_Athena_PartModifier_Generic.B_Athena_PartModifier_Generic_C
    ))
   > n3 = Show Message Dialog(
    Title=MakeLiteralText(Value="LOCTEXT(\"Parent Blueprint Is Selected\")")
    Message=MakeLiteralText(Value="LOCTEXT(\"Part Modifier Blueprint can not be \\\"B_Athena_PartModifier_Generic\\\". \\r\\n\\r\\nPlease add the proper blueprint or reach out to audio production or design to get acces to the proper assets. \\r\\n\\r\\nParent BP should not be used when added audio. Implementation will fail.\")")
    MessageType=Ok
    DefaultValue=No
    MessageCategory=Warning
   )
   Outputs!(Success=@n2.else, Fail:Exec=@n3.then)
  ]}
 SaveAllAssets {nodes [
   SaveAllAssets() {is=title K2Node_FunctionEntry}
   > n1 = Sequence!(execute) -> (then_0, then_1, then_2:Exec)
   > n2 = Get()
   > TransactObject(Object=@n2)
   n3 = Get SaveAssetRef() -> (SaveAssetRef:EUW IDSave Asset Object Reference)
   > Set SaveAsset!(execute, SaveAsset:bool=true, self:EUW IDSave Asset Object Reference=@n3) -> (then, Output_Get:bool)
   > SaveIdAsset!(execute, self:EUW IDSave Asset Object Reference=@n3, AssetToSave:Object Reference=@n2) -> (then)
   n4 = Get!(execute=@n1.then_1) -> (then, else, ShoeID:Cosmetic Shoes Item Definition Object Reference)
   > TransactObject(Object=@n4)
   n5 = Get SaveAssetRef()
   > Set SaveAsset(SaveAsset=true, self=@n5)
   > SaveIdAsset(self=@n5, AssetToSave=@n4)
  ]}
 Set Component Stats Variables to Blueprint {nodes [
   Set Component Stats Variables to Blueprint() {is=title K2Node_FunctionEntry}
   > n1 = For Each Loop(Array:Array of Struct Meta Sound Reactive Component Data Structures=Get ReactiveComponents()) -> (Array Element:Struct Meta Sound Reactive Component Data Structure)
   > n2 = Get!(execute) -> (then, else, As B Athena Part Modifier Generic:B Athena Part Modifier Generic Object Reference)
   > n3 = Gather Subobject Data For Instance(self=Get SubobjectDataSubsystem(), Context=@n2)
   > n4 = For Each Loop(Array:Array of Subobject Data Handle Structures=@n3) -> (Array Element:Subobject Data Handle Structure)
   > n5 = Find Stat Watch Component(Handle=@n4)
   n6 = Break Struct Meta Sound Reactive Component Data(Struct_MetaSound_Reactive_Component_Data=@n1) -> (
    UpdateAudioParameter_30_#4=~
    BroadcastInitialValue_32_#5=~
    InitialValueBroadcastDelay_23_#6=~
    ConsumingWaitingNotifications_27_#7=~
   )
   > Branch(Condition="Equal (Object)"(A=@n5, B=@n6.Object_43_#8))
   > Set WatcherStatTag(WatcherStatTag=@n6.WatcherStatTag_12_#2, self=@n5)
   > Set Default Value(Default Value=@n6.DefaultValue_14_#3, self=@n5) -> (Output_Get=0)
   > Set UpdateAudioParameter(UpdateAudioParameter=@n6.UpdateAudioParameter_30_#4, self=@n5)
   > Set Broadcast Initial Value(Broadcast Initial Value=@n6.BroadcastInitialValue_32_#5, self=@n5)
   > Set Initial Value Broadcast Delay(Initial Value Broadcast Delay=@n6.InitialValueBroadcastDelay_23_#6, self=@n5)
   > Set Consume Waiting Notifications(Consume Waiting Notifications=@n6.ConsumingWaitingNotifications_27_#7, self=@n5)
  ]}
 SetPartsModifierBlueprint {nodes [
   SetPartsModifierBlueprint() {is=title K2Node_FunctionEntry}
   n1 = Get PartModifierBlueprint()
   > n2 = Branch(Condition=IsValidSoftClassReference!(SoftClassReference:"Object Soft Class Reference (by ref)"=@n1) -> (ReturnValue:bool))
   > n3 = LoadClassAsset_Blocking(AssetClass=@n1)
   > n4 = GetClassDefaultObject!(execute, Class:Object Class Reference=@n3) -> (then, ReturnValue:Object Reference)
   > n5 = Cast To B_Athena_PartModifier_Generic!(execute, Object=@n4) -> (then, CastFailed, AsB Athena Part Modifier Generic:B Athena Part Modifier Generic Object Reference) {is=title K2Node_DynamicCast}
   > Set As B Athena Part Modifier Generic(As B Athena Part Modifier Generic:B Athena Part Modifier Generic Object Reference=@n5) -> (Output_Get:B Athena Part Modifier Generic Object Reference)
   Set As B Athena Part Modifier Generic(execute=@n2.else)
  ]}
 Widget Selector {nodes [
   n1 = Widget Selector!() -> (then, ShouldAdd:bool=~, Reference:"EUW Meta Sound Source Reference Object Reference (by ref)") {is=title K2Node_FunctionEntry}
   > n2 = Branch(Condition=@n1)
   n3 = Get SoundSourceReferenceWidgets() -> (SoundSourceReferenceWidgets:Array of EUW Meta Sound Source Reference Object References)
   > n4 = Add Unique(
    TargetArray:Array of EUW Meta Sound Source Reference Object References=@n3
    NewItem:"EUW Meta Sound Source Reference Object Reference (by ref)"=@n1.Reference
   )
   Set Current Source Reference Widgets!(
    execute=[@n5.then, @n4.then]
    self:EUW Meta Sound Cosmetic Preset Creator Object Reference=GetCreatorWidget() {is=title K2Node_CallFunction}
    CurrentActiveWidget:Widget Object Reference=Self-Reference()
    SourceReferences:Array of EUW Meta Sound Source Reference Object References=Get SoundSourceReferenceWidgets()
    CosmeticType:String=Reactive
   ) -> (then) {is=title K2Node_CallFunction}
   n5 = Remove Item(
    execute=@n2.else
    TargetArray:Array of EUW Meta Sound Source Reference Object References=@n3
    Item:"EUW Meta Sound Source Reference Object Reference (by ref)"=@n1.Reference
   )
  ]}
]
variables [
 Animation {type=AnimSequenceBase, default=null}
 AnimSoundContextClass {type=Class, default=/Game/Sounds/SoundLibrary/Contexts/Cosmetics/BP_Cosmetics_Animation_Context.BP_Cosmetics_Animation_Context_C}
 As B Athena Part Modifier Generic {type=B_Athena_PartModifier_Generic_C, default=null}
 bIsAnimationImplementation {type=boolean}
 CosmeticOption {type=Enum_MetaSound_Creator_CosmeticOptions, default=Elmination}
 CosmeticOptionsTable {
  type=DataTable
  default=&/DataTable_MetaSound_CosmeticOptions
 }
 Current Sound Library {type=SoundLibrary, default=null}
 LibrarySounds {default {}}
 LibReactiveParent {type=SoundLibrary, default=/Game/Sounds/SoundLibrary/Parents/Cosmetic/SL_Kicks_Additive_Cosmetic}
 NewVar {type="array<CosmeticItemDefinitionData_Variants>", default []}
 Notify Track Name {type=string, default=Audio}
 NotifyStartTime {type=number}
 PartModifierBlueprint {type=Class, default=null}
 ReactiveComponents {type="array<Struct_MetaSound_Reactive_Component_Data>", default []}
 ShoeID {type=CosmeticShoesItemDefinition, default=null}
 SoundLibrary {type=SoundLibrary, default=null}
 SoundSourceReference_Kicks {
  type="array<EUW_MetaSoundSourceReference_C>"
  default [
   &/EUW_MetaSound_Cosmetic_PresetCreator_Kicks.Default__EUW_MetaSound_Cosmetic_PresetCreator_Kicks_C:EUW_MetaSoundSourceReference_C_0
   &Default__EUW_MetaSound_Cosmetic_PresetCreator_Kicks_C:EUW_MetaSoundSourceReference_C_4
   &Default__EUW_MetaSound_Cosmetic_PresetCreator_Kicks_C:EUW_MetaSoundSourceReference_C_6
  ]
 }
 SoundSourceReference_Preview {
  type=EUW_MetaSoundSourceReference_C
  default=&Default__EUW_MetaSound_Cosmetic_PresetCreator_Kicks_C:EUW_MetaSoundSourceReference_C_5
 }
 SoundSourceReferenceWidgets {type="array<EUW_MetaSoundSourceReference_C>", default []}
 SoundSourceReferenceWidgetsTags {
  type="array<S_AdditiveKicksSoundLibraryGroups>"
  default [
   {
    sourceRef=&Default__EUW_MetaSound_Cosmetic_PresetCreator_Kicks_C:EUW_MetaSoundSourceReference_C_1
    soundLibraryTags [{tagName=SoundLibrary.Foley.Step}]
   }
   {
    sourceRef=&Default__EUW_MetaSound_Cosmetic_PresetCreator_Kicks_C:EUW_MetaSoundSourceReference_C_3
    soundLibraryTags [{tagName=SoundLibrary.Foley.Jump}, {tagName=SoundLibrary.Foley.Land}]
   }
   {
    sourceRef=&Default__EUW_MetaSound_Cosmetic_PresetCreator_Kicks_C:EUW_MetaSoundSourceReference_C_2
    soundLibraryTags [{tagName=SoundLibrary.Cosmetic.Kicks.Preview}]
   }
  ]
 }
 TagAlreadyApplied {type=boolean}
]
widgets {
 info {rootWidgetClass=&CanvasPanel, widgetCount=15}
 widgets [
  CanvasPanel_0 {widgetClassPath=&CanvasPanel, widget=CanvasPanel_0}
  VerticalBox_78 {
   widgetClassPath=&VerticalBox
   parent=CanvasPanel_0
   widget=VerticalBox_78
   slot {
    class=CanvasPanelSlot
    properties {bAutoSize=true, layoutData {alignment=<0 0>, anchors {maximum=<0 0>, minimum=<0 0>}, offsets=<ltrb 0 0 500 400>}}
   }
  }
  SizeBox {
   widgetClassPath=&SizeBox
   parent=VerticalBox_78
   widget=SizeBox
   properties {bOverride_WidthOverride=true, heightOverride=78.2408, maxDesiredWidth=460, minDesiredWidth=450, widthOverride=450}
   slot {class=VerticalBoxSlot, properties {horizontalAlignment=HAlign_Left, verticalAlignment=VAlign_Center}}
  }
  CosmeticShoes {
   widgetClassPath=&SinglePropertyView
   parent=SizeBox
   widget=CosmeticShoes
   properties {propertyName=ShoeID}
   slot {class=SizeBoxSlot, properties {padding=<ltrb 10 10 10 10>}}
  }
  SoundLibraryBox {
   widgetClassPath=&VerticalBox
   bIsVariable=true
   parent=VerticalBox_78
   widget=SoundLibraryBox
   slot {class=VerticalBoxSlot, properties {horizontalAlignment=HAlign_Left, verticalAlignment=VAlign_Center}}
  }
  SoundLibraryButton {
   widgetClassPath=&EditorUtilityButton
   parent=SoundLibraryBox
   widget=SoundLibraryButton
   properties {toolTipText=Reset the source reference back to default asset.}
   slot {class=VerticalBoxSlot, properties {horizontalAlignment=HAlign_Left, verticalAlignment=VAlign_Bottom, padding=<ltrb 5 10 5 5>}}
  }
  UEFN_TextBlock_C_4 {
   widgetClassPath=&UEFN_TextBlock
   parent=SoundLibraryButton
   widget=UEFN_TextBlock_C_4
   properties {
    justification=Center
    text=Create Sound Library
    font {
     bForceMonospaced=false
     bMaterialIsStencil=false
     fontMaterial=null
     fontObject=/LauncherSocial/BrutalType
     letterSpacing=0
     monospacedWidth=1
     size=9
     skewAmount=0
     typefaceFontName=Regular
     outlineSettings {
      bApplyOutlineToDropShadows=false
      bMiteredCorners=false
      bSeparateFillAlpha=false
      outlineMaterial=null
      outlineSize=0
      outlineColor=<rgba 0 0 0 1>
     }
    }
    shadowColorAndOpacity=<rgba 0 0 0 1>
   }
   slot {class=ButtonSlot}
  }
  SoundLibraryViewBox {
   widgetClassPath=&SizeBox
   bIsVariable=true
   parent=SoundLibraryBox
   widget=SoundLibraryViewBox
   properties {bOverride_WidthOverride=true, heightOverride=78.2408, maxDesiredWidth=460, minDesiredWidth=450, widthOverride=450}
   slot {class=VerticalBoxSlot, properties {horizontalAlignment=HAlign_Left, verticalAlignment=VAlign_Center}}
  }
  SoundLibraryView {
   widgetClassPath=&SinglePropertyView
   parent=SoundLibraryViewBox
   widget=SoundLibraryView
   properties {propertyName=SoundLibrary}
   slot {class=SizeBoxSlot, properties {padding=<ltrb 10 10 10 10>}}
  }
  SaveAssetRef {
   widgetClassPath=/Game/Sounds/Fort_Audio_Blueprints/Utilities/EditorUtilityWidgets/Metasounds/Tools/EUW_IDSaveAsset
   bIsVariable=true
   parent=SoundLibraryBox
   widget=SaveAssetRef
   properties {visibility=Collapsed}
   slot {class=VerticalBoxSlot, properties {padding=<ltrb 0 5 5 5>}}
  }
  EditorUtilityScrollBox_0 {
   widgetClassPath=&EditorUtilityScrollBox
   parent=VerticalBox_78
   widget=EditorUtilityScrollBox_0
   properties {scrollBarVisibility=Collapsed}
   slot {class=VerticalBoxSlot}
  }
  SizeBox_3 {
   widgetClassPath=&SizeBox
   parent=EditorUtilityScrollBox_0
   widget=SizeBox_3
   properties {bOverride_WidthOverride=true, heightOverride=78.2408, maxDesiredWidth=460, minDesiredWidth=450, widthOverride=450}
   slot {class=ScrollBoxSlot}
  }
  LibrarySoundsDetails {
   widgetClassPath=&DetailsView
   parent=SizeBox_3
   widget=LibrarySoundsDetails
   properties {bAutoLoadAsset=false, bIsEnabled=false, visibility=Collapsed, propertiesToShow [LibrarySounds]}
   slot {class=SizeBoxSlot}
  }
  ComponentsSizeBox {
   widgetClassPath=&SizeBox
   parent=EditorUtilityScrollBox_0
   widget=ComponentsSizeBox
   properties {bOverride_WidthOverride=true, heightOverride=78.2408, maxDesiredWidth=460, minDesiredWidth=450, widthOverride=450}
   slot {class=ScrollBoxSlot}
  }
  ReactiveComponentDetails {
   widgetClassPath=&DetailsView
   parent=ComponentsSizeBox
   widget=ReactiveComponentDetails
   properties {bIsEnabled=false, visibility=Collapsed, categoriesToShow [Reactive Components]}
   slot {class=SizeBoxSlot}
  }
 ]
}
