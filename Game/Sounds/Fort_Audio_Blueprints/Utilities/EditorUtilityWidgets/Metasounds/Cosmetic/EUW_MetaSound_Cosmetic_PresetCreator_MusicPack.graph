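# Editor utility widget that edits an Athena music pack item definition's front-end lobby music, preview start time and default-pack flag.
# Property views expose the widget's variables, and IDReference picks the definition.
# Selecting a valid definition binds MusicReference's sound-assignment delegate and loads current values through UpdateItemDefInfo; the invalid-selection message mistakenly says "Glider".
# When a sound arrives, SetEditorProperty writes all three fields, then the asset saves and the display refreshes.
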EditorUtilityWidgetBlueprint /Game/Sounds/Fort_Audio_Blueprints/Utilities/EditorUtilityWidgets/Metasounds/Cosmetic/EUW_MetaSound_Cosmetic_PresetCreator_MusicPack
graphs [
 EventGraph {nodes [
   OnSourcePropertyChanged() {is=title K2Node_CustomEvent}
   > n1 = Get!(execute:Exec) -> (then:Exec, else:Exec, MusicPackItemDefinition:Athena Music Pack Item Definition Object Reference)
   n2 = MusicPackSoundForAsset!() -> (OutputDelegate, then, Sound:Sound Base Soft Object Reference, SourceReference:EUW Meta Sound Source Reference Object Reference) {is=title K2Node_CustomEvent}
   Assign Sound for Asset!(
    execute:Exec=@n1.then
    self:EUW Meta Sound Source Reference Object Reference=Get MusicReference() -> (MusicReference:EUW Meta Sound Source Reference Object Reference)
    Delegate:"Delegate (by ref)"=@n2
   ) -> (then:Exec) {is=title K2Node_AssignDelegate}
   > UpdateItemDefInfo()
   PrintString(execute=@n1.else, InString="Not Valid Glider Selected!")
   Event PreConstruct()
   n3 = Get IDSource!(self:EUW IDReference Object Reference=Get IDReference() -> (IDReference:EUW IDReference Object Reference)) -> (IDSource:Single Property View Object Reference)
   > SetPropertyName!(execute, self:Single Property View Object Reference=@n3, NewPropertyName:Name=MusicPackItemDefinition) -> (then)
   > SetObject(self=@n3, NewObject=Self-Reference())
   > SetObject(self=Get PreviewStartTimeView() -> (PreviewStartTimeView:Single Property View Object Reference), NewObject=Self-Reference())
   > SetObject(self=Get DefaultMusicPackView() -> (DefaultMusicPackView:Single Property View Object Reference), NewObject=Self-Reference())
   > SetObject(self=Get MusicPackSoundView() -> (MusicPackSoundView:Single Property View Object Reference), NewObject=Self-Reference())
   Event Construct()
   > Add Unique(
    TargetArray:Array of EUW Meta Sound Source Reference Object References=Get SoundSourceReferenceWidgets() -> (SoundSourceReferenceWidgets:Array of EUW Meta Sound Source Reference Object References)
    NewItem:"EUW Meta Sound Source Reference Object Reference (by ref)"=Get MusicReference()
   )
   > Assign On Source Property Changed!(execute:Exec, self:EUW IDReference Object Reference=Get IDReference(), Delegate:"Delegate (by ref)"=Create Event()) -> (then:Exec) {is=title K2Node_AssignDelegate}
   UpdateItemDefInfo(execute=@n4.then)
   n4 = SaveIdAsset!(
    execute=@n7.then
    self:EUW IDSave Asset Object Reference=Get SaveAssetRef!(self:EUW IDReference Object Reference=Get IDReference()) -> (SaveAssetRef:EUW IDSave Asset Object Reference)
    AssetToSave:Object Reference=Get MusicPackItemDefinition() -> (MusicPackItemDefinition:Athena Music Pack Item Definition Object Reference)
   ) -> (then)
   n5 = Set MusicPackSound(execute=@n2.then, MusicPackSound:Sound Base Soft Object Reference=@n2.Sound) -> (Output_Get:Sound Base Soft Object Reference)
   > n6 = SetEditorProperty(Object=Get MusicPackItemDefinition(), PropertyName=FrontEndLobbyMusic, PropertyValue:Sound Base Soft Object Reference=@n5)
   > Branch(Condition=@n6)
   > SetEditorProperty(
    Object=Get MusicPackItemDefinition()
    PropertyName=MusicPreviewStartTime
    PropertyValue:double=Get PreviewStartTimeValue() -> (PreviewStartTimeValue:double)
   )
   > n7 = SetEditorProperty(
    Object=Get MusicPackItemDefinition()
    PropertyName=bIsDefaultMusicPack
    PropertyValue:bool=Get IsDefaultMusicPack() -> (IsDefaultMusicPack:bool)
   )
  ]}
 UpdateItemDefInfo {nodes [
   UpdateItemDefInfo() {is=title K2Node_FunctionEntry}
   > n1 = GetEditorProperty!(execute:Exec, Object:Object Reference=Get MusicPackItemDefinition(), PropertyName:Name=MusicPreviewStartTime) -> (then:Exec, PropertyValue:double, ReturnValue:bool)
   > Set PreviewStartTimeValue(PreviewStartTimeValue:double=@n1) -> (Output_Get:double)
   > n2 = GetEditorProperty(Object=Get MusicPackItemDefinition(), PropertyName=bIsDefaultMusicPack) -> (PropertyValue:bool)
   > Set IsDefaultMusicPack(IsDefaultMusicPack:bool=@n2) -> (Output_Get:bool)
   > n3 = GetEditorProperty(Object=Get MusicPackItemDefinition(), PropertyName=FrontEndLobbyMusic) -> (PropertyValue:Sound Base Soft Object Reference)
   > Set MusicPackSound(MusicPackSound=@n3)
  ]}
]
variables [
 IsDefaultMusicPack {type=boolean}
 MusicPackItemDefinition {type=AthenaMusicPackItemDefinition, default=null}
 MusicPackSound {type=SoundBase, default=null}
 MusicReference {
  type=EUW_MetaSoundSourceReference_C
  default=&/EUW_MetaSound_Cosmetic_PresetCreator_MusicPack.Default__EUW_MetaSound_Cosmetic_PresetCreator_MusicPack_C:EUW_MetaSoundSourceReference_C_0
 }
 PreviewStartTimeValue {type=number}
 SoundSourceReferenceWidgets {type="array<EUW_MetaSoundSourceReference_C>", default []}
]
widgets {
 info {rootWidgetClass=&CanvasPanel, widgetCount=8}
 widgets [
  CanvasPanel_0 {widgetClassPath=&CanvasPanel, widget=CanvasPanel_0}
  GridPanel_1 {
   widgetClassPath=&GridPanel
   parent=CanvasPanel_0
   widget=GridPanel_1
   slot {
    class=CanvasPanelSlot
    properties {bAutoSize=true, layoutData {alignment=<0 0>, anchors {maximum=<0 0>, minimum=<0 0>}, offsets=<ltrb 0 0 500 300>}}
   }
  }
  GridPanel_183 {
   widgetClassPath=&GridPanel
   parent=GridPanel_1
   widget=GridPanel_183
   slot {class=GridSlot, properties {row=1, padding=<ltrb 2 2 2 2>}}
  }
  PreviewStartTimeView {
   widgetClassPath=&SinglePropertyView
   parent=GridPanel_183
   widget=PreviewStartTimeView
   properties {propertyName=PreviewStartTimeValue, toolTipText=Start any preview of this music asset at the speicifc time.}
   slot {class=GridSlot, properties {column=1, padding=<ltrb 2 2 2 2>}}
  }
  DefaultMusicPackView {
   widgetClassPath=&SinglePropertyView
   parent=GridPanel_183
   widget=DefaultMusicPackView
   properties {propertyName=IsDefaultMusicPack}
   slot {class=GridSlot, properties {padding=<ltrb 2 2 2 2>}}
  }
  IDReference {
   widgetClassPath=/Game/Sounds/Fort_Audio_Blueprints/Utilities/EditorUtilityWidgets/Metasounds/Tools/EUW_IDReference
   bIsVariable=true
   parent=GridPanel_1
   widget=IDReference
   slot {class=GridSlot, properties {padding=<ltrb 2 2 2 2>}}
  }
  SizeBox_1 {
   widgetClassPath=&SizeBox
   parent=GridPanel_1
   widget=SizeBox_1
   properties {bOverride_WidthOverride=true, widthOverride=500}
   slot {class=GridSlot, properties {row=2, padding=<ltrb 2 2 2 2>}}
  }
  MusicPackSoundView {
   widgetClassPath=&SinglePropertyView
   parent=SizeBox_1
   widget=MusicPackSoundView
   properties {nameOverride=Front-End Lobby Music, propertyName=MusicPackSound}
   slot {class=SizeBoxSlot}
  }
 ]
}
