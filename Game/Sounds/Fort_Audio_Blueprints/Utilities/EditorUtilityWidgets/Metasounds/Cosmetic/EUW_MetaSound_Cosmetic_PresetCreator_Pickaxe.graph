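# Pickaxe panel of the editor-only MetaSound cosmetic preset creator, assigning Swing, Ready and Impact sounds to a melee weapon item definition.
# Toggles register source-reference widgets with the parent creator, with options for sequential Swing IDs or reusing Swing as Ready.
# A generated surface-button grid, seeded from the item, selects which impact map entries receive the sound.
# Sound delegates write the item's maps, mirror swing and impact to dual-wield off-hands, and debounce saving.
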EditorUtilityWidgetBlueprint /Game/Sounds/Fort_Audio_Blueprints/Utilities/EditorUtilityWidgets/Metasounds/Cosmetic/EUW_MetaSound_Cosmetic_PresetCreator_Pickaxe
graphs [
 CreatButtonObject {nodes [
   n1 = CreatButtonObject!() -> (then, Outer:Uniform Grid Panel Object Reference) {is=title K2Node_FunctionEntry}
   > n2 = Construct Object from Class(Class=&EditorUtilityButton, self=@n1) -> (ReturnValue:Editor Utility Button Object Reference)
   > n3 = AddChild(self=@n1, Content=@n2)
   > Return Node!(execute, PanelSlot:Panel Slot Object Reference=@n3, Button Object:Editor Utility Button Object Reference=@n2)
  ]}
 CreateButtonText {nodes [
   n1 = CreateButtonText!() -> (then, Outer:Editor Utility Button Object Reference, InText:Text) {is=title K2Node_FunctionEntry}
   > n2 = Construct Object from Class(Class=/Script/Blutility.EditorUtilityEditableText, self=@n1) -> (ReturnValue:Editor Utility Editable Text Object Reference)
   > "SetText (Editable Text)"!(execute, self:Editable Text Object Reference=@n2, InText:Text=@n1.InText) -> (then)
   > n3 = AddChild(self=@n1, Content=@n2)
   > SetVisibility(self=@n2, InVisibility=HitTestInvisible)
   > Return Node!(execute, PanelSlot=@n3, TextObject:Editor Utility Editable Text Object Reference=@n2)
  ]}
 EventGraph {nodes [
   "On Property Changed (SwingIdxView)"!() -> (OutputDelegate, then, PropertyName:Name=None) {is=title K2Node_ComponentBoundEvent}
   > SetVisibility(
    self=Get IdCountBox() -> (IdCountBox:Editor Utility Combo Box String Object Reference)
    InVisibility=Select(Option 0=Collapsed, Option 1=Visible, Index=Get IsIdSwing() -> (IsIdSwing:bool))
   )
   > ChangeIdCount()
   > InitializeActionWidgets()
   n1 = "On Selection Changed (IdCountBox)"!() -> (OutputDelegate, then, SelectedItem:String, SelectionType:ESelectInfo Enum) {is=title K2Node_ComponentBoundEvent}
   n2 = Switch on String!(
    execute:Exec=[@n1.then, @n33.then]
    Selection:String=GetSelectedOption!(self:"Combo Box (String) Object Reference"=Get IdCountBox()) -> (ReturnValue:String)
   ) -> (Default:Exec, Two:Exec, Three:Exec, Four:Exec) {pinNames [Two,Three,Four]}
   n3 = Set IdSourceWaves(
    execute=@n2.Four
    IdSourceWaves:Array of Strings=Make Array!("[0]":String=Swing.Idx1, "[1]":String=Swing.Idx2, "[2]":String=Swing.Idx3, "[3]":String=Swing.Idx4) -> (Array:Array of Strings)
   ) -> (Output_Get:Array of Strings)
   Set NamesDataStructure!(
    execute=[@n3.then, @n5.then, @n4.then]
    NamesDataStructure:Struct Meta Sound Source Reference Structure=Make Struct Meta Sound Source Reference!(
     SkipActionName_27_#1:bool=False
     DiableLoopOption_26_#2:bool=true
     SourceMetaSoundName_22_#3:String=Swing
     SourceWavesName_21_#4:Array of Strings=Get IdSourceWaves() -> (IdSourceWaves:Array of Strings)
     SameNameIdentifiers_32_#5:Array of Strings
    ) -> (Struct_MetaSoundSourceReference:Struct Meta Sound Source Reference Structure) {is=title K2Node_MakeStruct}
    self:EUW Meta Sound Source Reference Object Reference=Get SwingIdReference() -> (SwingIdReference:EUW Meta Sound Source Reference Object Reference)
   ) -> (then, Output_Get:Struct Meta Sound Source Reference Structure)
   > InitializeActionWidgets()
   n4 = Set IdSourceWaves(execute=@n2.Three, IdSourceWaves=Make Array!("[0]"=Swing.Idx1, "[1]"=Swing.Idx2, "[2]"=Swing.Idx3) -> (Array))
   n5 = Set IdSourceWaves(execute=@n2.Two, IdSourceWaves=Make Array!("[0]"=Swing.Idx1, "[1]"=Swing.Idx2) -> (Array))
   "On Property Changed (ReadyView)"!() -> (OutputDelegate, then, PropertyName:Name=None) {is=title K2Node_ComponentBoundEvent}
   > Widget Selector!(
    execute
    self
    ShouldAdd:bool=Get IsReady() -> (IsReady:bool)
    Reference:"EUW Meta Sound Source Reference Object Reference (by ref)"=Get ReadyReference() -> (ReadyReference:EUW Meta Sound Source Reference Object Reference)
   ) -> (then)
   n6 = Get IsReady()
   > SetIsEnabled(self=Get ReadySwingView() -> (ReadySwingView:Single Property View Object Reference), bInIsEnabled=@n6)
   > n7 = Branch(Condition=@n6)
   Set IsReadySwing(execute=@n7.else, IsReadySwing:bool=false) -> (Output_Get:bool)
   "On Property Changed (ImpactView)"!() -> (OutputDelegate, then, PropertyName:Name=None) {is=title K2Node_ComponentBoundEvent}
   > Widget Selector(
    ShouldAdd=Get IsImpact() -> (IsImpact:bool)
    Reference=Get ImpactReference() -> (ImpactReference:EUW Meta Sound Source Reference Object Reference)
   )
   > PhysicalSurfacePanelVisibility()
   "On Property Changed (SwingView)"!() -> (OutputDelegate, then, PropertyName:Name=None) {is=title K2Node_ComponentBoundEvent}
   > ChangeIdCount()
   > Widget Selector(
    ShouldAdd=Get IsSwing() -> (IsSwing:bool)
    Reference=Select(
     Option 0:"EUW Meta Sound Source Reference Object Reference (by ref)"=Get SwingReference() -> (SwingReference:EUW Meta Sound Source Reference Object Reference)
     Option 1:"EUW Meta Sound Source Reference Object Reference (by ref)"=Get SwingIdReference()
     Index=Get IsIdSwing()
    ) -> (ReturnValue:"EUW Meta Sound Source Reference Object Reference (by ref)")
   )
   "On Property Changed (ReadySwingView)"!() -> (OutputDelegate, then, PropertyName:Name=None) {is=title K2Node_ComponentBoundEvent}
   n8 = Get IsReadySwing() -> (IsReadySwing:bool)
   > n9 = Branch(Condition=@n8)
   > Widget Selector(ShouldAdd=NOT Boolean(A=@n8), Reference=Get ReadyReference())
   Widget Selector(execute=@n9.else, ShouldAdd=Get IsReady(), Reference=Get ReadyReference())
   OnSourcePropertyChanged() {is=title K2Node_CustomEvent}
   > n10 = Get!(execute:Exec) -> (then:Exec, else:Exec, Weapon Melee Item Definition:Fort Weapon Melee Item Definition Object Reference)
   > Set IsValidID(IsValidID:bool=true) -> (Output_Get:bool)
   > Clear!(
    execute
    TargetMap:Map of Strings to EPhysicalSurface Enums=Get CurrentPhysicalSurfacesMap() -> (CurrentPhysicalSurfacesMap:Map of Strings to EPhysicalSurface Enums)
   ) -> (then) {is=title K2Node_CallFunction}
   > Assign Sound for Asset!(execute:Exec, self:EUW Meta Sound Source Reference Object Reference=Get ReadyReference(), Delegate:"Delegate (by ref)"=Create Event()) -> (then:Exec) {is=title K2Node_AssignDelegate}
   > Assign Sound for Asset(self=Get SwingReference(), Delegate=Create Event())
   > Assign Sound for Asset(self=Get ImpactReference(), Delegate=Create Event())
   > PhysicalSurfacePanelVisibility()
   > n11 = GetItemData!(
    execute
    self:Item Definition Base Object Reference=Get Weapon Melee Item Definition() -> (Weapon Melee Item Definition:Fort Weapon Melee Item Definition Object Reference)
   ) -> (Found:Exec, NotFound:Exec, ItemData:"Fortnite Weapon Main Hand Impact Sounds Item Data Structure (by ref)")
   > n12 = Keys!(
    execute
    TargetMap:Map of EPhysicalSurface Enums to Sound Base Soft Object References=Break Fortnite Weapon Main Hand Impact Sounds Item Data!(FortItemComponentData_WeaponImpactSounds_MainHand:"Fortnite Weapon Main Hand Impact Sounds Item Data Structure (by ref)"=@n11) -> (ImpactPhysicalSurfaceSoundsMap:Map of EPhysicalSurface Enums to Sound Base Soft Object References) {is=title K2Node_BreakStruct}
   ) -> (then, Keys:Array of EPhysicalSurface Enums)
   > n13 = For Each Loop(Array:Array of EPhysicalSurface Enums=@n12) -> (Array Element:EPhysicalSurface Enum)
   UpdateGridImpactButtons(execute=@n13.Completed)
   Add!(
    execute=@n13.LoopBody
    TargetMap:Map of Strings to EPhysicalSurface Enums=Get CurrentPhysicalSurfacesMap()
    Key:"String (by ref)"=Enum to String!(Enumerator:EPhysicalSurface Enum=@n13) -> (ReturnValue:String)
    Value:"EPhysicalSurface Enum (by ref)"=@n13
   ) -> (then) {is=title K2Node_CallFunction}
   Set IsValidID(execute=@n10.else, IsValidID=false)
   > PhysicalSurfacePanelVisibility()
   n14 = SwingSoundForAsset!() -> (OutputDelegate, then, Sound:Sound Base Soft Object Reference, SourceReference:EUW Meta Sound Source Reference Object Reference) {is=title K2Node_CustomEvent}
   n15 = Make Map!(Key 0:EFortWeaponSoundState Enum=Normal, Value 0:Sound Base Soft Object Reference=@n14.Sound) -> (Map:Map of EFortWeaponSoundState Enums to Sound Base Soft Object References) {is=title K2Node_MakeMap}
   > SetPrimaryFireSoundsMap!(
    execute
    self:Fort Weapon Melee Item Definition Object Reference=Get Weapon Melee Item Definition()
    NewPrimaryFireSoundsMap:Map of EFortWeaponSoundState Enums to Sound Base Soft Object References=@n15
   ) -> (then)
   > SetOffHandPrimaryFire!(execute, self, PrimaryFireSoundMap:Map of EFortWeaponSoundState Enums to Sound Base Soft Object References=@n15) -> (then)
   > n16 = Branch(Condition=AND Boolean(A=Get IsReadySwing(), B=Get IsReady()))
   > n17 = ReadySoundForAsset!(execute, self, Sound:Sound Base Soft Object Reference=@n14.Sound, SourceReference:EUW Meta Sound Source Reference Object Reference) -> (then)
   RetriggerableDelay!(execute=[@n22.then, @n28.then, @n17.then, @n16.else], WorldContextObject:Object Reference, Duration=0.5) -> (then)
   > SaveIdAsset!(
    execute
    self:EUW IDSave Asset Object Reference=Get SaveAssetRef!(self:EUW IDReference Object Reference=Get IDReference() -> (IDReference:EUW IDReference Object Reference)) -> (SaveAssetRef:EUW IDSave Asset Object Reference)
    AssetToSave:Object Reference=Get Weapon Melee Item Definition()
   ) -> (then)
   n18 = ImpactSoundForAsset!() -> (OutputDelegate, then, Sound:Sound Base Soft Object Reference, SourceReference:EUW Meta Sound Source Reference Object Reference) {is=title K2Node_CustomEvent}
   > n19 = Values!(execute, TargetMap:Map of Strings to EPhysicalSurface Enums=Get CurrentPhysicalSurfacesMap()) -> (then, Values:Array of EPhysicalSurface Enums)
   > n20 = For Each Loop(Array=@n19)
   > Add(
    TargetMap:Map of EPhysicalSurface Enums to Sound Base Soft Object References=Get Impact Physical Surface Sounds Map() -> (Impact Physical Surface Sounds Map:Map of EPhysicalSurface Enums to Sound Base Soft Object References=SurfaceType_Default)
    Key:"EPhysicalSurface Enum (by ref)"=@n20
    Value:"Sound Base Soft Object Reference (by ref)"=@n18.Sound
   )
   n21 = Get Impact Physical Surface Sounds Map() -> (Impact Physical Surface Sounds Map=SurfaceType_Default)
   SetImpactPhysicalSurfaceSoundsMap!(
    execute=@n20.Completed
    self:Fort Weapon Melee Item Definition Object Reference=Get Weapon Melee Item Definition()
    NewImpactPhysicalSurfaceSoundsMap:Map of EPhysicalSurface Enums to Sound Base Soft Object References=@n21
   ) -> (then)
   > n22 = SetOffHandImpact!(execute, self, ImpactPhysicalSurfaceSoundsMap:Map of EPhysicalSurface Enums to Sound Base Soft Object References=@n21) -> (then)
   UpdateGridImpactButtons() {is=title K2Node_CustomEvent}
   > n23 = For Each Loop(Array:Array of Editor Utility Button Object References=Get ButtonArray() -> (ButtonArray:Array of Editor Utility Button Object References)) -> (Array Element:Editor Utility Button Object Reference)
   > Bind Event to On Pressed(self:Button Object Reference=@n23, Delegate=Create Event()) {is=title K2Node_AddDelegate}
   > n24 = Keys(TargetMap:Map of Strings to EPhysicalSurface Enums=Get CurrentPhysicalSurfacesMap()) -> (Keys:Array of Strings)
   n25 = "Get (a copy)"(Array:Array of Texts=Get ButtonTextArray() -> (ButtonTextArray:Array of Texts), Dimension 1=@n23.Array Index) -> (Output:Text)
   > n26 = Branch(Condition=Contains Item(TargetArray:Array of Strings=@n24, ItemToFind:"String (by ref)"="To String (Text)"(InText=@n25)))
   > SetBackgroundColor!(
    execute
    self:Button Object Reference=@n23
    InBackgroundColor:Linear Color Structure=Get ActiveColor() -> (ActiveColor:Linear Color Structure)
   ) -> (then)
   SetBackgroundColor(execute=@n26.else, self=@n23, InBackgroundColor=Get InactiveColor() -> (InactiveColor:Linear Color Structure))
   n27 = ReadySoundForAsset!() -> (OutputDelegate, then, Sound:Sound Base Soft Object Reference, SourceReference:EUW Meta Sound Source Reference Object Reference) {is=title K2Node_CustomEvent}
   > n28 = SetReloadSoundsMap!(
    execute
    self:Fort Weapon Melee Item Definition Object Reference=Get Weapon Melee Item Definition()
    NewReloadSoundsMap:Map of EFortReloadFXState Enums to Sound Base Soft Object References=Make Map(Key 0:EFortReloadFXState Enum=ReloadEnd, Value 0=@n27.Sound) -> (Map:Map of EFortReloadFXState Enums to Sound Base Soft Object References)
   ) -> (then)
   OnSurfaceButtonPressed() {is=title K2Node_CustomEvent}
   > n29 = For Each Loop(Array=Get ButtonArray())
   > Branch(Condition=IsPressed!(self:Button Object Reference=@n29) -> (ReturnValue:bool))
   > n30 = Keys(TargetMap=Get CurrentPhysicalSurfacesMap())
   n31 = "To String (Text)"(InText="Get (a copy)"(Array=Get ButtonTextArray(), Dimension 1=@n29.Array Index))
   > n32 = Branch(Condition=Contains Item(TargetArray=@n30, ItemToFind=@n31))
   > SetBackgroundColor(self=@n29, InBackgroundColor=Get InactiveColor())
   > Remove!(execute, TargetMap:Map of Strings to EPhysicalSurface Enums=Get CurrentPhysicalSurfacesMap(), Key:"String (by ref)"=@n31) -> (then, ReturnValue:bool)
   SetBackgroundColor(execute=@n32.else, self=@n29, InBackgroundColor=Get ActiveColor())
   > Add(
    TargetMap:Map of Strings to EPhysicalSurface Enums=Get CurrentPhysicalSurfacesMap()
    Key:"String (by ref)"=@n31
    Value:"EPhysicalSurface Enum (by ref)"=Find(
     TargetMap:Map of Strings to EPhysicalSurface Enums=Get AllPhyscalSurfacesMap() -> (AllPhyscalSurfacesMap:Map of Strings to EPhysicalSurface Enums)
     Key=@n31
    ) -> (Value:EPhysicalSurface Enum=SurfaceType_Default)
   )
   ChangeIdCount() {is=title K2Node_CustomEvent}
   > n33 = Get!(execute) -> (then, else, IsIdSwing:bool)
   UpdatePhysicalSurfaces() {is=title K2Node_CustomEvent}
   > Set PhysicalSurfaceGridRow(PhysicalSurfaceGridRow:int=0) -> (Output_Get:int=0)
   > Clear(TargetArray:Array of Editor Utility Button Object References=Get ButtonArray()) {is=title K2Node_CallArrayFunction}
   > Clear(TargetArray:Array of Texts=Get ButtonTextArray())
   > n34 = For Each Loop(Array:Array of EPhysicalSurface Enums=Get PhysicalSurfacesEnum() -> (PhysicalSurfacesEnum:Array of EPhysicalSurface Enums=SurfaceType_Default)) -> (Array Element:EPhysicalSurface Enum)
   n35 = Enum to String(Enumerator=@n34)
   > Add(TargetMap=Get AllPhyscalSurfacesMap(), Key=@n35, Value=@n34)
   > n36 = CreatButtonObject!(
    execute
    self
    Outer:Uniform Grid Panel Object Reference=Get PhysicalSurfaceGrid() -> (PhysicalSurfaceGrid:Uniform Grid Panel Object Reference)
   ) -> (then, PanelSlot:Panel Slot Object Reference, Button Object:Editor Utility Button Object Reference) {is=title K2Node_CallFunction}
   > n37 = CreateButtonText!(execute, self, Outer:Editor Utility Button Object Reference=@n36.Button Object, InText:Text="To Text (String)"(InString=@n35)) -> (then, PanelSlot:Panel Slot Object Reference, TextObject:Editor Utility Editable Text Object Reference) {is=title K2Node_CallFunction}
   > Add Unique(
    TargetArray:Array of Editor Utility Button Object References=Get ButtonArray()
    NewItem:"Editor Utility Button Object Reference (by ref)"=@n36.Button Object
   )
   > Add Unique(TargetArray:Array of Texts=Get ButtonTextArray(), NewItem:"Text (by ref)"=Get Text(self:Editable Text Object Reference=@n37.TextObject))
   > UpdateGridImpactButtons() {is=title K2Node_CallFunction}
   > n38 = Cast To UniformGridSlot!(execute, Object=@n36) -> (then, CastFailed, AsUniform Grid Slot:Uniform Grid Slot Object Reference) {is=title K2Node_DynamicCast}
   > SetHorizontalAlignment(self:Uniform Grid Slot Object Reference=@n38, InHorizontalAlignment=HAlign_Fill)
   n39 = "% (Integer)"(A=@n34.Array Index, B=3) -> (ReturnValue=0)
   > SetColumn!(execute, self:Uniform Grid Slot Object Reference=@n38, InColumn:int=@n39) -> (then)
   n40 = Get PhysicalSurfaceGridRow() -> (PhysicalSurfaceGridRow:int=0)
   > SetRow!(execute, self:Uniform Grid Slot Object Reference=@n38, InRow:int=@n40) -> (then)
   > Branch(Condition="Equal (Integer)"(A=@n39, B=2))
   > Increment Int(Value=@n40)
   Event PreConstruct()
   n41 = Get IDSource!(self:EUW IDReference Object Reference=Get IDReference()) -> (IDSource:Single Property View Object Reference)
   > SetPropertyName!(execute, self:Single Property View Object Reference=@n41, NewPropertyName:Name=Weapon Melee Item Definition) -> (then)
   > SetObject(self=@n41, NewObject=Self-Reference())
   > UpdatePhysicalSurfaces() {is=title K2Node_CallFunction}
   > SetObject(
    self=[
     Get ImpactView() -> (ImpactView:Single Property View Object Reference)
     Get SwingView() -> (SwingView:Single Property View Object Reference)
     Get ReadyView() -> (ReadyView:Single Property View Object Reference)
     Get ReadySwingView()
     Get SwingIdxView() -> (SwingIdxView:Single Property View Object Reference)
    ]
    NewObject=Self-Reference()
   )
   Event Construct()
   > Assign On Source Property Changed!(execute:Exec, self:EUW IDReference Object Reference=Get IDReference(), Delegate:"Delegate (by ref)"=Create Event()) -> (then:Exec) {is=title K2Node_AssignDelegate}
   > InitializeActionWidgets()
   > PhysicalSurfacePanelVisibility()
   > Set SwingReference(SwingReference:EUW Meta Sound Source Reference Object Reference=Get SwingNormRef() -> (SwingNormRef:EUW Meta Sound Source Reference Object Reference)) -> (Output_Get:EUW Meta Sound Source Reference Object Reference)
   Get IsReady()
  ]}
 GetCreatorWidget {nodes [
   GetCreatorWidget() {is=title K2Node_FunctionEntry}
   n1 = GetParent!(self:Widget Object Reference) -> (ReturnValue:Panel Widget Object Reference)
   > n2 = Cast To EUW_MetaSound_Cosmetic_PresetCreator!(execute, Object=GetOuterObject!(Object:Object Reference=GetOuterObject(Object=@n1)) -> (ReturnValue:Object Reference)) -> (then, CastFailed, AsEUW Meta Sound Cosmetic Preset Creator:EUW Meta Sound Cosmetic Preset Creator Object Reference) {is=title K2Node_DynamicCast}
   > Return Node!(execute, Costmetic Creator:EUW Meta Sound Cosmetic Preset Creator Object Reference=@n2)
  ]}
 InitializeActionWidgets {nodes [
   InitializeActionWidgets() {is=title K2Node_FunctionEntry}
   > n1 = Get!(execute) -> (then, else, IsSwing:bool)
   n2 = Get IsIdSwing()
   > Widget Selector(ShouldAdd=@n2, Reference=Get SwingIdReference())
   > n3 = Widget Selector(ShouldAdd=NOT Boolean(A=@n2), Reference=Get SwingReference())
   Widget Selector(execute=[@n3.then, @n1.else], ShouldAdd=Get IsReady(), Reference=Get ReadyReference())
   > Widget Selector(ShouldAdd=Get IsImpact(), Reference=Get ImpactReference())
  ]}
 PhysicalSurfacePanelVisibility {nodes [
   PhysicalSurfacePanelVisibility() {is=title K2Node_FunctionEntry}
   > SetIsEnabled(
    self=Get PhysicalSurfacePanel() -> (PhysicalSurfacePanel:Grid Panel Object Reference)
    bInIsEnabled=AND Boolean(A=Get IsImpact(), B=Get IsValidID() -> (IsValidID:bool))
   )
  ]}
 SetOffHandImpact {nodes [
   n1 = SetOffHandImpact!() -> (then, ImpactPhysicalSurfaceSoundsMap:Map of EPhysicalSurface Enums to Sound Base Soft Object References=SurfaceType_Default) {is=title K2Node_FunctionEntry}
   > n2 = Cast To FortWeaponMeleeDualWieldItemDefinition!(execute, Object=Get Weapon Melee Item Definition()) -> (then, CastFailed, AsFort Weapon Melee Dual Wield Item Definition:Fort Weapon Melee Dual Wield Item Definition Object Reference) {is=title K2Node_DynamicCast}
   > SetItemData!(
    execute
    self:Item Definition Base Object Reference=@n2
    ItemData:Fortnite Weapon Off Hand Impact Sounds Item Data Structure=Make Fortnite Weapon Off Hand Impact Sounds Item Data!(ImpactPhysicalSurfaceSoundsMap:Map of EPhysicalSurface Enums to Sound Base Soft Object References=@n1) -> (FortItemComponentData_WeaponImpactSounds_OffHand:Fortnite Weapon Off Hand Impact Sounds Item Data Structure) {is=title K2Node_MakeStruct}
   ) -> (then)
  ]}
 SetOffHandPrimaryFire {nodes [
   n1 = SetOffHandPrimaryFire!() -> (then, PrimaryFireSoundMap:Map of EFortWeaponSoundState Enums to Sound Base Soft Object References=Normal) {is=title K2Node_FunctionEntry}
   > n2 = Cast To FortWeaponMeleeDualWieldItemDefinition(Object=Get Weapon Melee Item Definition())
   > SetItemData(
    self=@n2
    ItemData:Fortnite Weapon Off Hand Primary Fire Sounds Item Data Structure=Make Fortnite Weapon Off Hand Primary Fire Sounds Item Data!(PrimaryFireSoundMap:Map of EFortWeaponSoundState Enums to Sound Base Soft Object References=@n1) -> (FortItemComponentData_WeaponPrimaryFireSounds_OffHand:Fortnite Weapon Off Hand Primary Fire Sounds Item Data Structure) {is=title K2Node_MakeStruct}
   )
  ]}
 Widget Selector {nodes [
   n1 = Widget Selector!() -> (then, ShouldAdd:bool=~, Reference:"EUW Meta Sound Source Reference Object Reference (by ref)") {is=title K2Node_FunctionEntry}
   > n2 = Branch(Condition=@n1)
   n3 = Get SoundSourceReferenceWidgets() -> (SoundSourceReferenceWidgets:Array of EUW Meta Sound Source Reference Object References)
   > n4 = Add Unique(
    TargetArray:Array of EUW Meta Sound Source Reference Object References=@n3
    NewItem:"EUW Meta Sound Source Reference Object Reference (by ref)"=@n1.Reference
   )
   Set Current Source Reference Widgets!(
    execute=[@n5.then, @n4.then]
    self:EUW Meta Sound Cosmetic Preset Creator Object Reference=GetCreatorWidget!() -> (Costmetic Creator:EUW Meta Sound Cosmetic Preset Creator Object Reference) {is=title K2Node_CallFunction}
    CurrentActiveWidget:Widget Object Reference=Self-Reference()
    SourceReferences:Array of EUW Meta Sound Source Reference Object References=Get SoundSourceReferenceWidgets()
    CosmeticType:String=Pickaxe
   ) -> (then) {is=title K2Node_CallFunction}
   n5 = Remove Item!(
    execute=@n2.else
    TargetArray:Array of EUW Meta Sound Source Reference Object References=@n3
    Item:"EUW Meta Sound Source Reference Object Reference (by ref)"=@n1.Reference
   ) -> (then, ReturnValue:bool)
  ]}
]
variables [
 ActiveColor {type=LinearColor, default=<rgba 0 0.4657 1 1>}
 AllPhyscalSurfacesMap {type="map<EPhysicalSurface>", default {}}
 ButtonArray {type="array<EditorUtilityButton>", default []}
 ButtonTextArray {type="array<string>", default []}
 CurrentPhysicalSurfacesMap {type="map<EPhysicalSurface>", default {}}
 IdSourceWaves {type="array<string>", default []}
 Impact Physical Surface Sounds Map {default {}}
 ImpactReference {
  type=EUW_MetaSoundSourceReference_C
  default=&/EUW_MetaSound_Cosmetic_PresetCreator_Pickaxe.Default__EUW_MetaSound_Cosmetic_PresetCreator_Pickaxe_C:EUW_MetaSoundSourceReference_C_5
 }
 InactiveColor {type=LinearColor, default=<rgba 1 1 1 1>}
 IsIdSwing {type=boolean}
 IsImpact {type=boolean, default=true}
 IsReady {type=boolean, default=true}
 IsReadySwing {type=boolean}
 IsSwing {type=boolean, default=true}
 IsValidID {type=boolean}
 PhysicalSurfaceGridRow {type=integer}
 PhysicalSurfacesEnum {
  type="array<EPhysicalSurface>"
  default [
   SurfaceType_Default
   SurfaceType1
   SurfaceType2
   SurfaceType3
   SurfaceType4
   SurfaceType5
   SurfaceType6
   SurfaceType7
   SurfaceType8
   SurfaceType9
   SurfaceType10
   SurfaceType11
   SurfaceType12
   SurfaceType13
   SurfaceType14
   SurfaceType15
   SurfaceType16
   SurfaceType17
   SurfaceType18
   SurfaceType19
   SurfaceType20
   SurfaceType21
   SurfaceType22
   SurfaceType23
   SurfaceType24
   SurfaceType25
   SurfaceType26
   SurfaceType27
   SurfaceType28
   SurfaceType29
   SurfaceType30
   SurfaceType31
   SurfaceType32
   SurfaceType33
   SurfaceType34
  ]
 }
 ReadyReference {
  type=EUW_MetaSoundSourceReference_C
  default=&Default__EUW_MetaSound_Cosmetic_PresetCreator_Pickaxe_C:EUW_MetaSoundSourceReference_C_3
 }
 SoundSourceReferenceWidgets {type="array<EUW_MetaSoundSourceReference_C>", default []}
 SwingIdReference {
  type=EUW_MetaSoundSourceReference_C
  default=&Default__EUW_MetaSound_Cosmetic_PresetCreator_Pickaxe_C:EUW_MetaSoundSourceReference_C_0
 }
 SwingNormRef {
  type=EUW_MetaSoundSourceReference_C
  default=&Default__EUW_MetaSound_Cosmetic_PresetCreator_Pickaxe_C:EUW_MetaSoundSourceReference_C_4
 }
 SwingReference {
  type=EUW_MetaSoundSourceReference_C
  default=&Default__EUW_MetaSound_Cosmetic_PresetCreator_Pickaxe_C:EUW_MetaSoundSourceReference_C_4
 }
 Weapon Melee Item Definition {type=FortWeaponMeleeItemDefinition, default=null}
]
widgets {
 info {rootWidgetClass=&CanvasPanel, widgetCount=22}
 widgets [
  CanvasPanel_0 {widgetClassPath=&CanvasPanel, widget=CanvasPanel_0}
  GridPanel_1 {
   widgetClassPath=&GridPanel
   parent=CanvasPanel_0
   widget=GridPanel_1
   slot {
    class=CanvasPanelSlot
    properties {bAutoSize=true, layoutData {alignment=<0 0>, anchors {maximum=<0 0>, minimum=<0 0>}, offsets=<ltrb 0 0 500 300>}}
   }
  }
  VerticalBox_78 {widgetClassPath=&VerticalBox, parent=GridPanel_1, widget=VerticalBox_78, slot {class=GridSlot, properties {row=1}}}
  HorizontalBox_116 {
   widgetClassPath=&HorizontalBox
   parent=VerticalBox_78
   widget=HorizontalBox_116
   slot {class=VerticalBoxSlot, properties {padding=<ltrb 10 0 10 10>}}
  }
  VerticalBox_102 {widgetClassPath=&VerticalBox, parent=HorizontalBox_116, widget=VerticalBox_102, slot {class=HorizontalBoxSlot}}
  ReadySwingView {
   widgetClassPath=&SinglePropertyView
   parent=VerticalBox_102
   widget=ReadySwingView
   properties {
    nameOverride=Add Swing MetaSound to Ready Slot in Item Definition
    propertyName=IsReadySwing
    toolTipText=Should Ready sounds be the same as Swing? Sets swing metasound in the ready slot of item definition. Nulls out the ready selection in editor widget.
   }
   slot {class=VerticalBoxSlot, properties {padding=<ltrb 10 10 10 10>}}
  }
  HorizontalBox_0 {
   widgetClassPath=&HorizontalBox
   parent=VerticalBox_102
   widget=HorizontalBox_0
   slot {class=VerticalBoxSlot, properties {padding=<ltrb 10 10 10 10>}}
  }
  SwingIdxView {
   widgetClassPath=&SinglePropertyView
   parent=HorizontalBox_0
   widget=SwingIdxView
   properties {
    nameOverride=Use swing ID for sequential sounds
    propertyName=IsIdSwing
    toolTipText=Should Ready sounds be the same as Swing? Sets swing metasound in the ready slot of item definition. Nulls out the ready selection in editor widget.
   }
   slot {class=HorizontalBoxSlot}
  }
  IdCountBox {
   widgetClassPath=/Script/Blutility.EditorUtilityComboBoxString
   bIsVariable=true
   parent=HorizontalBox_0
   widget=IdCountBox
   properties {
    selectedOption=Two
    visibility=Collapsed
    defaultOptions [Two,Three,Four]
    font {
     bForceMonospaced=false
     bMaterialIsStencil=false
     fontMaterial=null
     fontObject=/LauncherSocial/BrutalType
     letterSpacing=0
     monospacedWidth=1
     size=9
     skewAmount=0
     typefaceFontName=Regular
     outlineSettings {
      bApplyOutlineToDropShadows=false
      bMiteredCorners=false
      bSeparateFillAlpha=false
      outlineMaterial=null
      outlineSize=0
      outlineColor=<rgba 0 0 0 1>
     }
    }
    foregroundColor {colorUseRule=UseColor_Specified, specifiedColor=<rgba 0.5271 0.5271 0.5271 1>}
   }
   slot {class=HorizontalBoxSlot, properties {verticalAlignment=VAlign_Center, size {sizeRule=Fill, value=1}}}
  }
  IDReference {
   widgetClassPath=/Game/Sounds/Fort_Audio_Blueprints/Utilities/EditorUtilityWidgets/Metasounds/Tools/EUW_IDReference
   bIsVariable=true
   parent=VerticalBox_78
   widget=IDReference
   slot {class=VerticalBoxSlot}
  }
  PhysicalSurfacePanel {widgetClassPath=&GridPanel, bIsVariable=true, parent=VerticalBox_78, widget=PhysicalSurfacePanel, slot {class=VerticalBoxSlot}}
  Border_193 {
   widgetClassPath=&Border
   parent=PhysicalSurfacePanel
   widget=Border_193
   properties {
    horizontalAlignment=HAlign_Center
    verticalAlignment=VAlign_Center
    background {
     bIsDynamicallyLoaded=false
     drawAs=RoundedBox
     imageType=NoImage
     mirroring=NoMirror
     resourceName=None
     resourceObject=null
     tiling=NoTile
     imageSize=<32 32>
     margin=<ltrb 0 0 0 0>
     outlineSettings {
      bUseBrushTransparency=false
      roundingType=FixedRadius
      width=1
      color {colorUseRule=UseColor_Specified, specifiedColor=<rgba 0.1198 0.1198 0.1198 0.5>}
      cornerRadii=<0 0 0 0>
     }
     tintColor {colorUseRule=UseColor_Specified, specifiedColor=<rgba 0.0423 0.0423 0.0423 1>}
     uVRegion {bIsValid=false, max=<0 0>, min=<0 0>}
    }
    padding=<ltrb 4 4 4 4>
   }
   slot {class=GridSlot, properties {verticalAlignment=VAlign_Center}}
  }
  SoundListText {
   widgetClassPath=&EditorUtilityEditableText
   parent=Border_193
   widget=SoundListText
   properties {hintText=Source Reference Impact Physical Surfaces, text=Source Reference Impact Physical Surfaces}
   slot {class=BorderSlot, properties {horizontalAlignment=HAlign_Center, verticalAlignment=VAlign_Center, padding=<ltrb 4 4 4 4>}}
  }
  PhysicalSurfaceGrid {
   widgetClassPath=/Script/UMG.UniformGridPanel
   bIsVariable=true
   parent=PhysicalSurfacePanel
   widget=PhysicalSurfaceGrid
   slot {class=GridSlot, properties {row=1}}
  }
  GridPanel_183 {widgetClassPath=&GridPanel, parent=GridPanel_1, widget=GridPanel_183, slot {class=GridSlot}}
  HorizontalBox {
   widgetClassPath=&HorizontalBox
   parent=GridPanel_183
   widget=HorizontalBox
   slot {class=GridSlot, properties {horizontalAlignment=HAlign_Left, verticalAlignment=VAlign_Top, padding=<ltrb 10 10 10 10>}}
  }
  SwingView {
   widgetClassPath=&SinglePropertyView
   parent=HorizontalBox
   widget=SwingView
   properties {nameOverride=Swing:, propertyName=IsSwing}
   slot {class=HorizontalBoxSlot}
  }
  HorizontalBox_1 {
   widgetClassPath=&HorizontalBox
   parent=GridPanel_183
   widget=HorizontalBox_1
   slot {class=GridSlot, properties {column=1, horizontalAlignment=HAlign_Left, verticalAlignment=VAlign_Top, padding=<ltrb 10 10 10 10>}}
  }
  ReadyView {
   widgetClassPath=&SinglePropertyView
   parent=HorizontalBox_1
   widget=ReadyView
   properties {nameOverride=Ready:, propertyName=IsReady}
   slot {class=HorizontalBoxSlot}
  }
  HorizontalBox_2 {
   widgetClassPath=&HorizontalBox
   parent=GridPanel_183
   widget=HorizontalBox_2
   slot {class=GridSlot, properties {column=2, horizontalAlignment=HAlign_Left, verticalAlignment=VAlign_Top, padding=<ltrb 10 10 10 10>}}
  }
  ImpactView {
   widgetClassPath=&SinglePropertyView
   parent=HorizontalBox_2
   widget=ImpactView
   properties {nameOverride=Impact:, propertyName=IsImpact}
   slot {class=HorizontalBoxSlot}
  }
  HorizontalBox_3 {
   widgetClassPath=&HorizontalBox
   parent=GridPanel_183
   widget=HorizontalBox_3
   slot {class=GridSlot, properties {column=3, horizontalAlignment=HAlign_Left, verticalAlignment=VAlign_Top, padding=<ltrb 10 10 10 10>}}
  }
 ]
}
