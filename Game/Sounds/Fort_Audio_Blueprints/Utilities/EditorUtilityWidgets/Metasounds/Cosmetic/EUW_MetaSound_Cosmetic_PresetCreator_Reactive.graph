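# Editor utility panel in the cosmetic MetaSound preset creator that sets up audio for reactive cosmetics on character parts.
# It creates a parented Sound Library, fills it with MetaSounds tagged from a cosmetic-option data table, and attaches it to the character part.
# Unless set to animation mode, it adds and edits stat-watcher components on the part modifier Blueprint through SubobjectDataSubsystem.
# In animation mode, a button adds a Sound Library notify to the animation.
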EditorUtilityWidgetBlueprint /Game/Sounds/Fort_Audio_Blueprints/Utilities/EditorUtilityWidgets/Metasounds/Cosmetic/EUW_MetaSound_Cosmetic_PresetCreator_Reactive
graphs [
 Add MetaSounds To Library {nodes [
   n1 = Add MetaSounds To Library!() -> (then, Sounds:"Sound Base Soft Object Reference (by ref)") {is=title K2Node_FunctionEntry}
   > n2 = Get Cosmetic Options!(execute:Exec) -> (
    Success:Exec
    Failed:Exec
    CosmeticOptions:Enum_MetaSound_Creator_CosmeticOptions Enum=NewEnumerator1
    Stat Manager Tag:Fort Stat Manager Tag Structure
    SoundLibraryTag:Sound Library Tag Structure
   ) {is=title K2Node_MacroInstance}
   > n3 = AddMetasoundToLibrary!(
    execute
    Sound:"Sound Base Soft Object Reference (by ref)"=@n1
    SoundLibraryTag:Sound Library Tag Structure=@n2.SoundLibraryTag
    SoundLibrary:Sound Library Object Reference=Get SoundLibrary() -> (SoundLibrary:Sound Library Object Reference)
    __WorldContext:Object Reference
   ) -> (then, Sounds:Map of Gameplay Tag Structures to Sound Base Soft Object References, Success:bool)
   > Set LibrarySounds(LibrarySounds:Map of Gameplay Tag Structures to Sound Base Soft Object References=@n3) -> (Output_Get:Map of Gameplay Tag Structures to Sound Base Soft Object References)
   > Add Sound Library to Character Part()
   > n4 = Branch(Condition=Get bIsAnimationImplementation() -> (bIsAnimationImplementation:bool))
   Attempt To Create Reactive Component!(execute=@n4.else, self, InputPin:Fort Stat Manager Tag Structure=@n2.Stat Manager Tag) -> (then)
   Show Message Dialog(
    execute=@n2.Failed
    Title=MakeLiteralText(Value="LOCTEXT(\"Cosmetic Option Error\")")
    Message=MakeLiteralText(Value="LOCTEXT(\"Tags not found in Data table. Please reach out to audio producer or audio tech POCs.\")")
    MessageType=Ok
    DefaultValue=No
    MessageCategory=Warning
   )
   Add!(
    execute
    TargetMap:Map of Gameplay Tag Structures to Sound Base Soft Object References=Get Sounds!(self:Sound Library Object Reference=Get SoundLibrary()) -> (Sounds:Map of Gameplay Tag Structures to Sound Base Soft Object References)
    Key:"Gameplay Tag Structure (by ref)"
    Value:"Sound Base Soft Object Reference (by ref)"
   ) -> (then) {is=title K2Node_CallFunction}
   Get Sounds(self=Get SoundLibrary())
  ]}
 Add Reactive Object to Blueprint {nodes [
   n1 = Add Reactive Object to Blueprint!() -> (then, SubobjectHandle:Array of Subobject Data Handle Structures) {is=title K2Node_FunctionEntry}
   > n2 = Get Cosmetic Options() -> (CosmeticOptions=NewEnumerator1)
   n3 = Get SubobjectDataSubsystem()
   > Is Valid(InputObject=@n3)
   n4 = "Get (a ref)"(Array:Array of Subobject Data Handle Structures=@n1, Dimension 1=0) -> (Output:"Subobject Data Handle Structure (by ref)")
   > Branch(Condition=IsHandleValid!(DataHandle:"Subobject Data Handle Structure (by ref)"=@n4) -> (ReturnValue:bool))
   > n5 = AddNewSubobject!(
    execute
    self:Subobject Data Subsystem Object Reference=@n3
    Params_ParentHandle:Subobject Data Handle Structure=@n4
    Params_NewClass:Object Class Reference=/Game/Sounds/Fort_Audio_Blueprints/Cosmetics/B_Reactive_StatWatcher_Component.B_Reactive_StatWatcher_Component_C
    Params_BlueprintContext:Blueprint Object Reference=GetBlueprint!(Data:"Subobject Data Structure (by ref)"=GetData(DataHandle=@n4)) -> (ReturnValue:Blueprint Object Reference)
    Params_bSkipMarkBlueprintModified:bool=False
    Params_bConformTransformToParent:bool=false
   ) -> (then, FailReason:Text, ReturnValue:Subobject Data Handle Structure)
   n6 = Split(
    SourceString=Get Class Display Name(Class=&B_Reactive_StatWatcher_Component_C)
    InStr=_
    SearchDir=FromEnd
   )
   n7 = Append(A=_, B=GetCosmeticOption!() -> (ReturnValue:String))
   > RenameSubobject!(
    execute
    self:Subobject Data Subsystem Object Reference=@n3
    Handle:"Subobject Data Handle Structure (by ref)"=@n5.ReturnValue
    InNewName:"Text (by ref)"="To Text (String)"(InString=Append(A=@n6, B=@n7))
   ) -> (then, ReturnValue:bool)
   > n8 = Branch(Condition=IsHandleValid(DataHandle=@n5.ReturnValue))
   n9 = GetVariableName!(Data:"Subobject Data Structure (by ref)"=GetData(DataHandle=@n5.ReturnValue)) -> (ReturnValue:Name=None)
   > PrintString(InString=Append(A="Component Added: ", B="To String (Name)"(InName=@n9)))
   > n10 = Find Stat Watch Component!(exec, Handle:"Subobject Data Handle Structure (by ref)"=@n5.ReturnValue) -> (Success:Exec, Failed:Exec, AsB Reactive Stat Watcher Component:B Reactive Stat Watcher Component Object Reference) {is=title K2Node_MacroInstance}
   n11 = Make Struct Meta Sound Reactive Component Data!(
    ReactiveComponent_35_#1:B Reactive Stat Watcher Component Class Reference=&B_Reactive_StatWatcher_Component_C
    WatcherStatTag_12_#2:Fort Stat Manager Tag Structure=@n2.Stat Manager Tag
    DefaultValue_14_#3:int=-1
    UpdateAudioParameter_30_#4:bool=False
    BroadcastInitialValue_32_#5:bool=False
    InitialValueBroadcastDelay_23_#6:double=-1
    ConsumingWaitingNotifications_27_#7:bool=True
    Object_43_#8:B Reactive Stat Watcher Component Object Reference=@n10
   ) -> (Struct_MetaSound_Reactive_Component_Data:Struct Meta Sound Reactive Component Data Structure) {is=title K2Node_MakeStruct}
   > Set Array Elem!(
    execute
    TargetArray:Array of Struct Meta Sound Reactive Component Data Structures=Get ReactiveComponents() -> (ReactiveComponents:Array of Struct Meta Sound Reactive Component Data Structures)
    Index:int=0
    Item:"Struct Meta Sound Reactive Component Data Structure (by ref)"=@n11
    bSizeToFit:bool=true
   ) -> (then)
   n12 = Break Struct Meta Sound Reactive Component Data!(Struct_MetaSound_Reactive_Component_Data:"Struct Meta Sound Reactive Component Data Structure (by ref)"=@n11) -> (
    ReactiveComponent_35_#1:B Reactive Stat Watcher Component Class Reference
    WatcherStatTag_12_#2:Fort Stat Manager Tag Structure
    DefaultValue_14_#3:int
    UpdateAudioParameter_30_#4:bool=~
    BroadcastInitialValue_32_#5:bool=~
    InitialValueBroadcastDelay_23_#6:double=~
    ConsumingWaitingNotifications_27_#7:bool=~
    Object_43_#8:B Reactive Stat Watcher Component Object Reference
   ) {is=title K2Node_BreakStruct}
   > Set WatcherStatTag!(
    execute
    WatcherStatTag:Fort Stat Manager Tag Structure=@n12.WatcherStatTag_12_#2
    self:B Reactive Stat Watcher Component Object Reference=@n10
   ) -> (then, Output_Get:Fort Stat Manager Tag Structure)
   > Set Default Value!(
    execute
    Default Value:int=@n12.DefaultValue_14_#3
    self:B Reactive Stat Watcher Component Object Reference=@n10
   ) -> (then, Output_Get:int=0)
   > Set UpdateAudioParameter!(
    execute
    UpdateAudioParameter:bool=@n12.UpdateAudioParameter_30_#4
    self:B Reactive Stat Watcher Component Object Reference=@n10
   ) -> (then, Output_Get:bool)
   > Set Broadcast Initial Value!(
    execute
    Broadcast Initial Value:bool=@n12.BroadcastInitialValue_32_#5
    self:B Reactive Stat Watcher Component Object Reference=@n10
   ) -> (then, Output_Get:bool)
   > Set Initial Value Broadcast Delay!(
    execute
    Initial Value Broadcast Delay:float=@n12.InitialValueBroadcastDelay_23_#6
    self:B Reactive Stat Watcher Component Object Reference=@n10
   ) -> (then, Output_Get:float)
   > Set Consume Waiting Notifications!(
    execute
    Consume Waiting Notifications:bool=@n12.ConsumingWaitingNotifications_27_#7
    self:B Reactive Stat Watcher Component Object Reference=@n10
   ) -> (then, Output_Get:bool)
   > Set SoundLibraryTag!(execute, SoundLibraryTag:Sound Library Tag Structure=@n2.SoundLibraryTag, self:B Reactive Stat Watcher Component Object Reference=@n10) -> (then, Output_Get:Sound Library Tag Structure)
   PrintString(execute=@n8.else, InString="To String (Text)"(InText=@n5))
  ]}
 Add Sound Library to Character Part {nodes [
   Add Sound Library to Character Part() {is=title K2Node_FunctionEntry}
   > n1 = Get!(execute:Exec) -> (then:Exec, else:Exec, SoundLibrary:Sound Library Object Reference)
   > Set LibrarySounds(LibrarySounds=Get Sounds(self=@n1))
   > n2 = Get!(execute) -> (then, else, CharacterPart:Custom Character Part Object Reference)
   > Add Unique(
    TargetArray:Array of Sound Library Object References=Get SoundLibraries!(self:Custom Character Part Object Reference=@n2) -> (SoundLibraries:Array of Sound Library Object References)
    NewItem:"Sound Library Object Reference (by ref)"=@n1
   )
   > Set Current Sound Library(Current Sound Library:Sound Library Object Reference=@n1) -> (Output_Get:Sound Library Object Reference)
   Clear!(
    execute=@n1.else
    TargetMap:Map of Gameplay Tag Structures to Sound Base Soft Object References=Get LibrarySounds() -> (LibrarySounds:Map of Gameplay Tag Structures to Sound Base Soft Object References)
   ) -> (then) {is=title K2Node_CallFunction}
   > n3 = Get()
   > Remove Item!(
    execute
    TargetArray:Array of Sound Library Object References=Get SoundLibraries(self=@n3)
    Item:"Sound Library Object Reference (by ref)"=Get Current Sound Library() -> (Current Sound Library:Sound Library Object Reference)
   ) -> (then, ReturnValue:bool)
  ]}
 Add Widget Sound Libaries to Sound Library Asset {nodes [
   Add Widget Sound Libaries to Sound Library Asset() {is=title K2Node_FunctionEntry}
   > SetEditorProperty(
    Object=Get SoundLibrary()
    PropertyName=Sounds
    PropertyValue:Map of Gameplay Tag Structures to Sound Base Soft Object References=Get LibrarySounds()
   )
  ]}
 Attempt To Create Reactive Component {nodes [
   n1 = Attempt To Create Reactive Component!() -> (then, InputPin:Fort Stat Manager Tag Structure) {is=title K2Node_FunctionEntry}
   > n2 = Part Modifier Parent Check!(execute:Exec) -> (Success:Exec, Fail:Exec) {is=title K2Node_MacroInstance}
   > Set TagAlreadyApplied(TagAlreadyApplied:bool=false) -> (Output_Get:bool)
   > n3 = Get!(execute) -> (then, else, As B Athena Part Modifier Generic:B Athena Part Modifier Generic Object Reference)
   n4 = Get SubobjectDataSubsystem()
   > Is Valid(InputObject=@n4)
   > n5 = Gather Subobject Data For Instance(self=@n4, Context=@n3)
   > n6 = For Each Loop(Array:Array of Subobject Data Handle Structures=@n5) -> (Array Element:Subobject Data Handle Structure)
   n7 = GetObject!(
    Data:"Subobject Data Structure (by ref)"=FindSubobjectDataFromHandle!(self:Subobject Data Subsystem Object Reference=@n4, Handle:"Subobject Data Handle Structure (by ref)"=@n6) -> (OutData:Subobject Data Structure, ReturnValue:bool)
    bEvenIfPendingKill:bool=false
   ) -> (ReturnValue:Object Reference)
   > Cast To B_Reactive_StatWatcher_Component Class!(execute:Exec, Class:Object Class Reference=Get Class(Object=@n7) -> (ReturnValue:Object Class Reference)) -> (then:Exec, CastFailed:Exec, AsB Reactive Stat Watcher Component:B Reactive Stat Watcher Component Class Reference) {is=title K2Node_ClassDynamicCast}
   > n8 = Cast To B_Reactive_StatWatcher_Component!(execute, Object=@n7) -> (then, CastFailed, AsB Reactive Stat Watcher Component:B Reactive Stat Watcher Component Object Reference) {is=title K2Node_DynamicCast}
   > Branch(Condition=MatchesTag!(
     TagOne:Gameplay Tag Structure=Get WatcherStatTag!(self:B Reactive Stat Watcher Component Object Reference=@n8) -> (WatcherStatTag:Fort Stat Manager Tag Structure)
     TagTwo:Gameplay Tag Structure=@n1
     bExactMatch:bool=true
    ) -> (ReturnValue:bool))
   > Set TagAlreadyApplied(TagAlreadyApplied=true)
   n9 = Branch(execute=@n6.Completed, Condition=Get TagAlreadyApplied() -> (TagAlreadyApplied:bool))
   > PrintString(InString=Reactive Component already exists with same tag on character parts. No component will be created.)
   Add Reactive Object to Blueprint!(execute=@n9.else, self, SubobjectHandle:Array of Subobject Data Handle Structures=@n5) -> (then) {is=title K2Node_CallFunction}
   Show Message Dialog(
    execute=@n2.Fail
    Title=MakeLiteralText(Value="LOCTEXT(\"Part Modifier Blueprint Component Failed!\")")
    Message=MakeLiteralText(Value="LOCTEXT(\"Part Modifier Blueprint can not be \\\"B_Athena_PartModifier_Generic\\\". \\r\\n\\r\\nNo component will be added to this bluerpint.\\r\\n\\r\\nPlease add the proper blueprint or reach out to audio production or design to get acces to the proper assets. \\r\\n\")")
    MessageType=Ok
    DefaultValue=Continue
    MessageCategory=Error
   )
  ]}
 Check for Valid Character Part {nodes [
   n1 = Get ReactiveComponentDetails() -> (ReactiveComponentDetails:Details View Object Reference)
   n2 = SetIsEnabled(self=@n1, bInIsEnabled=false)
   Clear(execute=[@n2.then, @n4.else], TargetArray:Array of Struct Meta Sound Reactive Component Data Structures=Get ReactiveComponents()) {is=title K2Node_CallArrayFunction}
   > Show Message Dialog(
    Title=MakeLiteralText(Value="LOCTEXT(\"Missing Part Modifier Blueprint\")")
    Message=MakeLiteralText(Value="LOCTEXT(\"Part Modifier Blueprint is required to implement audio. \\r\\n\\r\\nPlease add the proper blueprint or reach out to audio production or design to get acces to the proper assets. \\r\\n\\r\\nDo not attempt to implement with out a part modifier blueprint, this widget will not work properly.\")")
    MessageType=Ok
    DefaultValue=No
    MessageCategory=Warning
   )
   n3 = SetIsEnabled(self=@n1, bInIsEnabled=true)
   Get Component Stats from Blueprint(execute=[@n3.then, @n5.Success]) {is=title K2Node_CallFunction}
   Check for Valid Character Part() {is=title K2Node_FunctionEntry}
   > SetPartsModifierBlueprint()
   > n4 = Get!(execute) -> (then, else, PartModifierBlueprint:Actor Soft Class Reference)
   > n5 = Part Modifier Parent Check()
  ]}
 Create Sound Library {nodes [
   n1 = Create Sound Library!() -> (then, PackagePath:String, AssetName:String) {is=title K2Node_FunctionEntry}
   > n2 = CreateSoundLibrary!(execute, PackagingPath:String=@n1, AssetsName:String=@n1.AssetName, __WorldContext:Object Reference) -> (then, Sound Library:Sound Library Object Reference, Successful:bool)
   > Branch(Condition=@n2.Successful)
   > SaveLoadedAsset(AssetToSave=@n2, bOnlyIfIsDirty=true)
   > Return Node!(execute, AsSound Library:Sound Library Object Reference=@n2)
  ]}
 Enable Options {nodes [
   Enable Options() {is=title K2Node_FunctionEntry}
   n1 = Get CharacterPart() -> (CharacterPart:Custom Character Part Object Reference)
   n2 = GetCreatorWidget!() -> (Costmetic Creator:EUW Meta Sound Cosmetic Preset Creator Object Reference)
   n3 = "GetText (Text Box)"!(self:Text Box Object Reference=Get PathInput!(self:EUW Meta Sound Asset Path Object Reference=Get AssetPath!(self:EUW Meta Sound Cosmetic Preset Creator Object Reference=@n2) -> (AssetPath:EUW Meta Sound Asset Path Object Reference)) -> (PathInput:Editor Utility Editable Text Box Object Reference)) -> (ReturnValue:Text)
   n4 = "To String (Text)"(InText=@n3)
   n5 = Get AssetPath(self=GetCreatorWidget())
   n6 = TextIsEmpty(InText="GetText (Text Box)"(self=Get AssetNameInput!(self:EUW Meta Sound Asset Path Object Reference=@n5) -> (AssetNameInput:Editor Utility Editable Text Box Object Reference)))
   n7 = TextIsEmpty(InText=@n3)
   > n8 = Branch(Condition=AND Boolean!(
     A=NOT Boolean(A=@n6)
     B=NOT Boolean(A=@n7)
     C:bool=ValidatePath!(InPath:String=@n4) -> (bDidSucceed:bool, OutReason:Text)
     D:bool=IsValid(Object=@n1)
    ) -> (ReturnValue))
   n9 = Get SoundLibraryBox() -> (SoundLibraryBox:Vertical Box Object Reference)
   > SetIsEnabled(self=@n9, bInIsEnabled=true)
   SetIsEnabled(execute=@n8.else, self=@n9, bInIsEnabled=false)
  ]}
 EventGraph {nodes [
   "On Property Changed (CharacterPartSource)"!() -> (OutputDelegate, then, PropertyName:Name=None) {is=title K2Node_ComponentBoundEvent}
   > n1 = Get!(execute) -> (then, else, CharacterPart:Custom Character Part Object Reference)
   > Set PartModifierBlueprint(PartModifierBlueprint:Actor Soft Class Reference=Get PartModifierBlueprint!(self:Custom Character Part Object Reference=@n1) -> (PartModifierBlueprint:Actor Soft Class Reference)) -> (Output_Get:Actor Soft Class Reference)
   > Set Combobox() {is=title K2Node_CallFunction}
   > Enable Options() {is=title K2Node_CallFunction}
   > Sequence()
   > SetIsEnabled(self=[Get ModifierBlueprintBox() -> (ModifierBlueprintBox:Size Box Object Reference), Get SoundLibraryBox()], bInIsEnabled=true)
   > Check for Valid Character Part() {is=title K2Node_CallFunction}
   Set PartModifierBlueprint(execute=@n1.else)
   > Enable Options()
   > SetIsEnabled(self=[Get SoundLibraryBox(), Get ModifierBlueprintBox()], bInIsEnabled=false)
   > Clear(TargetArray=Get ReactiveComponents())
   > Set SoundLibrary(SoundLibrary:Sound Library Object Reference) -> (Output_Get:Sound Library Object Reference)
   "On Pressed (SoundLibraryButton)"() {is=title K2Node_ComponentBoundEvent}
   n2 = Get PathInput(self=Get AssetPath(self=GetCreatorWidget()))
   n3 = GetCreatorWidget()
   n4 = "GetText (Text Box)"(self=Get AssetNameInput(self=Get AssetPath(self=@n3)))
   n5 = Get ReactiveComboBox() -> (ReactiveComboBox:Editor Utility Combo Box String Object Reference)
   > n6 = Create Sound Library!(
    execute
    self
    PackagePath:String="To String (Text)"(InText=Get Text(self=@n2))
    AssetName:String=Append!(
     A=SL
     B=_
     C:String=GetSelectedOption!(self:"Combo Box (String) Object Reference"=@n5) -> (ReturnValue:String)
     D:String=_
     E:String="To String (Text)"(InText=@n4)
    ) -> (ReturnValue)
   ) -> (then, AsSound Library:Sound Library Object Reference) {is=title K2Node_CallFunction}
   > n7 = Set SoundLibrary(SoundLibrary=@n6)
   > Set LibrarySounds(LibrarySounds=Get Sounds(self=@n7))
   > SetEditorProperty(
    Object=@n7
    PropertyName=Parent
    PropertyValue:Sound Library Object Reference=Get LibReactiveParent() -> (LibReactiveParent:Sound Library Object Reference)
   )
   "On Property Changed (ReactiveComponentDetails)"!() -> (OutputDelegate, then, PropertyName:Name=None) {is=title K2Node_ComponentBoundEvent}
   > Set Component Stats Variables to Blueprint() {is=title K2Node_CallFunction}
   "On Property Changed (LibrarySoundsDetails)"!() -> (OutputDelegate, then, PropertyName:Name=None) {is=title K2Node_ComponentBoundEvent}
   > Add Widget Sound Libaries to Sound Library Asset() {is=title K2Node_CallFunction}
   "On Property Changed (ModifierBlueprintSource)"!() -> (OutputDelegate, then, PropertyName:Name=None) {is=title K2Node_ComponentBoundEvent}
   > Check for Valid Character Part()
   > n8 = Get()
   n9 = Get PartModifierBlueprint!() -> (PartModifierBlueprint)
   > Branch(Condition="Not Equal (SoftClassReference)"(
     A:"Object Soft Class Reference (by ref)"=@n9
     B:"Object Soft Class Reference (by ref)"=Get PartModifierBlueprint!(self:Custom Character Part Object Reference=@n8) -> (PartModifierBlueprint)
    ) -> (ReturnValue:bool=~) {is=title K2Node_PromotableOperator})
   > Set PartModifierBlueprint!(execute, PartModifierBlueprint=@n9, self:Custom Character Part Object Reference=@n8) -> (then, Output_Get)
   "On Property Changed (AnimCheckView)"!() -> (OutputDelegate, then, PropertyName:Name=None) {is=title K2Node_ComponentBoundEvent}
   > n10 = Branch(Condition=Get bIsAnimationImplementation())
   n11 = Get ModifierBlueprintBox()
   > SetVisibility(self=@n11, InVisibility=Collapsed)
   > SetVisibility(self=Get AnimationBox() -> (AnimationBox:Vertical Box Object Reference), InVisibility=Visible)
   > SetIsEnabled(self=Get SoundLibraryBox(), bInIsEnabled=true)
   SetVisibility(execute=@n10.else, self=@n11, InVisibility=Visible)
   > SetVisibility(self=Get AnimationBox(), InVisibility=Collapsed)
   > Set Animation(Animation:Anim Sequence Base Object Reference) -> (Output_Get:Anim Sequence Base Object Reference)
   > SetIsEnabled(self=Get SoundLibraryBox(), bInIsEnabled=false)
   "On Property Changed (CosmeticOptionView)"!() -> (OutputDelegate, then, PropertyName:Name=None) {is=title K2Node_ComponentBoundEvent}
   > Set AssetSuffix!(
    execute
    AssetSuffix:String=GetCosmeticOption()
    self:EUW Meta Sound Source Reference Object Reference=Get SoundSourceReference_Reactive() -> (SoundSourceReference_Reactive:EUW Meta Sound Source Reference Object Reference)
   ) -> (then, Output_Get:String)
   "On Pressed (SoundNotifyButton)"() {is=title K2Node_ComponentBoundEvent}
   > n12 = Get!(execute) -> (then, else, Animation:Anim Sequence Base Object Reference)
   n13 = Get Notify Track Name() -> (Notify Track Name:Name=None)
   > AddAnimationNotifyTrack!(
    execute
    AnimationSequenceBase:Anim Sequence Base Object Reference=@n12
    NotifyTrackName:Name=@n13
    TrackColor:Linear Color Structure=(R=1,G=1,B=1,A=1)
   ) -> (then)
   > n14 = AddAnimationNotifyEvent!(
    execute
    AnimationSequenceBase:Anim Sequence Base Object Reference=@n12
    NotifyTrackName:Name=@n13
    StartTime:float=Get NotifyStartTime() -> (NotifyStartTime:float)
    NotifyClass:Anim Notify Class Reference=/Script/SoundLibrary.AnimNotify_SoundLibrary
   ) -> (then, ReturnValue:Anim Notify Object Reference)
   > n15 = Cast To AnimNotify_SoundLibrary!(execute, Object=@n14) -> (then, CastFailed, AsAnim Notify Sound Library:Anim Notify Sound Library Object Reference) {is=title K2Node_DynamicCast}
   > SetEditorProperty(
    Object=@n15
    PropertyName=NotifyColor
    PropertyValue:Color Structure=Make Color!(B:Byte=197, G:Byte=254, R:Byte=76, A:Byte=0) -> (Color:Color Structure) {is=title K2Node_MakeStruct}
   )
   > n16 = Construct Object from Class!(
    execute
    Class=Get AnimSoundContextClass() -> (AnimSoundContextClass:Sound Library Context Class Reference)
    self=@n12
    Actor:Actor Weak object
    BaseEventName:Gameplay Tag Structure
    bForwardToActorOwner:bool=True
    bForwardToAttachedActors:bool=True
    bAutoResetContext:bool=True
   ) -> (then, ReturnValue:Sound Library Context Object Reference)
   > SetEditorProperty(Object=@n15, PropertyName=SoundContext, PropertyValue:Sound Library Context Object Reference=@n16)
   > n17 = Cast To BP_Cosmetics_Animation_Context!(execute, Object=@n16) -> (then, CastFailed, AsBP Cosmetics Animation Context:BP Cosmetics Animation Context Object Reference) {is=title K2Node_DynamicCast}
   > n18 = Get Cosmetic Options() -> (CosmeticOptions=NewEnumerator1)
   > Set AnimSettings!(
    execute
    AnimSettings:Animation Notify Settings Structure=Make Animation Notify Settings!(
     AnimRateThreshold:Float Interval Structure=(Min=0,Max=0)
     EventName:Gameplay Tag Structure=@n18.SoundLibraryTag
     bOnThisActor:bool=False
     bAttachToActor:bool=true
     bStopWhenAttachedDestroyed:bool=False
     Offset:Vector="0,0,0"
     AttachName:Name=root
     FadeTime:float=0.5
     NotifyTriggerChance:float=1
     VolumeMultiplier:float=1
    ) -> (SoundLibraryAnimContextSettings:Animation Notify Settings Structure) {is=title K2Node_MakeStruct}
    self:Animation Notify Context Object Reference=@n17
   ) -> (then, Output_Get:Animation Notify Settings Structure)
   > n19 = Get()
   > TransactObject(Object=@n19)
   n20 = Get SaveAssetRef() -> (SaveAssetRef:EUW IDSave Asset Object Reference)
   > Set SaveAsset!(execute, SaveAsset:bool=true, self:EUW IDSave Asset Object Reference=@n20) -> (then, Output_Get:bool)
   > SaveIdAsset!(execute, self:EUW IDSave Asset Object Reference=@n20, AssetToSave:Object Reference=@n19) -> (then)
   Show Message Dialog(
    execute=@n18.Failed
    Title=MakeLiteralText(Value="LOCTEXT(\"Failed to Add Event Tag\")")
    Message=MakeLiteralText(Value="LOCTEXT(\"Failed to add Event Tag on Anim Notify Event Name. Please reach out to audio producer or audio tech POCs.\")")
    MessageType=Ok
    DefaultValue=No
    MessageCategory=Warning
   )
   "On Property Changed (SoundLibraryView)"!() -> (OutputDelegate, then, PropertyName:Name=None) {is=title K2Node_ComponentBoundEvent}
   > n21 = Get!(execute) -> (then, else, SoundLibrary:Sound Library Object Reference)
   > Set LibrarySounds(LibrarySounds=Get Sounds(self=@n21))
   Clear(execute=@n21.else, TargetMap=Get LibrarySounds()) {is=title K2Node_CallFunction}
   n22 = Add Sound to Assets!() -> (OutputDelegate, then, Sound:Sound Base Soft Object Reference, SourceReference:EUW Meta Sound Source Reference Object Reference) {is=title K2Node_CustomEvent}
   > Add MetaSounds To Library!(execute, self, Sounds:"Sound Base Soft Object Reference (by ref)"=@n22.Sound) -> (then) {is=title K2Node_CallFunction}
   > SaveAllAssets()
   OnTextValidation() {is=title K2Node_CustomEvent}
   > Enable Options()
   Event PreConstruct()
   > SetObject(
    self=[
     Get AnimCheckView() -> (AnimCheckView:Single Property View Object Reference)
     Get AnimationSource() -> (AnimationSource:Single Property View Object Reference)
     Get SoundLibraryView() -> (SoundLibraryView:Single Property View Object Reference)
     Get CosmeticOptionView() -> (CosmeticOptionView:Single Property View Object Reference)
     Get NotifyTimeSource() -> (NotifyTimeSource:Single Property View Object Reference)
     Get LibrarySoundsDetails() -> (LibrarySoundsDetails:Details View Object Reference)
     Get ReactiveComponentDetails()
     Get ModifierBlueprintSource() -> (ModifierBlueprintSource:Single Property View Object Reference)
     Get CharacterPartSource() -> (CharacterPartSource:Single Property View Object Reference)
    ]
    NewObject=Self-Reference()
   )
   Event Construct()
   > Widget Selector!(
    execute
    self
    ShouldAdd:bool=true
    Reference:"EUW Meta Sound Source Reference Object Reference (by ref)"=Get SoundSourceReference_Reactive()
   ) -> (then)
   > Bind Event to On Inputs Text Updated(self:EUW Meta Sound Asset Path Object Reference=Get AssetPath(self=GetCreatorWidget()), Delegate=Create Event()) {is=title K2Node_AddDelegate}
   > Enable Options()
   > Assign Sound for Asset!(execute:Exec, self:EUW Meta Sound Source Reference Object Reference=Get SoundSourceReference_Reactive(), Delegate:"Delegate (by ref)"=@n22) -> (then:Exec) {is=title K2Node_AssignDelegate}
   > SetIsEnabled(self=[Get SoundLibraryBox(), Get ModifierBlueprintBox()], bInIsEnabled=false)
   String To Name(InString=GetCosmeticOption())
   SetIsEnabled(self=Get LibrarySoundsDetails(), bInIsEnabled=true)
   SetIsEnabled(self=Get LibrarySoundsDetails(), bInIsEnabled=false)
   "On Property Changed (AnimationSource)"!() -> (OutputDelegate, then, PropertyName:Name=None) {is=title K2Node_ComponentBoundEvent}
   Get ReactiveComponentDetails()
   Get CosmeticOptionsTable() -> (CosmeticOptionsTable:Data Table Object Reference)
  ]}
 FindStatWatchComponent {nodes [
   n1 = Inputs!() -> (exec:Exec, Handle:"Subobject Data Handle Structure (by ref)")
   n2 = Get SubobjectDataSubsystem()
   > n3 = Is Valid(InputObject=@n2)
   > n4 = Cast To B_Reactive_StatWatcher_Component(Object=GetObject(Data=FindSubobjectDataFromHandle(self=@n2, Handle=@n1), bEvenIfPendingKill=true))
   > Outputs!(
    Success:Exec
    Failed:Exec=[@n4.CastFailed, @n3.Is Not Valid]
    AsB Reactive Stat Watcher Component:B Reactive Stat Watcher Component Object Reference=@n4
   )
   Show Message Dialog(
    Title=MakeLiteralText(Value="LOCTEXT(\"Warning: Stat Watcher Component Missing\")")
    Message=MakeLiteralText(Value="LOCTEXT(\"Please verify this asset has a reactive stat watch component\")")
    MessageType=Ok
    DefaultValue=No
    MessageCategory=Warning
   )
  ]}
 Get Component Stats from Blueprint {nodes [
   Get Component Stats from Blueprint() {is=title K2Node_FunctionEntry}
   > n1 = Get!(execute) -> (then, else, As B Athena Part Modifier Generic:B Athena Part Modifier Generic Object Reference)
   n2 = Get SubobjectDataSubsystem()
   > Is Valid(InputObject=@n2)
   > n3 = Gather Subobject Data For Instance(self=@n2, Context=@n1)
   > n4 = For Each Loop(Array=@n3)
   n5 = GetObject(Data=FindSubobjectDataFromHandle(self=@n2, Handle=@n4), bEvenIfPendingKill=false)
   > Cast To B_Reactive_StatWatcher_Component Class(Class=Get Class(Object=@n5))
   > n6 = Cast To B_Reactive_StatWatcher_Component(Object=@n5)
   > Add Unique(
    TargetArray:Array of Struct Meta Sound Reactive Component Data Structures=Get ReactiveComponents()
    NewItem:"Struct Meta Sound Reactive Component Data Structure (by ref)"=Make Struct Meta Sound Reactive Component Data(
     ReactiveComponent_35_#1=Get Class(Object=@n6) -> (ReturnValue:B Reactive Stat Watcher Component Class Reference)
     WatcherStatTag_12_#2=Get WatcherStatTag(self=@n6)
     DefaultValue_14_#3=Get Default Value!(self:B Reactive Stat Watcher Component Object Reference=@n6) -> (Default Value:int=0)
     UpdateAudioParameter_30_#4=Get UpdateAudioParameter!(self:B Reactive Stat Watcher Component Object Reference=@n6) -> (UpdateAudioParameter:bool)
     BroadcastInitialValue_32_#5=Get Broadcast Initial Value!(self:B Reactive Stat Watcher Component Object Reference=@n6) -> (Broadcast Initial Value:bool)
     InitialValueBroadcastDelay_23_#6=Get Initial Value Broadcast Delay!(self:B Reactive Stat Watcher Component Object Reference=@n6) -> (Initial Value Broadcast Delay:float)
     ConsumingWaitingNotifications_27_#7=Get Consume Waiting Notifications!(self:B Reactive Stat Watcher Component Object Reference=@n6) -> (Consume Waiting Notifications:bool)
     Object_43_#8=@n6
    )
   )
  ]}
 GetCosmeticOption {nodes [
   GetCosmeticOption() {is=title K2Node_FunctionEntry}
   n1 = Get CosmeticOption() -> (CosmeticOption:Enum_MetaSound_Creator_CosmeticOptions Enum=NewEnumerator1)
   > Return Node!(
    execute
    ReturnValue:String=Replace(SourceString=Enum to String!(Enumerator:Enum_MetaSound_Creator_CosmeticOptions Enum=@n1) -> (ReturnValue:String), From=" ")
   )
  ]}
 GetCosmeticOptions {nodes [
   n1 = String To Name(InString=GetCosmeticOption() {is=title K2Node_CallFunction})
   n2 = Get CosmeticOptionsTable()
   n3 = Get Data Table Row!(execute:Exec, DataTable:Data Table Object Reference=@n2, RowName:Name=@n1) -> (then:Exec, RowNotFound:Exec, ReturnValue:Struct Meta Sound Cosmetic Options Structure)
   n4 = GetCosmeticOptions!(execute=@n5.execute, DataTable:Data Table Object Reference=@n2, RowName:Name=@n1, __WorldContext:Object Reference) -> (
    then
    RowFound:bool
    CosmeticOption:Enum_MetaSound_Creator_CosmeticOptions Enum=NewEnumerator1
    Stat Manager Tag:Fort Stat Manager Tag Structure
    SoundLibraryTag:Sound Library Tag Structure
   )
   Outputs!(
    Success=[@n3.then, @n6.then]
    Failed=[@n3.RowNotFound, @n6.else]
    CosmeticOptions:Enum_MetaSound_Creator_CosmeticOptions Enum=@n4.CosmeticOption
    Stat Manager Tag:Fort Stat Manager Tag Structure=@n4.Stat Manager Tag
    SoundLibraryTag:Sound Library Tag Structure=@n4.SoundLibraryTag
   )
   n5 = Inputs!() -> (execute:Exec)
   Break Struct Meta Sound Cosmetic Options!(Struct_MetaSound_CosmeticOptions:"Struct Meta Sound Cosmetic Options Structure (by ref)"=@n3) -> (
    CosmeticOption_2_#9:Enum_MetaSound_Creator_CosmeticOptions Enum
    StatManagerTag_7_#10:Fort Stat Manager Tag Structure
    SoundLibraryTag_8_#11:Sound Library Tag Structure
   ) {is=title K2Node_BreakStruct}
   n6 = Branch(execute=@n4.then, Condition=@n4)
  ]}
 GetCreatorWidget {nodes [
   GetCreatorWidget() {is=title K2Node_FunctionEntry}
   > n1 = GetCosmeticCreator!(execute, Target:Widget Object Reference=Self-Reference(), __WorldContext:Object Reference) -> (then, AsEUW Meta Sound Cosmetic Preset Creator:EUW Meta Sound Cosmetic Preset Creator Object Reference)
   > Return Node!(execute, Costmetic Creator:EUW Meta Sound Cosmetic Preset Creator Object Reference=@n1)
  ]}
 On Combobox Change {nodes [
   n1 = On Combobox Change!() -> (then, Item:String) {is=title K2Node_FunctionEntry}
   n2 = Get SoundSourceReference_Reactive()
   n3 = ParseIntoArray(
    SourceString=GetDisplayName(Object=Get SourceReference!(self:EUW Meta Sound Source Reference Object Reference=@n2) -> (SourceReference:Meta Sound Source Object Reference))
    Delimiter=_
   )
   > Set AssetPrefix!(
    execute
    AssetPrefix:String=Append!(A="Get (a copy)"(Array=@n3), B=_, C=@n1) -> (ReturnValue)
    self:EUW Meta Sound Source Reference Object Reference=@n2
   ) -> (then, Output_Get:String)
   > Return Node!(execute, ReturnValue:Widget Object Reference)
   "Get (a copy)"(Array=@n3, Dimension 1=1)
  ]}
 PartModifierParentCheck {nodes [
   Inputs()
   n1 = Get PartModifierBlueprint!() -> (PartModifierBlueprint)
   > n2 = Branch(Condition="Equal Exactly (String)"(
     A="To String (SoftClassReference)"!(SoftClassReference:"Object Soft Class Reference (by ref)"=@n1) -> (ReturnValue:String)
     B=/Game/Athena/Cosmetics/Blueprints/B_Athena_PartModifier_Generic.B_Athena_PartModifier_Generic_C
    ))
   > n3 = Show Message Dialog(
    Title=MakeLiteralText(Value="LOCTEXT(\"Parent Blueprint Is Selected\")")
    Message=MakeLiteralText(Value="LOCTEXT(\"Part Modifier Blueprint can not be \\\"B_Athena_PartModifier_Generic\\\". \\r\\n\\r\\nPlease add the proper blueprint or reach out to audio production or design to get acces to the proper assets. \\r\\n\\r\\nParent BP should not be used when added audio. Implementation will fail.\")")
    MessageType=Ok
    DefaultValue=No
    MessageCategory=Warning
   )
   Outputs!(Success=@n2.else, Fail:Exec=@n3.then)
  ]}
 SaveAllAssets {nodes [
   SaveAllAssets() {is=title K2Node_FunctionEntry}
   > n1 = Sequence!(execute) -> (then_0, then_1, then_2:Exec)
   > n2 = Get!(execute) -> (then, else, SoundLibrary:Sound Library Object Reference)
   > TransactObject(Object=@n2)
   n3 = Get SaveAssetRef()
   > Set SaveAsset(SaveAsset=true, self=@n3)
   > SaveIdAsset(self=@n3, AssetToSave=@n2)
   n4 = Get!(execute=@n1.then_1) -> (then, else, CharacterPart:Custom Character Part Object Reference)
   > TransactObject(Object=@n4)
   n5 = Get SaveAssetRef()
   > Set SaveAsset(SaveAsset=true, self=@n5)
   > SaveIdAsset(self=@n5, AssetToSave=@n4)
   n6 = Get!(execute=@n1.then_2) -> (then, else, As B Athena Part Modifier Generic:B Athena Part Modifier Generic Object Reference)
   > TransactObject(Object=@n6)
   n7 = Get SaveAssetRef()
   > Set SaveAsset(SaveAsset=true, self=@n7)
   > SaveIdAsset(self=@n7, AssetToSave=@n6)
  ]}
 Set Combobox {nodes [
   Set Combobox() {is=title K2Node_FunctionEntry}
   n1 = Get CharacterPart()
   > n2 = Branch(Condition="Equal (Enum)"(
     A:EFortCustomPartType Enum=GetCharacterPartType!(self:Custom Character Part Object Reference=@n1) -> (ReturnValue:EFortCustomPartType Enum=Head)
     B:EFortCustomPartType Enum=Backpack
    ))
   > SetSelectedOption!(execute, self:"Combo Box (String) Object Reference"=Get ReactiveComboBox(), Option:String=BackBling) -> (then)
   SetSelectedOption(execute=@n2.else, self=Get ReactiveComboBox(), Option=Skin)
  ]}
 Set Component Stats Variables to Blueprint {nodes [
   Set Component Stats Variables to Blueprint() {is=title K2Node_FunctionEntry}
   > n1 = For Each Loop(Array:Array of Struct Meta Sound Reactive Component Data Structures=Get ReactiveComponents()) -> (Array Element:Struct Meta Sound Reactive Component Data Structure)
   > n2 = Get()
   > n3 = Gather Subobject Data For Instance(self=Get SubobjectDataSubsystem(), Context=@n2)
   > n4 = For Each Loop(Array:Array of Subobject Data Handle Structures=@n3) -> (Array Element:Subobject Data Handle Structure)
   > n5 = Find Stat Watch Component(Handle=@n4)
   n6 = Break Struct Meta Sound Reactive Component Data(Struct_MetaSound_Reactive_Component_Data=@n1) -> (
    UpdateAudioParameter_30_#4=~
    BroadcastInitialValue_32_#5=~
    InitialValueBroadcastDelay_23_#6=~
    ConsumingWaitingNotifications_27_#7=~
   )
   > Branch(Condition="Equal (Object)"(A=@n5, B=@n6.Object_43_#8))
   > Set WatcherStatTag(WatcherStatTag=@n6.WatcherStatTag_12_#2, self=@n5)
   > Set Default Value(Default Value=@n6.DefaultValue_14_#3, self=@n5) -> (Output_Get=0)
   > Set UpdateAudioParameter(UpdateAudioParameter=@n6.UpdateAudioParameter_30_#4, self=@n5)
   > Set Broadcast Initial Value(Broadcast Initial Value=@n6.BroadcastInitialValue_32_#5, self=@n5)
   > Set Initial Value Broadcast Delay(Initial Value Broadcast Delay=@n6.InitialValueBroadcastDelay_23_#6, self=@n5)
   > Set Consume Waiting Notifications(Consume Waiting Notifications=@n6.ConsumingWaitingNotifications_27_#7, self=@n5)
  ]}
 SetPartsModifierBlueprint {nodes [
   SetPartsModifierBlueprint() {is=title K2Node_FunctionEntry}
   n1 = Get PartModifierBlueprint()
   > n2 = Branch(Condition=IsValidSoftClassReference!(SoftClassReference:"Object Soft Class Reference (by ref)"=@n1) -> (ReturnValue:bool))
   > n3 = LoadClassAsset_Blocking(AssetClass=@n1)
   > n4 = GetClassDefaultObject!(execute, Class:Object Class Reference=@n3) -> (then, ReturnValue:Object Reference)
   > n5 = Cast To B_Athena_PartModifier_Generic!(execute, Object=@n4) -> (then, CastFailed, AsB Athena Part Modifier Generic:B Athena Part Modifier Generic Object Reference) {is=title K2Node_DynamicCast}
   > Set As B Athena Part Modifier Generic(As B Athena Part Modifier Generic:B Athena Part Modifier Generic Object Reference=@n5) -> (Output_Get:B Athena Part Modifier Generic Object Reference)
   Set As B Athena Part Modifier Generic(execute=@n2.else)
  ]}
 Widget Selector {nodes [
   n1 = Widget Selector!() -> (then, ShouldAdd:bool=~, Reference:"EUW Meta Sound Source Reference Object Reference (by ref)") {is=title K2Node_FunctionEntry}
   > n2 = Branch(Condition=@n1)
   n3 = Get SoundSourceReferenceWidgets() -> (SoundSourceReferenceWidgets:Array of EUW Meta Sound Source Reference Object References)
   > n4 = Add Unique(
    TargetArray:Array of EUW Meta Sound Source Reference Object References=@n3
    NewItem:"EUW Meta Sound Source Reference Object Reference (by ref)"=@n1.Reference
   )
   Set Current Source Reference Widgets!(
    execute=[@n5.then, @n4.then]
    self:EUW Meta Sound Cosmetic Preset Creator Object Reference=GetCreatorWidget() {is=title K2Node_CallFunction}
    CurrentActiveWidget:Widget Object Reference=Self-Reference()
    SourceReferences:Array of EUW Meta Sound Source Reference Object References=Get SoundSourceReferenceWidgets()
    CosmeticType:String=Reactive
   ) -> (then) {is=title K2Node_CallFunction}
   n5 = Remove Item(
    execute=@n2.else
    TargetArray:Array of EUW Meta Sound Source Reference Object References=@n3
    Item:"EUW Meta Sound Source Reference Object Reference (by ref)"=@n1.Reference
   )
  ]}
]
variables [
 Animation {type=AnimSequenceBase, default=null}
 AnimSoundContextClass {type=Class, default=/Game/Sounds/SoundLibrary/Contexts/Cosmetics/BP_Cosmetics_Animation_Context.BP_Cosmetics_Animation_Context_C}
 As B Athena Part Modifier Generic {type=B_Athena_PartModifier_Generic_C, default=null}
 bIsAnimationImplementation {type=boolean}
 CharacterPart {type=CustomCharacterPart, default=null}
 CosmeticOption {type=Enum_MetaSound_Creator_CosmeticOptions, default=Elmination}
 CosmeticOptionsTable {
  type=DataTable
  default=&/DataTable_MetaSound_CosmeticOptions
 }
 Current Sound Library {type=SoundLibrary, default=null}
 LibrarySounds {default {}}
 LibReactiveParent {type=SoundLibrary, default=/Game/Sounds/SoundLibrary/Parents/Cosmetic/Cosmetic_Reactive_Parent_Lib}
 Notify Track Name {type=string, default=Audio}
 NotifyStartTime {type=number}
 PartModifierBlueprint {type=Class, default=null}
 ReactiveComponents {type="array<Struct_MetaSound_Reactive_Component_Data>", default []}
 SoundLibrary {type=SoundLibrary, default=null}
 SoundSourceReference_Reactive {
  type=EUW_MetaSoundSourceReference_C
  default=&/EUW_MetaSound_Cosmetic_PresetCreator_Reactive.Default__EUW_MetaSound_Cosmetic_PresetCreator_Reactive_C:EUW_MetaSoundSourceReference_C_0
 }
 SoundSourceReferenceWidgets {type="array<EUW_MetaSoundSourceReference_C>", default []}
 TagAlreadyApplied {type=boolean}
]
widgets {
 info {rootWidgetClass=&CanvasPanel, widgetCount=30}
 widgets [
  CanvasPanel_0 {widgetClassPath=&CanvasPanel, widget=CanvasPanel_0}
  VerticalBox_78 {
   widgetClassPath=&VerticalBox
   parent=CanvasPanel_0
   widget=VerticalBox_78
   slot {
    class=CanvasPanelSlot
    properties {bAutoSize=true, layoutData {alignment=<0 0>, anchors {maximum=<0 0>, minimum=<0 0>}, offsets=<ltrb 0 0 500 400>}}
   }
  }
  HorizontalBox_531 {widgetClassPath=&HorizontalBox, parent=VerticalBox_78, widget=HorizontalBox_531, slot {class=VerticalBoxSlot}}
  SizeBox_107 {
   widgetClassPath=&SizeBox
   parent=HorizontalBox_531
   widget=SizeBox_107
   properties {
    bOverride_MaxDesiredWidth=true
    bOverride_MinDesiredWidth=true
    bOverride_WidthOverride=true
    maxDesiredWidth=200
    minDesiredWidth=150
    widthOverride=150
   }
   slot {class=HorizontalBoxSlot, properties {horizontalAlignment=HAlign_Left}}
  }
  ReactiveComboBox {
   widgetClassPath=/Script/Blutility.EditorUtilityComboBoxString
   bIsVariable=true
   parent=SizeBox_107
   widget=ReactiveComboBox
   properties {
    selectedOption=BackBling
    toolTipText=Select preset that will be used for backbling.
    contentPadding=<ltrb 2 2 2 2>
    defaultOptions [BackBling,Skin]
    font {
     bForceMonospaced=false
     bMaterialIsStencil=false
     fontMaterial=null
     fontObject=/LauncherSocial/BrutalType
     letterSpacing=0
     monospacedWidth=1
     size=9
     skewAmount=0
     typefaceFontName=Regular
     outlineSettings {
      bApplyOutlineToDropShadows=false
      bMiteredCorners=false
      bSeparateFillAlpha=false
      outlineMaterial=null
      outlineSize=0
      outlineColor=<rgba 0 0 0 1>
     }
    }
    foregroundColor {colorUseRule=UseColor_Specified, specifiedColor=<rgba 0.5271 0.5271 0.5271 1>}
   }
   slot {class=SizeBoxSlot, properties {verticalAlignment=VAlign_Center, padding=<ltrb 10 10 10 10>}}
  }
  CosmeticOptionBox {
   widgetClassPath=&SizeBox
   bIsVariable=true
   parent=HorizontalBox_531
   widget=CosmeticOptionBox
   properties {bOverride_WidthOverride=true, heightOverride=78.2408, maxDesiredWidth=460, minDesiredWidth=450, widthOverride=300}
   slot {class=HorizontalBoxSlot, properties {horizontalAlignment=HAlign_Left, verticalAlignment=VAlign_Center}}
  }
  CosmeticOptionView {
   widgetClassPath=&SinglePropertyView
   parent=CosmeticOptionBox
   widget=CosmeticOptionView
   properties {propertyName=CosmeticOption}
   slot {class=SizeBoxSlot, properties {padding=<ltrb 10 10 10 10>}}
  }
  AnimCheckView {
   widgetClassPath=&SinglePropertyView
   parent=VerticalBox_78
   widget=AnimCheckView
   properties {nameOverride=Use Animation Implmenation:, propertyName=bIsAnimationImplementation}
   slot {class=VerticalBoxSlot, properties {padding=<ltrb 10 10 10 10>}}
  }
  SizeBox {
   widgetClassPath=&SizeBox
   parent=VerticalBox_78
   widget=SizeBox
   properties {bOverride_WidthOverride=true, heightOverride=78.2408, maxDesiredWidth=460, minDesiredWidth=450, widthOverride=450}
   slot {class=VerticalBoxSlot, properties {horizontalAlignment=HAlign_Left, verticalAlignment=VAlign_Center}}
  }
  CharacterPartSource {
   widgetClassPath=&SinglePropertyView
   parent=SizeBox
   widget=CharacterPartSource
   properties {propertyName=CharacterPart}
   slot {class=SizeBoxSlot, properties {padding=<ltrb 10 10 10 10>}}
  }
  ModifierBlueprintBox {
   widgetClassPath=&SizeBox
   bIsVariable=true
   parent=VerticalBox_78
   widget=ModifierBlueprintBox
   properties {bOverride_WidthOverride=true, heightOverride=78.2408, maxDesiredWidth=460, minDesiredWidth=450, widthOverride=450}
   slot {class=VerticalBoxSlot, properties {horizontalAlignment=HAlign_Left, verticalAlignment=VAlign_Center}}
  }
  ModifierBlueprintSource {
   widgetClassPath=&SinglePropertyView
   parent=ModifierBlueprintBox
   widget=ModifierBlueprintSource
   properties {propertyName=PartModifierBlueprint}
   slot {class=SizeBoxSlot, properties {padding=<ltrb 10 10 10 10>}}
  }
  AnimationBox {
   widgetClassPath=&VerticalBox
   bIsVariable=true
   parent=VerticalBox_78
   widget=AnimationBox
   properties {visibility=Collapsed}
   slot {class=VerticalBoxSlot, properties {horizontalAlignment=HAlign_Left, verticalAlignment=VAlign_Center}}
  }
  SizeBox_1 {
   widgetClassPath=&SizeBox
   parent=AnimationBox
   widget=SizeBox_1
   properties {bOverride_WidthOverride=true, heightOverride=78.2408, maxDesiredWidth=460, minDesiredWidth=450, visibility=Visible, widthOverride=450}
   slot {class=VerticalBoxSlot}
  }
  AnimationSource {
   widgetClassPath=&SinglePropertyView
   parent=SizeBox_1
   widget=AnimationSource
   properties {propertyName=Animation}
   slot {class=SizeBoxSlot, properties {padding=<ltrb 10 10 10 10>}}
  }
  SizeBox_2 {
   widgetClassPath=&SizeBox
   parent=AnimationBox
   widget=SizeBox_2
   properties {heightOverride=78.2408, maxDesiredWidth=460, minDesiredWidth=450, visibility=Visible}
   slot {class=VerticalBoxSlot}
  }
  NotifyTimeSource {
   widgetClassPath=&SinglePropertyView
   parent=SizeBox_2
   widget=NotifyTimeSource
   properties {propertyName=NotifyStartTime}
   slot {class=SizeBoxSlot, properties {horizontalAlignment=HAlign_Left, verticalAlignment=VAlign_Center, padding=<ltrb 10 10 10 10>}}
  }
  SoundNotifyButton {
   widgetClassPath=&EditorUtilityButton
   parent=AnimationBox
   widget=SoundNotifyButton
   properties {toolTipText=Reset the source reference back to default asset.}
   slot {class=VerticalBoxSlot, properties {horizontalAlignment=HAlign_Left, verticalAlignment=VAlign_Bottom, padding=<ltrb 5 10 5 5>}}
  }
  UEFN_TextBlock_C {
   widgetClassPath=&UEFN_TextBlock
   parent=SoundNotifyButton
   widget=UEFN_TextBlock_C
   properties {
    justification=Center
    text=Add Sound Notify To Animation
    font {
     bForceMonospaced=false
     bMaterialIsStencil=false
     fontMaterial=null
     fontObject=&BrutalType
     letterSpacing=0
     monospacedWidth=1
     size=9
     skewAmount=0
     typefaceFontName=Regular
     outlineSettings {
      bApplyOutlineToDropShadows=false
      bMiteredCorners=false
      bSeparateFillAlpha=false
      outlineMaterial=null
      outlineSize=0
      outlineColor=<rgba 0 0 0 1>
     }
    }
    shadowColorAndOpacity=<rgba 0 0 0 1>
   }
   slot {class=ButtonSlot}
  }
  SoundLibraryBox {
   widgetClassPath=&VerticalBox
   bIsVariable=true
   parent=VerticalBox_78
   widget=SoundLibraryBox
   slot {class=VerticalBoxSlot, properties {horizontalAlignment=HAlign_Left, verticalAlignment=VAlign_Center}}
  }
  SoundLibraryButton {
   widgetClassPath=&EditorUtilityButton
   parent=SoundLibraryBox
   widget=SoundLibraryButton
   properties {toolTipText=Reset the source reference back to default asset.}
   slot {class=VerticalBoxSlot, properties {horizontalAlignment=HAlign_Left, verticalAlignment=VAlign_Bottom, padding=<ltrb 5 10 5 5>}}
  }
  UEFN_TextBlock_C_4 {
   widgetClassPath=&UEFN_TextBlock
   parent=SoundLibraryButton
   widget=UEFN_TextBlock_C_4
   properties {
    justification=Center
    text=Create Sound Library
    font {
     bForceMonospaced=false
     bMaterialIsStencil=false
     fontMaterial=null
     fontObject=&BrutalType
     letterSpacing=0
     monospacedWidth=1
     size=9
     skewAmount=0
     typefaceFontName=Regular
     outlineSettings {
      bApplyOutlineToDropShadows=false
      bMiteredCorners=false
      bSeparateFillAlpha=false
      outlineMaterial=null
      outlineSize=0
      outlineColor=<rgba 0 0 0 1>
     }
    }
    shadowColorAndOpacity=<rgba 0 0 0 1>
   }
   slot {class=ButtonSlot}
  }
  SoundLibraryViewBox {
   widgetClassPath=&SizeBox
   bIsVariable=true
   parent=SoundLibraryBox
   widget=SoundLibraryViewBox
   properties {bOverride_WidthOverride=true, heightOverride=78.2408, maxDesiredWidth=460, minDesiredWidth=450, widthOverride=450}
   slot {class=VerticalBoxSlot, properties {horizontalAlignment=HAlign_Left, verticalAlignment=VAlign_Center}}
  }
  SoundLibraryView {
   widgetClassPath=&SinglePropertyView
   parent=SoundLibraryViewBox
   widget=SoundLibraryView
   properties {propertyName=SoundLibrary}
   slot {class=SizeBoxSlot, properties {padding=<ltrb 10 10 10 10>}}
  }
  SaveAssetRef {
   widgetClassPath=/Game/Sounds/Fort_Audio_Blueprints/Utilities/EditorUtilityWidgets/Metasounds/Tools/EUW_IDSaveAsset
   bIsVariable=true
   parent=SoundLibraryBox
   widget=SaveAssetRef
   properties {visibility=Collapsed}
   slot {class=VerticalBoxSlot, properties {padding=<ltrb 0 5 5 5>}}
  }
  EditorUtilityScrollBox_0 {
   widgetClassPath=&EditorUtilityScrollBox
   parent=VerticalBox_78
   widget=EditorUtilityScrollBox_0
   properties {scrollBarVisibility=Collapsed}
   slot {class=VerticalBoxSlot}
  }
  SizeBox_3 {
   widgetClassPath=&SizeBox
   parent=EditorUtilityScrollBox_0
   widget=SizeBox_3
   properties {bOverride_WidthOverride=true, heightOverride=78.2408, maxDesiredWidth=460, minDesiredWidth=450, widthOverride=450}
   slot {class=ScrollBoxSlot}
  }
  LibrarySoundsDetails {
   widgetClassPath=&DetailsView
   parent=SizeBox_3
   widget=LibrarySoundsDetails
   properties {bAutoLoadAsset=false, bIsEnabled=false, visibility=Collapsed, propertiesToShow [LibrarySounds]}
   slot {class=SizeBoxSlot}
  }
  ComponentsSizeBox {
   widgetClassPath=&SizeBox
   parent=EditorUtilityScrollBox_0
   widget=ComponentsSizeBox
   properties {bOverride_WidthOverride=true, heightOverride=78.2408, maxDesiredWidth=460, minDesiredWidth=450, widthOverride=450}
   slot {class=ScrollBoxSlot}
  }
  ReactiveComponentDetails {
   widgetClassPath=&DetailsView
   parent=ComponentsSizeBox
   widget=ReactiveComponentDetails
   properties {bIsEnabled=false, visibility=Collapsed, categoriesToShow [Reactive Components]}
   slot {class=SizeBoxSlot}
  }
 ]
}
