EditorUtilityWidgetBlueprint /Game/Sounds/Fort_Audio_Blueprints/Utilities/EditorUtilityWidgets/Metasounds/Cosmetic/EUW_MetaSound_Cosmetic_PresetCreator
graphs [
 AddSelectedAssets {nodes [
   n1 = AddSelectedAssets!() -> (then, Assets:Array of Object References) {is=title K2Node_FunctionEntry}
   > UpdateSoundLists!(
    execute
    self:EUW Meta Sound Cosmetic Action Sounds List Object Reference=Get SoundNameListView() -> (SoundNameListView:EUW Meta Sound Cosmetic Action Sounds List Object Reference)
    Object:Array of Object References=@n1
   ) -> (then)
  ]}
 AssignMetaSoundInput {nodes [
   n1 = Inputs!() -> (execute, TargetArray:Array of Object References, InputName:String, ParameterID:String)
   > n2 = Branch(Condition=Is Not Empty!(TargetArray:Array of Object References=@n1) -> (ReturnValue:bool))
   n3 = Append(A=@n1.InputName, B=@n1.ParameterID)
   n4 = String To Name(InString=@n3)
   n5 = "Get (a copy)"(Array:Array of Object References=@n1) -> (Output:Object Reference)
   n6 = Get IsArray!(self:EUW Meta Sound Source Reference Object Reference=Get CurrentSoundSourceReference() -> (CurrentSoundSourceReference:EUW Meta Sound Source Reference Object Reference)) -> (IsArray:bool)
   > n7 = Create MetaSound Literal From AudioParameter!(
    execute
    self:Meta Sound Builder Subsystem Object Reference=Get MetaSoundBuilderSubsystem() -> (ReturnValue:Meta Sound Builder Subsystem Object Reference) {is=title K2Node_GetEngineSubsystem}
    Param:"Audio Parameter Structure (by ref)"=Select(
     Option 0:"Audio Parameter Structure (by ref)"=ObjectToAudioParameter!(Name:Name=@n4, Object:Object Reference=@n5) -> (ReturnValue:Audio Parameter Structure)
     Option 1:"Audio Parameter Structure (by ref)"=ObjectArrayToAudioParameter!(Name:Name=@n4, Objects:Array of Object References=@n1) -> (ReturnValue:Audio Parameter Structure)
     Index=@n6
    ) -> (ReturnValue:"Audio Parameter Structure (by ref)")
   ) -> (then, ReturnValue:Meta Sound Literal Structure)
   n8 = SelectString(A=@n3, B=@n1.ParameterID, bPickA=@n6)
   > n9 = SetGraphInputDefault!(
    execute
    self:Meta Sound Builder Base Object Reference=Get MetasoundBuilder() -> (MetasoundBuilder:Meta Sound Source Builder Object Reference)
    InputName:Name=String To Name(InString=@n8)
    Literal:"Meta Sound Literal Structure (by ref)"=@n7
    PageName:Name
   ) -> (Succeeded:Exec, Failed:Exec)
   n10 = LogString(execute=@n9.Failed, InString=Append!(A=***, B="string in selection is: ", C:String=@n8) -> (ReturnValue))
   Outputs!(Succeeded:Exec=@n9.Succeeded, Failed:Exec=@n10.then, NA:Exec=@n2.else)
  ]}
 AssignMetaoundInput {nodes [
   n1 = AssignMetaoundInput!() -> (then, TargetArray:Array of Object References, InputName:String, ParameterID:String) {is=title K2Node_FunctionEntry}
   > n2 = Branch(Condition=Is Not Empty(TargetArray=@n1))
   > Set LocalArray(LocalArray:Array of Object References=@n1) -> (Output_Get:Array of Object References)
   n3 = Append!(A=@n1.InputName, B=@n1.ParameterID) -> (ReturnValue)
   n4 = String To Name(InString=@n3)
   n5 = "Get (a copy)"(Array=@n1)
   n6 = Get IsArray(self=Get CurrentSoundSourceReference())
   > n7 = Create MetaSound Literal From AudioParameter(
    self=Get MetaSoundBuilderSubsystem()
    Param=Select(Option 0=ObjectToAudioParameter(Name=@n4, Object=@n5), Option 1=ObjectArrayToAudioParameter(Name=@n4, Objects=@n1), Index=@n6)
   )
   n8 = SelectString(A=@n3, B=@n1.ParameterID, bPickA=@n6)
   n9 = String To Name(InString=@n8)
   > n10 = SetGraphInputDefault(self=Get MetasoundBuilder(), InputName=@n9, Literal=@n7)
   LogString(execute=[@n15.Failed, @n10.Failed], InString=Append!(A=***, B="string in selection is: ", C:String=@n8) -> (ReturnValue))
   > n11 = Set Succeeded(Succeeded:bool=false) -> (Output_Get:bool)
   Return Node!(execute=[@n16.then, @n11.then], Success:bool=Get Succeeded() -> (Succeeded:bool), EarlyOut:bool=false)
   n12 = Get LocalArray() -> (LocalArray:Array of Object References)
   n13 = Branch(execute=@n10.Succeeded, Condition="integer >= integer"(A=Length(TargetArray:Array of Object References=@n12), B=3))
   > Resize!(execute, TargetArray:Array of Object References=@n12, Size:int=3) -> (then) {is=title K2Node_CallArrayFunction}
   > n14 = Create MetaSound Literal From AudioParameter(self=Get MetaSoundBuilderSubsystem(), Param=ObjectArrayToAudioParameter(Name=@n4, Objects=@n12))
   > n15 = SetGraphInputDefault(self=Get MetasoundBuilder(), InputName=@n9, Literal=@n14, PageName=Mobile)
   n16 = Set Succeeded(execute=[@n15.Succeeded, @n13.else], Succeeded=true)
   Return Node(execute=@n2.else, Success=false, EarlyOut=true)
  ]}
 AssignWaveAssets {nodes [
   n1 = AssignWaveAssets!() -> (then, InputName:String, Target:EUW Meta Sound Cosmetic Action Name Widget Object Reference) {is=title K2Node_FunctionEntry}
   > Clear(TargetArray:Array of Object References=Get LoopingWaves() -> (LoopingWaves:Array of Object References))
   > Clear(TargetArray=Get OneShotWaves() -> (OneShotWaves:Array of Object References))
   > n2 = For Each Loop(Array:Array of Widget Object References=GetAllChildren(self=Get ScrollBox!(self:EUW Meta Sound Cosmetic Action Name Widget Object Reference=@n1.Target) -> (ScrollBox:Editor Utility Scroll Box Object Reference))) -> (Array Element:Widget Object Reference)
   n3 = Sequence!(execute=@n2.Completed) -> (then_0, then_1, then_2:Exec)
   n4 = AssignMetaoundInput!(
    execute=@n3.then_1
    self
    TargetArray:Array of Object References=Get OneShotWaves()
    InputName:String=Get InputName() -> (InputName:String)
    ParameterID:String=Get NonLoopParameterID!(self:EUW Meta Sound Source Reference Object Reference=Get CurrentSoundSourceReference()) -> (NonLoopParameterID:String)
   ) -> (then, Success:bool, EarlyOut:bool) {is=title K2Node_CallFunction}
   > n5 = Branch(Condition=@n4.EarlyOut)
   Set Success(execute=@n5.else, Success:bool=@n4) -> (Output_Get:bool)
   n6 = AssignMetaoundInput(
    execute=@n3.then_0
    TargetArray=Get LoopingWaves()
    InputName=Get InputName()
    ParameterID=Get LoopParameterID!(self:EUW Meta Sound Source Reference Object Reference=Get CurrentSoundSourceReference()) -> (LoopParameterID:String)
   )
   > n7 = Branch(Condition=@n6.EarlyOut)
   Set Success(execute=@n7.else, Success=@n6)
   Return Node!(execute=@n3.then_2, bSuccess:bool=Get Success() -> (Success:bool))
   n8 = Cast To EUW_MetaSound_Cosmetic_SoundNameWidget!(execute=@n2.LoopBody, Object=@n2) -> (then, CastFailed, AsEUW Meta Sound Cosmetic Sound Name Widget:EUW Meta Sound Cosmetic Sound Name Widget Object Reference) {is=title K2Node_DynamicCast}
   > n9 = Branch(Condition=Get IsLooping!(self:EUW Meta Sound Cosmetic Sound Name Widget Object Reference=@n8) -> (IsLooping:bool))
   n10 = Get SoundObject!(self:EUW Meta Sound Cosmetic Sound Name Widget Object Reference=@n8) -> (SoundObject:Object Reference)
   > Add Unique(TargetArray:Array of Object References=Get LoopingWaves(), NewItem:"Object Reference (by ref)"=@n10)
   Add Unique(execute=@n9.else, TargetArray=Get OneShotWaves(), NewItem=@n10)
   Assign Meta Sound Input!(
    execute:Exec
    TargetArray:Array of Object References=Get OneShotWaves()
    InputName:String=Get InputName()
    ParameterID:String=Get NonLoopParameterID(self=Get CurrentSoundSourceReference())
   ) -> (Succeeded:Exec, Failed:Exec, NA:Exec) {is=title K2Node_MacroInstance}
   Assign Meta Sound Input(TargetArray=Get LoopingWaves(), InputName=Get InputName(), ParameterID=Get LoopParameterID(self=Get CurrentSoundSourceReference()))
  ]}
 EventGraph {nodes [
   "Event on Asset Path Button (AssetPath)"() {is=title K2Node_ComponentBoundEvent}
   > SetAssetPath!(
    execute
    self:EUW Meta Sound Asset Path Object Reference=Get AssetPath() -> (AssetPath:EUW Meta Sound Asset Path Object Reference)
    Object:Object Reference="Get (a copy)"(Array=Get SelectedAssets() -> (SelectedAssets:Array of Object References))
   ) -> (then)
   "Event on Make Preset Button (AssetPath)"() {is=title K2Node_ComponentBoundEvent}
   > Valid Path Check!(execute:Exec) -> (Success:Exec, Fail:Exec) {is=title K2Node_MacroInstance}
   > Valid Sounds Check!(execute:Exec) -> (Success:Exec, Fail:Exec) {is=title K2Node_MacroInstance}
   > Valid Asset Name Check!(execute:Exec) -> (Success:Exec, Fail:Exec) {is=title K2Node_MacroInstance}
   > Clear(TargetArray=Get Wave Array() -> (Wave Array:Array of Object References))
   > Clear(TargetArray:Array of Meta Sound Source Object References=Get Metasounds() -> (Metasounds:Array of Meta Sound Source Object References))
   > n1 = For Each Loop(Array:Array of EUW Meta Sound Source Reference Object References=Get SoundSourceReferenceWidgets() -> (SoundSourceReferenceWidgets:Array of EUW Meta Sound Source Reference Object References)) -> (Array Element:EUW Meta Sound Source Reference Object Reference)
   SetSourceViewer(execute=@n1.Completed)
   > Call Finish Setting Sounds!(execute:Exec, self:EUW Meta Sound Cosmetic Preset Creator Object Reference) -> (then:Exec) {is=title K2Node_CallDelegate}
   n2 = Set CurrentSoundSourceReference(execute=@n1.LoopBody, CurrentSoundSourceReference:EUW Meta Sound Source Reference Object Reference=@n1) -> (Output_Get:EUW Meta Sound Source Reference Object Reference)
   n3 = Get SourceReference!(self:EUW Meta Sound Source Reference Object Reference=@n2) -> (SourceReference:Meta Sound Source Object Reference)
   > Is Valid(InputObject=@n3)
   > Set Meta Sound Builder!(execute:Exec, Object:Object Reference=@n3) -> (Success:Exec, Failed:Exec) {is=title K2Node_MacroInstance}
   > Set AssignSuccess(AssignSuccess:bool=false) -> (Output_Get:bool)
   > Clear(TargetArray:Array of EUW Meta Sound Cosmetic Action Name Widget Object References=Get ActionWidgets() -> (ActionWidgets:Array of EUW Meta Sound Cosmetic Action Name Widget Object References))
   > n4 = For Each Loop(Array:Array of EUW Meta Sound Cosmetic Action Name Widget Object References=Get ActionNameWidgets!(self:EUW Meta Sound Cosmetic Action Sounds List Object Reference=Get SoundNameListView()) -> (ActionNameWidgets:Array of EUW Meta Sound Cosmetic Action Name Widget Object References)) -> (Array Element:EUW Meta Sound Cosmetic Action Name Widget Object Reference)
   n5 = Get ActionWidgets()
   Branch(execute=@n4.Completed, Condition=Is Not Empty(TargetArray:Array of EUW Meta Sound Cosmetic Action Name Widget Object References=@n5))
   > n6 = For Each Loop(Array=@n5)
   > n7 = AssignWaveAssets!(
    execute
    self
    InputName:String=Get BaseMetasoundName!(self:EUW Meta Sound Cosmetic Action Name Widget Object Reference=@n6) -> (BaseMetasoundName:String)
    Target:EUW Meta Sound Cosmetic Action Name Widget Object Reference=@n6
   ) -> (then, bSuccess:bool) {is=title K2Node_CallFunction}
   > Set AssignSuccess(AssignSuccess=@n7)
   n8 = Branch(execute=@n6.Completed, Condition=Get AssignSuccess() -> (AssignSuccess:bool))
   n9 = GetAssetName!(self:EUW Meta Sound Asset Path Object Reference=Get AssetPath()) -> (AssetText:Text, IsEmpty:bool)
   n10 = Get AssetSuffix!(self:EUW Meta Sound Source Reference Object Reference=Get CurrentSoundSourceReference()) -> (AssetSuffix:String)
   n11 = IsEmpty(InString=@n10)
   n12 = Append!(A=_, B=@n10) -> (ReturnValue)
   n13 = Get AssetPrefix!(self:EUW Meta Sound Source Reference Object Reference=Get CurrentSoundSourceReference()) -> (AssetPrefix:String)
   > n14 = BuildToAsset!(
    execute
    self:Meta Sound Editor Subsystem Object Reference=Get MetaSoundEditorSubsystem() -> (ReturnValue:Meta Sound Editor Subsystem Object Reference) {is=title K2Node_GetEditorSubsystem}
    InBuilder:Meta Sound Builder Base Object Reference=Get MetasoundBuilder()
    Author:String=SoundBaseAssetAction
    AssetName:String=Append!(A=Append!(A=@n13, B=_) -> (ReturnValue), B="To String (Text)"(InText=@n9), C:String=SelectString(B=@n12, bPickA=@n11)) -> (ReturnValue)
    PackagePath:String=SaveAssetLocation!() -> (ReturnValue:String)
    TemplateSoundWave:Sound Wave Object Reference
   ) -> (Succeeded:Exec, Failed:Exec, ReturnValue:Meta Sound Document Interface Interface)
   PrintString(execute=@n14.Failed, InString="FAIL!")
   PrintString(execute=@n14.Succeeded, InString="Success!")
   > n15 = Cast To MetaSoundSource!(execute, Object:Meta Sound Document Interface Interface=@n14) -> (then, CastFailed, AsMetaSound Source:Meta Sound Source Object Reference) {is=title K2Node_DynamicCast}
   > SetLicensedMetadata!(execute, self, SoundBase:Sound Base Object Reference=@n15) -> (then)
   > n16 = Branch(Condition=Get bSaveMetasounds() -> (bSaveMetasounds:bool))
   > n17 = SaveLoadedAsset(AssetToSave=@n15, bOnlyIfIsDirty=false)
   OutputLog!(
    execute=[@n17.then, @n16.else]
    self
    LoopObjects:Array of Object References=Get LoopingWaves()
    OneShotObjects:Array of Object References=Get OneShotWaves()
   ) -> (then)
   > Add(TargetArray:Array of Meta Sound Source Object References=Get Metasounds(), NewItem:"Meta Sound Source Object Reference (by ref)"=@n15) {is=title K2Node_CallArrayFunction}
   > SetSounds!(
    execute
    self:EUW Meta Sound Source Reference Object Reference=Get CurrentSoundSourceReference()
    Sound:Sound Base Soft Object Reference=Make Soft Reference!(Input:Meta Sound Source Object Reference=@n15) -> (Output:Meta Sound Source Soft Object Reference)
   ) -> (then)
   Show Message Dialog(
    execute=@n8.else
    Title=MakeLiteralText(Value="LOCTEXT(\"Error creating metasound\")")
    Message=MakeLiteralText(Value="LOCTEXT(\"Unable to assign wave values inside of metsaound, please reach out to tech audio for support\")")
    MessageType=Ok
    DefaultValue=No
    MessageCategory=Warning
   )
   Is Valid(
    exec=@n4.LoopBody
    InputObject=GetChildAt!(self:Panel Widget Object Reference=Get ScrollBox(self=@n4), Index:int=1) -> (ReturnValue:Widget Object Reference)
   )
   n18 = Get CurrentSoundSourceReference()
   n19 = Get NameOverride!(self:EUW Meta Sound Source Reference Object Reference=@n18) -> (NameOverride:String)
   n20 = Get OverrideName!(self:EUW Meta Sound Cosmetic Action Name Widget Object Reference=@n4) -> (OverrideName:String)
   n21 = Get!(self:EUW Meta Sound Source Reference Object Reference=@n18) -> (
    NamesDataStructure_SkipActionName_27_#1:bool=~
    NamesDataStructure_DiableLoopOption_26_#2:bool=~
    NamesDataStructure_SourceMetaSoundName_22_#3:String
    NamesDataStructure_SourceWavesName_21_#4:Array of Strings
    NamesDataStructure_SameNameIdentifiers_32_#5:Array of Strings
   )
   n22 = Get SourceMetaSoundName!(self:EUW Meta Sound Cosmetic Action Name Widget Object Reference=@n4) -> (SourceMetaSoundName:String)
   > Branch(Condition=AND Boolean(
     A="Equal Exactly (String)"(A=@n22, B=@n21.NamesDataStructure_SourceMetaSoundName_22_#3)
     B="Equal Exactly (String)"(A=@n20, B=@n19)
    ))
   > Add Unique(
    TargetArray:Array of EUW Meta Sound Cosmetic Action Name Widget Object References=Get ActionWidgets()
    NewItem:"EUW Meta Sound Cosmetic Action Name Widget Object Reference (by ref)"=@n4
   )
   "On Pressed (SetPickaxe)"() {is=title K2Node_ComponentBoundEvent}
   > Hide References()
   > n23 = Create Widget!(
    execute:Exec
    WorldContextObject:Object Reference
    Class:User Widget Class Reference=&/EUW_MetaSound_Cosmetic_PresetCreator_Pickaxe.EUW_MetaSound_Cosmetic_PresetCreator_Pickaxe_C
    OwningPlayer:Player Controller Object Reference
   ) -> (then:Exec, ReturnValue:EUW Meta Sound Cosmetic Preset Creator Pickaxe Object Reference)
   > n24 = Set PickaxeRef(PickaxeRef:EUW Meta Sound Cosmetic Preset Creator Pickaxe Object Reference=@n23) -> (Output_Get:EUW Meta Sound Cosmetic Preset Creator Pickaxe Object Reference)
   > AddChildToGrid!(
    execute
    self:Grid Panel Object Reference=Get ReferencesGrid() -> (ReferencesGrid:Grid Panel Object Reference)
    Content:Widget Object Reference=@n24
    InRow:int=0
    InColumn:int=0
   ) -> (then, ReturnValue:Grid Slot Object Reference)
   n25 = Get PickaxeRef() -> (PickaxeRef:EUW Meta Sound Cosmetic Preset Creator Pickaxe Object Reference)
   > Set Current Source Reference Widgets!(
    execute
    self
    CurrentActiveWidget:Widget Object Reference=@n25
    SourceReferences:Array of EUW Meta Sound Source Reference Object References=Get SoundSourceReferenceWidgets!(self:EUW Meta Sound Cosmetic Preset Creator Pickaxe Object Reference=@n25) -> (SoundSourceReferenceWidgets)
    CosmeticType:String=Pickaxe
   ) -> (then) {is=title K2Node_CallFunction}
   > SetColorAndOpacity(
    self=Get PickaxeText() -> (PickaxeText:Text Block Object Reference)
    InColorAndOpacity=Get ActivateTabColor() -> (ActivateTabColor:Slate Color Structure)
   )
   "On Pressed (SetReactive)"() {is=title K2Node_ComponentBoundEvent}
   > Hide References()
   > n26 = Create Widget(Class=&/EUW_MetaSound_Cosmetic_PresetCreator_Reactive.EUW_MetaSound_Cosmetic_PresetCreator_Reactive_C) -> (ReturnValue:EUW Meta Sound Cosmetic Preset Creator Reactive Object Reference)
   > n27 = Set ReactiveRef(ReactiveRef:EUW Meta Sound Cosmetic Preset Creator Reactive Object Reference=@n26) -> (Output_Get:EUW Meta Sound Cosmetic Preset Creator Reactive Object Reference)
   > AddChildToGrid(self=Get ReferencesGrid(), Content=@n27, InRow=0, InColumn=0)
   n28 = Get ReactiveRef() -> (ReactiveRef:EUW Meta Sound Cosmetic Preset Creator Reactive Object Reference)
   > Set Current Source Reference Widgets(
    CurrentActiveWidget=@n28
    SourceReferences=Get SoundSourceReferenceWidgets(self:EUW Meta Sound Cosmetic Preset Creator Reactive Object Reference=@n28)
    CosmeticType=Reactive
   )
   > SetColorAndOpacity(self=Get ReactiveText() -> (ReactiveText:Text Block Object Reference), InColorAndOpacity=Get ActivateTabColor())
   "On Clicked (SetCompanion)"() {is=title K2Node_ComponentBoundEvent}
   > Hide References()
   > n29 = Create Widget(Class=&/EUW_MetaSound_Cosmetic_PresetCreator_Companions.EUW_MetaSound_Cosmetic_PresetCreator_Companions_C) -> (ReturnValue:EUW Meta Sound Cosmetic Preset Creator Companions Object Reference)
   > n30 = Set CompanionRef(CompanionRef:EUW Meta Sound Cosmetic Preset Creator Companions Object Reference=@n29) -> (Output_Get:EUW Meta Sound Cosmetic Preset Creator Companions Object Reference)
   > AddChildToGrid(self=Get ReferencesGrid(), Content=@n30, InRow=0, InColumn=0)
   n31 = Get CompanionRef() -> (CompanionRef:EUW Meta Sound Cosmetic Preset Creator Companions Object Reference)
   > Set Current Source Reference Widgets(
    CurrentActiveWidget=@n31
    SourceReferences=Get SoundSourceReferenceWidgets(self:EUW Meta Sound Cosmetic Preset Creator Companions Object Reference=@n31)
    CosmeticType=Mimosa
   )
   > SetColorAndOpacity(self=Get CompanionText() -> (CompanionText:Text Block Object Reference), InColorAndOpacity=Get ActivateTabColor())
   "On Pressed (SetContrail)"() {is=title K2Node_ComponentBoundEvent}
   > Hide References()
   > n32 = Create Widget(Class=&/EUW_MetaSound_Cosmetic_PresetCreator_Contrail.EUW_MetaSound_Cosmetic_PresetCreator_Contrail_C) -> (ReturnValue:EUW Meta Sound Cosmetic Preset Creator Contrail Object Reference)
   > n33 = Set ContrailRef(ContrailRef:EUW Meta Sound Cosmetic Preset Creator Contrail Object Reference=@n32) -> (Output_Get:EUW Meta Sound Cosmetic Preset Creator Contrail Object Reference)
   > AddChildToGrid(self=Get ReferencesGrid(), Content=@n33, InRow=0, InColumn=0)
   n34 = Get ContrailRef() -> (ContrailRef:EUW Meta Sound Cosmetic Preset Creator Contrail Object Reference)
   > Set Current Source Reference Widgets(
    CurrentActiveWidget=@n34
    SourceReferences=Get SoundSourceReferenceWidgets(self:EUW Meta Sound Cosmetic Preset Creator Contrail Object Reference=@n34)
    CosmeticType=Contrail
   )
   > SetColorAndOpacity(self=Get ContrailText() -> (ContrailText:Text Block Object Reference), InColorAndOpacity=Get ActivateTabColor())
   "On Pressed (SetGlider)"() {is=title K2Node_ComponentBoundEvent}
   > Hide References()
   > n35 = Create Widget(Class=&/EUW_MetaSound_Cosmetic_PresetCreator_Glider.EUW_MetaSound_Cosmetic_PresetCreator_Glider_C) -> (ReturnValue:EUW Meta Sound Cosmetic Preset Creator Glider Object Reference)
   > n36 = Set GliderRef(GliderRef:EUW Meta Sound Cosmetic Preset Creator Glider Object Reference=@n35) -> (Output_Get:EUW Meta Sound Cosmetic Preset Creator Glider Object Reference)
   > AddChildToGrid(self=Get ReferencesGrid(), Content=@n36, InRow=0, InColumn=0)
   n37 = Get GliderRef() -> (GliderRef:EUW Meta Sound Cosmetic Preset Creator Glider Object Reference)
   > Set Current Source Reference Widgets(
    CurrentActiveWidget=@n37
    SourceReferences=Get SoundSourceReferenceWidgets(self:EUW Meta Sound Cosmetic Preset Creator Glider Object Reference=@n37)
    CosmeticType=Glider
   )
   > SetColorAndOpacity(self=Get GliderText() -> (GliderText:Text Block Object Reference), InColorAndOpacity=Get ActivateTabColor())
   "On Pressed (SetMusicPack)"() {is=title K2Node_ComponentBoundEvent}
   > Hide References()
   > n38 = Create Widget(Class=&/EUW_MetaSound_Cosmetic_PresetCreator_MusicPack.EUW_MetaSound_Cosmetic_PresetCreator_MusicPack_C) -> (ReturnValue:EUW Meta Sound Cosmetic Preset Creator Music Pack Object Reference)
   > n39 = Set MusicPack(MusicPack:EUW Meta Sound Cosmetic Preset Creator Music Pack Object Reference=@n38) -> (Output_Get:EUW Meta Sound Cosmetic Preset Creator Music Pack Object Reference)
   > AddChildToGrid(self=Get ReferencesGrid(), Content=@n39, InRow=0, InColumn=0)
   n40 = Get MusicPack() -> (MusicPack:EUW Meta Sound Cosmetic Preset Creator Music Pack Object Reference)
   > Set Current Source Reference Widgets(
    CurrentActiveWidget=@n40
    SourceReferences=Get SoundSourceReferenceWidgets(self:EUW Meta Sound Cosmetic Preset Creator Music Pack Object Reference=@n40)
    CosmeticType=Reactive
   )
   > SetColorAndOpacity(self=Get MusicPackText() -> (MusicPackText:Text Block Object Reference), InColorAndOpacity=Get ActivateTabColor())
   "On Pressed (SetEmote)"() {is=title K2Node_ComponentBoundEvent}
   > Hide References()
   > n41 = Create Widget(Class=&/EUW_MetaSound_Cosmetic_PresetCreator_Emote.EUW_MetaSound_Cosmetic_PresetCreator_Emote_C) -> (ReturnValue:EUW Meta Sound Cosmetic Preset Creator Emote Object Reference)
   > n42 = Set EmoteRef(EmoteRef:EUW Meta Sound Cosmetic Preset Creator Emote Object Reference=@n41) -> (Output_Get:EUW Meta Sound Cosmetic Preset Creator Emote Object Reference)
   > AddChildToGrid(self=Get ReferencesGrid(), Content=@n42, InRow=0, InColumn=0)
   n43 = Get EmoteRef() -> (EmoteRef:EUW Meta Sound Cosmetic Preset Creator Emote Object Reference)
   > Set Current Source Reference Widgets(
    CurrentActiveWidget=@n43
    SourceReferences=Get SoundSourceReferenceWidgets(self:EUW Meta Sound Cosmetic Preset Creator Emote Object Reference=@n43)
    CosmeticType=Emote
   )
   > SetColorAndOpacity(self=Get EmoteText() -> (EmoteText:Text Block Object Reference), InColorAndOpacity=Get ActivateTabColor())
   "On Pressed (SetKicks)"() {is=title K2Node_ComponentBoundEvent}
   > Hide References()
   > n44 = Create Widget(Class=&/EUW_MetaSound_Cosmetic_PresetCreator_Kicks.EUW_MetaSound_Cosmetic_PresetCreator_Kicks_C) -> (ReturnValue:EUW Meta Sound Cosmetic Preset Creator Kicks Object Reference)
   > n45 = Set KicksRef(KicksRef:EUW Meta Sound Cosmetic Preset Creator Kicks Object Reference=@n44) -> (Output_Get:EUW Meta Sound Cosmetic Preset Creator Kicks Object Reference)
   > AddChildToGrid(self=Get ReferencesGrid(), Content=@n45, InRow=0, InColumn=0)
   n46 = Get KicksRef() -> (KicksRef:EUW Meta Sound Cosmetic Preset Creator Kicks Object Reference)
   > Set Current Source Reference Widgets(
    CurrentActiveWidget=@n46
    SourceReferences=Get SoundSourceReferenceWidgets(self:EUW Meta Sound Cosmetic Preset Creator Kicks Object Reference=@n46)
    CosmeticType=Kicks
   )
   > SetColorAndOpacity(self=Get KicksText() -> (KicksText:Text Block Object Reference), InColorAndOpacity=Get ActivateTabColor())
   SetSourceViewer() {is=title K2Node_CustomEvent}
   > ClearChildren(self=Get MsSourceBox() -> (MsSourceBox:Uniform Grid Panel Object Reference))
   > Set MsGridRow(MsGridRow:int=0) -> (Output_Get:int=0)
   > n47 = For Each Loop(Array:Array of Meta Sound Source Object References=Get Metasounds()) -> (Array Element:Meta Sound Source Object Reference)
   n48 = Make Soft Reference(Input=@n47)
   > n49 = Construct Object from Class!(
    execute
    Class=&/EUW_MetaSound_Cosmetic_MsPreview.EUW_MetaSound_Cosmetic_MsPreview_C
    self
    AudPath:String="To String (SoftObjectReference)"!(SoftObjectReference:"Soft Object Reference (by ref)"=@n48) -> (ReturnValue:String)
    MsSource:Meta Sound Source Soft Object Reference=@n48
   ) -> (then, ReturnValue:EUW Meta Sound Cosmetic Ms Preview Object Reference)
   > n50 = AddChild(self=Get MsSourceBox(), Content=@n49)
   > SetNameOverride!(
    execute
    self:Single Property View Object Reference=Get SoundView!(self:EUW Meta Sound Cosmetic Ms Preview Object Reference=@n49) -> (SoundView:Single Property View Object Reference)
    NewPropertyName:Text="LOCTEXT(\" \")"
   ) -> (then)
   > n51 = Cast To UniformGridSlot!(execute, Object=@n50) -> (then, CastFailed, AsUniform Grid Slot:Uniform Grid Slot Object Reference) {is=title K2Node_DynamicCast}
   > SetHorizontalAlignment(self:Uniform Grid Slot Object Reference=@n51, InHorizontalAlignment=HAlign_Fill)
   n52 = "% (Integer)"(A=@n47.Array Index, B=3) -> (ReturnValue=0)
   > SetColumn!(execute, self:Uniform Grid Slot Object Reference=@n51, InColumn:int=@n52) -> (then)
   n53 = Get MsGridRow() -> (MsGridRow:int=0)
   > SetRow!(execute, self:Uniform Grid Slot Object Reference=@n51, InRow:int=@n53) -> (then)
   > Branch(Condition="Equal (Integer)"(A=@n52, B=2))
   > Increment Int(Value=@n53)
   GetSelectedAssets() {is=title K2Node_CustomEvent}
   > n54 = GetSelectedAssetsOfClass(AssetClass=/Script/Engine.SoundWave)
   > AddSelectedAssets!(execute, self, Assets:Array of Object References=@n54) -> (then) {is=title K2Node_CallFunction}
   Event Destruct()
   > Clear and Invalidate Timer by Handle!(execute, WorldContextObject:Object Reference, Handle=Get GetSelectedAssetTimer() -> (GetSelectedAssetTimer:Timer Handle Structure)) -> (then)
   Event Construct()
   > Hide References()
   > n55 = Set Timer by Event(Delegate=Create Event(), Time=0.2, bLooping=true)
   > Set GetSelectedAssetTimer(GetSelectedAssetTimer:Timer Handle Structure=@n55) -> (Output_Get:Timer Handle Structure)
   Event PreConstruct()
   > SetObject(
    self=[
     Get CopyrightSoundsView() -> (CopyrightSoundsView:Single Property View Object Reference)
     Get SaveMetasoundsView() -> (SaveMetasoundsView:Single Property View Object Reference)
    ]
    NewObject=Self-Reference()
   )
   "To String (Text)"(InText="GetText (Text)"!(self:Text Object Reference=Get PresetText!(self:EUW Meta Sound Cosmetic Action Name Widget Object Reference=@n6) -> (PresetText:Text Object Reference)) -> (ReturnValue:Text))
   Valid List Check!(execute:Exec) -> (Success:Exec, Fail:Exec) {is=title K2Node_MacroInstance}
   Get NewVar() -> (NewVar:Fort Item Definition Object Reference)
   Get SourceMetaSoundName()
  ]}
 FinishSettingSounds {nodes [FinishSettingSounds() {is=title K2Node_FunctionEntry}]}
 Hide References {nodes [
   Hide References() {is=title K2Node_FunctionEntry}
   > RemoveFromParent(self=[Get ContrailRef(), Get PickaxeRef(), Get EmoteRef(), Get MusicPack(), Get ReactiveRef(), Get CompanionRef(), Get KicksRef(), Get GliderRef()])
   > Unbind all Events from Finish Setting Sounds!(execute:Exec, self:EUW Meta Sound Cosmetic Preset Creator Object Reference) -> (then:Exec) {is=title K2Node_ClearDelegate}
   > SetColorAndOpacity(
    self=[
     Get GliderText()
     Get MusicPackText()
     Get KicksText()
     Get ContrailText()
     Get CompanionText()
     Get ReactiveText()
     Get EmoteText()
     Get PickaxeText()
    ]
    InColorAndOpacity=Make Slate Color(SpecifiedColor=(R=1,G=1,B=1,A=1))
   )
   > Set IDReference(IDReference:EUW IDReference Object Reference) -> (Output_Get:EUW IDReference Object Reference)
  ]}
 OutputLog {nodes [
   n1 = OutputLog!() -> (then, LoopObjects:Array of Object References, OneShotObjects:Array of Object References) {is=title K2Node_FunctionEntry}
   > Set LoopStrings(LoopStrings:String) -> (Output_Get:String)
   > Set OneShotStrings(OneShotStrings:String) -> (Output_Get:String)
   > n2 = For Each Loop(Array:Array of Object References=@n1) -> (Array Element:Object Reference)
   n3 = For Each Loop(Exec=@n2.Completed, Array=@n1.OneShotObjects)
   > Set OneShotStrings(OneShotStrings=Append!(
     A=Get OneShotStrings() -> (OneShotStrings:String)
     B="\r\n"
     C:String="To String (Object)"!(InObj:Object Reference=@n3) -> (ReturnValue:String)
    ) -> (ReturnValue))
   n4 = Append!(A="Loops: ", B=Get LoopStrings() -> (LoopStrings:String)) -> (ReturnValue)
   n5 = IsEmpty(InString=@n4)
   n6 = Append(A="OneShots: ", B=Get OneShotStrings())
   n7 = IsEmpty(InString=@n6)
   n8 = Get CurrentSoundSourceReference()
   PrintString(
    execute=@n3.Completed
    InString=Append!(
     A=Get(self=@n8) -> (
      NamesDataStructure_SkipActionName_27_#1=~
      NamesDataStructure_DiableLoopOption_26_#2=~
     ).NamesDataStructure_SourceMetaSoundName_22_#3
     B="\r\n"
     C:String=SelectString(B=@n6, bPickA=@n7)
     D:String=SelectString(B="\r\n\r\n", bPickA=@n5)
     E:String=SelectString(B=@n4, bPickA=@n5)
    ) -> (ReturnValue)
   )
   Set LoopStrings(execute=@n2.LoopBody, LoopStrings=Append!(A=Get LoopStrings(), B="\r\n", C="To String (Object)"(InObj=@n2)) -> (ReturnValue))
  ]}
 SaveAssetLocation {nodes [
   SaveAssetLocation() {is=title K2Node_FunctionEntry}
   n1 = Get PathInput!(self:EUW Meta Sound Asset Path Object Reference=Get AssetPath()) -> (PathInput:Editor Utility Editable Text Box Object Reference)
   > Return Node!(execute, ReturnValue:String="To String (Text)"(InText=Get Text(self=@n1)))
  ]}
 Set Current Source Reference Widgets {nodes [
   n1 = Set Current Source Reference Widgets!() -> (
    then
    CurrentActiveWidget:Widget Object Reference
    SourceReferences:Array of EUW Meta Sound Source Reference Object References
    CosmeticType:String
   ) {is=title K2Node_FunctionEntry}
   > Set Current Active Widget(Current Active Widget:Widget Object Reference=@n1) -> (Output_Get:Widget Object Reference)
   > Clear(TargetArray:Array of EUW Meta Sound Source Reference Object References=Get SoundSourceReferenceWidgets!() -> (SoundSourceReferenceWidgets))
   > Set Cosmetic Type(Cosmetic Type:String=@n1.CosmeticType) -> (Output_Get:String)
   > ResetActionWidgets(self:EUW Meta Sound Cosmetic Action Sounds List Object Reference=Get SoundNameListView())
   > n2 = For Each Loop(Array:Array of EUW Meta Sound Source Reference Object References=@n1.SourceReferences) -> (Array Element:EUW Meta Sound Source Reference Object Reference)
   n3 = For Each Loop(Exec=@n2.Completed, Array=Get SoundSourceReferenceWidgets())
   SetVisibility(execute=@n3.Completed, self=Get Current Active Widget() -> (Current Active Widget:Widget Object Reference), InVisibility=Visible)
   GetSoundNameReferences!(
    execute=@n3.LoopBody
    self:EUW Meta Sound Cosmetic Action Sounds List Object Reference=Get SoundNameListView()
    CosmeticType:String=Get Cosmetic Type() -> (Cosmetic Type:String)
    Names Data Structure:Struct Meta Sound Source Reference Structure=Get NamesDataStructure!(self:EUW Meta Sound Source Reference Object Reference=@n3) -> (NamesDataStructure:Struct Meta Sound Source Reference Structure)
    NameOverride:String=Get NameOverride(self=@n3)
   ) -> (then)
   Is Valid(exec=@n2.LoopBody, InputObject=@n2)
   > Add Unique(
    TargetArray:Array of EUW Meta Sound Source Reference Object References=Get SoundSourceReferenceWidgets()
    NewItem:"EUW Meta Sound Source Reference Object Reference (by ref)"=@n2
   )
   AddAnimationNotifyTrack!(
    execute
    AnimationSequenceBase:Anim Sequence Base Object Reference
    NotifyTrackName:Name=None
    TrackColor:Linear Color Structure=(R=1,G=1,B=1,A=1)
   ) -> (then)
  ]}
 SetLicensedMetadata {nodes [
   n1 = SetLicensedMetadata!() -> (then, SoundBase:Sound Base Object Reference) {is=title K2Node_FunctionEntry}
   > n2 = Branch(Condition=Get bLicensedAudio() -> (bLicensedAudio:bool))
   > n3 = AddMuteForStreamSafeTag!(execute, AudioAsset:Sound Base Object Reference=@n1, __WorldContext:Object Reference) -> (then, Success:bool)
   > n4 = Branch(Condition=@n3)
   Return Node!(execute=[@n4.then, @n2.else, @n6.then])
   n5 = GetDisplayName(Object=@n1)
   n6 = Show Message Dialog(
    execute=@n4.else
    Title=MakeLiteralText(Value="LOCTEXT(\"Audio copyright tags not added\")")
    Message="To Text (String)"(InString=Append(
      A="Asset: "
      B=@n5
      C=" isn't able to add new streaming tags. Please ensure the following:\r\n- there are no tags currently listed in the item user data\r\n- You are saving your content to /BrCosmetics, and not /Game\r\n\r\nReach out to tech audio for further support if needed."
     ))
    MessageType=Ok
    DefaultValue=No
    MessageCategory=Warning
   )
   n7 = Branch(Condition=Get bLicensedAudio())
   > n8 = HasAssetUserDataOfClass!(
    execute
    self:Interface Asset User Data Interface
    InUserDataClass:Asset User Data Class Reference=Get AudioUserDataClass() -> (AudioUserDataClass:Asset User Data Class Reference)
   ) -> (then, ReturnValue:bool)
   > n9 = Branch(Condition=@n8)
   n10 = GetAssetUserDataOfClass!(
    execute=[@n9.then, @n13.then]
    self:Interface Asset User Data Interface=Get SoundBase() -> (SoundBase:Sound Base Object Reference)
    InUserDataClass:Asset User Data Class Reference=Get AudioUserDataClass()
   ) -> (then, ReturnValue:Asset User Data Object Reference)
   > n11 = Cast To AudioAssetUserData!(execute, Object=@n10) -> (then, CastFailed, AsAudio Asset User Data:Audio Asset User Data Object Reference) {is=title K2Node_DynamicCast}
   > AddTag!(
    execute
    Container:"Gameplay Tag Container Structure (by ref)"=Get MetadataTags!(self:Audio Asset User Data Object Reference=@n11) -> (MetadataTags:Gameplay Tag Container Structure)
    GameplayTag:"Gameplay Tag Structure (by ref)"=Get LicensedTag() -> (LicensedTag:Gameplay Tag Structure)
   ) -> (then)
   n12 = AddAssetUserDataOfClass!(
    execute=@n9.else
    self:Interface Asset User Data Interface=Get SoundBase()
    InUserDataClass:Asset User Data Class Reference=Get AudioUserDataClass()
   ) -> (then, ReturnValue:bool)
   > n13 = Branch(Condition=@n12)
   Return Node(execute=@n13.else)
   Return Node(execute=@n7.else)
  ]}
 SetMetaSoundBuilder {nodes [
   n1 = Inputs!() -> (execute, Object:Object Reference)
   > n2 = Cast To MetaSoundDocumentInterface!(execute, Object=@n1) -> (then, CastFailed, AsMetaSound Document Interface:Meta Sound Document Interface Interface) {is=title K2Node_DynamicCast}
   > n3 = CreateSourcePresetBuilder!(
    execute
    self:Meta Sound Builder Subsystem Object Reference=Get MetaSoundBuilderSubsystem()
    BuilderName:Name=Builder
    ReferencedSourceClass:"Meta Sound Document Interface Interface (by ref)"=@n2
   ) -> (Succeeded:Exec, Failed:Exec, ReturnValue:Meta Sound Source Builder Object Reference)
   > Set MetasoundBuilder(MetasoundBuilder:Meta Sound Source Builder Object Reference=@n3) -> (Output_Get:Meta Sound Source Builder Object Reference)
   > Outputs!(Success:Exec, Failed=@n4.then)
   n4 = PrintString(execute=@n2.CastFailed, InString="Builder Failed! Check Set Meta Sound Builder Macro")
  ]}
 UpdateActivateWidget {nodes [UpdateActivateWidget!() -> (then, ActivateWidget:Widget Object Reference) {is=title K2Node_FunctionEntry}]}
 ValidAssetNameCheck {nodes [
   Inputs!() -> (execute)
   n1 = Get AssetPath()
   > n2 = Branch(Condition=NOT Boolean(A=GetAssetName(self=@n1).IsEmpty))
   > Outputs!(Success, Fail:Exec=@n2.else)
  ]}
 ValidListCheck {nodes [
   Inputs()
   n1 = Get ActionNameWidgets(self=Get SoundNameListView())
   n2 = GetAllChildren(self=Get ScrollBox(self="Get (a copy)"(Array:Array of EUW Meta Sound Cosmetic Action Name Widget Object References=@n1) -> (Output:EUW Meta Sound Cosmetic Action Name Widget Object Reference)))
   > n3 = Branch(Condition=NOT Boolean(A=Is Valid Index!(TargetArray:Array of Widget Object References=@n2, IndexToTest:int=1) -> (ReturnValue:bool)))
   > Outputs(Fail=@n4.then)
   n4 = Show Message Dialog(
    execute=@n3.else
    Title=MakeLiteralText(Value="LOCTEXT(\"Undefined Action Still has Waves Available\")")
    Message=MakeLiteralText(Value="LOCTEXT(\"Remove or Add waves to proper action widget locations. No asset should be in undefined section.\")")
    MessageType=Ok
    DefaultValue=No
    MessageCategory=Warning
   )
  ]}
 ValidPathCheck {nodes [
   Inputs()
   n1 = Get AssetPath()
   > n2 = Branch(Condition=Get ValidFilePath!(self:EUW Meta Sound Asset Path Object Reference=@n1) -> (ValidFilePath:bool))
   n3 = "To String (Text)"(InText=Get Text(self=Get PathInput(self=@n1)))
   > n4 = Branch(Condition=NOR Boolean!(A:bool=Contains(SearchIn=@n3, Substring=/CosmeticCompanions), B:bool=Contains(SearchIn=@n3, Substring=/BRCosmetics/Sounds)) -> (ReturnValue:bool))
   > n5 = Show Message Dialog(
    Title=MakeLiteralText(Value="LOCTEXT(\"Outside of expected directory\")")
    Message=MakeLiteralText(Value="LOCTEXT(\"You are trying to create sounds outside of the designated BR audio plugins, if this is intended you may continue. \\r\\n\\r\\nIf you feel you are getting this error unintendedly, please reach out to the Audio Tech POI for cosmetics and ask for this directory to be included\")")
    MessageType=CancelRetryContinue
    DefaultValue=No
    MessageCategory=Error
   )
   > n6 = Switch on EAppReturnType(Selection=@n5)
   Outputs(Success=[@n6.Continue, @n4.else], Fail=[@n7.then, @n6.Cancel, @n6.Retry])
   n7 = Show Message Dialog(
    execute=@n2.else
    Title=MakeLiteralText(Value="LOCTEXT(\"Invalid File Path\")")
    Message=MakeLiteralText(Value="LOCTEXT(\"File path not valid. Please select proper content browser path.\")")
    MessageType=Ok
    DefaultValue=No
    MessageCategory=Error
   )
  ]}
 ValidSoundsCheck {nodes [
   Inputs()
   n1 = Get SoundNameListView()
   > n2 = Branch(Condition=Is Valid Index(TargetArray:Array of EUW Meta Sound Cosmetic Sound Name Widget Object References=Get SoundNameWidgets!(self:EUW Meta Sound Cosmetic Action Sounds List Object Reference=@n1) -> (SoundNameWidgets:Array of EUW Meta Sound Cosmetic Sound Name Widget Object References)))
   > Outputs(Fail=@n3.then)
   n3 = Show Message Dialog(
    execute=@n2.else
    Title=MakeLiteralText(Value="LOCTEXT(\"No Sounds Added\")")
    Message=MakeLiteralText(Value="LOCTEXT(\"Missing sound waves. None selected.\")")
    MessageType=Ok
    DefaultValue=No
    MessageCategory=Warning
   )
  ]}
]
variables [
 ActionWidgets {type="array<EUW_MetaSound_Cosmetic_ActionNameWidget_C>", default []}
 ActivateTabColor {type=SlateColor, default {colorUseRule=UseColor_Specified, specifiedColor=<rgba 0.2745 1 0.2314 1>}}
 Asset Prefix Name {type=string, default=""}
 AssignSuccess {type=boolean}
 AudioUserDataClass {type=Class, default=/Script/AudioGameplay.AudioAssetUserData}
 bLicensedAudio {type=boolean}
 bSaveMetasounds {type=boolean, default=true}
 BuiltMS {type="array<MetaSoundSource>", default []}
 CompanionRef {type=EUW_MetaSound_Cosmetic_PresetCreator_Companions_C, default=null}
 ContrailRef {type=EUW_MetaSound_Cosmetic_PresetCreator_Contrail_C, default=null}
 Cosmetic Type {type=string, default=""}
 Current Active Widget {type=Widget, default=null}
 CurrentSoundSourceReference {type=EUW_MetaSoundSourceReference_C, default=null}
 EmoteRef {type=EUW_MetaSound_Cosmetic_PresetCreator_Emote_C, default=null}
 FinishSettingSounds {default=()}
 GetSelectedAssetTimer {type=TimerHandle}
 GliderRef {type=EUW_MetaSound_Cosmetic_PresetCreator_Glider_C, default=null}
 IDReference {type=EUW_IDReference_C, default=null}
 KicksRef {type=EUW_MetaSound_Cosmetic_PresetCreator_Kicks_C, default=null}
 LicensedTag {type=GameplayTag, default {tagName=Audio.Metadata.MuteForStreamSafe}}
 LoopingWaves {type="array<Object>", default []}
 MetasoundBuilder {type=MetaSoundSourceBuilder, default=null}
 Metasounds {type="array<MetaSoundSource>", default []}
 MsGridRow {type=integer}
 MusicPack {type=EUW_MetaSound_Cosmetic_PresetCreator_MusicPack_C, default=null}
 NewVar {type=FortItemDefinition, default=null}
 OneShotWaves {type="array<Object>", default []}
 PickaxeRef {type=EUW_MetaSound_Cosmetic_PresetCreator_Pickaxe_C, default=null}
 ReactiveRef {type=EUW_MetaSound_Cosmetic_PresetCreator_Reactive_C, default=null}
 SelectedAssets {type="array<Object>", default []}
 SoundSourceReferenceWidgets {type="array<EUW_MetaSoundSourceReference_C>", default []}
 UpdateActivateWidget {default=()}
 Wave Array {type="array<Object>", default []}
]
widgets {
 info {rootWidgetClass=&CanvasPanel, widgetCount=31}
 widgets [
  CanvasPanel_202 {
   widgetClassPath=&CanvasPanel
   widget=CanvasPanel_202
   properties {toolTipText=Set cosmetic options to glider presets., renderTransform {angle=0, scale=<1 1>, shear=<0 0>, translation=<-7 2>}}
  }
  GridPanel_0 {
   widgetClassPath=&GridPanel
   parent=CanvasPanel_202
   widget=GridPanel_0
   slot {class=CanvasPanelSlot, properties {layoutData {alignment=<0 0>, anchors {maximum=<0 0>, minimum=<0 0>}, offsets=<ltrb 8 8 481.2312 50>}}}
  }
  VerticalBox_2 {widgetClassPath=&VerticalBox, parent=GridPanel_0, widget=VerticalBox_2, slot {class=GridSlot}}
  SizeBox_0 {
   widgetClassPath=&SizeBox
   parent=VerticalBox_2
   widget=SizeBox_0
   properties {bOverride_HeightOverride=true, heightOverride=60}
   slot {class=VerticalBoxSlot, properties {verticalAlignment=VAlign_Center, padding=<ltrb 2 2 2 2>}}
  }
  VerticalBox_0 {
   widgetClassPath=&VerticalBox
   parent=SizeBox_0
   widget=VerticalBox_0
   properties {bOverride_Cursor=true}
   slot {class=SizeBoxSlot, properties {padding=<ltrb 1 1 1 1>}}
  }
  HorizontalBox_60 {widgetClassPath=&HorizontalBox, parent=VerticalBox_0, widget=HorizontalBox_60, slot {class=VerticalBoxSlot}}
  SetGlider {
   widgetClassPath=&EditorUtilityButton
   parent=HorizontalBox_60
   widget=SetGlider
   properties {toolTipText=Set cosmetic options to glider presets.}
   slot {class=HorizontalBoxSlot, properties {padding=<ltrb 5 5 5 5>, size {sizeRule=Fill, value=1}}}
  }
  GliderText {
   widgetClassPath=&UEFN_TextBlock
   bIsVariable=true
   parent=SetGlider
   widget=GliderText
   properties {
    justification=Center
    text=Glider
    font {
     bForceMonospaced=false
     bMaterialIsStencil=false
     fontMaterial=null
     fontObject=&Roboto
     letterSpacing=0
     monospacedWidth=1
     size=9
     skewAmount=0
     typefaceFontName=Regular
     outlineSettings {
      bApplyOutlineToDropShadows=false
      bMiteredCorners=false
      bSeparateFillAlpha=false
      outlineMaterial=null
      outlineSize=0
      outlineColor=<rgba 0 0 0 1>
     }
    }
    shadowColorAndOpacity=<rgba 0 0 0 1>
   }
   slot {class=ButtonSlot}
  }
  SetPickaxe {
   widgetClassPath=&EditorUtilityButton
   parent=HorizontalBox_60
   widget=SetPickaxe
   properties {toolTipText=Set cosmetic options to pickaxe presets.}
   slot {class=HorizontalBoxSlot, properties {padding=<ltrb 5 5 5 5>, size {sizeRule=Fill, value=1}}}
  }
  PickaxeText {
   widgetClassPath=&UEFN_TextBlock
   bIsVariable=true
   parent=SetPickaxe
   widget=PickaxeText
   properties {
    justification=Center
    text=Pickaxe
    font {
     bForceMonospaced=false
     bMaterialIsStencil=false
     fontMaterial=null
     fontObject=&Roboto
     letterSpacing=0
     monospacedWidth=1
     size=9
     skewAmount=0
     typefaceFontName=Regular
     outlineSettings {
      bApplyOutlineToDropShadows=false
      bMiteredCorners=false
      bSeparateFillAlpha=false
      outlineMaterial=null
      outlineSize=0
      outlineColor=<rgba 0 0 0 1>
     }
    }
    shadowColorAndOpacity=<rgba 0 0 0 1>
   }
   slot {class=ButtonSlot}
  }
  SetEmote {
   widgetClassPath=&EditorUtilityButton
   parent=HorizontalBox_60
   widget=SetEmote
   properties {toolTipText=Set cosmetic options to emote presets.}
   slot {class=HorizontalBoxSlot, properties {padding=<ltrb 5 5 5 5>, size {sizeRule=Fill, value=1}}}
  }
  EmoteText {
   widgetClassPath=&UEFN_TextBlock
   bIsVariable=true
   parent=SetEmote
   widget=EmoteText
   properties {
    justification=Center
    text=Emote
    font {
     bForceMonospaced=false
     bMaterialIsStencil=false
     fontMaterial=null
     fontObject=&Roboto
     letterSpacing=0
     monospacedWidth=1
     size=9
     skewAmount=0
     typefaceFontName=Regular
     outlineSettings {
      bApplyOutlineToDropShadows=false
      bMiteredCorners=false
      bSeparateFillAlpha=false
      outlineMaterial=null
      outlineSize=0
      outlineColor=<rgba 0 0 0 1>
     }
    }
    shadowColorAndOpacity=<rgba 0 0 0 1>
   }
   slot {class=ButtonSlot}
  }
  SetReactive {
   widgetClassPath=&EditorUtilityButton
   parent=HorizontalBox_60
   widget=SetReactive
   properties {toolTipText=Set cosmetic options to backbling presets.}
   slot {class=HorizontalBoxSlot, properties {padding=<ltrb 5 5 5 5>, size {sizeRule=Fill, value=1}}}
  }
  ReactiveText {
   widgetClassPath=&UEFN_TextBlock
   bIsVariable=true
   parent=SetReactive
   widget=ReactiveText
   properties {
    justification=Center
    text=Reactive
    font {
     bForceMonospaced=false
     bMaterialIsStencil=false
     fontMaterial=null
     fontObject=&Roboto
     letterSpacing=0
     monospacedWidth=1
     size=9
     skewAmount=0
     typefaceFontName=Regular
     outlineSettings {
      bApplyOutlineToDropShadows=false
      bMiteredCorners=false
      bSeparateFillAlpha=false
      outlineMaterial=null
      outlineSize=0
      outlineColor=<rgba 0 0 0 1>
     }
    }
    shadowColorAndOpacity=<rgba 0 0 0 1>
   }
   slot {class=ButtonSlot}
  }
  HorizontalBox {widgetClassPath=&HorizontalBox, parent=VerticalBox_0, widget=HorizontalBox, slot {class=VerticalBoxSlot}}
  SetMusicPack {
   widgetClassPath=&EditorUtilityButton
   parent=HorizontalBox
   widget=SetMusicPack
   properties {toolTipText=Set cosmetic options to backbling presets.}
   slot {class=HorizontalBoxSlot, properties {padding=<ltrb 5 5 5 5>, size {sizeRule=Fill, value=1}}}
  }
  MusicPackText {
   widgetClassPath=&UEFN_TextBlock
   bIsVariable=true
   parent=SetMusicPack
   widget=MusicPackText
   properties {
    justification=Center
    text=Music Pack
    font {
     bForceMonospaced=false
     bMaterialIsStencil=false
     fontMaterial=null
     fontObject=&Roboto
     letterSpacing=0
     monospacedWidth=1
     size=9
     skewAmount=0
     typefaceFontName=Regular
     outlineSettings {
      bApplyOutlineToDropShadows=false
      bMiteredCorners=false
      bSeparateFillAlpha=false
      outlineMaterial=null
      outlineSize=0
      outlineColor=<rgba 0 0 0 1>
     }
    }
    shadowColorAndOpacity=<rgba 0 0 0 1>
   }
   slot {class=ButtonSlot}
  }
  SetContrail {
   widgetClassPath=&EditorUtilityButton
   parent=HorizontalBox
   widget=SetContrail
   properties {toolTipText=Set cosmetic options to contrail presets.}
   slot {class=HorizontalBoxSlot, properties {padding=<ltrb 5 5 5 5>, size {sizeRule=Fill, value=1}}}
  }
  ContrailText {
   widgetClassPath=&UEFN_TextBlock
   bIsVariable=true
   parent=SetContrail
   widget=ContrailText
   properties {
    justification=Center
    text=Contrail
    font {
     bForceMonospaced=false
     bMaterialIsStencil=false
     fontMaterial=null
     fontObject=&Roboto
     letterSpacing=0
     monospacedWidth=1
     size=9
     skewAmount=0
     typefaceFontName=Regular
     outlineSettings {
      bApplyOutlineToDropShadows=false
      bMiteredCorners=false
      bSeparateFillAlpha=false
      outlineMaterial=null
      outlineSize=0
      outlineColor=<rgba 0 0 0 1>
     }
    }
    shadowColorAndOpacity=<rgba 0 0 0 1>
   }
   slot {class=ButtonSlot}
  }
  SetCompanion {
   widgetClassPath=&EditorUtilityButton
   parent=HorizontalBox
   widget=SetCompanion
   properties {toolTipText=Set cosmetic options to contrail presets.}
   slot {class=HorizontalBoxSlot, properties {padding=<ltrb 5 5 5 5>, size {sizeRule=Fill, value=1}}}
  }
  CompanionText {
   widgetClassPath=&UEFN_TextBlock
   bIsVariable=true
   parent=SetCompanion
   widget=CompanionText
   properties {
    justification=Center
    text=Companion
    font {
     bForceMonospaced=false
     bMaterialIsStencil=false
     fontMaterial=null
     fontObject=&Roboto
     letterSpacing=0
     monospacedWidth=1
     size=9
     skewAmount=0
     typefaceFontName=Regular
     outlineSettings {
      bApplyOutlineToDropShadows=false
      bMiteredCorners=false
      bSeparateFillAlpha=false
      outlineMaterial=null
      outlineSize=0
      outlineColor=<rgba 0 0 0 1>
     }
    }
    shadowColorAndOpacity=<rgba 0 0 0 1>
   }
   slot {class=ButtonSlot}
  }
  SetKicks {
   widgetClassPath=&EditorUtilityButton
   parent=HorizontalBox
   widget=SetKicks
   properties {toolTipText=Set cosmetic options to contrail presets.}
   slot {class=HorizontalBoxSlot, properties {padding=<ltrb 5 5 5 5>, size {sizeRule=Fill, value=1}}}
  }
  KicksText {
   widgetClassPath=&UEFN_TextBlock
   bIsVariable=true
   parent=SetKicks
   widget=KicksText
   properties {
    justification=Center
    text=Kicks
    font {
     bForceMonospaced=false
     bMaterialIsStencil=false
     fontMaterial=null
     fontObject=&Roboto
     letterSpacing=0
     monospacedWidth=1
     size=9
     skewAmount=0
     typefaceFontName=Regular
     outlineSettings {
      bApplyOutlineToDropShadows=false
      bMiteredCorners=false
      bSeparateFillAlpha=false
      outlineMaterial=null
      outlineSize=0
      outlineColor=<rgba 0 0 0 1>
     }
    }
    shadowColorAndOpacity=<rgba 0 0 0 1>
   }
   slot {class=ButtonSlot}
  }
  GridPanel_89 {
   widgetClassPath=&GridPanel
   parent=VerticalBox_2
   widget=GridPanel_89
   slot {class=VerticalBoxSlot, properties {padding=<ltrb 2 2 2 2>, size {sizeRule=Fill, value=1}}}
  }
  AssetPath {
   widgetClassPath=/Game/Sounds/Fort_Audio_Blueprints/Utilities/EditorUtilityWidgets/Metasounds/Tools/EUW_MetaSoundAssetPath
   bIsVariable=true
   parent=GridPanel_89
   widget=AssetPath
   slot {class=GridSlot, properties {horizontalAlignment=HAlign_Center, verticalAlignment=VAlign_Center, padding=<ltrb 2 2 2 2>}}
  }
  ReferencesGrid {
   widgetClassPath=&GridPanel
   bIsVariable=true
   parent=GridPanel_89
   widget=ReferencesGrid
   slot {class=GridSlot, properties {row=2, padding=<ltrb 2 2 2 2>}}
  }
  SaveMetasoundsView {
   widgetClassPath=&SinglePropertyView
   parent=GridPanel_89
   widget=SaveMetasoundsView
   properties {nameOverride=Save Metasounds:, propertyName=bSaveMetasounds}
   slot {class=GridSlot, properties {horizontalAlignment=HAlign_Left, row=1}}
  }
  CopyrightSoundsView {
   widgetClassPath=&SinglePropertyView
   parent=GridPanel_89
   widget=CopyrightSoundsView
   properties {nameOverride=Licensed Audio, propertyName=bLicensedAudio, renderTransform {angle=0, scale=<1 1>, shear=<0 0>, translation=<-202.4024 0>}}
   slot {class=GridSlot, properties {horizontalAlignment=HAlign_Right, row=1}}
  }
  SoundNameListView {
   widgetClassPath=&/EUW_MetaSound_Cosmetic_ActionSoundsList
   bIsVariable=true
   parent=CanvasPanel_202
   widget=SoundNameListView
   slot {class=CanvasPanelSlot, properties {layoutData {alignment=<0 0>, anchors {maximum=<0 0>, minimum=<0 0>}, offsets=<ltrb 508 8 100 30>}}}
  }
  EditorUtilityScrollBox_0 {
   widgetClassPath=&EditorUtilityScrollBox
   parent=CanvasPanel_202
   widget=EditorUtilityScrollBox_0
   slot {
    class=CanvasPanelSlot
    properties {layoutData {alignment=<0 0>, anchors {maximum=<0 0>, minimum=<0 0>}, offsets=<ltrb 512 612 592.4925 139.3093>}}
   }
  }
  MsSourceBox {
   widgetClassPath=/Script/UMG.UniformGridPanel
   bIsVariable=true
   parent=EditorUtilityScrollBox_0
   widget=MsSourceBox
   slot {class=ScrollBoxSlot}
  }
 ]
}
